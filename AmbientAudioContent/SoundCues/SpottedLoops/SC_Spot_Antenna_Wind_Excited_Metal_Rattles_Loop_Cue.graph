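# Looping spot ambience cue for a wind-excited antenna with rattling metal, played as a positional outdoor sound.
# Mixes exterior wind, wire wind and distant metal rattle loops, each through nested looping envelopes of differing lengths that drift pitch and volume; the wire layer adds random pitch and a short fade-in.
# Only the first two quality levels receive the mix.
# Overrides give capsule attenuation with low-pass and listener focus, limited concurrency, and an exterior ambient bus send.
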SoundCue /AmbientAudioContent/SoundCues/SpottedLoops/SC_Spot_Antenna_Wind_Excited_Metal_Rattles_Loop_Cue
root=QualityLevel_0
nodes [
 {
  id=Enveloper_0
  children [WavePlayer_0]
  properties {
   bLoop=true
   bLoopIndefinitely=true
   loopEnd=30
   pitchCurve {editorCurveData {keys [{interpMode=RCIM_Cubic, time=0}, <at 15.9468 0.9181 RCIM_Cubic>, {interpMode=RCIM_Cubic, time=30}]}}
   volumeCurve {editorCurveData {keys [<at 0 1>]}}
  }
 }
 {
  id=Enveloper_1
  children [Modulator_1]
  properties {
   bLoop=true
   bLoopIndefinitely=true
   loopEnd=30
   pitchCurve {editorCurveData {keys [{interpMode=RCIM_Cubic, time=0}, <at 15.9468 0.9 RCIM_Cubic>, {interpMode=RCIM_Cubic, time=30}]}}
   volumeCurve {editorCurveData {keys [<at 0 1>]}}
  }
 }
 {
  id=Enveloper_12
  children [Modulator_0]
  properties {
   bLoop=true
   bLoopIndefinitely=true
   loopEnd=30
   pitchCurve {editorCurveData {keys [{interpMode=RCIM_Cubic, time=0}, <at 15.9468 0.92 RCIM_Cubic>, {interpMode=RCIM_Cubic, time=30}]}}
   volumeCurve {editorCurveData {keys [<at 0 1>]}}
  }
 }
 {
  id=Enveloper_17
  children [Enveloper_12]
  properties {
   bLoop=true
   bLoopIndefinitely=true
   loopEnd=40
   pitchCurve {editorCurveData {keys [{time=0}]}}
   volumeCurve {editorCurveData {keys [
      {arriveTangent=-0.01, interpMode=RCIM_Cubic, time=0, value=1}
      {arriveTangent=-0.0031, interpMode=RCIM_Cubic, leaveTangent=-0.0031, time=9.8969, value=0.3718}
      {arriveTangent=-0.0105, interpMode=RCIM_Cubic, leaveTangent=-0.0105, tangentMode=RCTM_User, time=18.5053, value=0.9432}
      {arriveTangent=-0.0515, interpMode=RCIM_Cubic, leaveTangent=-0.0515, time=23.4399, value=0.2987}
      {arriveTangent=0.0013, interpMode=RCIM_Cubic, leaveTangent=0.0013, tangentMode=RCTM_User, time=28.1028, value=0.4487}
      {arriveTangent=-0.0006, interpMode=RCIM_Cubic, leaveTangent=-0.0006, tangentMode=RCTM_User, time=32.9324, value=0.3744}
      <at 40 1>
     ]}}
  }
 }
 {
  id=Enveloper_2
  children [Enveloper_1]
  properties {
   bLoop=true
   bLoopIndefinitely=true
   loopEnd=120
   pitchCurve {editorCurveData {keys [{time=0}]}}
   volumeCurve {editorCurveData {keys [
      {arriveTangent=0.0001, interpMode=RCIM_Cubic, leaveTangent=0.0082, tangentMode=RCTM_User, time=0.4639, value=0.3721}
      {arriveTangent=0.0321, interpMode=RCIM_Cubic, leaveTangent=0.0321, time=10, value=0.6}
      {arriveTangent=-0.003, interpMode=RCIM_Cubic, leaveTangent=-0.003, time=20, value=1}
      <at 45.7834 0.4918 RCIM_Cubic>
      {arriveTangent=0.0153, interpMode=RCIM_Cubic, leaveTangent=0.0153, time=60, value=1}
      {arriveTangent=-0.0202, interpMode=RCIM_Cubic, leaveTangent=-0.0202, tangentMode=RCTM_User, time=82.1697, value=1.05}
      <at 119.9639 0.4918>
     ]}}
  }
 }
 {
  id=Enveloper_29
  children [Enveloper_2]
  properties {
   pitchCurve {editorCurveData {keys [{time=0}]}}
   volumeCurve {editorCurveData {keys [{arriveTangent=3.4293, interpMode=RCIM_Cubic, leaveTangent=3.4293, tangentMode=RCTM_User, time=0, value=0}, <at 0.4 1 RCIM_Cubic>]}}
  }
 }
 {
  id=Enveloper_3
  children [Enveloper_0]
  properties {
   bLoop=true
   bLoopIndefinitely=true
   loopEnd=20
   pitchCurve {editorCurveData {keys [{time=0}]}}
   volumeCurve {editorCurveData {keys [
      {arriveTangent=-0.01, interpMode=RCIM_Cubic, time=0, value=1}
      <at 10 1.2 RCIM_Cubic>
      {interpMode=RCIM_Cubic, leaveTangent=0.01, time=20, value=1}
     ]}}
  }
 }
 {id=Mixer_1, children [Enveloper_3,Enveloper_29,Enveloper_17], properties {inputVolume [0.4,1,2]}}
 {id=Modulator_0, children [WavePlayer_4], properties {pitchMax=1, pitchMin=1}}
 {id=Modulator_1, children [WavePlayer_2], properties {pitchMax=1.2, pitchMin=1}}
 {id=QualityLevel_0, children [Mixer_1,Mixer_1,null,null]}
 {id=WavePlayer_0, properties {bLooping=true, soundWaveAssetPtr=/AmbientAudioContent/Waves/SpottedLoops/Spot_Winterfest_Exterior_Wind_Loop}}
 {id=WavePlayer_2, properties {bLooping=true, soundWaveAssetPtr=/AmbientAudioContent/Waves/SpottedLoops/Spot_Ascender_Ambient_Wire_Wind_Loop}}
 {
  id=WavePlayer_4
  properties {bLooping=true, soundWaveAssetPtr=/Game/Sounds/Asteria/POI/JunkYard/Ambience_Metal_Distant_Rattling_In_Wind_Loop_03_Quad}
 }
]
properties {
 bOverrideAttenuation=true
 bOverrideConcurrency=true
 maxDistance=2050
 soundClassObject=&Fort_Ambient_Outdoors_Athena
 volumeMultiplier=0.35
 attenuationOverrides {
  attenuationShape=Capsule
  bAttenuateWithLPF=true
  bEnableFocusInterpolation=true
  bEnableListenerFocus=true
  falloffDistance=1500
  focusReleaseInterpSpeed=0.9
  lPFFrequencyAtMax=750
  lPFRadiusMax=4000
  lPFRadiusMin=1000
  nonFocusVolumeAttenuation=0.75
  attenuationShapeExtents=<550 300>
 }
 busSends [{audioBus=null, minSendLevel=0, sendLevel=1, soundSourceBus=&AmbientAudio_Exterior_SourceBus}]
 concurrencyOverrides {maxCount=2}
}
