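# Looping ambient spot cue that scatters distant arcade stinger one-shots.
# A loop waits a random delay, applies slight pitch variation, then randomly picks one of twelve waves or an empty slot, with Mixer nodes trimming some waves' levels.
# The first two quality levels share this chain; the other two are empty.
# Overrides add custom falloff, air absorption, occlusion, listener focus and a tight voice limit, alongside two distance-scaled reverb sends.
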SoundCue /AmbientAudioContent/SoundCues/SpottedLoops/SC_Spot_Arcade_Stingers_Distant_Loop_IntExt_Cue
root=QualityLevel_0
nodes [
 {id=Delay_0, children [Modulator_0], properties {delayMax=8, delayMin=3}}
 {id=Looping_0, children [Delay_0]}
 {id=Mixer_0, children [WavePlayer_29], properties {inputVolume [0.4]}}
 {id=Mixer_1, children [WavePlayer_22], properties {inputVolume [0.4]}}
 {id=Mixer_2, children [WavePlayer_23], properties {inputVolume [0.4]}}
 {id=Mixer_4, children [WavePlayer_28], properties {inputVolume [0.75]}}
 {id=Mixer_5, children [WavePlayer_27], properties {inputVolume [1.25]}}
 {id=Mixer_6, children [WavePlayer_17], properties {inputVolume [0.4]}}
 {id=Mixer_7, children [WavePlayer_18], properties {inputVolume [0.78]}}
 {id=Modulator_0, children [Random_2], properties {pitchMax=1.1, pitchMin=0.9}}
 {id=QualityLevel_0, children [Looping_0,Looping_0,null,null]}
 {
  id=Random_2
  children [
   Mixer_6
   Mixer_7
   Mixer_1
   Mixer_2
   Mixer_5
   Mixer_4
   Mixer_0
   WavePlayer_30
   WavePlayer_24
   WavePlayer_25
   WavePlayer_20
   WavePlayer_21
   null
  ]
  properties {weights [1,1,1,1,1,1,1,1,1,1,1,1,1]}
 }
 {id=WavePlayer_17, properties {soundWaveAssetPtr=/Game/Sounds/Artemis/POI/Hollow_Base_Takover/Spots/Arcade/Spot_Arcade_Stingers_Distant_01}}
 {id=WavePlayer_18, properties {soundWaveAssetPtr=/Game/Sounds/Artemis/POI/Hollow_Base_Takover/Spots/Arcade/Spot_Arcade_Stingers_Distant_02}}
 {id=WavePlayer_20, properties {soundWaveAssetPtr=/Game/Sounds/Artemis/POI/Hollow_Base_Takover/Spots/Arcade/Spot_Arcade_Stingers_Distant_05}}
 {id=WavePlayer_21, properties {soundWaveAssetPtr=/Game/Sounds/Artemis/POI/Hollow_Base_Takover/Spots/Arcade/Spot_Arcade_Stingers_Distant_06}}
 {id=WavePlayer_22, properties {soundWaveAssetPtr=/Game/Sounds/Artemis/POI/Hollow_Base_Takover/Spots/Arcade/Spot_Arcade_Stingers_Distant_07}}
 {id=WavePlayer_23, properties {soundWaveAssetPtr=/Game/Sounds/Artemis/POI/Hollow_Base_Takover/Spots/Arcade/Spot_Arcade_Stingers_Distant_08}}
 {id=WavePlayer_24, properties {soundWaveAssetPtr=/Game/Sounds/Artemis/POI/Hollow_Base_Takover/Spots/Arcade/Spot_Arcade_Stingers_Distant_09}}
 {id=WavePlayer_25, properties {soundWaveAssetPtr=/Game/Sounds/Artemis/POI/Hollow_Base_Takover/Spots/Arcade/Spot_Arcade_Stingers_Distant_10}}
 {id=WavePlayer_27, properties {soundWaveAssetPtr=/Game/Sounds/Artemis/POI/Hollow_Base_Takover/Spots/Arcade/Spot_Arcade_Stingers_Distant_12}}
 {id=WavePlayer_28, properties {soundWaveAssetPtr=/Game/Sounds/Artemis/POI/Hollow_Base_Takover/Spots/Arcade/Spot_Arcade_Stingers_Distant_13}}
 {id=WavePlayer_29, properties {soundWaveAssetPtr=/Game/Sounds/Artemis/POI/Hollow_Base_Takover/Spots/Arcade/Spot_Arcade_Stingers_Distant_14}}
 {id=WavePlayer_30, properties {soundWaveAssetPtr=/Game/Sounds/Artemis/POI/Hollow_Base_Takover/Spots/Arcade/Spot_Arcade_Stingers_Distant_03}}
]
properties {
 bEnableBusSends=false
 bOverrideAttenuation=true
 bOverrideConcurrency=true
 maxDistance=3600
 soundClassObject=&Fort_Ambient_OneShots
 volumeMultiplier=0.8
 attenuationOverrides {
  absorptionMethod=CustomCurve
  bAttenuateWithLPF=true
  bEnableFocusInterpolation=true
  bEnableListenerFocus=true
  bEnableLogFrequencyScaling=true
  bEnableOcclusion=true
  distanceAlgorithm=Custom
  focusAttackInterpSpeed=0.5
  focusAzimuth=45
  focusDistanceScale=1.1
  focusReleaseInterpSpeed=0.5
  hPFFrequencyAtMax=750
  lPFFrequencyAtMax=500
  lPFRadiusMax=4000
  lPFRadiusMin=650
  nonFocusAzimuth=80
  nonFocusVolumeAttenuation=0.8
  occlusionLowPassFilterFrequency=2000
  occlusionTraceChannel=ECC_GameTraceChannel11
  occlusionVolumeAttenuation=0.75
  attenuationShapeExtents {x=0}
  customAttenuationCurve {editorCurveData {keys [<at 0 1>, {arriveTangent=-0.5305, interpMode=RCIM_Cubic, leaveTangent=-0.5305, tangentMode=RCTM_User, time=1, value=0}]}}
  customHighpassAirAbsorptionCurve {externalCurve=/Game/Sounds/Attenuation/Curves/Fort_Default_Logarithmic_AirAbsorption_Float}
  customLowpassAirAbsorptionCurve {externalCurve=&Fort_Default_Logarithmic_AirAbsorption_Float}
 }
 concurrencyOverrides {maxCount=2}
 soundSubmixSends [
  {
   maxSendDistance=6000
   maxSendLevel=0.45
   minSendDistance=1500
   minSendLevel=0
   sendLevelControlMethod=Linear
   soundSubmix=&Fortnite_Reverb_Distant
  }
  {
   maxSendDistance=6000
   maxSendLevel=0.8
   minSendDistance=500
   minSendLevel=0.1
   sendLevelControlMethod=Linear
   soundSubmix=&Fortnite_Reverb_Default
  }
 ]
}
