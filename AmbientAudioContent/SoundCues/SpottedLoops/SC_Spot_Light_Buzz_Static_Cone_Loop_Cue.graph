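# Sound cue for a looping static electrical light buzz, an outdoor ambient spot emitter with cone-shaped falloff.
# It mixes a randomly chosen high-pitched hologram loop, weighted louder, with one of two neon sign buzz loops, using modulators to vary pitch.
# A random branch optionally delays the start briefly before feeding an attenuation node with HRTF, occlusion, low-pass and listener focus.
# Only the first two quality levels connect to sound, and concurrency is capped.
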SoundCue /AmbientAudioContent/SoundCues/SpottedLoops/SC_Spot_Light_Buzz_Static_Cone_Loop_Cue
root=QualityLevel_0
nodes [
 {
  id=Attenuation_5
  children [Random_3]
  properties {
   bOverrideAttenuation=true
   attenuationOverrides {
    attenuationShape=Cone
    bAttenuateWithLPF=true
    bEnableFocusInterpolation=true
    bEnableListenerFocus=true
    bEnableLogFrequencyScaling=true
    bEnableOcclusion=true
    coneSphereFalloffDistance=500
    coneSphereRadius=150
    distanceAlgorithm=NaturalSound
    falloffDistance=750
    focusAttackInterpSpeed=0.5
    focusAzimuth=15
    focusReleaseInterpSpeed=0.5
    focusVolumeAttenuation=1.2
    hPFFrequencyAtMax=100
    hPFFrequencyAtMin=50
    lPFFrequencyAtMax=500
    lPFRadiusMax=1500
    lPFRadiusMin=150
    nonFocusAzimuth=45
    nonFocusVolumeAttenuation=0.2
    occlusionInterpolationTime=2
    occlusionLowPassFilterFrequency=1000
    occlusionTraceChannel=ECC_GameTraceChannel11
    occlusionVolumeAttenuation=0
    spatializationAlgorithm=SPATIALIZATION_HRTF
    stereoSpread=0
    attenuationShapeExtents {y=44.7792, z=29.6789}
    customAttenuationCurve {editorCurveData {keys [
       {arriveTangent=-2.6039, interpMode=RCIM_Cubic, leaveTangent=-2.6039, tangentMode=RCTM_User, time=0, value=1}
       {arriveTangent=0.0377, interpMode=RCIM_Cubic, leaveTangent=0.0377, tangentMode=RCTM_User, time=0.5, value=0}
      ]}}
   }
  }
 }
 {id=Delay_0, children [Modulator_0], properties {delayMax=0.25, delayMin=0.25}}
 {id=Mixer_0, children [Modulator_1,Modulator_3], properties {inputVolume [2,1]}}
 {id=Modulator_0, children [Mixer_0], properties {pitchMax=1.1}}
 {id=Modulator_1, children [Random_2], properties {pitchMax=2, pitchMin=1.8}}
 {id=Modulator_3, children [Random_0], properties {pitchMax=1.1}}
 {id=Modulator_4, children [WavePlayer_12], properties {pitchMax=0.7, pitchMin=0.7, volumeMax=0.6, volumeMin=0.6}}
 {id=QualityLevel_0, children [Attenuation_5, Attenuation_5, null, null]}
 {id=Random_0, children [Modulator_4, WavePlayer_13], properties {bShouldExcludeFromBranchCulling=true, weights [1,1]}}
 {id=Random_2, children [WavePlayer_0,WavePlayer_7,WavePlayer_8,WavePlayer_9], properties {weights [1,1,1,1]}}
 {id=Random_3, children [Delay_0,Modulator_0], properties {weights [1,1]}}
 {id=WavePlayer_0, properties {bLooping=true, soundWaveAssetPtr=/AmbientAudioContent/Waves/SpottedLoops/Spot_Keys_And_Locks_HologramLoop_04}}
 {id=WavePlayer_12, properties {bLooping=true, soundWaveAssetPtr=/AmbientAudioContent/Waves/SpottedLoops/spot_neon_sign_buzz_loop_03}}
 {id=WavePlayer_13, properties {bLooping=true, soundWaveAssetPtr=/AmbientAudioContent/Waves/SpottedLoops/spot_neon_sign_buzz_loop_01}}
 {id=WavePlayer_7, properties {bLooping=true, soundWaveAssetPtr=/AmbientAudioContent/Waves/SpottedLoops/Spot_Keys_And_Locks_HologramLoop_03}}
 {id=WavePlayer_8, properties {bLooping=true, soundWaveAssetPtr=/AmbientAudioContent/Waves/SpottedLoops/Spot_Keys_And_Locks_HologramLoop_02}}
 {id=WavePlayer_9, properties {bLooping=true, soundWaveAssetPtr=/AmbientAudioContent/Waves/SpottedLoops/Spot_Keys_And_Locks_HologramLoop_01}}
]
properties {
 bOverrideConcurrency=true
 maxDistance=1150
 soundClassObject=&Fort_Ambient_Outdoors_Athena
 volumeMultiplier=0.3
 attenuationOverrides {falloffDistance=2000}
 concurrencyOverrides {maxCount=2, resolutionRule=PreventNew}
 concurrencySet [&Weapon_Drop_StopOldest]
 soundSubmixSends [{minSendLevel=0.25, sendLevelControlMethod=Linear, soundSubmix=&Fortnite_Reverb_Default}]
}
