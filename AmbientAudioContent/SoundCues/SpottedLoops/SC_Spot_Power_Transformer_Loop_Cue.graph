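# Sound cue for a looping power transformer ambient spot sound, routed to the Fort_Quests sound class with quest transformer attenuation.
# An equal-weight root Random either plays immediately or passes through a short fixed Delay, which offsets the start between instances.
# Both paths share one pitch-varying Modulator feeding a second Random that picks one of two looping waves.
# The cue primes on load.
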SoundCue /AmbientAudioContent/SoundCues/SpottedLoops/SC_Spot_Power_Transformer_Loop_Cue
root=Random_2
nodes [
 {id=Delay_0, children [Modulator_0], properties {delayMax=0.3, delayMin=0.3}}
 {id=Modulator_0, children [Random_1], properties {pitchMax=1.2}}
 {id=Random_1, children [WavePlayer_0,WavePlayer_1], properties {weights [1,1]}}
 {id=Random_2, children [Delay_0,Modulator_0], properties {weights [1,1]}}
 {id=WavePlayer_0, properties {bLooping=true, soundWaveAssetPtr=/AmbientAudioContent/Waves/SpottedLoops/Spot_Power_Transformer_Loop_01}}
 {id=WavePlayer_1, properties {bLooping=true, soundWaveAssetPtr=/AmbientAudioContent/Waves/SpottedLoops/Spot_Power_Transformer_Loop_02}}
]
properties {
 attenuationSettings=/Game/Athena/Sounds/Quests/PowerTransformer/Fort_Power_Transformer_Loop_Att
 bPrimeOnLoad=true
 maxDistance=2550
 soundClassObject=&Fort_Quests
 volumeMultiplier=1
 attenuationOverrides {falloffDistance=700, attenuationShapeExtents {x=100}}
}
