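# Ambient Sound Cue for a discreet security camera idle loop, played quietly in the indoor ambient sound class.
# A Mixer blends a looping movement wave, started after a brief delay, with a beep repeated through a Looping node after a six-second delay, each passed through a Modulator.
# Overrides add cone-shaped HRTF attenuation with filtering, occlusion, listener focus and a distant reverb send, plus low-count concurrency that stops the quietest.
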SoundCue /AmbientAudioContent/SoundCues/SpottedLoops/SC_Spot_Security_Camera_Discreet_Idle_Loop_Cue
root=Mixer_1
nodes [
 {id=Delay_0, children [Modulator_1], properties {delayMax=6, delayMin=6}}
 {id=Delay_2, children [Modulator_0], properties {delayMax=0.08, delayMin=0.08}}
 {id=Looping_0, children [Delay_0]}
 {id=Mixer_1, children [Delay_2,Looping_0], properties {inputVolume [0.5,0.6]}}
 {id=Modulator_0, children [WavePlayer_2]}
 {id=Modulator_1, children [WavePlayer_1]}
 {id=WavePlayer_1, properties {soundWaveAssetPtr=/AmbientAudioContent/Waves/SpottedLoops/Spot_Security_Camera_MoveLoop_Beep_01}}
 {id=WavePlayer_2, properties {bLooping=true, soundWaveAssetPtr=/AmbientAudioContent/Waves/SpottedLoops/Spot_Security_Camera_MoveLoop_01}}
]
properties {
 bOverrideAttenuation=true
 bOverrideConcurrency=true
 maxDistance=1350
 soundClassObject=&Fort_Ambient_Indoors_Athena
 volumeMultiplier=0.15
 attenuationOverrides {
  attenuationShape=Cone
  bAttenuateWithLPF=true
  bEnableFocusInterpolation=true
  bEnableListenerFocus=true
  bEnableLogFrequencyScaling=true
  bEnableOcclusion=true
  bEnableSubmixSends=true
  distanceAlgorithm=NaturalSound
  falloffDistance=1000
  hPFFrequencyAtMax=600
  lPFFrequencyAtMax=1200
  lPFRadiusMax=1000
  lPFRadiusMin=0
  nonFocusVolumeAttenuation=0.5
  occlusionInterpolationTime=1
  occlusionLowPassFilterFrequency=600
  occlusionTraceChannel=ECC_GameTraceChannel11
  occlusionVolumeAttenuation=0.5
  spatializationAlgorithm=SPATIALIZATION_HRTF
  stereoSpread=0
  attenuationShapeExtents=<350 70 20>
  submixSendSettings [{
    bEnableHPFCutoff=false
    bEnableLPFCutoff=false
    disableManualSendClamp=false
    hPFCutoff=20
    lPFCutoff=20000
    maxSendDistance=7500
    maxSendLevel=0.25
    minSendDistance=0
    minSendLevel=0
    sendLevel=0.2
    sendLevelControlMethod=Linear
    soundSubmix=&Fortnite_Reverb_Distant
   }]
 }
 concurrencyOverrides {maxCount=3, resolutionRule=StopQuietest}
}
