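# Sound cue for a dreamy pad tone used as a spot ambience loop at the Utopia City fountain gate.
# A single looping Wave Player feeds a stereo pad wave straight to the output with boosted volume, under the indoor ambient sound class.
# Cue-level overrides give it capsule attenuation with stereo spread and no reverb send, plus concurrency that stops the quietest instance when the limit is hit.
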SoundCue /AmbientAudioContent/SoundCues/SpottedLoops/SC_Spot_UtopiaCity_Dreamy_Tone_FountainGate
root=WavePlayer_4
nodes [{id=WavePlayer_4, properties {bLooping=true, soundWaveAssetPtr=/AmbientAudioContent/Waves/POI/Ambience_UtopiaCity_Dreamy_Pad_Stereo_Loop}}]
properties {
 bOverrideAttenuation=true
 bOverrideConcurrency=true
 maxDistance=1300
 soundClassObject=&Fort_Ambient_Indoors_Athena
 volumeMultiplier=1.4
 attenuationOverrides {
  attenuationShape=Capsule
  bEnableReverbSend=false
  falloffDistance=1000
  nonSpatializedRadiusEnd=50
  stereoSpread=120
  attenuationShapeExtents=<300 300>
 }
 concurrencyOverrides {maxCount=4, resolutionRule=StopQuietest}
}
