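# Looping ambient spot cue for an Academy hologram window, reusing the NeonCity UltraSam hologram loop in the Athena ambient sound class.
# A Modulator plays the looping wave at a fixed doubled pitch, with reduced cue volume.
# Overridden attenuation uses a custom short-range falloff curve with distance filtering and listener focus.
# A concurrency override allows one instance, and a distance-driven send feeds the distant reverb submix.
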SoundCue /AmbientAudioContent/SoundCues/SpottedLoops/Spot_Academy_Hologram_Window_Loop_Cue
root=Modulator_0
nodes [
 {id=Modulator_0, children [WavePlayer_2], properties {pitchMax=2, pitchMin=2, volumeMin=1}}
 {id=WavePlayer_2, properties {bLooping=true, soundWaveAssetPtr=/AmbientAudioContent/Waves/POI/Spot_NeonCity_Hologram_UltraSam_Loop}}
]
properties {
 bOverrideAttenuation=true
 bOverrideConcurrency=true
 maxDistance=1500
 soundClassObject=&Fort_Ambient_Athena
 volumeMultiplier=0.301
 attenuationOverrides {
  bAttenuateWithLPF=true
  bEnableFocusInterpolation=true
  bEnableListenerFocus=true
  distanceAlgorithm=Custom
  falloffDistance=1500
  focusAttackInterpSpeed=0.5
  focusReleaseInterpSpeed=0.5
  hPFFrequencyAtMax=300
  lPFFrequencyAtMax=600
  lPFRadiusMax=1500
  lPFRadiusMin=0
  nonFocusVolumeAttenuation=0.75
  stereoSpread=150
  attenuationShapeExtents {x=0}
  customAttenuationCurve {editorCurveData {keys [
     <at 0 1 RCIM_Cubic>
     {arriveTangent=-1, interpMode=RCIM_Cubic, leaveTangent=-1, time=0.2498, value=0.3196}
     {arriveTangent=-0.3683, interpMode=RCIM_Cubic, leaveTangent=-0.9149, tangentMode=RCTM_User, time=1, value=0}
    ]}}
 }
 concurrencyOverrides {maxCount=1}
 concurrencySet [&Weapon_Drop_StopOldest]
 soundSubmixSends [{
   maxSendDistance=4000
   maxSendLevel=0.7
   minSendDistance=0
   minSendLevel=0.05
   sendLevelControlMethod=Linear
   soundSubmix=&Fortnite_Reverb_Distant
  }]
}
