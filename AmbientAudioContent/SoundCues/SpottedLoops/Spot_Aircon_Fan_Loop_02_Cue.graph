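# Sound cue for a spotted ambient air-conditioner fan loop, classed as an ambient prop sound.
# A Random node picks one of three looping fan waves with equal weight, the third boosted through a single-input Mixer.
# A root Modulator on default settings adds slight variation, and the cue plays at reduced volume.
# It uses an outpost wall fan attenuation asset and overrides concurrency to cap simultaneous instances.
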SoundCue /AmbientAudioContent/SoundCues/SpottedLoops/Spot_Aircon_Fan_Loop_02_Cue
root=Modulator_0
nodes [
 {id=Mixer_0, children [WavePlayer_2], properties {inputVolume [1.2]}}
 {id=Modulator_0, children [Random_0]}
 {id=Random_0, children [WavePlayer_0,WavePlayer_1,Mixer_0], properties {weights [1,1,1]}}
 {id=WavePlayer_0, properties {bLooping=true, soundWaveAssetPtr=/AmbientAudioContent/Waves/SpottedLoops/Spot_AC_Fan_01}}
 {id=WavePlayer_1, properties {bLooping=true, soundWaveAssetPtr=/AmbientAudioContent/Waves/SpottedLoops/Spot_AC_Fan_02}}
 {id=WavePlayer_2, properties {bLooping=true, soundWaveAssetPtr=/AmbientAudioContent/Waves/SpottedLoops/Spot_AC_Fan_03}}
]
properties {
 attenuationSettings=/Game/Sounds/Artemis/Biomes/Grasslands/Grassland_Spots/Spot_Outpost_Wall_Fan_Smaller_Att
 bOverrideConcurrency=true
 maxDistance=1200
 soundClassObject=&Fort_Ambient_Props_Athena
 volumeMultiplier=0.5
 attenuationOverrides {attenuationShapeExtents {x=0}}
 concurrencyOverrides {maxCount=3}
}
