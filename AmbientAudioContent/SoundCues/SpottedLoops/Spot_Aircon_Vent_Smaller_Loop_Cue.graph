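# Sound cue for a looping spot ambient sound of a smaller air conditioning vent, in the Fort_Ambient_Athena sound class.
# A Random node picks with equal weight among three looping AC duct air waves and feeds a Modulator at the root.
# An attenuation asset and reduced volume shape the playback, while a concurrency override caps simultaneous instances at a few.
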SoundCue /AmbientAudioContent/SoundCues/SpottedLoops/Spot_Aircon_Vent_Smaller_Loop_Cue
root=Modulator_0
nodes [
 {id=Modulator_0, children [Random_1]}
 {id=Random_1, children [WavePlayer_0,WavePlayer_1,WavePlayer_2], properties {weights [1,1,1]}}
 {id=WavePlayer_0, properties {bLooping=true, soundWaveAssetPtr=/AmbientAudioContent/Waves/SpottedLoops/Spot_AC_Duct_Air_01}}
 {id=WavePlayer_1, properties {bLooping=true, soundWaveAssetPtr=/AmbientAudioContent/Waves/SpottedLoops/Spot_AC_Duct_Air_02}}
 {id=WavePlayer_2, properties {bLooping=true, soundWaveAssetPtr=/AmbientAudioContent/Waves/SpottedLoops/Spot_AC_Duct_Air_03}}
]
properties {
 attenuationSettings=/AmbientAudioBase/Attenuations/Spot_Aircon_Vent_Loop_Smaller_Att
 bOverrideConcurrency=true
 maxDistance=1250
 soundClassObject=&Fort_Ambient_Athena
 volumeMultiplier=0.4
 attenuationOverrides {attenuationShapeExtents {x=0}}
 concurrencyOverrides {maxCount=3}
}
