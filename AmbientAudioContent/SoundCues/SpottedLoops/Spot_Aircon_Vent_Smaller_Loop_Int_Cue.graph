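# Sound cue for a small indoor air-conditioning vent ambient loop, routed to an indoor ambient sound class.
# A root Random picks either the fan or the duct-air branch, each passing through a Modulator into an equally weighted choice of three looping waves.
# Smaller-vent attenuation shapes falloff, a concurrency override caps it at two instances, and a custom-curve send feeds more distant reverb as distance grows.
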SoundCue /AmbientAudioContent/SoundCues/SpottedLoops/Spot_Aircon_Vent_Smaller_Loop_Int_Cue
root=Random_2
nodes [
 {id=Modulator_0, children [Random_1]}
 {id=Modulator_1, children [Random_0]}
 {id=Random_0, children [WavePlayer_3,WavePlayer_4,WavePlayer_5], properties {weights [1,1,1]}}
 {id=Random_1, children [WavePlayer_0,WavePlayer_1,WavePlayer_2], properties {weights [1,1,1]}}
 {id=Random_2, children [Modulator_1,Modulator_0], properties {weights [1,1]}}
 {id=WavePlayer_0, properties {bLooping=true, soundWaveAssetPtr=/AmbientAudioContent/Waves/SpottedLoops/Spot_AC_Duct_Air_01}}
 {id=WavePlayer_1, properties {bLooping=true, soundWaveAssetPtr=/AmbientAudioContent/Waves/SpottedLoops/Spot_AC_Duct_Air_02}}
 {id=WavePlayer_2, properties {bLooping=true, soundWaveAssetPtr=/AmbientAudioContent/Waves/SpottedLoops/Spot_AC_Duct_Air_03}}
 {id=WavePlayer_3, properties {bLooping=true, soundWaveAssetPtr=/AmbientAudioContent/Waves/SpottedLoops/Spot_AC_Fan_01}}
 {id=WavePlayer_4, properties {bLooping=true, soundWaveAssetPtr=/AmbientAudioContent/Waves/SpottedLoops/Spot_AC_Fan_02}}
 {id=WavePlayer_5, properties {bLooping=true, soundWaveAssetPtr=/AmbientAudioContent/Waves/SpottedLoops/Spot_AC_Fan_03}}
]
properties {
 attenuationSettings=/AmbientAudioBase/Attenuations/Spot_Aircon_Vent_Loop_Smaller_Att
 bOverrideConcurrency=true
 maxDistance=1250
 soundClassObject=&Fort_Ambient_Indoors_Athena
 volumeMultiplier=0.4
 attenuationOverrides {attenuationShapeExtents {x=0}}
 concurrencyOverrides {maxCount=2}
 soundSubmixSends [{
   minSendLevel=0
   sendLevelControlMethod=CustomCurve
   soundSubmix=&Fortnite_Reverb_Distant
   customSendLevelCurve {editorCurveData {keys [<at 0 0>,<at 1 0.4>]}}
  }]
}
