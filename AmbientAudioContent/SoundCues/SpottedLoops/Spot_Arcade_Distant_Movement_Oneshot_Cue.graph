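# Sound cue for distant one-shot arcade ambience, playing random Neon City arcade sounds on the ambient one-shots sound class.
# A Concatenator plays one random pick, then a second after a random delay, both from the same wave pool with pitch and volume modulation.
# A custom attenuation curve keeps it silent up close, peaking at mid-range and fading out far away, with distance-based filtering.
# Concurrency overrides cap simultaneous instances.
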SoundCue /AmbientAudioContent/SoundCues/SpottedLoops/Spot_Arcade_Distant_Movement_Oneshot_Cue
root=Concatenator_0
nodes [
 {id=Concatenator_0, children [Modulator_1,Delay_0], properties {inputVolume [1,1]}}
 {id=Delay_0, children [Modulator_0], properties {delayMax=5, delayMin=2}}
 {id=Modulator_0, children [Random_3], properties {pitchMax=1.2, pitchMin=0.85, volumeMax=1, volumeMin=0.8}}
 {id=Modulator_1, children [Random_2], properties {pitchMax=1.2, pitchMin=0.85, volumeMax=1, volumeMin=0.8}}
 {
  id=Random_2
  children [
   WavePlayer_36
   WavePlayer_37
   WavePlayer_38
   WavePlayer_39
   WavePlayer_40
   WavePlayer_41
   WavePlayer_42
   WavePlayer_43
   WavePlayer_44
   WavePlayer_45
   WavePlayer_46
   WavePlayer_47
   WavePlayer_48
   WavePlayer_49
   WavePlayer_50
   WavePlayer_51
   WavePlayer_52
   WavePlayer_53
  ]
  properties {weights [1,1,1,1,1,1,1,1,1,1,1,1,1,1,1,1,1,1]}
 }
 {
  id=Random_3
  children [
   WavePlayer_54
   WavePlayer_55
   WavePlayer_56
   WavePlayer_57
   WavePlayer_58
   WavePlayer_59
   WavePlayer_60
   WavePlayer_61
   WavePlayer_62
   WavePlayer_63
   WavePlayer_64
   WavePlayer_65
   WavePlayer_66
   WavePlayer_67
   WavePlayer_68
   WavePlayer_69
   WavePlayer_70
   WavePlayer_71
  ]
  properties {weights [1,1,1,1,1,1,1,1,1,1,1,1,1,1,1,1,1,1]}
 }
 {id=WavePlayer_36, properties {soundWaveAssetPtr=/AmbientAudioContent/Waves/OneShots/Spot_Neon_City_Arcade_Sounds_18}}
 {id=WavePlayer_37, properties {soundWaveAssetPtr=/AmbientAudioContent/Waves/OneShots/Spot_Neon_City_Arcade_Sounds_17}}
 {id=WavePlayer_38, properties {soundWaveAssetPtr=/AmbientAudioContent/Waves/OneShots/Spot_Neon_City_Arcade_Sounds_16}}
 {id=WavePlayer_39, properties {soundWaveAssetPtr=/AmbientAudioContent/Waves/OneShots/Spot_Neon_City_Arcade_Sounds_15}}
 {id=WavePlayer_40, properties {soundWaveAssetPtr=/AmbientAudioContent/Waves/OneShots/Spot_Neon_City_Arcade_Sounds_14}}
 {id=WavePlayer_41, properties {soundWaveAssetPtr=/AmbientAudioContent/Waves/OneShots/Spot_Neon_City_Arcade_Sounds_13}}
 {id=WavePlayer_42, properties {soundWaveAssetPtr=/AmbientAudioContent/Waves/OneShots/Spot_Neon_City_Arcade_Sounds_12}}
 {id=WavePlayer_43, properties {soundWaveAssetPtr=/AmbientAudioContent/Waves/OneShots/Spot_Neon_City_Arcade_Sounds_11}}
 {id=WavePlayer_44, properties {soundWaveAssetPtr=/AmbientAudioContent/Waves/OneShots/Spot_Neon_City_Arcade_Sounds_10}}
 {id=WavePlayer_45, properties {soundWaveAssetPtr=/AmbientAudioContent/Waves/OneShots/Spot_Neon_City_Arcade_Sounds_09}}
 {id=WavePlayer_46, properties {soundWaveAssetPtr=/AmbientAudioContent/Waves/OneShots/Spot_Neon_City_Arcade_Sounds_08}}
 {id=WavePlayer_47, properties {soundWaveAssetPtr=/AmbientAudioContent/Waves/OneShots/Spot_Neon_City_Arcade_Sounds_07}}
 {id=WavePlayer_48, properties {soundWaveAssetPtr=/AmbientAudioContent/Waves/OneShots/Spot_Neon_City_Arcade_Sounds_06}}
 {id=WavePlayer_49, properties {soundWaveAssetPtr=/AmbientAudioContent/Waves/OneShots/Spot_Neon_City_Arcade_Sounds_05}}
 {id=WavePlayer_50, properties {soundWaveAssetPtr=/AmbientAudioContent/Waves/OneShots/Spot_Neon_City_Arcade_Sounds_04}}
 {id=WavePlayer_51, properties {soundWaveAssetPtr=/AmbientAudioContent/Waves/OneShots/Spot_Neon_City_Arcade_Sounds_03}}
 {id=WavePlayer_52, properties {soundWaveAssetPtr=/AmbientAudioContent/Waves/OneShots/Spot_Neon_City_Arcade_Sounds_02}}
 {id=WavePlayer_53, properties {soundWaveAssetPtr=/AmbientAudioContent/Waves/OneShots/Spot_Neon_City_Arcade_Sounds_01}}
 {id=WavePlayer_54, properties {soundWaveAssetPtr=&Spot_Neon_City_Arcade_Sounds_18}}
 {id=WavePlayer_55, properties {soundWaveAssetPtr=&Spot_Neon_City_Arcade_Sounds_17}}
 {id=WavePlayer_56, properties {soundWaveAssetPtr=&Spot_Neon_City_Arcade_Sounds_16}}
 {id=WavePlayer_57, properties {soundWaveAssetPtr=&Spot_Neon_City_Arcade_Sounds_15}}
 {id=WavePlayer_58, properties {soundWaveAssetPtr=&Spot_Neon_City_Arcade_Sounds_14}}
 {id=WavePlayer_59, properties {soundWaveAssetPtr=&Spot_Neon_City_Arcade_Sounds_13}}
 {id=WavePlayer_60, properties {soundWaveAssetPtr=&Spot_Neon_City_Arcade_Sounds_12}}
 {id=WavePlayer_61, properties {soundWaveAssetPtr=&Spot_Neon_City_Arcade_Sounds_11}}
 {id=WavePlayer_62, properties {soundWaveAssetPtr=&Spot_Neon_City_Arcade_Sounds_10}}
 {id=WavePlayer_63, properties {soundWaveAssetPtr=&Spot_Neon_City_Arcade_Sounds_09}}
 {id=WavePlayer_64, properties {soundWaveAssetPtr=&Spot_Neon_City_Arcade_Sounds_08}}
 {id=WavePlayer_65, properties {soundWaveAssetPtr=&Spot_Neon_City_Arcade_Sounds_07}}
 {id=WavePlayer_66, properties {soundWaveAssetPtr=&Spot_Neon_City_Arcade_Sounds_06}}
 {id=WavePlayer_67, properties {soundWaveAssetPtr=&Spot_Neon_City_Arcade_Sounds_05}}
 {id=WavePlayer_68, properties {soundWaveAssetPtr=&Spot_Neon_City_Arcade_Sounds_04}}
 {id=WavePlayer_69, properties {soundWaveAssetPtr=&Spot_Neon_City_Arcade_Sounds_03}}
 {id=WavePlayer_70, properties {soundWaveAssetPtr=&Spot_Neon_City_Arcade_Sounds_02}}
 {id=WavePlayer_71, properties {soundWaveAssetPtr=&Spot_Neon_City_Arcade_Sounds_01}}
]
properties {
 bEnableSubmixSends=false
 bOverrideAttenuation=true
 bOverrideConcurrency=true
 maxDistance=8500
 soundClassObject=&Fort_Ambient_OneShots
 volumeMultiplier=0.752
 attenuationOverrides {
  bAttenuateWithLPF=true
  bEnableLogFrequencyScaling=true
  distanceAlgorithm=Custom
  falloffDistance=8000
  hPFFrequencyAtMax=150
  lPFFrequencyAtMax=300
  lPFFrequencyAtMin=12000
  lPFRadiusMax=8000
  lPFRadiusMin=0
  stereoSpread=8000
  attenuationShapeExtents {x=500}
  customAttenuationCurve {editorCurveData {keys [
     <at 0.2 -0.0163 RCIM_Cubic>
     {arriveTangent=0.0203, interpMode=RCIM_Cubic, leaveTangent=0.0203, time=0.5, value=1}
     {arriveTangent=-0.0395, interpMode=RCIM_Cubic, leaveTangent=-0.0395, tangentMode=RCTM_User, time=1, value=0}
    ]}}
 }
 concurrencyOverrides {maxCount=3, retriggerTime=3}
}
