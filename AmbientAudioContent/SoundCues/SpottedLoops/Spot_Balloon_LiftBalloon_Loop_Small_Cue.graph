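# Looping sound cue for a small lift balloon's cloth ambience, using a small-balloon attenuation asset.
# A looping cloth wave feeds a quality switch that fills only the first two of four slots.
# Two nested endless Envelopers add variation from that single wave: a 30-second pitch swell inside a 60-second volume contour with two dips.
# The cue plays quieter and higher-pitched, with overridden concurrency allowing one instance.
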SoundCue /AmbientAudioContent/SoundCues/SpottedLoops/Spot_Balloon_LiftBalloon_Loop_Small_Cue
root=Enveloper_2
nodes [
 {
  id=Enveloper_1
  children [QualityLevel_0]
  properties {
   bLoop=true
   bLoopIndefinitely=true
   loopEnd=30
   pitchCurve {editorCurveData {keys [{interpMode=RCIM_Cubic, time=0}, <at 15.9468 0.9 RCIM_Cubic>, {interpMode=RCIM_Cubic, time=30}]}}
   volumeCurve {editorCurveData {keys [<at 0 1>]}}
  }
 }
 {
  id=Enveloper_2
  children [Enveloper_1]
  properties {
   bLoop=true
   bLoopIndefinitely=true
   loopEnd=60
   pitchCurve {editorCurveData {keys [{time=0}]}}
   volumeCurve {editorCurveData {keys [
      {arriveTangent=-0.01, interpMode=RCIM_Cubic, time=0, value=1}
      <at 10 0.2 RCIM_Cubic>
      {arriveTangent=0.0037, interpMode=RCIM_Cubic, leaveTangent=0.0037, tangentMode=RCTM_User, time=20, value=1}
      <at 30 0.1 RCIM_Cubic>
      <at 40 1>
      <at 60 1>
     ]}}
  }
 }
 {id=QualityLevel_0, children [WavePlayer_1,WavePlayer_1,null,null]}
 {
  id=WavePlayer_1
  properties {bLooping=true, soundWaveAssetPtr=/AmbientAudioContent/Waves/SpottedLoops/Spot_Balloon_LiftBalloon_Cloth_Loop_01}
 }
]
properties {
 attenuationSettings=/AmbientAudioBase/Attenuations/Spot_Balloon_LiftBalloon_Small_Att
 bOverrideConcurrency=true
 maxDistance=1750
 pitchMultiplier=1.2
 soundClassObject=&Fort_Gadgets_Items
 volumeMultiplier=0.3
 attenuationOverrides {attenuationShape=Capsule, falloffDistance=5000, reverbDistanceMin=1000, reverbWetLevelMax=3, attenuationShapeExtents {x=4000}}
 concurrencyOverrides {maxCount=1, voiceStealReleaseTime=5}
 concurrencySet [/Game/Sounds/Concurrency_Rules/ThreeMaximum_Owner_PreventNew]
}
