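# Sound cue for a barricade door sliding open, using door attenuation and the doors sound class.
# A random node picks between two equally weighted slide waves, and a modulator slightly varies their boosted volume.
# All four quality-level inputs share that one chain, and no looping node appears despite the Loop name.
# Output also feeds an exterior ambient source bus and the default reverb submix, with concurrency settings for spotted ambient sounds.
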SoundCue /AmbientAudioContent/SoundCues/SpottedLoops/Spot_BarricadeDoor_Slide_Open_Loop_Cue
root=QualityLevel_0
nodes [
 {id=Modulator_2, children [Random_1], properties {volumeMax=1.2, volumeMin=1.1}}
 {id=QualityLevel_0, children [Modulator_2,Modulator_2,Modulator_2,Modulator_2]}
 {id=Random_1, children [WavePlayer_3,WavePlayer_4], properties {weights [1,1]}}
 {id=WavePlayer_3, properties {soundWaveAssetPtr=/AmbientAudioContent/Waves/OneShots/Spot_Door_Barricade_Sliding_Open_01}}
 {id=WavePlayer_4, properties {soundWaveAssetPtr=/AmbientAudioContent/Waves/OneShots/Spot_Door_Barricade_Sliding_Open_02}}
]
properties {
 attenuationSettings=/Game/Sounds/Attenuation/Fort_Doors_Att
 maxDistance=2400
 soundClassObject=&Fort_Doors
 volumeMultiplier=1
 busSends [{audioBus=null, minSendLevel=0, sendLevel=1, soundSourceBus=&AmbientAudio_Exterior_SourceBus}]
 concurrencyOverrides {resolutionRule=PreventNew}
 concurrencySet [/AmbientAudioBase/Concurrency/AmbientConc_3Max_Spot_Default_Concurrency]
 soundSubmixSends [{minSendLevel=0.6, soundSubmix=&Fortnite_Reverb_Default}]
}
