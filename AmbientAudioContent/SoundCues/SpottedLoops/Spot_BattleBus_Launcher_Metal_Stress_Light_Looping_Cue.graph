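# Looping ambient SoundCue that plays intermittent light metal stress creaks, named for the Battle Bus launcher.
# Each loop waits a random 3–8 seconds, then picks one of two pitch modulators, slightly raised or lowered.
# Both modulators share one random pool of ten Tilted Towers metal stress one-shots, giving varied repeats from few assets.
# Capsule attenuation and concurrency overrides localize and limit it, with a send to the exterior ambient bus.
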SoundCue /AmbientAudioContent/SoundCues/SpottedLoops/Spot_BattleBus_Launcher_Metal_Stress_Light_Looping_Cue
root=Looping_0
nodes [
 {id=Delay_0, children [Random_0], properties {delayMax=8, delayMin=3}}
 {id=Looping_0, children [Delay_0]}
 {id=Modulator_0, children [Random_1], properties {pitchMax=1.1, pitchMin=1}}
 {id=Modulator_1, children [Random_1], properties {pitchMax=0.9, pitchMin=0.85}}
 {id=Random_0, children [Modulator_0,Modulator_1], properties {weights [1,1]}}
 {
  id=Random_1
  children [
   WavePlayer_0
   WavePlayer_1
   WavePlayer_2
   WavePlayer_13
   WavePlayer_14
   WavePlayer_15
   WavePlayer_16
   WavePlayer_17
   WavePlayer_18
   WavePlayer_19
  ]
  properties {weights [1,1,1,1,1,1,1,1,1,1]}
 }
 {id=WavePlayer_0, properties {soundWaveAssetPtr=/AmbientAudioContent/Waves/OneShots/Spot_Tilted_Towers_Metal_Stress_Light_01}}
 {id=WavePlayer_1, properties {soundWaveAssetPtr=/AmbientAudioContent/Waves/OneShots/Spot_Tilted_Towers_Metal_Stress_Light_02}}
 {id=WavePlayer_13, properties {soundWaveAssetPtr=/AmbientAudioContent/Waves/OneShots/Spot_Tilted_Towers_Metal_Stress_Light_04}}
 {id=WavePlayer_14, properties {soundWaveAssetPtr=/AmbientAudioContent/Waves/OneShots/Spot_Tilted_Towers_Metal_Stress_Light_05}}
 {id=WavePlayer_15, properties {soundWaveAssetPtr=/AmbientAudioContent/Waves/OneShots/Spot_Tilted_Towers_Metal_Stress_Light_06}}
 {id=WavePlayer_16, properties {soundWaveAssetPtr=/AmbientAudioContent/Waves/OneShots/Spot_Tilted_Towers_Metal_Stress_Light_07}}
 {id=WavePlayer_17, properties {soundWaveAssetPtr=/AmbientAudioContent/Waves/OneShots/Spot_Tilted_Towers_Metal_Stress_Light_08}}
 {id=WavePlayer_18, properties {soundWaveAssetPtr=/AmbientAudioContent/Waves/OneShots/Spot_Tilted_Towers_Metal_Stress_Light_09}}
 {id=WavePlayer_19, properties {soundWaveAssetPtr=/AmbientAudioContent/Waves/OneShots/Spot_Tilted_Towers_Metal_Stress_Light_10}}
 {id=WavePlayer_2, properties {soundWaveAssetPtr=/AmbientAudioContent/Waves/OneShots/Spot_Tilted_Towers_Metal_Stress_Light_03}}
]
properties {
 bEnableSubmixSends=false
 bOverrideAttenuation=true
 bOverrideConcurrency=true
 maxDistance=1500
 soundClassObject=&Fort_Ambient_OneShots
 volumeMultiplier=1
 attenuationOverrides {attenuationShape=Capsule, falloffDistance=500, attenuationShapeExtents=<1000 250>}
 busSends [{audioBus=null, minSendLevel=1, sendLevel=1, soundSourceBus=&AmbientAudio_Exterior_SourceBus}]
 concurrencyOverrides {maxCount=2}
}
