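# Looping ambient SoundCue of big wooden stress creaks, built from bridge corridor creak one-shots for spot placement with an interior creak attenuation.
# A loop waits a random few seconds, then plays a random creak, with a modulator varying pitch and volume.
# A quality-level node picks the pool: the first two levels draw from nine waves, the last two from a four-wave subset.
# It plays at low volume under the Fort_Doors sound class.
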SoundCue /AmbientAudioContent/SoundCues/SpottedLoops/Spot_Big_Wooden_Stress_Creaks_Loop_Cue
root=Modulator_0
nodes [
 {id=Delay_0, children [QualityLevel_0], properties {delayMax=8, delayMin=3}}
 {id=Looping_0, children [Delay_0]}
 {id=Modulator_0, children [Looping_0], properties {pitchMin=0.8, volumeMin=0.8}}
 {id=QualityLevel_0, children [Random_2,Random_2,Random_1,Random_1]}
 {id=Random_1, children [WavePlayer_9, WavePlayer_10, WavePlayer_11, WavePlayer_12], properties {weights [1,1,1,1]}}
 {
  id=Random_2
  children [WavePlayer_13, WavePlayer_14, WavePlayer_15, WavePlayer_16, WavePlayer_17, WavePlayer_18, WavePlayer_19, WavePlayer_20, WavePlayer_21]
  properties {weights [1,1,1,1,1,1,1,1,1]}
 }
 {id=WavePlayer_10, properties {soundWaveAssetPtr=/AmbientAudioContent/Waves/OneShots/Spot_Bridge_Wooden_Corridor_Creaks_01}}
 {id=WavePlayer_11, properties {soundWaveAssetPtr=/AmbientAudioContent/Waves/OneShots/Spot_Bridge_Wooden_Corridor_Creaks_02}}
 {id=WavePlayer_12, properties {soundWaveAssetPtr=/AmbientAudioContent/Waves/OneShots/Spot_Bridge_Wooden_Corridor_Creaks_03}}
 {id=WavePlayer_13, properties {soundWaveAssetPtr=&Spot_Bridge_Wooden_Corridor_Creaks_01}}
 {id=WavePlayer_14, properties {soundWaveAssetPtr=&Spot_Bridge_Wooden_Corridor_Creaks_02}}
 {id=WavePlayer_15, properties {soundWaveAssetPtr=&Spot_Bridge_Wooden_Corridor_Creaks_03}}
 {id=WavePlayer_16, properties {soundWaveAssetPtr=/AmbientAudioContent/Waves/OneShots/Spot_Bridge_Wooden_Corridor_Creaks_04}}
 {id=WavePlayer_17, properties {soundWaveAssetPtr=/AmbientAudioContent/Waves/OneShots/Spot_Bridge_Wooden_Corridor_Creaks_05}}
 {id=WavePlayer_18, properties {soundWaveAssetPtr=/AmbientAudioContent/Waves/OneShots/Spot_Bridge_Wooden_Corridor_Creaks_06}}
 {id=WavePlayer_19, properties {soundWaveAssetPtr=/AmbientAudioContent/Waves/OneShots/Spot_Bridge_Wooden_Corridor_Creaks_07}}
 {id=WavePlayer_20, properties {soundWaveAssetPtr=/AmbientAudioContent/Waves/OneShots/Spot_Bridge_Wooden_Corridor_Creaks_08}}
 {id=WavePlayer_21, properties {soundWaveAssetPtr=/AmbientAudioContent/Waves/OneShots/Spot_Bridge_Wooden_Corridor_Creaks_09}}
 {id=WavePlayer_9, properties {soundWaveAssetPtr=&Spot_Bridge_Wooden_Corridor_Creaks_09}}
]
properties {
 attenuationSettings=/Game/Sounds/Attenuation/Fort_TileAmbient_OneShot_Big_Interior_Creak_Att
 maxDistance=2641.8145
 soundClassObject=&Fort_Doors
 volumeMultiplier=0.15
}
