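# Sound cue for the ambient fire loop of a Brimstone wall brazier, in the Athena ambient sound class.
# A mixer layers two looping Coliseum brazier beds, close mono and stereo, with a half-volume random pick of three non-looping urn fire waves.
# The urn layer has its own close-range attenuation, separate from the cue's small-brazier attenuation.
# A root modulator varies pitch slightly, and a shared ambient concurrency set limits instances.
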SoundCue /AmbientAudioContent/SoundCues/SpottedLoops/Spot_Brimstone_Wall_Brazier_Fire_Loop_Cue
root=Modulator_0
nodes [
 {
  id=Attenuation_2
  children [Random_2]
  properties {attenuationSettings=/Game/Sounds/Asteria/Environment_Actors/Brazier_LightSource/Spot_Bazier_LightSource_Fire_Close_Att}
 }
 {id=Mixer_0, children [Attenuation_2, Mixer_1], properties {inputVolume [0.5,1]}}
 {id=Mixer_1, children [WavePlayer_1,WavePlayer_2], properties {inputVolume [1,1]}}
 {id=Modulator_0, children [Mixer_0], properties {pitchMax=1.1, pitchMin=0.9}}
 {id=Random_2, children [WavePlayer_9, WavePlayer_10, WavePlayer_11], properties {weights [1,1,1]}}
 {id=WavePlayer_1, properties {bLooping=true, soundWaveAssetPtr=/AmbientAudioContent/Waves/OneShots/Spot_Coliseum_Brazier_Fire_Close_Mono}}
 {id=WavePlayer_10, properties {soundWaveAssetPtr=/AmbientAudioContent/Waves/SpottedLoops/Spot_Fire_Urn_02}}
 {id=WavePlayer_11, properties {soundWaveAssetPtr=/AmbientAudioContent/Waves/SpottedLoops/Spot_Fire_Urn_03}}
 {id=WavePlayer_2, properties {bLooping=true, soundWaveAssetPtr=/AmbientAudioContent/Waves/OneShots/Spot_Coliseum_Brazier_Fire_Stereo}}
 {id=WavePlayer_9, properties {soundWaveAssetPtr=/AmbientAudioContent/Waves/SpottedLoops/Spot_Fire_Urn_01}}
]
properties {
 attenuationSettings=/Game/Sounds/Helios/Environment_Actors/Brazier/Fire_Brazier_Stand_Small_Att
 maxDistance=1275
 soundClassObject=&Fort_Ambient_Athena
 subtitlePriority=5000
 volumeMultiplier=0.401
 attenuationOverrides {
  bEnableFocusInterpolation=true
  bEnableListenerFocus=true
  bEnableOcclusion=true
  dBAttenuationAtMax=-40
  falloffDistance=512
  focusAzimuth=15
  focusVolumeAttenuation=1.5
  nonFocusAzimuth=45
  nonFocusVolumeAttenuation=0.5
  occlusionInterpolationTime=1
  occlusionLowPassFilterFrequency=10000
  occlusionTraceChannel=ECC_GameTraceChannel11
  occlusionVolumeAttenuation=0.5
  stereoSpread=512
  attenuationShapeExtents=<256 64.5942 90>
 }
 concurrencyOverrides {bLimitToOwner=true, maxCount=3}
 concurrencySet [/AmbientAudioBase/Concurrency/AmbientConc_3Max_Spot_Default_Concurrency]
}
