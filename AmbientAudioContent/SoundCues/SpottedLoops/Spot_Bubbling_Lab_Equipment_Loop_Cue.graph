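# Ambient spot sound cue for bubbling lab equipment, layering a looping fish tank hum with a randomly chosen viscous bubbling one-shot.
# Each layer gets pitch modulation before mixing, and a Random node routes the mix either directly or through a short delay.
# A root Enveloper loops indefinitely with a slow volume swell.
# Overridden attenuation adds logarithmic falloff, low-pass filtering and listener focus, with a reverb submix send and an ambient concurrency set.
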SoundCue /AmbientAudioContent/SoundCues/SpottedLoops/Spot_Bubbling_Lab_Equipment_Loop_Cue
root=Enveloper_4
nodes [
 {id=Delay_0, children [Mixer_0], properties {delayMax=0.3, delayMin=0.3}}
 {
  id=Enveloper_4
  children [Random_0]
  properties {
   bLoop=true
   bLoopIndefinitely=true
   loopEnd=20
   pitchCurve {editorCurveData {keys [{time=0}]}}
   volumeCurve {editorCurveData {keys [
      {arriveTangent=-0.01, interpMode=RCIM_Cubic, time=0, value=1}
      <at 10 1.2 RCIM_Cubic>
      {interpMode=RCIM_Cubic, leaveTangent=0.01, time=20, value=1}
     ]}}
  }
 }
 {id=Mixer_0, children [Modulator_0,Modulator_1], properties {inputVolume [1,1.5]}}
 {id=Modulator_0, children [Random_1], properties {pitchMax=1.1}}
 {id=Modulator_1, children [WavePlayer_0], properties {pitchMax=1.1}}
 {id=Random_0, children [Delay_0,Mixer_0], properties {weights [1,1]}}
 {id=Random_1, children [WavePlayer_6, WavePlayer_7, WavePlayer_8, WavePlayer_9, WavePlayer_10], properties {weights [1,1,1,1,1]}}
 {id=WavePlayer_0, properties {bLooping=true, soundWaveAssetPtr=/AmbientAudioContent/Waves/SpottedLoops/Spot_Fish_Tank_Hum_Ext_Loop}}
 {id=WavePlayer_10, properties {soundWaveAssetPtr=/AmbientAudioContent/Waves/OneShots/Spot_Liquid_Viscous_Flow_Bubbling_Slowly_B_01}}
 {id=WavePlayer_6, properties {soundWaveAssetPtr=/AmbientAudioContent/Waves/OneShots/Spot_Liquid_Viscous_Flow_Bubbling_Slowly_B_05}}
 {id=WavePlayer_7, properties {soundWaveAssetPtr=/AmbientAudioContent/Waves/OneShots/Spot_Liquid_Viscous_Flow_Bubbling_Slowly_B_04}}
 {id=WavePlayer_8, properties {soundWaveAssetPtr=/AmbientAudioContent/Waves/OneShots/Spot_Liquid_Viscous_Flow_Bubbling_Slowly_B_03}}
 {id=WavePlayer_9, properties {soundWaveAssetPtr=/AmbientAudioContent/Waves/OneShots/Spot_Liquid_Viscous_Flow_Bubbling_Slowly_B_02}}
]
properties {
 bOverrideAttenuation=true
 maxDistance=750
 soundClassObject=&Fort_Ambient_Athena
 volumeMultiplier=0.25
 attenuationOverrides {
  bAttenuateWithLPF=true
  bEnableListenerFocus=true
  distanceAlgorithm=Logarithmic
  falloffDistance=750
  focusAzimuth=15
  lPFFrequencyAtMax=8000
  lPFRadiusMax=1000
  lPFRadiusMin=500
  nonFocusAzimuth=45
  nonFocusVolumeAttenuation=0.4
  stereoSpread=50
  attenuationShapeExtents {x=0}
 }
 concurrencyOverrides {maxCount=2, voiceStealReleaseTime=2}
 concurrencySet [/AmbientAudioBase/Concurrency/AmbientConc_3Max_Spot_Default_Concurrency]
 soundSubmixSends [{maxSendLevel=1.2, minSendLevel=1, sendLevel=1.2, soundSubmix=&Fortnite_Reverb_Default}]
}
