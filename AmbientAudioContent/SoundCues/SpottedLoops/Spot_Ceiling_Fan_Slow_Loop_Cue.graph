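# Sound cue for a slow ceiling fan ambient loop, used as a spotted prop sound in the Athena ambient props sound class.
# A Random node with equal weights picks one of three looping wave variations.
# It uses a standing electric fan attenuation asset at reduced volume.
# A concurrency override caps it at two voices with a timed release on voice stealing, alongside an Athena prop concurrency set reference.
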SoundCue /AmbientAudioContent/SoundCues/SpottedLoops/Spot_Ceiling_Fan_Slow_Loop_Cue
root=Random_0
nodes [
 {id=Random_0, children [WavePlayer_1,WavePlayer_2,WavePlayer_4], properties {weights [1,1,1]}}
 {id=WavePlayer_1, properties {bLooping=true, soundWaveAssetPtr=/AmbientAudioContent/Waves/SpottedLoops/Spot_CeilingFan_Slow_Loop_01}}
 {id=WavePlayer_2, properties {bLooping=true, soundWaveAssetPtr=/AmbientAudioContent/Waves/SpottedLoops/Spot_CeilingFan_Slow_Loop_02}}
 {id=WavePlayer_4, properties {bLooping=true, soundWaveAssetPtr=/AmbientAudioContent/Waves/SpottedLoops/Spot_CeilingFan_Slow_Loop_03}}
]
properties {
 attenuationSettings=/AmbientAudioBase/Attenuations/Spot_Standing_Electric_Fan_Att
 bEnableBusSends=false
 bOverrideConcurrency=true
 maxDistance=750
 soundClassObject=&Fort_Ambient_Props_Athena
 volumeMultiplier=0.4
 concurrencyOverrides {maxCount=2, voiceStealReleaseTime=2}
 concurrencySet [/Game/Sounds/Concurrency_Rules/Athena_Prop]
 preEffectBusSends [{}]
}
