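# Sound cue for a looping centrifuge pump spot ambience, played quietly and pitched down in the Fortnite Athena ambient sound class.
# A random node starts the looping wave immediately or after a short fixed delay, through an enveloper fade-in and a modulator.
# All quality levels share one attenuation node, which overrides falloff with a capsule shape, custom curve and listener focus.
# The cue limits simultaneous voices and sends to the default reverb submix.
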SoundCue /AmbientAudioContent/SoundCues/SpottedLoops/Spot_Centrifuge_Pump_Loop_04_Cue
root=QualityLevel_0
nodes [
 {
  id=Attenuation_0
  children [Random_1]
  properties {
   bOverrideAttenuation=true
   attenuationOverrides {
    attenuationShape=Capsule
    bEnableListenerFocus=true
    distanceAlgorithm=Custom
    falloffDistance=2000
    nonFocusVolumeAttenuation=0.75
    occlusionInterpolationTime=2
    occlusionLowPassFilterFrequency=5000
    occlusionTraceChannel=ECC_GameTraceChannel11
    occlusionVolumeAttenuation=0.5
    stereoSpread=100
    attenuationShapeExtents {x=250}
    customAttenuationCurve {editorCurveData {keys [
       <at 0 1>
       {arriveTangent=-1, interpMode=RCIM_Cubic, leaveTangent=-1, time=0.4986, value=0.2546}
       {arriveTangent=-1.0909, interpMode=RCIM_Cubic, leaveTangent=-1.0909, tangentMode=RCTM_User, time=1, value=0}
      ]}}
    submixSendSettings [{
      bEnableHPFCutoff=false
      bEnableLPFCutoff=false
      disableManualSendClamp=false
      hPFCutoff=20
      lPFCutoff=20000
      maxSendDistance=6000
      maxSendLevel=1
      minSendDistance=400
      minSendLevel=0
      sendLevel=0.2
      sendLevelControlMethod=Linear
      soundSubmix=null
     }]
   }
  }
 }
 {id=Delay_0, children [Enveloper_1], properties {delayMax=0.3, delayMin=0.3}}
 {
  id=Enveloper_1
  children [Modulator_0]
  properties {pitchCurve {editorCurveData {keys [{time=0}]}}, volumeCurve {editorCurveData {keys [<at 0 0>,<at 1 1>]}}}
 }
 {id=Modulator_0, children [WavePlayer_0]}
 {id=QualityLevel_0, children [Attenuation_0, Attenuation_0, Attenuation_0, Attenuation_0]}
 {id=Random_1, children [Delay_0,Enveloper_1], properties {weights [1,1]}}
 {id=WavePlayer_0, properties {bLooping=true, soundWaveAssetPtr=/AmbientAudioContent/Waves/SpottedLoops/Spot_Centrifuge_Pump_Loop_04}}
]
properties {
 bEnableBusSends=false
 bOverrideConcurrency=true
 maxDistance=2250
 pitchMultiplier=0.8
 soundClassObject=&Fort_Ambient_Athena
 volumeMultiplier=0.2
 attenuationOverrides {
  bEnableListenerFocus=true
  distanceAlgorithm=Logarithmic
  falloffDistance=1000
  nonFocusVolumeAttenuation=0.75
  stereoSpread=100
  attenuationShapeExtents {x=250}
 }
 concurrencyOverrides {maxCount=2, voiceStealReleaseTime=0.2}
 concurrencySet [/AmbientAudioBase/Concurrency/AmbientConc_3Max_Spot_Default_Concurrency]
 soundSubmixSends [{
   maxSendDistance=2000
   minSendLevel=0
   soundSubmix=&Fortnite_Reverb_Default
   customSendLevelCurve {editorCurveData {keys [<at 0 0>,<at 1 1>]}}
  }]
}
