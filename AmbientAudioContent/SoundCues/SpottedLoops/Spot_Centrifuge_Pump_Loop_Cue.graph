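# Ambient spot sound cue that plays a looping centrifuge pump, pitched down, under the Athena ambient sound class.
# A mixer blends a near loop with a half-level distant loop, each with custom attenuation curves that crossfade them by distance.
# Each layer randomly starts immediately or after a short delay, then passes through a fade-in enveloper and modulator for variation.
# All quality levels share that mixer, with capped concurrency and a reverb send.
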SoundCue /AmbientAudioContent/SoundCues/SpottedLoops/Spot_Centrifuge_Pump_Loop_Cue
root=QualityLevel_0
nodes [
 {
  id=Attenuation_0
  children [Random_1]
  properties {
   bOverrideAttenuation=true
   attenuationOverrides {
    bEnableListenerFocus=true
    distanceAlgorithm=Custom
    falloffDistance=2000
    nonFocusVolumeAttenuation=0.75
    attenuationShapeExtents {x=250}
    customAttenuationCurve {editorCurveData {keys [<at 0 1>,<at 0.4985 0.2>,<at 1 0>]}}
    submixSendSettings [{
      bEnableHPFCutoff=false
      bEnableLPFCutoff=false
      disableManualSendClamp=false
      hPFCutoff=20
      lPFCutoff=20000
      maxSendDistance=6000
      maxSendLevel=1
      minSendDistance=400
      minSendLevel=0
      sendLevel=0.2
      sendLevelControlMethod=Linear
      soundSubmix=null
     }]
   }
  }
 }
 {
  id=Attenuation_2
  children [Random_3]
  properties {
   bOverrideAttenuation=true
   attenuationOverrides {
    bEnableListenerFocus=true
    distanceAlgorithm=Custom
    falloffDistance=2000
    nonFocusVolumeAttenuation=0.4
    stereoSpread=800
    attenuationShapeExtents {x=250}
    customAttenuationCurve {editorCurveData {keys [<at 0 0>,<at 0.2 0>,<at 0.4985 1 RCIM_Cubic>,<at 1 0 RCIM_Cubic>]}}
    submixSendSettings [{
      bEnableHPFCutoff=false
      bEnableLPFCutoff=false
      disableManualSendClamp=false
      hPFCutoff=20
      lPFCutoff=20000
      maxSendDistance=6000
      maxSendLevel=1
      minSendDistance=400
      minSendLevel=0
      sendLevel=0.2
      sendLevelControlMethod=Linear
      soundSubmix=null
     }]
   }
  }
 }
 {id=Delay_0, children [Enveloper_1], properties {delayMax=0.3, delayMin=0.3}}
 {id=Delay_1, children [Enveloper_0], properties {delayMax=0.3, delayMin=0.3}}
 {
  id=Enveloper_0
  children [Modulator_1]
  properties {pitchCurve {editorCurveData {keys [{time=0}]}}, volumeCurve {editorCurveData {keys [<at 0 0>,<at 1 1>]}}}
 }
 {
  id=Enveloper_1
  children [Modulator_0]
  properties {pitchCurve {editorCurveData {keys [{time=0}]}}, volumeCurve {editorCurveData {keys [<at 0 0>,<at 1 1>]}}}
 }
 {id=Mixer_0, children [Attenuation_0, Attenuation_2], properties {inputVolume [1,0.5]}}
 {id=Modulator_0, children [WavePlayer_1]}
 {id=Modulator_1, children [WavePlayer_2], properties {pitchMax=1.1, pitchMin=1}}
 {id=QualityLevel_0, children [Mixer_0,Mixer_0,Mixer_0,Mixer_0]}
 {id=Random_1, children [Delay_0,Enveloper_1], properties {weights [1,1]}}
 {id=Random_3, children [Delay_1,Enveloper_0], properties {weights [1,1]}}
 {id=WavePlayer_1, properties {bLooping=true, soundWaveAssetPtr=/AmbientAudioContent/Waves/SpottedLoops/Spot_Centrifuge_Pump_Loop_01}}
 {id=WavePlayer_2, properties {bLooping=true, soundWaveAssetPtr=/AmbientAudioContent/Waves/SpottedLoops/Spot_Centrifuge_Pump_Loop_Distant_01}}
]
properties {
 bEnableBusSends=false
 bOverrideConcurrency=true
 maxDistance=2250
 pitchMultiplier=0.6
 soundClassObject=&Fort_Ambient_Athena
 volumeMultiplier=0.4
 attenuationOverrides {
  bEnableListenerFocus=true
  distanceAlgorithm=Logarithmic
  falloffDistance=1000
  nonFocusVolumeAttenuation=0.75
  stereoSpread=100
  attenuationShapeExtents {x=250}
 }
 concurrencyOverrides {maxCount=2, voiceStealReleaseTime=0.2}
 concurrencySet [/AmbientAudioBase/Concurrency/AmbientConc_3Max_Spot_Default_Concurrency]
 soundSubmixSends [{
   maxSendDistance=2000
   minSendLevel=0
   soundSubmix=&Fortnite_Reverb_Default
   customSendLevelCurve {editorCurveData {keys [<at 0 0>,<at 1 1>]}}
  }]
}
