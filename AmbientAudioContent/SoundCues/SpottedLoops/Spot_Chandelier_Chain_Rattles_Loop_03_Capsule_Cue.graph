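# Looping ambient spot SoundCue that plays randomized chandelier chain rattles at irregular, delayed intervals.
# Each loop mixes an enveloped, pitch- and volume-varied slow chain rattle with a quieter chandelier rattle, both randomly picked.
# Each layer has capsule attenuation with occlusion, listener focus and reverb sends; the chandelier layer reaches farther and stays silent up close.
# Only the first two quality levels are connected, and a concurrency override limits instances.
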SoundCue /AmbientAudioContent/SoundCues/SpottedLoops/Spot_Chandelier_Chain_Rattles_Loop_03_Capsule_Cue
root=QualityLevel_0
nodes [
 {
  id=Attenuation_1
  children [Modulator_0]
  properties {
   bOverrideAttenuation=true
   attenuationOverrides {
    attenuationShape=Capsule
    bEnableFocusInterpolation=true
    bEnableListenerFocus=true
    bEnableOcclusion=true
    bEnableSubmixSends=true
    distanceAlgorithm=Custom
    falloffDistance=1500
    focusReleaseInterpSpeed=0.25
    nonFocusVolumeAttenuation=0.4
    occlusionInterpolationTime=2
    occlusionLowPassFilterFrequency=950
    occlusionTraceChannel=ECC_GameTraceChannel11
    occlusionVolumeAttenuation=0.75
    attenuationShapeExtents {x=300}
    customAttenuationCurve {editorCurveData {keys [<at 0.0051 0.4771 RCIM_Cubic>, {arriveTangent=-0.4796, interpMode=RCIM_Cubic, leaveTangent=-0.4796, time=0.4, value=1}, <at 1 0>]}}
    submixSendSettings [
     {
      bEnableHPFCutoff=false
      bEnableLPFCutoff=false
      disableManualSendClamp=false
      hPFCutoff=20
      lPFCutoff=20000
      maxSendDistance=1500
      maxSendLevel=1
      minSendDistance=500
      minSendLevel=0.2
      sendLevel=1
      sendLevelControlMethod=Linear
      soundSubmix=&Fortnite_Reverb_Default
      customSendLevelCurve {editorCurveData {keys [<at 0 0.2>,<at 1 1>]}}
     }
     {
      bEnableHPFCutoff=false
      bEnableLPFCutoff=false
      disableManualSendClamp=false
      hPFCutoff=20
      lPFCutoff=20000
      maxSendDistance=1500
      maxSendLevel=0.4
      minSendDistance=400
      minSendLevel=0
      sendLevel=0.2
      sendLevelControlMethod=Linear
      soundSubmix=&Fortnite_Reverb_Distant
      customSendLevelCurve {externalCurve=/Game/Sounds/SubMixes/Reverb/Fortnite_Reverb_Distant_Send}
     }
    ]
   }
  }
 }
 {
  id=Attenuation_2
  children [Modulator_1]
  properties {
   bOverrideAttenuation=true
   attenuationOverrides {
    attenuationShape=Capsule
    bEnableFocusInterpolation=true
    bEnableListenerFocus=true
    bEnableOcclusion=true
    bEnableSubmixSends=true
    distanceAlgorithm=Custom
    falloffDistance=3000
    focusReleaseInterpSpeed=0.25
    nonFocusVolumeAttenuation=0.5
    nonSpatializedRadiusStart=2000
    occlusionInterpolationTime=2
    occlusionLowPassFilterFrequency=950
    occlusionTraceChannel=ECC_GameTraceChannel11
    occlusionVolumeAttenuation=0.75
    attenuationShapeExtents {x=300}
    customAttenuationCurve {editorCurveData {keys [
       <at 0.0051 0 RCIM_Cubic>
       {arriveTangent=-0.0064, interpMode=RCIM_Cubic, leaveTangent=-0.0064, tangentMode=RCTM_User, time=0.2, value=0}
       <at 0.4 1 RCIM_Cubic>
       <at 1 0>
      ]}}
    submixSendSettings [
     {
      bEnableHPFCutoff=false
      bEnableLPFCutoff=false
      disableManualSendClamp=false
      hPFCutoff=20
      lPFCutoff=20000
      maxSendDistance=1500
      maxSendLevel=1
      minSendDistance=500
      minSendLevel=0.2
      sendLevel=1
      sendLevelControlMethod=Linear
      soundSubmix=&Fortnite_Reverb_Default
      customSendLevelCurve {editorCurveData {keys [<at 0 0.2>,<at 1 1>]}}
     }
     {
      bEnableHPFCutoff=false
      bEnableLPFCutoff=false
      disableManualSendClamp=false
      hPFCutoff=20
      lPFCutoff=20000
      maxSendDistance=1500
      maxSendLevel=0.4
      minSendDistance=400
      minSendLevel=0
      sendLevel=0.2
      sendLevelControlMethod=Linear
      soundSubmix=&Fortnite_Reverb_Distant
      customSendLevelCurve {externalCurve=&Fortnite_Reverb_Distant_Send}
     }
    ]
   }
  }
 }
 {id=Delay_0, children [Mixer_0], properties {delayMax=35, delayMin=6}}
 {
  id=Enveloper_0
  children [Random_2]
  properties {pitchCurve {editorCurveData {keys [{time=0}]}}, volumeCurve {editorCurveData {keys [<at 0 1>,<at 3 1>,<at 4 0>]}}}
 }
 {id=Looping_0, children [Delay_0]}
 {id=Mixer_0, children [Attenuation_1, Attenuation_2], properties {inputVolume [1,0.5]}}
 {id=Modulator_0, children [Enveloper_0], properties {pitchMin=0.8, volumeMin=0.8}}
 {id=Modulator_1, children [Random_3], properties {pitchMax=1, pitchMin=1}}
 {id=QualityLevel_0, children [Looping_0,Looping_0,null,null]}
 {
  id=Random_2
  children [WavePlayer_14, WavePlayer_15, WavePlayer_16, WavePlayer_17, WavePlayer_18, WavePlayer_19, WavePlayer_20, WavePlayer_21]
  properties {weights [1,1,1,1,1,1,1,1]}
 }
 {
  id=Random_3
  children [WavePlayer_22, WavePlayer_23, WavePlayer_24, WavePlayer_25, WavePlayer_26, WavePlayer_27]
  properties {weights [1,1,1,1,1,1]}
 }
 {id=WavePlayer_14, properties {soundWaveAssetPtr=/AmbientAudioContent/Waves/OneShots/Spot_ChainRattle_Slow_01}}
 {id=WavePlayer_15, properties {soundWaveAssetPtr=/AmbientAudioContent/Waves/OneShots/Spot_ChainRattle_Slow_02}}
 {id=WavePlayer_16, properties {soundWaveAssetPtr=/AmbientAudioContent/Waves/OneShots/Spot_ChainRattle_Slow_03}}
 {id=WavePlayer_17, properties {soundWaveAssetPtr=/AmbientAudioContent/Waves/OneShots/Spot_ChainRattle_Slow_04}}
 {id=WavePlayer_18, properties {soundWaveAssetPtr=/AmbientAudioContent/Waves/OneShots/Spot_ChainRattle_Slow_05}}
 {id=WavePlayer_19, properties {soundWaveAssetPtr=/AmbientAudioContent/Waves/OneShots/Spot_ChainRattle_Slow_06}}
 {id=WavePlayer_20, properties {soundWaveAssetPtr=/AmbientAudioContent/Waves/OneShots/Spot_ChainRattle_Slow_07}}
 {id=WavePlayer_21, properties {soundWaveAssetPtr=/AmbientAudioContent/Waves/OneShots/Spot_ChainRattle_Slow_08}}
 {id=WavePlayer_22, properties {soundWaveAssetPtr=/AmbientAudioContent/Waves/OneShots/Spot_Nitro_Wasteland_Chandelier_Rattle_6}}
 {id=WavePlayer_23, properties {soundWaveAssetPtr=/AmbientAudioContent/Waves/OneShots/Spot_Nitro_Wasteland_Chandelier_Rattle_5}}
 {id=WavePlayer_24, properties {soundWaveAssetPtr=/AmbientAudioContent/Waves/OneShots/Spot_Nitro_Wasteland_Chandelier_Rattle_4}}
 {id=WavePlayer_25, properties {soundWaveAssetPtr=/AmbientAudioContent/Waves/OneShots/Spot_Nitro_Wasteland_Chandelier_Rattle_3}}
 {id=WavePlayer_26, properties {soundWaveAssetPtr=/AmbientAudioContent/Waves/OneShots/Spot_Nitro_Wasteland_Chandelier_Rattle_2}}
 {id=WavePlayer_27, properties {soundWaveAssetPtr=/AmbientAudioContent/Waves/OneShots/Spot_Nitro_Wasteland_Chandelier_Rattle_1}}
]
properties {
 bOverrideConcurrency=true
 maxDistance=3300
 soundClassObject=/Game/Sounds/FortSoundClassesAndModes/Fort_Earthquakes
 volumeMultiplier=0.4
 attenuationOverrides {
  attenuationShape=Capsule
  bAttenuate=false
  bAttenuateWithLPF=true
  bEnableOcclusion=true
  distanceAlgorithm=Custom
  falloffDistance=2500
  lPFFrequencyAtMax=1000
  lPFRadiusMax=3000
  lPFRadiusMin=1000
  occlusionInterpolationTime=3
  occlusionLowPassFilterFrequency=1000
  occlusionTraceChannel=ECC_GameTraceChannel11
  occlusionVolumeAttenuation=0.5
  stereoSpread=100
  attenuationShapeExtents=<300 50>
  customAttenuationCurve {editorCurveData {keys [
     <at 0 0>
     {arriveTangent=2.5, interpMode=RCIM_Cubic, leaveTangent=2.5, time=0.25, value=0.0018}
     <at 0.4 1>
     {arriveTangent=-0.0071, interpMode=RCIM_Cubic, leaveTangent=-0.0071, tangentMode=RCTM_User, time=0.6, value=1}
     <at 1 0>
    ]}}
 }
 busSends [{audioBus=null, minSendLevel=1, sendLevel=1, soundSourceBus=&AmbientAudio_Exterior_SourceBus}]
 concurrencyOverrides {maxCount=2}
 concurrencySet [/AmbientAudioBase/Concurrency/AmbientConc_3Max_Spot_Default_Concurrency]
}
