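# Ambient spot sound cue for a grungy, flickering city fluorescent tube light, with capsule attenuation and reverb send.
# A mixer blends a random neon loop, a grungy hum loop with slow oscillating volume, and random bulb flickers.
# Flickers retrigger through a short delay loop with pitch and volume variation, under their own tighter attenuation.
# Only the first two quality levels play the mix; the other two are empty.
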SoundCue /AmbientAudioContent/SoundCues/SpottedLoops/Spot_City_Fluorescent_Tube_Light_Hum_Grungy_02_Flicker_Caspule_Loop_Cue
root=QualityLevel_0
nodes [
 {
  id=Attenuation_1
  children [Looping_1]
  properties {
   bOverrideAttenuation=true
   attenuationOverrides {
    attenuationShape=Capsule
    bAttenuateWithLPF=true
    bEnableListenerFocus=true
    bEnableLogFrequencyScaling=true
    distanceAlgorithm=Logarithmic
    falloffDistance=1200
    hPFFrequencyAtMax=1000
    hPFFrequencyAtMin=300
    lPFRadiusMax=2000
    lPFRadiusMin=300
    nonFocusVolumeAttenuation=0
    stereoSpread=0
    attenuationShapeExtents=<600 400>
   }
  }
 }
 {id=Delay_0, children [Modulator_1], properties {delayMax=0.2, delayMin=0.2}}
 {id=Looping_1, children [Delay_0]}
 {id=Mixer_0, children [Attenuation_1, Modulator_0, Oscillator_0], properties {inputVolume [0.5,0.616,0.12]}}
 {id=Modulator_0, children [Random_0], properties {pitchMin=1}}
 {id=Modulator_1, children [Random_1], properties {pitchMax=1.2, pitchMin=1.1, volumeMin=0.7}}
 {
  id=Oscillator_0
  children [WavePlayer_1]
  properties {amplitudeMax=0.3, amplitudeMin=0.2, bModulateVolume=true, centerMax=1, centerMin=1, frequencyMax=0.1, frequencyMin=0.1}
 }
 {id=QualityLevel_0, children [Mixer_0,Mixer_0,null,null]}
 {id=Random_0, children [WavePlayer_0,WavePlayer_2,WavePlayer_3], properties {bShouldExcludeFromBranchCulling=true, weights [1,1,1]}}
 {id=Random_1, children [WavePlayer_6,WavePlayer_7,WavePlayer_8], properties {bShouldExcludeFromBranchCulling=true, weights [1,1,1]}}
 {id=WavePlayer_0, properties {bLooping=true, soundWaveAssetPtr=/AmbientAudioContent/Waves/SpottedLoops/Spot_City_NeonLight_01}}
 {
  id=WavePlayer_1
  properties {bLooping=true, soundWaveAssetPtr=/AmbientAudioContent/Waves/SpottedLoops/Spot_Fluorescent_Tube_Light_Hum_Grungy_01_Loop}
 }
 {id=WavePlayer_2, properties {bLooping=true, soundWaveAssetPtr=/AmbientAudioContent/Waves/SpottedLoops/Spot_City_NeonLight_02}}
 {id=WavePlayer_3, properties {bLooping=true, soundWaveAssetPtr=/AmbientAudioContent/Waves/SpottedLoops/Spot_City_NeonLight_03}}
 {id=WavePlayer_6, properties {soundWaveAssetPtr=/Game/Sounds/Asteria/Environment_Actors/NeonLightFlicker/Spot_Bulb_Light_Flicker_Single_03}}
 {id=WavePlayer_7, properties {soundWaveAssetPtr=/Game/Sounds/Asteria/Environment_Actors/NeonLightFlicker/Spot_Bulb_Light_Flicker_Single_04}}
 {id=WavePlayer_8, properties {soundWaveAssetPtr=/Game/Sounds/Asteria/Environment_Actors/NeonLightFlicker/Spot_Bulb_Light_Flicker_Single_05}}
]
properties {
 bOverrideAttenuation=true
 maxDistance=2287.812
 priority=10
 soundClassObject=&Fort_Ambient_Athena
 volumeMultiplier=1.2
 attenuationOverrides {
  attenuationShape=Capsule
  bAttenuateWithLPF=true
  bEnableListenerFocus=true
  distanceAlgorithm=Logarithmic
  falloffDistance=1687.8119
  focusAzimuth=20
  focusVolumeAttenuation=1.2
  hPFFrequencyAtMax=1000
  hPFFrequencyAtMin=300
  lPFRadiusMax=2000
  lPFRadiusMin=100
  nonFocusAzimuth=40
  nonFocusVolumeAttenuation=0.4
  attenuationShapeExtents=<600 400>
 }
 concurrencyOverrides {maxCount=1, voiceStealReleaseTime=3}
 concurrencySet [/AmbientAudioBase/Concurrency/AmbientConc_3Max_Spot_Default_Concurrency]
 soundSubmixSends [{minSendLevel=1, soundSubmix=&Fortnite_Reverb_Default}]
}
