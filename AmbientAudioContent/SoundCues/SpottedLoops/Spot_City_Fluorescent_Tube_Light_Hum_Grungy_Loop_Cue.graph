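# Sound cue for a spotted ambient loop of a grungy fluorescent tube light hum in city settings, on the Athena ambient sound class.
# The root randomly starts immediately or after a short fixed delay, likely to desync instances, into a quality node with two of four slots filled.
# Each filled slot shares a random pick between two looping waves, each slowly volume-modulated by an oscillator.
# Overrides add short-range logarithmic attenuation with filtering and listener focus, alongside concurrency limiting and two reverb sends.
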SoundCue /AmbientAudioContent/SoundCues/SpottedLoops/Spot_City_Fluorescent_Tube_Light_Hum_Grungy_Loop_Cue
root=Random_1
nodes [
 {id=Delay_0, children [QualityLevel_0], properties {delayMax=0.35, delayMin=0.35}}
 {
  id=Oscillator_0
  children [WavePlayer_1]
  properties {amplitudeMax=0.3, amplitudeMin=0.2, bModulateVolume=true, centerMax=1, centerMin=1, frequencyMax=0.1, frequencyMin=0.1}
 }
 {
  id=Oscillator_1
  children [WavePlayer_0]
  properties {amplitudeMax=0.3, amplitudeMin=0.2, bModulateVolume=true, centerMax=1, centerMin=1, frequencyMax=0.1, frequencyMin=0.1}
 }
 {id=QualityLevel_0, children [Random_0,Random_0,null,null]}
 {id=Random_0, children [Oscillator_1,Oscillator_0], properties {weights [1,1]}}
 {id=Random_1, children [Delay_0, QualityLevel_0], properties {weights [1,1]}}
 {
  id=WavePlayer_0
  properties {bLooping=true, soundWaveAssetPtr=/AmbientAudioContent/Waves/SpottedLoops/Spot_Fluorescent_Tube_Light_Hum_Grungy_01_Loop}
 }
 {
  id=WavePlayer_1
  properties {bLooping=true, soundWaveAssetPtr=/AmbientAudioContent/Waves/SpottedLoops/Spot_Fluorescent_Tube_Light_Hum_Grungy_02_Loop}
 }
]
properties {
 bOverrideAttenuation=true
 maxDistance=900
 priority=10
 soundClassObject=&Fort_Ambient_Athena
 volumeMultiplier=0.2
 attenuationOverrides {
  bAttenuateWithLPF=true
  bEnableListenerFocus=true
  distanceAlgorithm=Logarithmic
  falloffDistance=750
  focusAzimuth=20
  focusDistanceScale=0.9
  hPFFrequencyAtMax=1000
  hPFFrequencyAtMin=300
  lPFRadiusMax=2000
  lPFRadiusMin=100
  nonFocusAzimuth=40
  nonFocusVolumeAttenuation=0.75
  attenuationShapeExtents {x=150}
 }
 concurrencyOverrides {maxCount=1, voiceStealReleaseTime=3}
 concurrencySet [/AmbientAudioBase/Concurrency/AmbientConc_2Max_Spot_Concurrency]
 soundSubmixSends [
  {minSendLevel=1, soundSubmix=&Fortnite_Reverb_Default}
  {
   maxSendDistance=3000
   minSendDistance=0
   minSendLevel=0
   sendLevelControlMethod=CustomCurve
   soundSubmix=&Fortnite_Reverb_Distant
   customSendLevelCurve {editorCurveData {keys [<at 0 0>,<at 1 0.25>]}}
  }
 ]
}
