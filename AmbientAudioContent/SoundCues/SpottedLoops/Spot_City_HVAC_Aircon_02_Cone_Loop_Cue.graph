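# Looping spot ambience cue for a city HVAC air conditioning unit, using cone attenuation and the Athena ambient sound class.
# A quality-level node routes its first two slots and last two slots to separate Modulator branches, both looping the same 02a wave.
# An Enveloper at the root ramps volume up at the start.
# The cue plays at reduced volume with a small concurrency limit overridden on the cue itself.
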SoundCue /AmbientAudioContent/SoundCues/SpottedLoops/Spot_City_HVAC_Aircon_02_Cone_Loop_Cue
root=Enveloper_0
nodes [
 {
  id=Enveloper_0
  children [QualityLevel_0]
  properties {pitchCurve {editorCurveData {keys [{time=0}]}}, volumeCurve {editorCurveData {keys [<at 0 0>,<at 1 1>]}}}
 }
 {id=Modulator_0, children [WavePlayer_3]}
 {id=Modulator_1, children [WavePlayer_5]}
 {id=QualityLevel_0, children [Modulator_0,Modulator_0,Modulator_1,Modulator_1]}
 {id=WavePlayer_3, properties {bLooping=true, soundWaveAssetPtr=/AmbientAudioContent/Waves/SpottedLoops/Spot_City_HVAC_Aircon_Unit_Loop_02a}}
 {id=WavePlayer_5, properties {bLooping=true, soundWaveAssetPtr=&Spot_City_HVAC_Aircon_Unit_Loop_02a}}
]
properties {
 attenuationSettings=/AmbientAudioBase/Attenuations/ATT_Spot_City_HVAC_Aircon_Cone
 bOverrideConcurrency=true
 maxDistance=1375
 soundClassObject=&Fort_Ambient_Athena
 volumeMultiplier=0.25
 attenuationOverrides {attenuationShapeExtents {x=0}}
 concurrencyOverrides {maxCount=3}
}
