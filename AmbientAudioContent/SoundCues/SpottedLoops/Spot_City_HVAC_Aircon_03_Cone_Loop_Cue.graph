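# Looping ambient Sound Cue for a city HVAC air-conditioner prop, using cone attenuation and concurrency capped at two voices.
# An Enveloper ramps volume in, then a Quality Level node selects the layering.
# The first two levels mix randomly chosen duct-air loops at half volume with randomly chosen unit loops, each layer passing through a Modulator for variation.
# The last two levels play only one modulated unit loop as a cheaper version.
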SoundCue /AmbientAudioContent/SoundCues/SpottedLoops/Spot_City_HVAC_Aircon_03_Cone_Loop_Cue
root=Enveloper_0
nodes [
 {
  id=Enveloper_0
  children [QualityLevel_0]
  properties {pitchCurve {editorCurveData {keys [{time=0}]}}, volumeCurve {editorCurveData {keys [<at 0 0>,<at 1 1>]}}}
 }
 {id=Mixer_0, children [Modulator_2,Modulator_0], properties {inputVolume [0.5,1]}}
 {id=Modulator_0, children [Random_0]}
 {id=Modulator_1, children [WavePlayer_5]}
 {id=Modulator_2, children [Random_1]}
 {id=QualityLevel_0, children [Mixer_0,Mixer_0,Modulator_1,Modulator_1]}
 {id=Random_0, children [WavePlayer_3,WavePlayer_1], properties {weights [1,1]}}
 {id=Random_1, children [WavePlayer_0,WavePlayer_2], properties {weights [1,1]}}
 {id=WavePlayer_0, properties {bLooping=true, soundWaveAssetPtr=/AmbientAudioContent/Waves/SpottedLoops/Spot_AC_Duct_Air_03}}
 {id=WavePlayer_1, properties {bLooping=true, soundWaveAssetPtr=/AmbientAudioContent/Waves/SpottedLoops/Spot_City_HVAC_Aircon_Unit_Loop_03b}}
 {id=WavePlayer_2, properties {bLooping=true, soundWaveAssetPtr=/AmbientAudioContent/Waves/SpottedLoops/Spot_AC_Duct_Air_02}}
 {id=WavePlayer_3, properties {bLooping=true, soundWaveAssetPtr=/AmbientAudioContent/Waves/SpottedLoops/Spot_City_HVAC_Aircon_Unit_Loop_03a}}
 {id=WavePlayer_5, properties {bLooping=true, soundWaveAssetPtr=&Spot_City_HVAC_Aircon_Unit_Loop_03a}}
]
properties {
 attenuationSettings=/AmbientAudioBase/Attenuations/ATT_Spot_City_HVAC_Aircon_Cone
 bOverrideConcurrency=true
 maxDistance=1375
 pitchMultiplier=0.9
 soundClassObject=&Fort_Ambient_Props_Athena
 volumeMultiplier=0.25
 attenuationOverrides {attenuationShapeExtents {x=0}}
 concurrencyOverrides {maxCount=2, voiceStealReleaseTime=0.5}
}
