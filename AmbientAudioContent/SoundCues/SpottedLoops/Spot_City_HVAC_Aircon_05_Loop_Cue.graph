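# Sound cue for a spotted, looping ambient city HVAC air-conditioner unit sound, variant 05.
# A looping wave player feeds one air-conditioner wave into a Modulator at default settings, which is the cue's root.
# It uses a shared aircon-loop attenuation asset and the Fort_Ambient_Athena sound class.
# A concurrency override on the cue caps how many instances play at once.
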SoundCue /AmbientAudioContent/SoundCues/SpottedLoops/Spot_City_HVAC_Aircon_05_Loop_Cue
root=Modulator_0
nodes [
 {id=Modulator_0, children [WavePlayer_0]}
 {id=WavePlayer_0, properties {bLooping=true, soundWaveAssetPtr=/AmbientAudioContent/Waves/SpottedLoops/Spot_City_HVAC_Aircon_Unit_Loop_05a}}
]
properties {
 attenuationSettings=&/ATT_Spot_City_HVAC_Aircon_Loop
 bOverrideConcurrency=true
 maxDistance=1560
 soundClassObject=&Fort_Ambient_Athena
 attenuationOverrides {attenuationShapeExtents {x=0}}
 concurrencyOverrides {maxCount=3}
}
