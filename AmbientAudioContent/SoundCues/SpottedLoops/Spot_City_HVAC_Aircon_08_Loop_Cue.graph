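# Looping city HVAC air-conditioner ambience cue for spotted placement, in the Athena ambient sound class with its own attenuation asset.
# A quality-level node sends the first two levels through a modulator over an equal-weight random choice of looping waves 08a and 08b.
# The last two levels use a second modulator over wave 08a alone, dropping the variation.
# A concurrency override caps simultaneous instances at a small count.
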SoundCue /AmbientAudioContent/SoundCues/SpottedLoops/Spot_City_HVAC_Aircon_08_Loop_Cue
root=QualityLevel_0
nodes [
 {id=Modulator_0, children [Random_0]}
 {id=Modulator_1, children [WavePlayer_2]}
 {id=QualityLevel_0, children [Modulator_0,Modulator_0,Modulator_1,Modulator_1]}
 {id=Random_0, children [WavePlayer_0,WavePlayer_1], properties {weights [1,1]}}
 {id=WavePlayer_0, properties {bLooping=true, soundWaveAssetPtr=/AmbientAudioContent/Waves/SpottedLoops/Spot_City_HVAC_Aircon_Unit_Loop_08b}}
 {id=WavePlayer_1, properties {bLooping=true, soundWaveAssetPtr=/AmbientAudioContent/Waves/SpottedLoops/Spot_City_HVAC_Aircon_Unit_Loop_08a}}
 {id=WavePlayer_2, properties {bLooping=true, soundWaveAssetPtr=&Spot_City_HVAC_Aircon_Unit_Loop_08a}}
]
properties {
 attenuationSettings=&/ATT_Spot_City_HVAC_Aircon_Loop
 bOverrideConcurrency=true
 maxDistance=1560
 soundClassObject=&Fort_Ambient_Athena
 attenuationOverrides {attenuationShapeExtents {x=0}}
 concurrencyOverrides {maxCount=3}
}
