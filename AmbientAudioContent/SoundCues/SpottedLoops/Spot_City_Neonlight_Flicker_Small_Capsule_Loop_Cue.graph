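# Ambient spot sound cue for a small flickering city neon light, mixing a neon layer with repeating bulb flickers.
# One branch randomly picks one of three neon waves; the other loops random flicker one-shots after random delays, mixed slightly lower.
# Each branch has modulation and its own capsule HRTF attenuation with distant-reverb send; only the flicker branch's attenuation enables occlusion.
# Concurrency caps the cue at two instances, stopping the quietest.
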SoundCue /AmbientAudioContent/SoundCues/SpottedLoops/Spot_City_Neonlight_Flicker_Small_Capsule_Loop_Cue
root=Mixer_0
nodes [
 {
  id=Attenuation_0
  children [Modulator_0]
  properties {
   bOverrideAttenuation=true
   attenuationOverrides {
    attenuationShape=Capsule
    bAttenuateWithLPF=true
    bEnableListenerFocus=true
    bEnableLogFrequencyScaling=true
    bEnableSubmixSends=true
    distanceAlgorithm=Custom
    falloffDistance=1000
    hPFFrequencyAtMax=1500
    lPFFrequencyAtMax=3000
    lPFRadiusMax=500
    lPFRadiusMin=100
    nonFocusVolumeAttenuation=0.5
    spatializationAlgorithm=SPATIALIZATION_HRTF
    stereoSpread=0
    attenuationShapeExtents {x=500}
    customAttenuationCurve {editorCurveData {keys [
       <at 0 1 RCIM_Cubic>
       {arriveTangent=-2.4787, interpMode=RCIM_Cubic, leaveTangent=-2.4787, time=0.0861, value=0.8966}
       {arriveTangent=-1.0106, interpMode=RCIM_Cubic, leaveTangent=-1.0106, time=0.2539, value=0.3707}
       {arriveTangent=-0.4969, interpMode=RCIM_Cubic, leaveTangent=-0.4969, time=0.819, value=0.156}
       {arriveTangent=-1.0199, interpMode=RCIM_Cubic, leaveTangent=-1.0199, tangentMode=RCTM_User, time=1, value=0}
      ]}}
    submixSendSettings [{
      bEnableHPFCutoff=false
      bEnableLPFCutoff=false
      disableManualSendClamp=false
      hPFCutoff=20
      lPFCutoff=20000
      maxSendDistance=2000
      maxSendLevel=0.25
      minSendDistance=0
      minSendLevel=0
      sendLevel=0.2
      sendLevelControlMethod=Linear
      soundSubmix=&Fortnite_Reverb_Distant
     }]
   }
  }
 }
 {
  id=Attenuation_1
  children [Modulator_1]
  properties {
   bOverrideAttenuation=true
   attenuationOverrides {
    attenuationShape=Capsule
    bAttenuateWithLPF=true
    bEnableListenerFocus=true
    bEnableLogFrequencyScaling=true
    bEnableOcclusion=true
    bEnableSubmixSends=true
    distanceAlgorithm=Custom
    falloffDistance=1000
    hPFFrequencyAtMax=300
    lPFRadiusMax=1800
    lPFRadiusMin=300
    nonFocusAzimuth=90
    nonFocusVolumeAttenuation=0.5
    occlusionLowPassFilterFrequency=1000
    occlusionTraceChannel=ECC_GameTraceChannel11
    occlusionVolumeAttenuation=0.4
    spatializationAlgorithm=SPATIALIZATION_HRTF
    stereoSpread=0
    attenuationShapeExtents {x=500}
    customAttenuationCurve {editorCurveData {keys [<at 0 1>,<at 0.0421 0.958>,<at 0.2071 0.4102>,<at 0.5096 0.1406>,<at 1 0>]}}
    submixSendSettings [{
      bEnableHPFCutoff=false
      bEnableLPFCutoff=false
      disableManualSendClamp=false
      hPFCutoff=20
      lPFCutoff=20000
      maxSendDistance=1000
      maxSendLevel=0.25
      minSendDistance=0
      minSendLevel=0
      sendLevel=0.2
      sendLevelControlMethod=Linear
      soundSubmix=&Fortnite_Reverb_Distant
     }]
   }
  }
 }
 {id=Delay_1, children [Random_2], properties {delayMax=2, delayMin=0.1}}
 {id=Looping_1, children [Delay_1]}
 {id=Mixer_0, children [Attenuation_1, Attenuation_0], properties {inputVolume [0.8,1]}}
 {id=Modulator_0, children [Random_3], properties {pitchMin=1}}
 {id=Modulator_1, children [Looping_1], properties {pitchMax=1.5, volumeMin=0.7}}
 {id=Random_2, children [WavePlayer_4, WavePlayer_8, WavePlayer_9, WavePlayer_10], properties {weights [1,1,1,1]}}
 {id=Random_3, children [WavePlayer_11, WavePlayer_12, WavePlayer_13], properties {weights [1,1,1]}}
 {id=WavePlayer_10, properties {soundWaveAssetPtr=/AmbientAudioContent/Waves/OneShots/Spot_Bulb_Light_Flicker_Single_05}}
 {id=WavePlayer_11, properties {soundWaveAssetPtr=/AmbientAudioContent/Waves/SpottedLoops/Spot_City_NeonLight_01}}
 {id=WavePlayer_12, properties {soundWaveAssetPtr=/AmbientAudioContent/Waves/SpottedLoops/Spot_City_NeonLight_02}}
 {id=WavePlayer_13, properties {soundWaveAssetPtr=/AmbientAudioContent/Waves/SpottedLoops/Spot_City_NeonLight_03}}
 {id=WavePlayer_4, properties {soundWaveAssetPtr=/AmbientAudioContent/Waves/OneShots/Spot_Bulb_Light_Flicker_Single_01}}
 {id=WavePlayer_8, properties {soundWaveAssetPtr=/AmbientAudioContent/Waves/OneShots/Spot_Bulb_Light_Flicker_Single_03}}
 {id=WavePlayer_9, properties {soundWaveAssetPtr=/AmbientAudioContent/Waves/OneShots/Spot_Bulb_Light_Flicker_Single_04}}
]
properties {
 bOverrideConcurrency=true
 maxDistance=1500
 soundClassObject=&Fort_Ambient_Outdoors_Athena
 volumeMultiplier=1.2
 attenuationOverrides {
  bEnableOcclusion=true
  falloffDistance=925
  nonFocusVolumeAttenuation=0
  nonSpatializedRadiusStart=900
  occlusionLowPassFilterFrequency=2000
  occlusionTraceChannel=ECC_GameTraceChannel6
  occlusionVolumeAttenuation=0.7
  stereoSpread=1500
  attenuationShapeExtents {x=900}
 }
 concurrencyOverrides {bVolumeScaleCanRelease=true, maxCount=2, resolutionRule=StopQuietest}
 soundSubmixSends [{minSendLevel=0.5, soundSubmix=&Fortnite_Reverb_Default}]
}
