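# Sound Cue for a spatial outdoor ambient neon light in a city, layering a looping electrical buzz with intermittent bulb flickers.
# The buzz randomly picks one of three looping waves, while the flicker branch loops over a random delay and one of four one-shots with pitch and volume variation.
# Each branch has its own attenuation asset before mixing, with flickers slightly quieter.
# Cue-level concurrency caps instances, stopping the quietest.
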SoundCue /AmbientAudioContent/SoundCues/SpottedLoops/Spot_City_Neonlight_Loop_Cue
root=Mixer_0
nodes [
 {
  id=Attenuation_0
  children [Modulator_0]
  properties {attenuationSettings=/Game/Sounds/Helios/Environment_Actors/NeonLightSignBuzz/Spot_Neon_Light_Buzz_Sign_Loop_Narrow_Att}
 }
 {
  id=Attenuation_1
  children [Modulator_1]
  properties {attenuationSettings=/Game/Sounds/Helios/Environment_Actors/NeonLightSignBuzz/Spot_Neon_Light_Buzz_Sign_Loop_Narrow_Flicker_Att}
 }
 {id=Delay_1, children [Random_2], properties {delayMax=2, delayMin=0.1}}
 {id=Looping_1, children [Delay_1]}
 {id=Mixer_0, children [Attenuation_1, Attenuation_0], properties {inputVolume [0.8,1]}}
 {id=Modulator_0, children [Random_3], properties {pitchMin=1}}
 {id=Modulator_1, children [Looping_1], properties {pitchMax=1.5, volumeMin=0.7}}
 {id=Random_2, children [WavePlayer_8, WavePlayer_9, WavePlayer_10, WavePlayer_11], properties {weights [1,1,1,1]}}
 {id=Random_3, children [WavePlayer_12, WavePlayer_13, WavePlayer_14], properties {weights [1,1,1]}}
 {id=WavePlayer_10, properties {soundWaveAssetPtr=/AmbientAudioContent/Waves/OneShots/Spot_Bulb_Light_Flicker_Single_04}}
 {id=WavePlayer_11, properties {soundWaveAssetPtr=/AmbientAudioContent/Waves/OneShots/Spot_Bulb_Light_Flicker_Single_05}}
 {id=WavePlayer_12, properties {bLooping=true, soundWaveAssetPtr=/AmbientAudioContent/Waves/SpottedLoops/Spot_City_NeonLight_01}}
 {id=WavePlayer_13, properties {bLooping=true, soundWaveAssetPtr=/AmbientAudioContent/Waves/SpottedLoops/Spot_City_NeonLight_02}}
 {id=WavePlayer_14, properties {bLooping=true, soundWaveAssetPtr=/AmbientAudioContent/Waves/SpottedLoops/Spot_City_NeonLight_03}}
 {id=WavePlayer_8, properties {soundWaveAssetPtr=/AmbientAudioContent/Waves/OneShots/Spot_Bulb_Light_Flicker_Single_01}}
 {id=WavePlayer_9, properties {soundWaveAssetPtr=/AmbientAudioContent/Waves/OneShots/Spot_Bulb_Light_Flicker_Single_03}}
]
properties {
 bOverrideConcurrency=true
 maxDistance=3000
 soundClassObject=&Fort_Ambient_Outdoors_Athena
 volumeMultiplier=1.2
 attenuationOverrides {
  bEnableOcclusion=true
  falloffDistance=925
  nonFocusVolumeAttenuation=0
  nonSpatializedRadiusStart=900
  occlusionLowPassFilterFrequency=2000
  occlusionTraceChannel=ECC_GameTraceChannel6
  occlusionVolumeAttenuation=0.7
  stereoSpread=1500
  attenuationShapeExtents {x=900}
 }
 concurrencyOverrides {bVolumeScaleCanRelease=true, maxCount=2, resolutionRule=StopQuietest}
 soundSubmixSends [{minSendLevel=0.5, soundSubmix=&Fortnite_Reverb_Default}]
}
