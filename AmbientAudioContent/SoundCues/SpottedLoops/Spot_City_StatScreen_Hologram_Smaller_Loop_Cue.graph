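# Sound cue for a small hologram spot loop, named for a city stat screen, that reuses Keys and Locks hologram loop waves.
# An equally weighted Random node picks one of four waves and feeds a Modulator at the root.
# The cue plays quiet and pitched up, with gadget sound class and small government computer screen attenuation.
# Its concurrency override prevents new instances, though a stop-oldest weapon-drop concurrency set is also referenced.
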SoundCue /AmbientAudioContent/SoundCues/SpottedLoops/Spot_City_StatScreen_Hologram_Smaller_Loop_Cue
root=Modulator_0
nodes [
 {id=Modulator_0, children [Random_1], properties {pitchMin=1}}
 {id=Random_1, children [WavePlayer_0,WavePlayer_5,WavePlayer_6,WavePlayer_7], properties {weights [1,1,1,1]}}
 {id=WavePlayer_0, properties {soundWaveAssetPtr=/AmbientAudioContent/Waves/SpottedLoops/Spot_Keys_And_Locks_HologramLoop_01}}
 {id=WavePlayer_5, properties {soundWaveAssetPtr=/AmbientAudioContent/Waves/SpottedLoops/Spot_Keys_And_Locks_HologramLoop_02}}
 {id=WavePlayer_6, properties {soundWaveAssetPtr=/AmbientAudioContent/Waves/SpottedLoops/Spot_Keys_And_Locks_HologramLoop_03}}
 {id=WavePlayer_7, properties {soundWaveAssetPtr=/AmbientAudioContent/Waves/SpottedLoops/Spot_Keys_And_Locks_HologramLoop_04}}
]
properties {
 attenuationSettings=/Game/Sounds/Artemis/Environment_Actors/Computer_Screen/Government_Computer_Smaller_Att
 bOverrideConcurrency=true
 maxDistance=700
 pitchMultiplier=2
 soundClassObject=&Fort_Gadgets_Items
 volumeMultiplier=0.35
 attenuationOverrides {falloffDistance=2000}
 concurrencyOverrides {resolutionRule=PreventNew}
 concurrencySet [&Weapon_Drop_StopOldest]
}
