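# Spotted ambient Sound Cue that loops a sealed city drainage pipe water sound for outdoor environments.
# A looping wave feeds a slow volume oscillator, then a modulator that varies pitch only.
# The quality-level root routes this chain to its first two slots and leaves the other two empty.
# It uses sealed-pipe attenuation, a shared spot-ambience concurrency set, the outdoor ambient sound class and a reverb submix send.
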SoundCue /AmbientAudioContent/SoundCues/SpottedLoops/Spot_City_Water_Drainage_Pipe_01_Loop_Cue
root=QualityLevel_0
nodes [
 {id=Modulator_1, children [Oscillator_0], properties {pitchMax=1, volumeMax=1, volumeMin=1}}
 {
  id=Oscillator_0
  children [WavePlayer_1]
  properties {amplitudeMax=0.3, amplitudeMin=0.2, bModulateVolume=true, centerMax=1, centerMin=1, frequencyMax=0.1, frequencyMin=0.1}
 }
 {id=QualityLevel_0, children [Modulator_1,Modulator_1,null,null]}
 {
  id=WavePlayer_1
  properties {bLooping=true, soundWaveAssetPtr=/AmbientAudioContent/Waves/SpottedLoops/Spot_City_Drainage_Pipe_Sealed_01a_Loop}
 }
]
properties {
 attenuationSettings=/AmbientAudioBase/Attenuations/Spot_Water_Pipe_Sealed_Att
 maxDistance=1000
 priority=10
 soundClassObject=&Fort_Ambient_Outdoors_Athena
 volumeMultiplier=0.5
 concurrencyOverrides {maxCount=1, voiceStealReleaseTime=3}
 concurrencySet [/AmbientAudioBase/Concurrency/AmbientConc_3Max_Spot_Default_Concurrency]
 soundSubmixSends [{minSendLevel=1, soundSubmix=&Fortnite_Reverb_Default}]
}
