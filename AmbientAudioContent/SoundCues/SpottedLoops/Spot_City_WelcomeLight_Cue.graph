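# Sound cue for the ambient loop of a city welcome light sign, layering a steady looping bed with intermittent bulb flickers.
# The bed randomly picks one of two looping waves with slight pitch variation, mixed above the flicker layer.
# The flicker layer repeatedly plays one of four single-flicker waves after a short random delay, with volume modulation.
# It uses a shared attenuation asset, a reverb submix send, and a two-voice concurrency limit that stops the quietest.
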SoundCue /AmbientAudioContent/SoundCues/SpottedLoops/Spot_City_WelcomeLight_Cue
root=Mixer_0
nodes [
 {id=Delay_1, children [Random_1], properties {delayMax=0.7, delayMin=0.1}}
 {id=Looping_1, children [Delay_1]}
 {id=Mixer_0, children [Modulator_1,Modulator_0], properties {inputVolume [0.4,1]}}
 {id=Modulator_0, children [Random_2], properties {pitchMax=1.01, pitchMin=1}}
 {id=Modulator_1, children [Looping_1], properties {pitchMax=1, volumeMin=0.7}}
 {id=Random_1, children [WavePlayer_0,WavePlayer_5,WavePlayer_6,WavePlayer_7], properties {weights [1,1,1,1]}}
 {id=Random_2, children [WavePlayer_8,WavePlayer_9], properties {weights [1,1]}}
 {id=WavePlayer_0, properties {soundWaveAssetPtr=/Game/Sounds/Asteria/Environment_Actors/NeonLightFlicker/Spot_Bulb_Light_Flicker_Single_01}}
 {id=WavePlayer_5, properties {soundWaveAssetPtr=/Game/Sounds/Asteria/Environment_Actors/NeonLightFlicker/Spot_Bulb_Light_Flicker_Single_03}}
 {id=WavePlayer_6, properties {soundWaveAssetPtr=/Game/Sounds/Asteria/Environment_Actors/NeonLightFlicker/Spot_Bulb_Light_Flicker_Single_04}}
 {id=WavePlayer_7, properties {soundWaveAssetPtr=/Game/Sounds/Asteria/Environment_Actors/NeonLightFlicker/Spot_Bulb_Light_Flicker_Single_05}}
 {id=WavePlayer_8, properties {bLooping=true, soundWaveAssetPtr=/AmbientAudioContent/Waves/SpottedLoops/Spot_City_Welcome_Light_01a}}
 {id=WavePlayer_9, properties {bLooping=true, soundWaveAssetPtr=/AmbientAudioContent/Waves/SpottedLoops/Spot_City_Welcome_Light_01b}}
]
properties {
 attenuationSettings=/Game/Sounds/Helios/Environment_Actors/WelcomeLight/Spot_Welcome_Light_Buzz_Sign_Loop_Att
 bOverrideConcurrency=true
 maxDistance=3150
 soundClassObject=&Fort_Ambient_Outdoors_Athena
 volumeMultiplier=0.6
 attenuationOverrides {
  bEnableOcclusion=true
  falloffDistance=925
  nonFocusVolumeAttenuation=0
  nonSpatializedRadiusStart=900
  occlusionLowPassFilterFrequency=2000
  occlusionTraceChannel=ECC_GameTraceChannel6
  occlusionVolumeAttenuation=0.7
  stereoSpread=1500
  attenuationShapeExtents {x=900}
 }
 concurrencyOverrides {bVolumeScaleCanRelease=true, maxCount=2, resolutionRule=StopQuietest}
 soundSubmixSends [{minSendLevel=0.5, soundSubmix=&Fortnite_Reverb_Default}]
}
