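# Spotted ambient Sound Cue that plays a looping indoor clock tick from two big-clock one-shot waves.
# Each loop pass waits a fixed one-second delay, then plays a randomly chosen tick with constant raised pitch and reduced volume.
# An attenuation node at the root overrides its settings for short-range falloff, occlusion filtering and a reverb submix send.
# The cue uses the indoor ambient one-shot sound class and a shared spot concurrency limit.
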SoundCue /AmbientAudioContent/SoundCues/SpottedLoops/Spot_Clock_Ticks_Loop_Int_Cue
root=Attenuation_0
nodes [
 {
  id=Attenuation_0
  children [Looping_0]
  properties {
   attenuationSettings=/Game/Sounds/Attenuation/Fort_BigClock_Att
   bOverrideAttenuation=true
   attenuationOverrides {
    bEnableOcclusion=true
    bEnableSubmixSends=true
    falloffDistance=400
    occlusionLowPassFilterFrequency=850
    occlusionTraceChannel=ECC_GameTraceChannel11
    occlusionVolumeAttenuation=0.25
    attenuationShapeExtents {x=200}
    submixSendSettings [{
      bEnableHPFCutoff=false
      bEnableLPFCutoff=false
      disableManualSendClamp=false
      hPFCutoff=20
      lPFCutoff=20000
      maxSendDistance=6000
      maxSendLevel=1
      minSendDistance=400
      minSendLevel=1
      sendLevel=0.2
      sendLevelControlMethod=Linear
      soundSubmix=&Fortnite_Reverb_Default
     }]
   }
  }
 }
 {id=Delay_0, children [Modulator_0], properties {delayMax=1, delayMin=1}}
 {id=Looping_0, children [Delay_0]}
 {id=Modulator_0, children [Random_0], properties {pitchMax=1.4, pitchMin=1.4, volumeMax=0.7, volumeMin=0.7}}
 {id=Random_0, children [WavePlayer_3,WavePlayer_4], properties {weights [1,1]}}
 {id=WavePlayer_3, properties {soundWaveAssetPtr=/AmbientAudioContent/Waves/OneShots/Spot_BigClock_Tick_01}}
 {id=WavePlayer_4, properties {soundWaveAssetPtr=/AmbientAudioContent/Waves/OneShots/Spot_BigClock_Tick_02}}
]
properties {
 maxDistance=600
 pitchMultiplier=1.4
 soundClassObject=/Game/Sounds/FortSoundClassesAndModes/Fort_Ambient_OneShots_Indoors
 volumeMultiplier=1
 attenuationOverrides {falloffDistance=400, attenuationShapeExtents {x=200}}
 concurrencySet [/AmbientAudioBase/Concurrency/AmbientConc_3Max_Spot_Default_Concurrency]
}
