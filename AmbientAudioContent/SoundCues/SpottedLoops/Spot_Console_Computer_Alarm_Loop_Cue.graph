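# Spotted ambient loop cue for an alarming computer console, mixing four attenuated layers.
# Three layers repeat after random delays: a weighted pick of lab alarms and computer blips, a pitched-down one-shot, and pitch-varied one-shots.
# The fourth is a continuous, heavily pitched-down transmitter hum; one alarm uses an envelope fade-out and another a finite loop count.
# The quality node routes the mix to its first two levels, leaving the other two empty.
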SoundCue /AmbientAudioContent/SoundCues/SpottedLoops/Spot_Console_Computer_Alarm_Loop_Cue
root=QualityLevel_0
nodes [
 {
  id=Attenuation_0
  children [Looping_0]
  properties {attenuationSettings=/Game/Sounds/Artemis/Environment_Actors/Computer_Screen/Government_Computer_Smaller_Log_Att}
 }
 {
  id=Attenuation_1
  children [Looping_1]
  properties {attenuationSettings=&Government_Computer_Smaller_Log_Att}
 }
 {
  id=Attenuation_2
  children [Looping_2]
  properties {attenuationSettings=&Government_Computer_Smaller_Log_Att}
 }
 {
  id=Attenuation_3
  children [Modulator_6]
  properties {attenuationSettings=/Game/Sounds/Artemis/Environment_Actors/Computer_Screen/Government_Computer_Att}
 }
 {id=Delay_0, children [Random_3], properties {delayMax=1, delayMin=0.5}}
 {id=Delay_1, children [Modulator_0], properties {delayMax=8, delayMin=3}}
 {id=Delay_2, children [Modulator_4], properties {delayMax=1, delayMin=0.5}}
 {
  id=Enveloper_0
  children [Mixer_8]
  properties {pitchCurve {editorCurveData {keys [{time=0}]}}, volumeCurve {editorCurveData {keys [<at 0 1>,<at 8 1>,<at 8.5 0>]}}}
 }
 {id=Looping_0, children [Delay_0]}
 {id=Looping_1, children [Delay_1]}
 {id=Looping_2, children [Delay_2]}
 {id=Looping_3, children [Mixer_2], properties {bLoopIndefinitely=false, loopCount=2}}
 {id=Mixer_0, children [WavePlayer_19], properties {inputVolume [1]}}
 {id=Mixer_2, children [WavePlayer_17], properties {inputVolume [1]}}
 {id=Mixer_3, children [WavePlayer_6], properties {inputVolume [0.5]}}
 {id=Mixer_4, children [WavePlayer_7], properties {inputVolume [0.5]}}
 {id=Mixer_5, children [Attenuation_0, Attenuation_1, Attenuation_2, Attenuation_3], properties {inputVolume [0.6,0.3,0.7,1]}}
 {id=Mixer_6, children [WavePlayer_8], properties {inputVolume [0.5]}}
 {id=Mixer_7, children [WavePlayer_14], properties {inputVolume [0.75]}}
 {id=Mixer_8, children [WavePlayer_22], properties {inputVolume [0.5]}}
 {id=Modulator_0, children [WavePlayer_4], properties {pitchMax=0.8, pitchMin=0.8, volumeMax=1, volumeMin=1}}
 {id=Modulator_2, children [Mixer_3], properties {pitchMax=1.15, volumeMax=1, volumeMin=1}}
 {id=Modulator_3, children [Mixer_4], properties {pitchMax=1.15, volumeMax=1, volumeMin=1}}
 {id=Modulator_4, children [Random_5], properties {pitchMax=1.3, pitchMin=0.7, volumeMax=1, volumeMin=1}}
 {id=Modulator_5, children [Mixer_6], properties {pitchMax=1.15, volumeMax=1, volumeMin=1}}
 {id=Modulator_6, children [WavePlayer_0], properties {pitchMax=0.5, pitchMin=0.2}}
 {id=QualityLevel_0, children [Mixer_5,Mixer_5,null,null]}
 {id=Random_3, children [Looping_3,Mixer_0,Enveloper_0,Modulator_2,Modulator_3,Modulator_5], properties {weights [1,1,1.5,1,1,1]}}
 {id=Random_5, children [WavePlayer_1,WavePlayer_2,WavePlayer_3,Mixer_7], properties {weights [1,1,1,1]}}
 {
  id=WavePlayer_0
  properties {bLooping=true, soundWaveAssetPtr=/Game/Athena/Sounds/Props/MountainBaseComputer/ComputerTransimtter_Running_Loop_01}
 }
 {id=WavePlayer_1, properties {soundWaveAssetPtr=/AmbientAudioContent/Waves/OneShots/Spot_SecretComplex_Computer_Oneshot_10}}
 {id=WavePlayer_14, properties {soundWaveAssetPtr=/AmbientAudioContent/Waves/OneShots/Spot_SecretComplex_Computer_Oneshot_15}}
 {id=WavePlayer_17, properties {soundWaveAssetPtr=/AmbientAudioContent/Waves/SpottedLoops/Spot_Weapon_Lab_Alarm_Loop_01}}
 {id=WavePlayer_19, properties {soundWaveAssetPtr=/AmbientAudioContent/Waves/SpottedLoops/Spot_Weapon_Lab_Alarm_Loop_03}}
 {id=WavePlayer_2, properties {soundWaveAssetPtr=/AmbientAudioContent/Waves/OneShots/Spot_SecretComplex_Computer_Oneshot_11}}
 {id=WavePlayer_22, properties {soundWaveAssetPtr=/AmbientAudioContent/Waves/SpottedLoops/Spot_Weapon_Lab_Alarm_Loop_06}}
 {id=WavePlayer_3, properties {soundWaveAssetPtr=/AmbientAudioContent/Waves/OneShots/Spot_SecretComplex_Computer_Oneshot_12}}
 {id=WavePlayer_4, properties {soundWaveAssetPtr=/AmbientAudioContent/Waves/OneShots/Spot_SecretComplex_Computer_Oneshot_01}}
 {id=WavePlayer_6, properties {soundWaveAssetPtr=/AmbientAudioContent/Waves/OneShots/Spot_SecretComplex_Computer_Oneshot_03}}
 {id=WavePlayer_7, properties {soundWaveAssetPtr=/AmbientAudioContent/Waves/OneShots/Spot_SecretComplex_Computer_Oneshot_04}}
 {id=WavePlayer_8, properties {soundWaveAssetPtr=/AmbientAudioContent/Waves/OneShots/Spot_SecretComplex_Computer_Oneshot_05}}
]
properties {
 maxDistance=750
 soundClassObject=&Fort_Gadgets
 volumeMultiplier=0.4
 attenuationOverrides {
  bEnableFocusInterpolation=true
  falloffDistance=600
  nonFocusVolumeAttenuation=0.3
  occlusionVolumeAttenuation=0.3
  stereoSpread=0
  attenuationShapeExtents {x=200}
 }
 concurrencyOverrides {maxCount=2}
 concurrencySet [/AmbientAudioBase/Concurrency/AmbientConc_3Max_Spot_Default_Concurrency]
}
