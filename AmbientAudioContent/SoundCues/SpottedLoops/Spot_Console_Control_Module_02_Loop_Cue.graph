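# Ambient sound cue for a spotted console control module prop, combining a hum loop with intermittent beeps.
# A mixer blends a random looping wave, shaped by fade-in and slow-swell envelopers, with two quieter Looping branches that fire randomly delayed, pitch-varied beep one-shots.
# One beep variant is pitched up through an extra modulator.
# All quality levels share this mix, which uses terminal attenuation, occlusion, listener focus, concurrency limits and a reverb send.
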SoundCue /AmbientAudioContent/SoundCues/SpottedLoops/Spot_Console_Control_Module_02_Loop_Cue
root=QualityLevel_0
nodes [
 {id=Delay_0, children [Modulator_1], properties {delayMax=3, delayMin=0.65}}
 {id=Delay_1, children [Modulator_2], properties {delayMax=2, delayMin=0.5}}
 {
  id=Enveloper_0
  children [Enveloper_1]
  properties {pitchCurve {editorCurveData {keys [{time=0}]}}, volumeCurve {editorCurveData {keys [<at 0 0>,<at 1 1>]}}}
 }
 {
  id=Enveloper_1
  children [Modulator_0]
  properties {
   loopEnd=15
   pitchCurve {editorCurveData {keys [{arriveTangent=0.0098, interpMode=RCIM_Cubic, time=0}, {interpMode=RCIM_Cubic, time=15}]}}
   volumeCurve {editorCurveData {keys [<at 0 1 RCIM_Cubic>,<at 7.5 0.8 RCIM_Cubic>,<at 15 1>]}}
  }
 }
 {id=Looping_0, children [Delay_0]}
 {id=Looping_1, children [Delay_1]}
 {id=Mixer_0, children [Looping_0,Enveloper_0,Looping_1], properties {inputVolume [0.5,1,0.25]}}
 {id=Mixer_3, children [WavePlayer_16], properties {inputVolume [0.6]}}
 {id=Mixer_4, children [WavePlayer_15], properties {inputVolume [0.6]}}
 {id=Modulator_0, children [Random_0], properties {pitchMax=0.9, pitchMin=0.9}}
 {id=Modulator_1, children [Random_1], properties {pitchMax=1.2, pitchMin=0.9, volumeMin=0.9}}
 {id=Modulator_2, children [Random_5], properties {pitchMax=1.2, pitchMin=0.9}}
 {id=Modulator_3, children [WavePlayer_41], properties {pitchMax=2, pitchMin=2}}
 {id=QualityLevel_0, children [Mixer_0,Mixer_0,Mixer_0,Mixer_0]}
 {id=Random_0, children [WavePlayer_7, WavePlayer_8, WavePlayer_9, WavePlayer_10, WavePlayer_11], properties {weights [1,1,1,1,1]}}
 {id=Random_1, children [Mixer_3, Mixer_4, WavePlayer_17, WavePlayer_19, WavePlayer_20, WavePlayer_22], properties {weights [1,1,1,1,1,1]}}
 {
  id=Random_5
  children [Modulator_3, WavePlayer_41, WavePlayer_42, WavePlayer_43, WavePlayer_44, WavePlayer_45, WavePlayer_47]
  properties {weights [1,1,1,1,1,1,1]}
 }
 {id=WavePlayer_10, properties {bLooping=true, soundWaveAssetPtr=/AmbientAudioContent/Waves/SpottedLoops/Spot_Console_Control_Module_Loop_04}}
 {id=WavePlayer_11, properties {bLooping=true, soundWaveAssetPtr=/AmbientAudioContent/Waves/SpottedLoops/Spot_Console_Control_Module_Loop_05}}
 {id=WavePlayer_15, properties {soundWaveAssetPtr=/AmbientAudioContent/Waves/OneShots/Spot_Cosole_Control_Module_Beep_02}}
 {id=WavePlayer_16, properties {soundWaveAssetPtr=/AmbientAudioContent/Waves/OneShots/Spot_Cosole_Control_Module_Beep_03}}
 {id=WavePlayer_17, properties {soundWaveAssetPtr=/AmbientAudioContent/Waves/OneShots/Spot_Cosole_Control_Module_Beep_04}}
 {id=WavePlayer_19, properties {soundWaveAssetPtr=/AmbientAudioContent/Waves/OneShots/Spot_Cosole_Control_Module_Beep_06}}
 {id=WavePlayer_20, properties {soundWaveAssetPtr=/AmbientAudioContent/Waves/OneShots/Spot_Cosole_Control_Module_Beep_07}}
 {id=WavePlayer_22, properties {soundWaveAssetPtr=/AmbientAudioContent/Waves/OneShots/Spot_Cosole_Control_Module_Beep_09}}
 {id=WavePlayer_41, properties {soundWaveAssetPtr=&Spot_Cosole_Control_Module_Beep_03}}
 {id=WavePlayer_42, properties {soundWaveAssetPtr=&Spot_Cosole_Control_Module_Beep_04}}
 {id=WavePlayer_43, properties {soundWaveAssetPtr=/AmbientAudioContent/Waves/OneShots/Spot_Cosole_Control_Module_Beep_05}}
 {id=WavePlayer_44, properties {soundWaveAssetPtr=&Spot_Cosole_Control_Module_Beep_06}}
 {id=WavePlayer_45, properties {soundWaveAssetPtr=&Spot_Cosole_Control_Module_Beep_07}}
 {id=WavePlayer_47, properties {soundWaveAssetPtr=&Spot_Cosole_Control_Module_Beep_09}}
 {id=WavePlayer_7, properties {bLooping=true, soundWaveAssetPtr=/AmbientAudioContent/Waves/SpottedLoops/Spot_Console_Control_Module_Loop_01}}
 {id=WavePlayer_8, properties {bLooping=true, soundWaveAssetPtr=/AmbientAudioContent/Waves/SpottedLoops/Spot_Console_Control_Module_Loop_02}}
 {id=WavePlayer_9, properties {bLooping=true, soundWaveAssetPtr=/AmbientAudioContent/Waves/SpottedLoops/Spot_Console_Control_Module_Loop_03}}
]
properties {
 attenuationSettings=/Game/Sounds/Helios/Environment_Actors/ConsoleTerminals/Spot_Console_Terminal_Communication_Hum_Att
 maxDistance=1000
 pitchMultiplier=1.1
 soundClassObject=&Fort_Ambient_Props_Athena
 subtitlePriority=5000
 volumeMultiplier=0.4
 attenuationOverrides {
  bEnableFocusInterpolation=true
  bEnableListenerFocus=true
  bEnableOcclusion=true
  dBAttenuationAtMax=-40
  falloffDistance=512
  focusAzimuth=15
  focusVolumeAttenuation=1.5
  nonFocusAzimuth=45
  nonFocusVolumeAttenuation=0.5
  occlusionInterpolationTime=1
  occlusionLowPassFilterFrequency=10000
  occlusionTraceChannel=ECC_GameTraceChannel11
  occlusionVolumeAttenuation=0.5
  stereoSpread=512
  attenuationShapeExtents=<256 64.5942 90>
 }
 concurrencyOverrides {bLimitToOwner=true, maxCount=2}
 concurrencySet [/AmbientAudioBase/Concurrency/AmbientConc_3Max_Spot_Default_Concurrency]
 soundSubmixSends [{minSendLevel=1, soundSubmix=&Fortnite_Reverb_Default}]
}
