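# Ambient sound cue for a spotted console control module hum, classed as an ambient prop sound, with overridden capsule attenuation, listener focus and occlusion.
# A mixer blends one random non-looping pick of four communication hum waves with a looping server rack hum, randomized by a modulator and faded in by an enveloper.
# All quality levels share that mixer.
# Concurrency uses a shared three-max spot ambience set.
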SoundCue /AmbientAudioContent/SoundCues/SpottedLoops/Spot_Console_Control_Module_Capsule_Loop_Cue
root=QualityLevel_0
nodes [
 {
  id=Enveloper_0
  children [Modulator_0]
  properties {pitchCurve {editorCurveData {keys [{time=0}]}}, volumeCurve {editorCurveData {keys [<at 0 0>,<at 1 1>]}}}
 }
 {id=Mixer_0, children [Random_3,Enveloper_0], properties {inputVolume [0.5,0.6]}}
 {id=Modulator_0, children [WavePlayer_0]}
 {id=QualityLevel_0, children [Mixer_0,Mixer_0,Mixer_0,Mixer_0]}
 {id=Random_3, children [WavePlayer_2,WavePlayer_4,WavePlayer_5,WavePlayer_6], properties {weights [1,1,1,1]}}
 {
  id=WavePlayer_0
  properties {bLooping=true, soundWaveAssetPtr=/AmbientAudioContent/Waves/SpottedLoops/Spot_Computer_Server_Rack_Tower_Hum_Loop_01}
 }
 {id=WavePlayer_2, properties {soundWaveAssetPtr=/AmbientAudioContent/Waves/SpottedLoops/Spot_Console_Terminal_Communication_Hum_Loop_04}}
 {id=WavePlayer_4, properties {soundWaveAssetPtr=/AmbientAudioContent/Waves/SpottedLoops/Spot_Console_Terminal_Communication_Hum_Loop_03}}
 {id=WavePlayer_5, properties {soundWaveAssetPtr=/AmbientAudioContent/Waves/SpottedLoops/Spot_Console_Terminal_Communication_Hum_Loop_02}}
 {id=WavePlayer_6, properties {soundWaveAssetPtr=/AmbientAudioContent/Waves/SpottedLoops/Spot_Console_Terminal_Communication_Hum_Loop_01}}
]
properties {
 attenuationSettings=/Game/Sounds/Helios/Environment_Actors/ConsoleTerminals/Spot_Console_Terminal_Communication_Hum_Att
 bOverrideAttenuation=true
 maxDistance=550
 pitchMultiplier=1.1
 soundClassObject=&Fort_Ambient_Props_Athena
 subtitlePriority=5000
 volumeMultiplier=0.301
 attenuationOverrides {
  attenuationShape=Capsule
  bEnableFocusInterpolation=true
  bEnableListenerFocus=true
  bEnableOcclusion=true
  dBAttenuationAtMax=-40
  falloffDistance=350
  focusAzimuth=15
  focusVolumeAttenuation=1.5
  nonFocusAzimuth=45
  nonFocusVolumeAttenuation=0.5
  occlusionInterpolationTime=1
  occlusionLowPassFilterFrequency=10000
  occlusionTraceChannel=ECC_GameTraceChannel11
  occlusionVolumeAttenuation=0.5
  stereoSpread=512
  attenuationShapeExtents=<200 100 90>
 }
 concurrencyOverrides {bLimitToOwner=true, maxCount=2}
 concurrencySet [/AmbientAudioBase/Concurrency/AmbientConc_3Max_Spot_Default_Concurrency]
}
