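# Sound cue for a quiet spotted ambient loop on a smaller dam control console, classed as an ambient prop sound.
# It mixes a randomly picked, pitched-down server-stack hum, given slow cyclic pitch drift by looping envelopers, with one of four looping tech drone hums.
# The mix starts after a short random delay under a fade-in envelope, on only the first two of four quality slots.
# Overridden attenuation adds listener focus and occlusion; a reverb send is included.
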SoundCue /AmbientAudioContent/SoundCues/SpottedLoops/Spot_Console_Dam_Controls_Loop_Smaller_Cue
root=QualityLevel_1
nodes [
 {id=Delay_0, children [Mixer_5], properties {delayMax=0.1, delayMin=0.05}}
 {
  id=Enveloper_0
  children [Modulator_0]
  properties {
   bLoop=true
   bLoopIndefinitely=true
   loopEnd=60
   pitchCurve {editorCurveData {keys [{interpMode=RCIM_Cubic, time=0}, <at 30 0.95 RCIM_Cubic>, {interpMode=RCIM_Cubic, time=60}]}}
   volumeCurve {editorCurveData {keys [<at 0 1>]}}
  }
 }
 {
  id=Enveloper_1
  children [Modulator_2]
  properties {
   bLoop=true
   bLoopIndefinitely=true
   loopEnd=60
   pitchCurve {editorCurveData {keys [{interpMode=RCIM_Cubic, time=0}, <at 30 0.95 RCIM_Cubic>, {interpMode=RCIM_Cubic, time=60}]}}
   volumeCurve {editorCurveData {keys [<at 0 1>]}}
  }
 }
 {
  id=Enveloper_2
  children [Modulator_3]
  properties {
   bLoop=true
   bLoopIndefinitely=true
   loopEnd=60
   pitchCurve {editorCurveData {keys [{interpMode=RCIM_Cubic, time=0}, <at 30 0.95 RCIM_Cubic>, {interpMode=RCIM_Cubic, time=60}]}}
   volumeCurve {editorCurveData {keys [<at 0 1>]}}
  }
 }
 {
  id=Enveloper_29
  children [Delay_0]
  properties {
   pitchCurve {editorCurveData {keys [{time=0}]}}
   volumeCurve {editorCurveData {keys [{arriveTangent=3.4293, interpMode=RCIM_Cubic, leaveTangent=3.4293, tangentMode=RCTM_User, time=0, value=0}, <at 0.4 1 RCIM_Cubic>]}}
  }
 }
 {id=Mixer_0, children [Modulator_7], properties {inputVolume [0.8]}}
 {id=Mixer_1, children [WavePlayer_13], properties {inputVolume [1.2]}}
 {id=Mixer_5, children [Random_1,Random_3], properties {inputVolume [1,1]}}
 {id=Modulator_0, children [WavePlayer_0], properties {pitchMax=0.9, pitchMin=0.8, volumeMax=0.85, volumeMin=0.8}}
 {id=Modulator_2, children [WavePlayer_2], properties {pitchMax=0.9, pitchMin=0.8}}
 {id=Modulator_3, children [WavePlayer_3], properties {pitchMax=0.9, pitchMin=0.8}}
 {id=Modulator_7, children [WavePlayer_11], properties {pitchMax=0.8, pitchMin=0.8}}
 {id=QualityLevel_1, children [Enveloper_29,Enveloper_29,null,null]}
 {id=Random_1, children [Enveloper_0,Enveloper_1,Enveloper_2], properties {weights [1,1,1]}}
 {id=Random_3, children [WavePlayer_4, Mixer_0, WavePlayer_12, Mixer_1], properties {weights [1,1,1,1]}}
 {id=WavePlayer_0, properties {bLooping=true, soundWaveAssetPtr=/AmbientAudioContent/Waves/SpottedLoops/Spot_Server_Stack_Hum_05}}
 {id=WavePlayer_11, properties {bLooping=true, soundWaveAssetPtr=/AmbientAudioContent/Waves/SpottedLoops/Spot_Console_Tech_Drone_Hum_Loop_02}}
 {id=WavePlayer_12, properties {bLooping=true, soundWaveAssetPtr=/AmbientAudioContent/Waves/SpottedLoops/Spot_Console_Tech_Drone_Hum_Loop_03}}
 {id=WavePlayer_13, properties {bLooping=true, soundWaveAssetPtr=/AmbientAudioContent/Waves/SpottedLoops/Spot_Console_Tech_Drone_Hum_Loop_04}}
 {id=WavePlayer_2, properties {bLooping=true, soundWaveAssetPtr=/AmbientAudioContent/Waves/SpottedLoops/Spot_Server_Stack_Hum_06}}
 {id=WavePlayer_3, properties {bLooping=true, soundWaveAssetPtr=/AmbientAudioContent/Waves/SpottedLoops/Spot_Server_Stack_Hum_07}}
 {id=WavePlayer_4, properties {bLooping=true, soundWaveAssetPtr=/AmbientAudioContent/Waves/SpottedLoops/Spot_Console_Tech_Drone_Hum_Loop_01}}
]
properties {
 attenuationSettings=/Game/Sounds/Helios/Environment_Actors/ConsoleTerminals/Spot_Console_Terminal_Communication_Hum_Att
 bOverrideAttenuation=true
 maxDistance=650
 soundClassObject=&Fort_Ambient_Props_Athena
 subtitlePriority=5000
 volumeMultiplier=0.15
 attenuationOverrides {
  bEnableFocusInterpolation=true
  bEnableListenerFocus=true
  bEnableOcclusion=true
  dBAttenuationAtMax=-40
  falloffDistance=500
  focusAzimuth=15
  nonFocusAzimuth=45
  nonFocusVolumeAttenuation=0.6
  occlusionInterpolationTime=1
  occlusionLowPassFilterFrequency=10000
  occlusionTraceChannel=ECC_GameTraceChannel11
  occlusionVolumeAttenuation=0.5
  attenuationShapeExtents=<150 175 90>
 }
 concurrencyOverrides {bLimitToOwner=true, maxCount=2}
 concurrencySet [/AmbientAudioBase/Concurrency/AmbientConc_3Max_Spot_Default_Concurrency]
 soundSubmixSends [{minSendLevel=1, soundSubmix=&Fortnite_Reverb_Default}]
}
