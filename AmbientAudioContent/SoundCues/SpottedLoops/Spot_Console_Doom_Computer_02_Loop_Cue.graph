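# Looping ambient Sound Cue for a spotted console computer, routed to the indoor ambient sound class.
# A mixer blends a pitched-down server loop with two quieter looping layers of randomly delayed one-shots: computer sounds and radio telemetry.
# Modulators vary one-shot pitch (plus computer volume), and an Enveloper ramps in the telemetry layer.
# All quality levels share this mix, with overridden attenuation using occlusion and listener focus, and per-owner concurrency limits.
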SoundCue /AmbientAudioContent/SoundCues/SpottedLoops/Spot_Console_Doom_Computer_02_Loop_Cue
root=QualityLevel_0
nodes [
 {id=Delay_0, children [Modulator_2], properties {delayMax=1, delayMin=0.5}}
 {id=Delay_6, children [Random_2], properties {delayMax=4, delayMin=0.2}}
 {
  id=Enveloper_0
  children [Looping_2]
  properties {pitchCurve {editorCurveData {keys [{time=0}]}}, volumeCurve {editorCurveData {keys [<at 0 0>,<at 1.5 1>]}}}
 }
 {id=Looping_0, children [Delay_0]}
 {id=Looping_2, children [Modulator_4]}
 {id=Mixer_0, children [Modulator_0,Looping_0,Enveloper_0], properties {inputVolume [0.85,0.4,0.35]}}
 {id=Modulator_0, children [WavePlayer_2], properties {pitchMax=0.8, pitchMin=0.8}}
 {id=Modulator_2, children [Random_3], properties {pitchMax=1.15, volumeMax=1, volumeMin=0.4}}
 {id=Modulator_4, children [Delay_6], properties {pitchMax=1.2, pitchMin=0.85}}
 {id=QualityLevel_0, children [Mixer_0,Mixer_0,Mixer_0,Mixer_0]}
 {
  id=Random_2
  children [WavePlayer_0, WavePlayer_1, WavePlayer_4, WavePlayer_14, WavePlayer_15, WavePlayer_16]
  properties {weights [1,1,1,1,1,1]}
 }
 {id=Random_3, children [WavePlayer_17, WavePlayer_18, WavePlayer_19, WavePlayer_20], properties {weights [1,1,1,1]}}
 {id=WavePlayer_0, properties {soundWaveAssetPtr=/AmbientAudioContent/Waves/OneShots/Spot_Radio_Telemetry_A_1}}
 {id=WavePlayer_1, properties {soundWaveAssetPtr=/AmbientAudioContent/Waves/OneShots/Spot_Radio_Telemetry_A_2}}
 {id=WavePlayer_14, properties {soundWaveAssetPtr=/AmbientAudioContent/Waves/OneShots/Spot_Radio_Telemetry_A_4}}
 {id=WavePlayer_15, properties {soundWaveAssetPtr=/AmbientAudioContent/Waves/OneShots/Spot_Radio_Telemetry_A_5}}
 {id=WavePlayer_16, properties {soundWaveAssetPtr=/AmbientAudioContent/Waves/OneShots/Spot_Radio_Telemetry_A_6}}
 {id=WavePlayer_17, properties {soundWaveAssetPtr=/AmbientAudioContent/Waves/OneShots/Spot_SecretComplex_Computer_Oneshot_02}}
 {id=WavePlayer_18, properties {soundWaveAssetPtr=/AmbientAudioContent/Waves/OneShots/Spot_SecretComplex_Computer_Oneshot_03}}
 {id=WavePlayer_19, properties {soundWaveAssetPtr=/AmbientAudioContent/Waves/OneShots/Spot_SecretComplex_Computer_Oneshot_04}}
 {id=WavePlayer_2, properties {bLooping=true, soundWaveAssetPtr=/AmbientAudioContent/Waves/SpottedLoops/Spot_Agency_Computer_Server_Loop}}
 {id=WavePlayer_20, properties {soundWaveAssetPtr=/AmbientAudioContent/Waves/OneShots/Spot_SecretComplex_Computer_Oneshot_05}}
 {id=WavePlayer_4, properties {soundWaveAssetPtr=/AmbientAudioContent/Waves/OneShots/Spot_Radio_Telemetry_A_3}}
]
properties {
 bOverrideAttenuation=true
 bOverrideConcurrency=true
 maxDistance=1300
 soundClassObject=&Fort_Ambient_Indoors_Athena
 subtitlePriority=5000
 volumeMultiplier=0.501
 attenuationOverrides {
  bEnableFocusInterpolation=true
  bEnableListenerFocus=true
  bEnableOcclusion=true
  dBAttenuationAtMax=-40
  falloffDistance=900
  focusAzimuth=15
  focusVolumeAttenuation=1.5
  nonFocusAzimuth=45
  nonFocusVolumeAttenuation=0.5
  occlusionInterpolationTime=1
  occlusionLowPassFilterFrequency=10000
  occlusionTraceChannel=ECC_Camera
  occlusionVolumeAttenuation=0.5
  stereoSpread=512
  attenuationShapeExtents {y=64.5942, z=90}
 }
 concurrencyOverrides {bLimitToOwner=true, maxCount=2}
}
