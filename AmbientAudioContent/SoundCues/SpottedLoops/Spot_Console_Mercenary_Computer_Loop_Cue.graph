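# Spotted ambient loop cue for a mercenary computer console, using its own attenuation asset and a limited-concurrency set.
# A mixer layers a boosted looping terminal hum with two quieter streams of randomized computer one-shots.
# One stream repeats a four-sound random pick back to back, while the other inserts a short random delay before each two-sound pick.
# Modulators vary the pitch of the one-shots to disguise repetition.
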SoundCue /AmbientAudioContent/SoundCues/SpottedLoops/Spot_Console_Mercenary_Computer_Loop_Cue
root=Mixer_5
nodes [
 {id=Delay_1, children [Modulator_2], properties {delayMax=1, delayMin=0.5}}
 {id=Looping_0, children [Modulator_1]}
 {id=Looping_1, children [Delay_1]}
 {id=Mixer_5, children [Looping_0,Looping_1,Modulator_0], properties {inputVolume [0.4,0.2,3]}}
 {id=Modulator_0, children [WavePlayer_0], properties {pitchMin=1}}
 {id=Modulator_1, children [Random_1], properties {pitchMax=1.15, volumeMax=1, volumeMin=1}}
 {id=Modulator_2, children [Random_2], properties {pitchMax=1.15, volumeMax=1, volumeMin=1}}
 {id=Random_1, children [WavePlayer_1,WavePlayer_2,WavePlayer_3,WavePlayer_9], properties {weights [1,1,1,1]}}
 {id=Random_2, children [WavePlayer_4, WavePlayer_10], properties {weights [1,1]}}
 {
  id=WavePlayer_0
  properties {bLooping=true, soundWaveAssetPtr=/AmbientAudioContent/Waves/SpottedLoops/Spot_Console_Terminal_Communication_Hum_Loop_04}
 }
 {id=WavePlayer_1, properties {soundWaveAssetPtr=/AmbientAudioContent/Waves/OneShots/Spot_SecretComplex_Computer_Oneshot_02}}
 {id=WavePlayer_10, properties {soundWaveAssetPtr=/AmbientAudioContent/Waves/OneShots/Spot_SecretComplex_Computer_Oneshot_07}}
 {id=WavePlayer_2, properties {soundWaveAssetPtr=/AmbientAudioContent/Waves/OneShots/Spot_SecretComplex_Computer_Oneshot_03}}
 {id=WavePlayer_3, properties {soundWaveAssetPtr=/AmbientAudioContent/Waves/OneShots/Spot_SecretComplex_Computer_Oneshot_04}}
 {id=WavePlayer_4, properties {soundWaveAssetPtr=/AmbientAudioContent/Waves/OneShots/Spot_SecretComplex_Computer_Oneshot_01}}
 {id=WavePlayer_9, properties {soundWaveAssetPtr=/AmbientAudioContent/Waves/OneShots/Spot_SecretComplex_Computer_Oneshot_05}}
]
properties {
 attenuationSettings=/Game/Sounds/Helios/Environment_Actors/Console_Mercenary/Spot_Console_Mercenary_Computer_Loop_Att
 maxDistance=625
 soundClassObject=&Fort_Gadgets_Items
 volumeMultiplier=0.2
 attenuationOverrides {falloffDistance=2000}
 concurrencyOverrides {resolutionRule=PreventNew}
 concurrencySet [/AmbientAudioBase/Concurrency/AmbientConc_3Max_Spot_Default_Concurrency]
}
