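# Spot ambient Sound Cue that plays intermittent walkie-talkie radio chatter, named for a console radio communication prop.
# An endless loop waits a random short delay, then randomly picks one of six equally weighted one-shots with slight pitch variation.
# An external attenuation asset wraps the chain for spatialization.
# The quiet cue routes to an ambient props sound class and references a shared spot concurrency set to limit simultaneous instances.
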SoundCue /AmbientAudioContent/SoundCues/SpottedLoops/Spot_Console_RadioCommunication_Loop_01_Cue
root=Attenuation_0
nodes [
 {
  id=Attenuation_0
  children [Looping_0]
  properties {attenuationSettings=/Game/Sounds/Helios/Environment_Actors/Console_RadioCommunication/Spot_Console_RadioCommunication_Loop_Att}
 }
 {id=Delay_0, children [Modulator_0], properties {delayMax=3}}
 {id=Looping_0, children [Delay_0]}
 {id=Modulator_0, children [Random_1], properties {pitchMax=1.1, pitchMin=0.9}}
 {
  id=Random_1
  children [WavePlayer_0, WavePlayer_7, WavePlayer_8, WavePlayer_9, WavePlayer_10, WavePlayer_11]
  properties {weights [1,1,1,1,1,1]}
 }
 {id=WavePlayer_0, properties {soundWaveAssetPtr=/AmbientAudioContent/Waves/OneShots/Spot_Radio_Communication_Walkie_Talkie_01}}
 {id=WavePlayer_10, properties {soundWaveAssetPtr=/AmbientAudioContent/Waves/OneShots/Spot_Radio_Communication_Walkie_Talkie_05}}
 {id=WavePlayer_11, properties {soundWaveAssetPtr=/AmbientAudioContent/Waves/OneShots/Spot_Radio_Communication_Walkie_Talkie_06}}
 {id=WavePlayer_7, properties {soundWaveAssetPtr=/AmbientAudioContent/Waves/OneShots/Spot_Radio_Communication_Walkie_Talkie_02}}
 {id=WavePlayer_8, properties {soundWaveAssetPtr=/AmbientAudioContent/Waves/OneShots/Spot_Radio_Communication_Walkie_Talkie_03}}
 {id=WavePlayer_9, properties {soundWaveAssetPtr=/AmbientAudioContent/Waves/OneShots/Spot_Radio_Communication_Walkie_Talkie_04}}
]
properties {
 maxDistance=850
 soundClassObject=&Fort_Ambient_Props_Athena
 volumeMultiplier=0.125
 concurrencyOverrides {maxCount=4, resolutionRule=PreventNew, volumeScale=0.5}
 concurrencySet [/AmbientAudioBase/Concurrency/AmbientConc_3Max_Spot_Default_Concurrency]
}
