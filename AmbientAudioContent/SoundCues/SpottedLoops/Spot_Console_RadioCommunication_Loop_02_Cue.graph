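# Spotted ambient Sound Cue that plays looping, intermittent walkie-talkie radio chatter, named for a radio communication console.
# A Looping node repeats a random delay, pitch modulation and an equal-weight pick from six one-shot waves, all under a shared attenuation asset.
# The quiet cue uses the ambient props sound class and a spot concurrency set, with sends to the exterior ambient source bus and default reverb submix.
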SoundCue /AmbientAudioContent/SoundCues/SpottedLoops/Spot_Console_RadioCommunication_Loop_02_Cue
root=Attenuation_0
nodes [
 {
  id=Attenuation_0
  children [Looping_0]
  properties {attenuationSettings=/Game/Sounds/Helios/Environment_Actors/Console_RadioCommunication/Spot_Console_RadioCommunication_Loop_Att}
 }
 {id=Delay_0, children [Modulator_0], properties {delayMax=3}}
 {id=Looping_0, children [Delay_0]}
 {id=Modulator_0, children [Random_1], properties {pitchMax=1.1, pitchMin=0.9}}
 {
  id=Random_1
  children [WavePlayer_1, WavePlayer_7, WavePlayer_8, WavePlayer_9, WavePlayer_10, WavePlayer_11]
  properties {weights [1,1,1,1,1,1]}
 }
 {id=WavePlayer_1, properties {soundWaveAssetPtr=/AmbientAudioContent/Waves/OneShots/Spot_Radio_Communication_Walkie_Talkie_01}}
 {id=WavePlayer_10, properties {soundWaveAssetPtr=/AmbientAudioContent/Waves/OneShots/Spot_Radio_Communication_Walkie_Talkie_05}}
 {id=WavePlayer_11, properties {soundWaveAssetPtr=/AmbientAudioContent/Waves/OneShots/Spot_Radio_Communication_Walkie_Talkie_06}}
 {id=WavePlayer_7, properties {soundWaveAssetPtr=/AmbientAudioContent/Waves/OneShots/Spot_Radio_Communication_Walkie_Talkie_02}}
 {id=WavePlayer_8, properties {soundWaveAssetPtr=/AmbientAudioContent/Waves/OneShots/Spot_Radio_Communication_Walkie_Talkie_03}}
 {id=WavePlayer_9, properties {soundWaveAssetPtr=/AmbientAudioContent/Waves/OneShots/Spot_Radio_Communication_Walkie_Talkie_04}}
]
properties {
 maxDistance=850
 soundClassObject=&Fort_Ambient_Props_Athena
 volumeMultiplier=0.25
 busSends [{
   audioBus=null
   minSendLevel=0.25
   sendLevel=1
   soundSourceBus=&AmbientAudio_Exterior_SourceBus
   customSendLevelCurve {editorCurveData {keys [<at 0 0.25>,<at 1 1>]}}
  }]
 concurrencyOverrides {maxCount=2, volumeScale=0.5}
 concurrencySet [/AmbientAudioBase/Concurrency/AmbientConc_3Max_Spot_Default_Concurrency]
 soundSubmixSends [{
   maxSendDistance=3000
   minSendLevel=0.25
   soundSubmix=&Fortnite_Reverb_Default
   customSendLevelCurve {editorCurveData {keys [<at 0 0.25>,<at 1 1>]}}
  }]
}
