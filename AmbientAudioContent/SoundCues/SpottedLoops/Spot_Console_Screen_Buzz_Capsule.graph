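# Looping ambient spot sound cue for a console screen buzz, classed as an ambient prop sound.
# A Random node picks one of three equally weighted looping hologram waves, then feeds a Modulator for variation and a Mixer that halves the volume.
# Overridden attenuation uses a short-range capsule with listener focus, so loudness depends on facing.
# A shared ambient spot concurrency asset governs simultaneous instances.
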SoundCue /AmbientAudioContent/SoundCues/SpottedLoops/Spot_Console_Screen_Buzz_Capsule
root=Mixer_0
nodes [
 {id=Mixer_0, children [Modulator_1], properties {inputVolume [0.5]}}
 {id=Modulator_1, children [Random_0], properties {pitchMin=1}}
 {id=Random_0, children [WavePlayer_1,WavePlayer_2,WavePlayer_3], properties {weights [1,1,1]}}
 {id=WavePlayer_1, properties {bLooping=true, soundWaveAssetPtr=/AmbientAudioContent/Waves/SpottedLoops/Spot_Keys_And_Locks_HologramLoop_01}}
 {id=WavePlayer_2, properties {bLooping=true, soundWaveAssetPtr=/AmbientAudioContent/Waves/SpottedLoops/Spot_Keys_And_Locks_HologramLoop_02}}
 {id=WavePlayer_3, properties {bLooping=true, soundWaveAssetPtr=/AmbientAudioContent/Waves/SpottedLoops/Spot_Keys_And_Locks_HologramLoop_03}}
]
properties {
 attenuationSettings=/Game/Sounds/Helios/Environment_Actors/LaptopScreenStatic/Spot_Laptop_Screen_Static_Sphere_Att
 bOverrideAttenuation=true
 maxDistance=453
 soundClassObject=&Fort_Ambient_Props_Athena
 volumeMultiplier=1
 attenuationOverrides {
  attenuationShape=Capsule
  bEnableListenerFocus=true
  falloffDistance=350
  focusAzimuth=25
  focusVolumeAttenuation=1.2
  nonFocusAzimuth=55
  nonFocusVolumeAttenuation=0.75
  attenuationShapeExtents=<103 50>
 }
 concurrencyOverrides {maxCount=6}
 concurrencySet [/AmbientAudioBase/Concurrency/AmbientConc_3Max_Spot_Default_Concurrency]
}
