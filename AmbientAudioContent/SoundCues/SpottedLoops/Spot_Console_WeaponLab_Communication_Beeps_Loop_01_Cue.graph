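# Looping spot ambience cue for communication console beeps, named for a weapon lab console.
# A Random node picks one of four looping beep waves with equal weight, and every quality level feeds from that same node.
# It plays quietly in an ambient props sound class, using a console terminal hum attenuation asset and a shared spot-ambience concurrency set.
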SoundCue /AmbientAudioContent/SoundCues/SpottedLoops/Spot_Console_WeaponLab_Communication_Beeps_Loop_01_Cue
root=QualityLevel_0
nodes [
 {id=QualityLevel_0, children [Random_0,Random_0,Random_0,Random_0]}
 {id=Random_0, children [WavePlayer_1,WavePlayer_5,WavePlayer_6,WavePlayer_7], properties {weights [1,1,1,1]}}
 {
  id=WavePlayer_1
  properties {bLooping=true, soundWaveAssetPtr=/AmbientAudioContent/Waves/SpottedLoops/Spot_Console_Communication_Beeping_Loop_01}
 }
 {
  id=WavePlayer_5
  properties {bLooping=true, soundWaveAssetPtr=/AmbientAudioContent/Waves/SpottedLoops/Spot_Console_Communication_Beeping_Loop_04}
 }
 {
  id=WavePlayer_6
  properties {bLooping=true, soundWaveAssetPtr=/AmbientAudioContent/Waves/SpottedLoops/Spot_Console_Communication_Beeping_Loop_05}
 }
 {
  id=WavePlayer_7
  properties {bLooping=true, soundWaveAssetPtr=/AmbientAudioContent/Waves/SpottedLoops/Spot_Console_Communication_Beeping_Loop_06}
 }
]
properties {
 attenuationSettings=/Game/Sounds/Helios/Environment_Actors/ConsoleTerminals/Spot_Console_Terminal_Communication_Hum_Att
 maxDistance=1000
 soundClassObject=&Fort_Ambient_Props_Athena
 subtitlePriority=5000
 volumeMultiplier=0.251
 attenuationOverrides {
  bEnableFocusInterpolation=true
  bEnableListenerFocus=true
  bEnableOcclusion=true
  dBAttenuationAtMax=-40
  falloffDistance=512
  focusAzimuth=15
  focusVolumeAttenuation=1.5
  nonFocusAzimuth=45
  nonFocusVolumeAttenuation=0.5
  occlusionInterpolationTime=1
  occlusionLowPassFilterFrequency=10000
  occlusionTraceChannel=ECC_GameTraceChannel11
  occlusionVolumeAttenuation=0.5
  stereoSpread=512
  attenuationShapeExtents=<256 64.5942 90>
 }
 concurrencyOverrides {bLimitToOwner=true, maxCount=2}
 concurrencySet [/AmbientAudioBase/Concurrency/AmbientConc_3Max_Spot_Default_Concurrency]
}
