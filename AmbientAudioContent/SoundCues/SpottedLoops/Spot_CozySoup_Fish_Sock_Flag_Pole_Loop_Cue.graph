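# Looping ambient sound cue for a CozySoup fish sock flag pole prop, in the Athena ambient props sound class.
# A mixer blends pitched-up flag and banner flapping loops with a quiet, pitched-down, looped chainlink rattle, then boosts the result.
# An attenuation node adds listener focus and occlusion, while cue-level overrides disable attenuation and spatialization and cap concurrency.
# The quality switch plays this chain on the first two levels and leaves the other two empty.
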SoundCue /AmbientAudioContent/SoundCues/SpottedLoops/Spot_CozySoup_Fish_Sock_Flag_Pole_Loop_Cue
root=QualityLevel_0
nodes [
 {
  id=Attenuation_0
  children [Modulator_2]
  properties {
   bOverrideAttenuation=true
   attenuationOverrides {
    bEnableFocusInterpolation=true
    bEnableListenerFocus=true
    bEnableOcclusion=true
    falloffDistance=1500
    focusAzimuth=15
    focusVolumeAttenuation=1.2
    nonFocusAzimuth=45
    nonFocusVolumeAttenuation=0.4
    occlusionTraceChannel=ECC_GameTraceChannel11
    occlusionVolumeAttenuation=0.2
    attenuationShapeExtents {x=20}
   }
  }
 }
 {id=Mixer_0, children [Modulator_0,Modulator_1,Modulator_3], properties {inputVolume [1,0.7,0.8]}}
 {id=Modulator_0, children [WavePlayer_2], properties {pitchMax=1.4, pitchMin=1.2}}
 {id=Modulator_1, children [WavePlayer_3], properties {pitchMax=0.65, pitchMin=0.55, volumeMax=0.1, volumeMin=0.05}}
 {id=Modulator_2, children [Mixer_0], properties {volumeMax=2.35, volumeMin=2.25}}
 {id=Modulator_3, children [WavePlayer_0], properties {pitchMax=1.2, pitchMin=1.2}}
 {id=QualityLevel_0, children [Attenuation_0, Attenuation_0, null, null]}
 {id=WavePlayer_0, properties {bLooping=true, soundWaveAssetPtr=/AmbientAudioContent/Waves/SpottedLoops/Spot_Flag_Flap_Cloth_Banner_Slow_Loop}}
 {id=WavePlayer_2, properties {bLooping=true, soundWaveAssetPtr=/AmbientAudioContent/Waves/SpottedLoops/Spot_CTF_FlagLoop_01}}
 {
  id=WavePlayer_3
  properties {bLooping=true, soundWaveAssetPtr=/AmbientAudioContent/Waves/OneShots/Spot_Wind_Excited_Chainlink_Fence_Rattle_01}
 }
]
properties {
 attenuationSettings=&Fort_Ambient_Athena_Att
 bOverrideAttenuation=true
 bOverrideConcurrency=true
 maxDistance=2097152
 soundClassObject=&Fort_Ambient_Props_Athena
 volumeMultiplier=1.1
 attenuationOverrides {
  bAttenuate=false
  bEnableFocusInterpolation=true
  bEnableReverbSend=false
  bEnableSendToAudioLink=false
  bSpatialize=false
  binauralRadius=1
  dBAttenuationAtMax=-40
  distanceAlgorithm=Logarithmic
  falloffDistance=1000
  focusAzimuth=15
  focusVolumeAttenuation=1.1
  nonFocusAzimuth=45
  nonFocusVolumeAttenuation=0.4
  nonSpatializedRadiusEnd=20
  nonSpatializedRadiusStart=30
  occlusionTraceChannel=ECC_GameTraceChannel11
  occlusionVolumeAttenuation=0.2
  stereoSpread=2000
  attenuationShapeExtents=<10 150>
  customAttenuationCurve {editorCurveData {keys [<at 0 1>,<at 400 0>]}}
 }
 concurrencyOverrides {maxCount=6}
}
