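# Spotted ambient loop cue for a rotating metal air-con prop, playing one of two looping exhaust-stack waves at random.
# All quality levels share a random branch that starts playback either immediately or after a short fixed delay, staggering instances.
# Both branches feed an enveloper volume ramp and a modulator above the wave choice.
# It uses desk-fan attenuation, overrides concurrency to cap voices, and sends to reverb by distance.
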SoundCue /AmbientAudioContent/SoundCues/SpottedLoops/Spot_CrewCut_AirCon_Metal_Rotating_Loop_Cue
root=QualityLevel_0
nodes [
 {id=Delay_0, children [Enveloper_1], properties {delayMax=0.3, delayMin=0.3}}
 {
  id=Enveloper_1
  children [Modulator_0]
  properties {pitchCurve {editorCurveData {keys [{time=0}]}}, volumeCurve {editorCurveData {keys [<at 0 0>,<at 1 1>]}}}
 }
 {id=Modulator_0, children [Random_2]}
 {id=QualityLevel_0, children [Random_1,Random_1,Random_1,Random_1]}
 {id=Random_1, children [Delay_0,Enveloper_1], properties {weights [1,1]}}
 {id=Random_2, children [WavePlayer_3,WavePlayer_4], properties {weights [1,1]}}
 {
  id=WavePlayer_3
  properties {bLooping=true, soundWaveAssetPtr=/AmbientAudioContent/Waves/SpottedLoops/Spot_ExhaustStackA_Rotating_Metal_Loop_01}
 }
 {
  id=WavePlayer_4
  properties {bLooping=true, soundWaveAssetPtr=/AmbientAudioContent/Waves/SpottedLoops/Spot_ExhaustStackA_Rotating_Metal_Loop_02}
 }
]
properties {
 attenuationSettings=/AmbientAudioBase/Attenuations/Spot_Desk_Electric_Fan_Att
 bEnableBusSends=false
 bOverrideConcurrency=true
 maxDistance=600
 soundClassObject=&Fort_Ambient_Props_Athena
 volumeMultiplier=0.3
 attenuationOverrides {distanceAlgorithm=Logarithmic, falloffDistance=850, nonFocusVolumeAttenuation=0.4, stereoSpread=0, attenuationShapeExtents {x=250}}
 concurrencyOverrides {maxCount=2, voiceStealReleaseTime=0.2}
 concurrencySet [/AmbientAudioBase/Concurrency/AmbientConc_3Max_Spot_Default_Concurrency]
 soundSubmixSends [{
   maxSendDistance=750
   maxSendLevel=0.6
   minSendDistance=250
   minSendLevel=0.25
   soundSubmix=&Fortnite_Reverb_Default
   customSendLevelCurve {editorCurveData {keys [<at 0 0>,<at 1 1>]}}
  }]
}
