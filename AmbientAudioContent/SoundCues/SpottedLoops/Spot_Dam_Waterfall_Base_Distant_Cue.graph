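# Looping ambient SoundCue for the distant layer of the dam waterfall base, routed to the outdoor Athena ambient sound class.
# A single looping Wave Player feeds the output at roughly half volume.
# Overridden attenuation uses a custom curve that is silent at the source, peaks about a third of the way out and fades to zero at maximum range.
# Wide stereo spread, near-field non-spatialization and slowly interpolated low-pass occlusion shape the sound.
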SoundCue /AmbientAudioContent/SoundCues/SpottedLoops/Spot_Dam_Waterfall_Base_Distant_Cue
root=WavePlayer_0
nodes [{id=WavePlayer_0, properties {bLooping=true, soundWaveAssetPtr=/AmbientAudioContent/Waves/Water/Spot_Dam_Waterfall_Base_Distant_01}}]
properties {
 bOverrideAttenuation=true
 maxDistance=20000
 soundClassObject=&Fort_Ambient_Outdoors_Athena
 volumeMultiplier=0.501
 attenuationOverrides {
  bEnableFocusInterpolation=true
  bEnableOcclusion=true
  distanceAlgorithm=Custom
  falloffDistance=20000
  focusAzimuth=60
  nonFocusAzimuth=90
  nonFocusVolumeAttenuation=0.75
  nonSpatializedRadiusStart=5000
  occlusionInterpolationTime=3
  occlusionLowPassFilterFrequency=700
  occlusionTraceChannel=ECC_GameTraceChannel11
  stereoSpread=10000
  attenuationShapeExtents {x=0}
  customAttenuationCurve {editorCurveData {keys [
     <at 0 0>
     <at 0.3 1>
     {arriveTangent=-0.3968, interpMode=RCIM_Cubic, leaveTangent=-0.2706, tangentMode=RCTM_Break, time=0.45, value=0.9637}
     <at 1 0>
    ]}}
 }
}
