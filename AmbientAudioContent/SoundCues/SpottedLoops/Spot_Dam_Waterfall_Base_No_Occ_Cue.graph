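# Looping ambient spot cue for the base of a dam waterfall, played on the water interaction sound class with no occlusion.
# A Random node picks one of three equally weighted, looping close-up waterfall waves.
# Attenuation overrides add distance low-pass filtering, listener focus that lowers off-focus volume, and a non-spatialized blend near the source.
# Concurrency caps simultaneous instances at three, stopping the quietest with a release fade.
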SoundCue /AmbientAudioContent/SoundCues/SpottedLoops/Spot_Dam_Waterfall_Base_No_Occ_Cue
root=Random_1
nodes [
 {id=Random_1, children [WavePlayer_0,WavePlayer_1,WavePlayer_2], properties {weights [1,1,1]}}
 {id=WavePlayer_0, properties {bLooping=true, soundWaveAssetPtr=/AmbientAudioContent/Waves/Water/Spot_Dam_Waterfall_Base_Close_01a}}
 {id=WavePlayer_1, properties {bLooping=true, soundWaveAssetPtr=/AmbientAudioContent/Waves/Water/Spot_Dam_Waterfall_Base_Close_01b}}
 {id=WavePlayer_2, properties {bLooping=true, soundWaveAssetPtr=/AmbientAudioContent/Waves/Water/Spot_Dam_Waterfall_Base_Close_01c}}
]
properties {
 bOverrideAttenuation=true
 bOverrideConcurrency=true
 maxDistance=8500
 soundClassObject=/Game/Sounds/FortSoundClassesAndModes/Fort_Water_Interaction
 attenuationOverrides {
  bAttenuateWithLPF=true
  bEnableFocusInterpolation=true
  bEnableListenerFocus=true
  falloffDistance=6500
  lPFFrequencyAtMax=1000
  lPFRadiusMax=4000
  lPFRadiusMin=1000
  nonFocusVolumeAttenuation=0.6
  nonSpatializedRadiusEnd=2000
  nonSpatializedRadiusStart=1000
  occlusionInterpolationTime=2
  occlusionLowPassFilterFrequency=1000
  occlusionTraceChannel=ECC_GameTraceChannel11
  stereoSpread=500
  attenuationShapeExtents {x=2000}
 }
 concurrencyOverrides {maxCount=3, resolutionRule=StopQuietest, voiceStealReleaseTime=2}
}
