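# Looping spot ambience cue for a dam waterfall intake, playing one water wave through the Fort_Ambient_TileMix sound class at reduced volume.
# A single wave player feeds the output, with capsule-shaped attenuation overridden on the cue itself.
# The attenuation adds listener focus, distance-based reverb send and a non-spatialized near radius.
# Despite the No_Occ name, occlusion is enabled on a game trace channel, with volume reduction and low-pass filtering.
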SoundCue /AmbientAudioContent/SoundCues/SpottedLoops/Spot_Dam_Waterfall_Intake_No_Occ_Cue
root=WavePlayer_0
nodes [{id=WavePlayer_0, properties {bLooping=true, soundWaveAssetPtr=/AmbientAudioContent/Waves/Water/Spot_Dam_Waterfall_Intake_01}}]
properties {
 bOverrideAttenuation=true
 maxDistance=6828.0054
 soundClassObject=/Game/Sounds/FortSoundClassesAndModes/Fort_Ambient_TileMix
 volumeMultiplier=0.6
 attenuationOverrides {
  attenuationShape=Capsule
  bEnableFocusInterpolation=true
  bEnableListenerFocus=true
  bEnableOcclusion=true
  falloffDistance=6000
  focusAzimuth=40
  nonFocusAzimuth=80
  nonFocusVolumeAttenuation=0.6
  nonSpatializedRadiusEnd=2000
  nonSpatializedRadiusStart=1000
  occlusionInterpolationTime=1
  occlusionLowPassFilterFrequency=16000
  occlusionTraceChannel=ECC_GameTraceChannel11
  occlusionVolumeAttenuation=0.4
  reverbDistanceMax=6000
  reverbWetLevelMax=0.5
  attenuationShapeExtents=<828.0052 180>
 }
}
