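# Looping spot ambience cue for a dark statue presence, pitched down, lowered in volume and assigned to the Fort_Ambient_Athena sound class.
# A Mixer layers two looping waves, a castle throne hall ambience and a brimstone pillar tonal loop.
# Two nested, endlessly looping Envelopers of different lengths add slow movement, the inner varying pitch and the outer gently dipping volume.
# Overridden attenuation uses a capsule shape with logarithmic falloff and listener focus.
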SoundCue /AmbientAudioContent/SoundCues/SpottedLoops/Spot_Dark_Statue_Pressence_02_Loop_Cue
root=Enveloper_5
nodes [
 {
  id=Enveloper_4
  children [Mixer_0]
  properties {
   bLoop=true
   bLoopIndefinitely=true
   loopEnd=30
   pitchCurve {editorCurveData {keys [{interpMode=RCIM_Cubic, time=0}, <at 15.9468 0.9181 RCIM_Cubic>, {interpMode=RCIM_Cubic, time=30}]}}
   volumeCurve {editorCurveData {keys [<at 0 1>]}}
  }
 }
 {
  id=Enveloper_5
  children [Enveloper_4]
  properties {
   bLoop=true
   bLoopIndefinitely=true
   loopEnd=20
   pitchCurve {editorCurveData {keys [{time=0}]}}
   volumeCurve {editorCurveData {keys [
      {arriveTangent=0, interpMode=RCIM_Cubic, leaveTangent=0, time=0, value=1}
      {arriveTangent=0, interpMode=RCIM_Cubic, leaveTangent=0, time=10, value=0.9}
      {arriveTangent=0, interpMode=RCIM_Cubic, leaveTangent=0, time=20, value=0.9992}
     ]}}
  }
 }
 {id=Mixer_0, children [WavePlayer_1, WavePlayer_50], properties {inputVolume [1.2,1]}}
 {
  id=WavePlayer_1
  properties {bLooping=true, soundWaveAssetPtr=/Game/Sounds/Helios/POI/CastleDoom/Ambience_DCastle_Throne_Hall_Interior_Loop_Stereo}
 }
 {
  id=WavePlayer_50
  properties {bLooping=true, soundWaveAssetPtr=/AmbientAudioContent/Waves/SpottedLoops/Spot_Brimstone_Pillar_Tonal_Middle_Loop_01}
 }
]
properties {
 attenuationSettings=&Fort_Ambient_Athena_Att
 bOverrideAttenuation=true
 maxDistance=2750
 pitchMultiplier=0.8
 soundClassObject=&Fort_Ambient_Athena
 volumeMultiplier=0.5
 attenuationOverrides {
  attenuationShape=Capsule
  bEnableListenerFocus=true
  distanceAlgorithm=Logarithmic
  falloffDistance=1500
  focusVolumeAttenuation=1.2
  nonFocusVolumeAttenuation=0.4
  attenuationShapeExtents=<1250 750>
 }
 preEffectBusSends [{}]
}
