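# Looping interior-window ambience cue of desert wind and blowing sand, spatialized by an interior window wind attenuation.
# A mixer layers endlessly looping, weighted-random gust one-shots over a half-volume sand-blowing bed, with modulators varying pitch and volume.
# Nested looping envelopers lower the bed's pitch and slowly drift its pitch and volume.
# Quality switches give the first two levels larger gust and bed pools, and the last two fewer gusts and one bed wave.
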SoundCue /AmbientAudioContent/SoundCues/SpottedLoops/Spot_Desert_Window_Int_Loop_Cue
root=Mixer_1
nodes [
 {
  id=Enveloper_4
  children [Enveloper_5]
  properties {
   bLoop=true
   bLoopIndefinitely=true
   loopEnd=100
   pitchCurve {editorCurveData {keys [<at 0 0.8>]}}
   volumeCurve {editorCurveData {keys [<at 0 1 RCIM_Cubic>,<at 50 0.9 RCIM_Cubic>,<at 100 1 RCIM_Cubic>]}}
  }
 }
 {
  id=Enveloper_5
  children [Modulator_6]
  properties {
   bLoop=true
   bLoopIndefinitely=true
   loopEnd=30
   pitchCurve {editorCurveData {keys [{interpMode=RCIM_Cubic, time=0}, <at 15.9468 0.9181 RCIM_Cubic>, {interpMode=RCIM_Cubic, time=30}]}}
   volumeCurve {editorCurveData {keys [<at 0 1>]}}
  }
 }
 {
  id=Enveloper_6
  children [Enveloper_7]
  properties {
   bLoop=true
   bLoopIndefinitely=true
   loopEnd=100
   pitchCurve {editorCurveData {keys [<at 0 0.8>]}}
   volumeCurve {editorCurveData {keys [<at 0 1 RCIM_Cubic>,<at 50 0.9 RCIM_Cubic>,<at 100 1 RCIM_Cubic>]}}
  }
 }
 {
  id=Enveloper_7
  children [Modulator_7]
  properties {
   bLoop=true
   bLoopIndefinitely=true
   loopEnd=30
   pitchCurve {editorCurveData {keys [{interpMode=RCIM_Cubic, time=0}, <at 15.9468 0.9181 RCIM_Cubic>, {interpMode=RCIM_Cubic, time=30}]}}
   volumeCurve {editorCurveData {keys [<at 0 1>]}}
  }
 }
 {id=Looping_2, children [Modulator_4]}
 {id=Looping_3, children [Modulator_5]}
 {id=Mixer_1, children [QualityLevel_2, QualityLevel_3], properties {inputVolume [1,0.5]}}
 {id=Modulator_4, children [Random_1], properties {pitchMax=1.1, pitchMin=1, volumeMax=1.1, volumeMin=0.9}}
 {id=Modulator_5, children [Random_7], properties {pitchMax=1.1, volumeMax=1.1, volumeMin=0.9}}
 {id=Modulator_6, children [Random_21], properties {pitchMax=1.1, pitchMin=1, volumeMax=1.1, volumeMin=0.9}}
 {id=Modulator_7, children [WavePlayer_46], properties {pitchMax=1.1, pitchMin=1, volumeMax=1.1, volumeMin=0.9}}
 {id=QualityLevel_2, children [Looping_2,Looping_2,Looping_3,Looping_3]}
 {id=QualityLevel_3, children [Enveloper_4,Enveloper_4,Enveloper_6,Enveloper_6]}
 {id=Random_1, children [Random_16,Random_12,Random_11], properties {weights [1,0.33,0.1]}}
 {
  id=Random_11
  children [WavePlayer_5, WavePlayer_6, WavePlayer_7, WavePlayer_17, WavePlayer_18, WavePlayer_19]
  properties {weights [1,1,1,1,1,1]}
 }
 {
  id=Random_12
  children [WavePlayer_21, WavePlayer_39, WavePlayer_40, WavePlayer_41, WavePlayer_42, WavePlayer_43, WavePlayer_44, WavePlayer_45]
  properties {weights [1,1,1,1,1,1,1,1]}
 }
 {id=Random_15, children [WavePlayer_57, WavePlayer_58, WavePlayer_65], properties {weights [1,1,1]}}
 {id=Random_16, children [Random_17,Random_18], properties {weights [1,1]}}
 {id=Random_17, children [WavePlayer_47, WavePlayer_48, WavePlayer_49, WavePlayer_50, WavePlayer_51], properties {weights [1,1,1,1,1]}}
 {id=Random_18, children [WavePlayer_52, WavePlayer_53, WavePlayer_54, WavePlayer_55, WavePlayer_56], properties {weights [1,1,1,1,1]}}
 {id=Random_19, children [WavePlayer_73, WavePlayer_66, WavePlayer_67], properties {weights [1,1,1]}}
 {id=Random_20, children [WavePlayer_74, WavePlayer_75, WavePlayer_76, WavePlayer_77], properties {weights [1,1,1,1]}}
 {id=Random_21, children [WavePlayer_78, WavePlayer_79, WavePlayer_80, WavePlayer_81], properties {weights [1,1,1,1]}}
 {id=Random_7, children [Random_15,Random_19,Random_20], properties {weights [1,0.33,0.1]}}
 {id=WavePlayer_17, properties {soundWaveAssetPtr=/AmbientAudioContent/Waves/OneShots/Spot_Desert_WindGust_Med_hv_Sandy_03}}
 {id=WavePlayer_18, properties {soundWaveAssetPtr=/AmbientAudioContent/Waves/OneShots/Spot_Desert_WindGust_Med_hv_Sandy_04}}
 {id=WavePlayer_19, properties {soundWaveAssetPtr=/AmbientAudioContent/Waves/OneShots/Spot_Desert_WindGust_Med_hv_Sandy_05}}
 {id=WavePlayer_21, properties {soundWaveAssetPtr=/Game/Sounds/Apollo/Biomes/Tower_Aftermath/AMB_Tower_AM_Sandy_Wind_Oneshot_01}}
 {id=WavePlayer_39, properties {soundWaveAssetPtr=/Game/Sounds/Apollo/Biomes/Tower_Aftermath/AMB_Tower_AM_Sandy_Wind_Oneshot_12}}
 {id=WavePlayer_40, properties {soundWaveAssetPtr=/Game/Sounds/Apollo/Biomes/Tower_Aftermath/AMB_Tower_AM_Sandy_Wind_Oneshot_02}}
 {id=WavePlayer_41, properties {soundWaveAssetPtr=/Game/Sounds/Apollo/Biomes/Tower_Aftermath/AMB_Tower_AM_Sandy_Wind_Oneshot_03}}
 {id=WavePlayer_42, properties {soundWaveAssetPtr=/Game/Sounds/Apollo/Biomes/Tower_Aftermath/AMB_Tower_AM_Sandy_Wind_Oneshot_04}}
 {id=WavePlayer_43, properties {soundWaveAssetPtr=/Game/Sounds/Apollo/Biomes/Tower_Aftermath/AMB_Tower_AM_Sandy_Wind_Oneshot_05}}
 {id=WavePlayer_44, properties {soundWaveAssetPtr=/Game/Sounds/Apollo/Biomes/Tower_Aftermath/AMB_Tower_AM_Sandy_Wind_Oneshot_06}}
 {id=WavePlayer_45, properties {soundWaveAssetPtr=/Game/Sounds/Apollo/Biomes/Tower_Aftermath/AMB_Tower_AM_Sandy_Wind_Oneshot_07}}
 {
  id=WavePlayer_46
  properties {bLooping=true, soundWaveAssetPtr=/AmbientAudioContent/Waves/Biomes/Spot_Nitro_Wasteland_Desert_Sand_Blowing_Loop_01}
 }
 {id=WavePlayer_47, properties {soundWaveAssetPtr=/AmbientAudioContent/Waves/OneShots/Spot_Desert_Wind_Sand_Gust_01}}
 {id=WavePlayer_48, properties {soundWaveAssetPtr=/AmbientAudioContent/Waves/OneShots/Spot_Desert_Wind_Sand_Gust_02}}
 {id=WavePlayer_49, properties {soundWaveAssetPtr=/AmbientAudioContent/Waves/OneShots/Spot_Desert_Wind_Sand_Gust_03}}
 {id=WavePlayer_5, properties {soundWaveAssetPtr=/AmbientAudioContent/Waves/OneShots/Spot_Desert_WindGust_Med_hv_Sandy_06}}
 {id=WavePlayer_50, properties {soundWaveAssetPtr=/AmbientAudioContent/Waves/OneShots/Spot_Desert_Wind_Sand_Gust_04}}
 {id=WavePlayer_51, properties {soundWaveAssetPtr=/AmbientAudioContent/Waves/OneShots/Spot_Desert_Wind_Sand_Gust_05}}
 {id=WavePlayer_52, properties {soundWaveAssetPtr=/AmbientAudioContent/Waves/OneShots/Spot_Desert_WindGust_Med_Lt_Debris_01}}
 {id=WavePlayer_53, properties {soundWaveAssetPtr=/AmbientAudioContent/Waves/OneShots/Spot_Desert_WindGust_Med_Lt_Debris_02}}
 {id=WavePlayer_54, properties {soundWaveAssetPtr=/AmbientAudioContent/Waves/OneShots/Spot_Desert_WindGust_Med_Lt_Debris_03}}
 {id=WavePlayer_55, properties {soundWaveAssetPtr=/AmbientAudioContent/Waves/OneShots/Spot_Desert_WindGust_Med_Lt_Debris_04}}
 {id=WavePlayer_56, properties {soundWaveAssetPtr=/AmbientAudioContent/Waves/OneShots/Spot_Desert_WindGust_Med_Lt_Debris_05}}
 {id=WavePlayer_57, properties {soundWaveAssetPtr=&Spot_Desert_WindGust_Med_Lt_Debris_01}}
 {id=WavePlayer_58, properties {soundWaveAssetPtr=&Spot_Desert_WindGust_Med_Lt_Debris_02}}
 {id=WavePlayer_6, properties {soundWaveAssetPtr=/AmbientAudioContent/Waves/OneShots/Spot_Desert_WindGust_Med_hv_Sandy_01}}
 {id=WavePlayer_65, properties {soundWaveAssetPtr=&Spot_Desert_WindGust_Med_Lt_Debris_03}}
 {id=WavePlayer_66, properties {soundWaveAssetPtr=&AMB_Tower_AM_Sandy_Wind_Oneshot_12}}
 {id=WavePlayer_67, properties {soundWaveAssetPtr=&AMB_Tower_AM_Sandy_Wind_Oneshot_02}}
 {id=WavePlayer_7, properties {soundWaveAssetPtr=/AmbientAudioContent/Waves/OneShots/Spot_Desert_WindGust_Med_hv_Sandy_02}}
 {id=WavePlayer_73, properties {soundWaveAssetPtr=&AMB_Tower_AM_Sandy_Wind_Oneshot_01}}
 {id=WavePlayer_74, properties {soundWaveAssetPtr=&Spot_Desert_WindGust_Med_hv_Sandy_02}}
 {id=WavePlayer_75, properties {soundWaveAssetPtr=&Spot_Desert_WindGust_Med_hv_Sandy_03}}
 {id=WavePlayer_76, properties {soundWaveAssetPtr=&Spot_Desert_WindGust_Med_hv_Sandy_04}}
 {id=WavePlayer_77, properties {soundWaveAssetPtr=&Spot_Desert_WindGust_Med_hv_Sandy_05}}
 {
  id=WavePlayer_78
  properties {bLooping=true, soundWaveAssetPtr=&Spot_Nitro_Wasteland_Desert_Sand_Blowing_Loop_01}
 }
 {
  id=WavePlayer_79
  properties {bLooping=true, soundWaveAssetPtr=/AmbientAudioContent/Waves/Biomes/Spot_Nitro_Wasteland_Desert_Sand_Blowing_Loop_02}
 }
 {
  id=WavePlayer_80
  properties {bLooping=true, soundWaveAssetPtr=/AmbientAudioContent/Waves/Biomes/Spot_Nitro_Wasteland_Desert_Sand_Blowing_Loop_03}
 }
 {
  id=WavePlayer_81
  properties {bLooping=true, soundWaveAssetPtr=/AmbientAudioContent/Waves/Biomes/Spot_Nitro_Wasteland_Desert_Sand_Blowing_Loop_04}
 }
]
properties {
 attenuationSettings=/Game/Sounds/Helios/Biomes/DesertWasteland/Spots/Spot_Interior_Window_Wind_Pan_Att
 bEnableBusSends=false
 maxDistance=1253.7231
 soundClassObject=/Game/Sounds/FortSoundClassesAndModes/Fort_Ambient_OneShots_Indoors
 volumeMultiplier=0.251
 preEffectBusSends [{}]
}
