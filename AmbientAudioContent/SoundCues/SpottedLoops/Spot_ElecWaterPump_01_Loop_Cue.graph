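# Looping ambient spot Sound Cue for an electric water pump, using capsule pipe attenuation and an exterior ambient bus send.
# A randomized start delay feeds a quality switch, with every level sharing one two-input mixer.
# The mixer blends a quieter random generator or electrical-box loop with a louder random water-pump or energy-mech loop.
# Modulators randomize pitch and set per-layer volume, while a cue-wide multiplier doubles pitch.
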SoundCue /AmbientAudioContent/SoundCues/SpottedLoops/Spot_ElecWaterPump_01_Loop_Cue
root=Delay_0
nodes [
 {id=Delay_0, children [QualityLevel_0], properties {delayMax=1}}
 {id=Mixer_1, children [Modulator_0,Random_1], properties {inputVolume [0.3,2]}}
 {id=Modulator_0, children [Random_0], properties {pitchMax=1.1, pitchMin=0.85}}
 {id=Modulator_1, children [WavePlayer_0], properties {pitchMax=1.1, pitchMin=0.9, volumeMax=0.8, volumeMin=0.8}}
 {id=Modulator_4, children [WavePlayer_4], properties {pitchMax=1.1, pitchMin=0.9, volumeMax=0.4, volumeMin=0.4}}
 {id=Modulator_5, children [WavePlayer_5], properties {pitchMax=1.2, pitchMin=1.1, volumeMax=0.4, volumeMin=0.4}}
 {id=QualityLevel_0, children [Mixer_1,Mixer_1,Mixer_1,Mixer_1]}
 {id=Random_0, children [WavePlayer_3,WavePlayer_1], properties {weights [1,1]}}
 {id=Random_1, children [Modulator_1,Modulator_4,Modulator_5], properties {weights [1,1,1]}}
 {id=WavePlayer_0, properties {bLooping=true, soundWaveAssetPtr=/AmbientAudioContent/Waves/SpottedLoops/Spot_Outpost_WaterPump_Loop_01}}
 {
  id=WavePlayer_1
  properties {bLooping=true, soundWaveAssetPtr=/AmbientAudioContent/Waves/SpottedLoops/Spot_RadarBase_Tinfoil_ElectricalBox_Loop_01}
 }
 {id=WavePlayer_3, properties {bLooping=true, soundWaveAssetPtr=/AmbientAudioContent/Waves/SpottedLoops/Spot_Electric_Generator_Loop_03}}
 {
  id=WavePlayer_4
  properties {bLooping=true, soundWaveAssetPtr=/AmbientAudioContent/Waves/SpottedLoops/Spot_HollowBase_Coaster_Energy_Mech_Loop1}
 }
 {
  id=WavePlayer_5
  properties {bLooping=true, soundWaveAssetPtr=&Spot_HollowBase_Coaster_Energy_Mech_Loop1}
 }
]
properties {
 attenuationSettings=/Game/Sounds/Helios/Environment_Actors/SewerPipes/Spot_UndergroundPipe_Loop_Capsule_Att
 maxDistance=1000
 pitchMultiplier=2
 soundClassObject=&Fort_Ambient_Props_Athena
 volumeMultiplier=1
 attenuationOverrides {falloffDistance=2000}
 busSends [{audioBus=null, minSendLevel=0.6, sendLevel=1, soundSourceBus=&AmbientAudio_Exterior_SourceBus}]
 concurrencyOverrides {maxCount=2}
 concurrencySet [/AmbientAudioBase/Concurrency/AmbientConc_3Max_Spot_Default_Concurrency]
}
