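# Looping spot ambience cue for an electric desk fan prop, routed to the Athena ambient props sound class.
# One slow-fan wave with slight random pitch variation feeds both branches of the root mixer through separate attenuation nodes.
# The second branch is mixed louder with a short custom falloff override, layering a close-up boost over the standard fan attenuation.
# Overridden concurrency caps it at two voices.
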SoundCue /AmbientAudioContent/SoundCues/SpottedLoops/Spot_Electric_Desk_Fan_Loop_Cue
root=Mixer_1
nodes [
 {id=Attenuation_0, children [Mixer_0], properties {attenuationSettings=/AmbientAudioBase/Attenuations/Spot_Desk_Electric_Fan_Att}}
 {
  id=Attenuation_1
  children [Mixer_0]
  properties {
   attenuationSettings=&Spot_Desk_Electric_Fan_Att
   bOverrideAttenuation=true
   attenuationOverrides {
    distanceAlgorithm=Custom
    falloffDistance=100
    attenuationShapeExtents {x=0}
    customAttenuationCurve {editorCurveData {keys [<at 0 1>,<at 1 0>]}}
   }
  }
 }
 {id=Mixer_0, children [Modulator_0,null], properties {inputVolume [0.6,1.1]}}
 {id=Mixer_1, children [Attenuation_0, Attenuation_1], properties {inputVolume [1,2]}}
 {id=Modulator_0, children [WavePlayer_0], properties {pitchMax=1.1, pitchMin=1}}
 {id=WavePlayer_0, properties {bLooping=true, soundWaveAssetPtr=/AmbientAudioContent/Waves/SpottedLoops/Spot_Desk_Fan_Slow_Loop}}
]
properties {
 attenuationSettings=&Spot_Desk_Electric_Fan_Att
 bEnableBusSends=false
 bOverrideConcurrency=true
 maxDistance=600
 soundClassObject=&Fort_Ambient_Props_Athena
 volumeMultiplier=0.3
 concurrencyOverrides {maxCount=2, voiceStealReleaseTime=2}
 concurrencySet [/Game/Sounds/Concurrency_Rules/Athena_Prop]
 preEffectBusSends [{}]
}
