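# Spotted ambient loop cue for an electric floor fan prop, in the Fort ambient props sound class.
# A mixer layers two looping waves at equal volume: a fast desk fan loop and a mono blimp jump pad ambience borrowed from game content.
# A standing-fan attenuation asset shapes falloff, with overall volume kept low and bus sends disabled.
# Overridden concurrency caps playback at two voices, fading out stolen ones.
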SoundCue /AmbientAudioContent/SoundCues/SpottedLoops/Spot_Electric_Floor_Fan_Loop_Cue
root=Mixer_0
nodes [
 {id=Mixer_0, children [WavePlayer_3,WavePlayer_0], properties {inputVolume [1,1]}}
 {id=WavePlayer_0, properties {bLooping=true, soundWaveAssetPtr=/Game/Sounds/Artemis/POI/Blimps/Spots/Blimp_Jump_Pad_Ambient_Loop_Mono}}
 {id=WavePlayer_3, properties {bLooping=true, soundWaveAssetPtr=/AmbientAudioContent/Waves/SpottedLoops/Spot_Desk_Fan_Fast_Loop}}
]
properties {
 attenuationSettings=/AmbientAudioBase/Attenuations/Spot_Standing_Electric_Fan_Att
 bEnableBusSends=false
 bOverrideConcurrency=true
 maxDistance=750
 soundClassObject=&Fort_Ambient_Props_Athena
 volumeMultiplier=0.2
 concurrencyOverrides {maxCount=2, voiceStealReleaseTime=2}
 concurrencySet [/Game/Sounds/Concurrency_Rules/Athena_Prop]
 preEffectBusSends [{}]
}
