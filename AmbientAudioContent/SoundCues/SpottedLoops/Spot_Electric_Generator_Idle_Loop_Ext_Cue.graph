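# Sound cue for an exterior electric generator idle loop, played as a spotted ambient prop emitter with large-prop attenuation and prop concurrency.
# A looping water pump wave at full level is mixed with a quieter looping power receiver layer, then passed through a Modulator.
# All quality levels share that branch, which feeds an Enveloper volume fade-in and a short random start Delay at the root.
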SoundCue /AmbientAudioContent/SoundCues/SpottedLoops/Spot_Electric_Generator_Idle_Loop_Ext_Cue
root=Delay_0
nodes [
 {id=Delay_0, children [Enveloper_0], properties {delayMax=1}}
 {
  id=Enveloper_0
  children [QualityLevel_0]
  properties {pitchCurve {editorCurveData {keys [{time=0}]}}, volumeCurve {editorCurveData {keys [<at 0 0>,<at 1 1>]}}}
 }
 {id=Mixer_0, children [WavePlayer_2,WavePlayer_0], properties {inputVolume [0.15,1]}}
 {id=Modulator_0, children [Mixer_0], properties {volumeMax=1, volumeMin=1}}
 {id=QualityLevel_0, children [Modulator_0,Modulator_0,Modulator_0,Modulator_0]}
 {id=WavePlayer_0, properties {bLooping=true, soundWaveAssetPtr=/AmbientAudioContent/Waves/SpottedLoops/Spot_Outpost_WaterPump_Loop_01}}
 {id=WavePlayer_2, properties {bLooping=true, soundWaveAssetPtr=/AmbientAudioContent/Waves/POI/Spot_Mole_AmbientLoop_Object_PowerReceivers_04}}
]
properties {
 attenuationSettings=/AmbientAudioBase/Attenuations/Fort_Amb_Emitter_Prop_Large_Att
 bEnableBusSends=false
 maxDistance=2000
 soundClassObject=&Fort_Ambient_Props_Athena
 volumeMultiplier=1
 concurrencyOverrides {maxCount=2}
 concurrencySet [/Game/Sounds/Concurrency_Rules/Athena_Prop]
 preEffectBusSends [{}]
}
