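# Spotted ambient loop cue for a small electric motor, in the Fort_Ambient_Athena sound class.
# It mixes a looping servo mechanism wave through a Modulator with a quieter looping motor wave whose volume slowly oscillates.
# An attenuation node at the root adds distance low-pass filtering, listener focus and trace-based occlusion, with reverb send disabled.
# A concurrency override caps how many instances play at once.
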SoundCue /AmbientAudioContent/SoundCues/SpottedLoops/Spot_Electric_Motor_Small_01_Loop_Cue
root=Attenuation_1
nodes [
 {
  id=Attenuation_1
  children [Mixer_0]
  properties {
   bOverrideAttenuation=true
   attenuationOverrides {
    bAttenuateWithLPF=true
    bEnableFocusInterpolation=true
    bEnableListenerFocus=true
    bEnableOcclusion=true
    bEnableReverbSend=false
    falloffDistance=1400
    focusReleaseInterpSpeed=0.5
    lPFFrequencyAtMax=2000
    lPFRadiusMax=3000
    lPFRadiusMin=500
    nonFocusVolumeAttenuation=0.5
    occlusionInterpolationTime=1
    occlusionLowPassFilterFrequency=2000
    occlusionTraceChannel=ECC_GameTraceChannel11
    occlusionVolumeAttenuation=0.5
    stereoSpread=512
    attenuationShapeExtents {x=160}
   }
  }
 }
 {id=Mixer_0, children [Modulator_0,Oscillator_2], properties {inputVolume [1,0.5]}}
 {id=Modulator_0, children [WavePlayer_3]}
 {
  id=Oscillator_2
  children [WavePlayer_4]
  properties {
   amplitudeMax=0.25
   amplitudeMin=0.25
   bModulateVolume=true
   centerMax=1
   centerMin=1
   frequencyMax=0.1
   frequencyMin=0.1
   offsetMax=1.1
   offsetMin=0.9
  }
 }
 {id=WavePlayer_3, properties {bLooping=true, soundWaveAssetPtr=/AmbientAudioContent/Waves/SpottedLoops/Spot_Light_Servo_Motor_Mechanism_Loop}}
 {id=WavePlayer_4, properties {bLooping=true, soundWaveAssetPtr=/AmbientAudioContent/Waves/SpottedLoops/Spot_Electric_Motor_Small_01}}
]
properties {
 bOverrideConcurrency=true
 maxDistance=1560
 soundClassObject=&Fort_Ambient_Athena
 volumeMultiplier=0.42
 attenuationOverrides {attenuationShapeExtents {x=0}}
 concurrencyOverrides {maxCount=3}
}
