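# Ambient spot cue for a buzzing neon mascot sign, mixing a looping transformer buzz with a neon sign wave.
# A random node picks that mix either directly or through a slight pitch-raising modulator.
# Two nested, endlessly looping envelopers of different lengths add slow pitch and volume curves over the result.
# Capsule attenuation with occlusion and listener focus, plus distance-scaled concurrency and an ambient props sound class, control playback in the world.
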SoundCue /AmbientAudioContent/SoundCues/SpottedLoops/Spot_Electric_Neon_Mascot_Sign_Buzz_02_Cue
root=Enveloper_3
nodes [
 {
  id=Enveloper_0
  children [Random_1]
  properties {
   bLoop=true
   bLoopIndefinitely=true
   loopEnd=30
   pitchCurve {editorCurveData {keys [
      {arriveTangent=-0.0033, interpMode=RCIM_Cubic, leaveTangent=-0.0033, time=0}
      <at 15 0.95 RCIM_Cubic>
      {arriveTangent=0.0033, interpMode=RCIM_Cubic, leaveTangent=0.0033, time=30}
     ]}}
   volumeCurve {editorCurveData {keys [<at 0 1>]}}
  }
 }
 {
  id=Enveloper_3
  children [Enveloper_0]
  properties {
   bLoop=true
   bLoopIndefinitely=true
   loopEnd=20
   pitchCurve {editorCurveData {keys [{time=0}]}}
   volumeCurve {editorCurveData {keys [
      {arriveTangent=-0.01, interpMode=RCIM_Cubic, time=0, value=1}
      <at 10 1.1 RCIM_Cubic>
      {interpMode=RCIM_Cubic, leaveTangent=0.01, time=20, value=1}
     ]}}
  }
 }
 {id=Mixer_0, children [WavePlayer_2,WavePlayer_3], properties {inputVolume [1,1]}}
 {id=Modulator_0, children [Mixer_0], properties {pitchMax=1.1, pitchMin=1}}
 {id=Random_1, children [Modulator_0,Mixer_0], properties {weights [1,1]}}
 {id=WavePlayer_2, properties {bLooping=true, soundWaveAssetPtr=/AmbientAudioContent/Waves/SpottedLoops/Spot_Transformer_Large_Buzz_02}}
 {id=WavePlayer_3, properties {soundWaveAssetPtr=/AmbientAudioContent/Waves/SpottedLoops/Spot_PowerPlant_Neon_Sign_Loop_02}}
]
properties {
 bOverrideAttenuation=true
 bOverrideConcurrency=true
 maxDistance=1700
 soundClassObject=&Fort_Ambient_Props_Athena
 volumeMultiplier=0.2
 attenuationOverrides {
  attenuationShape=Capsule
  bEnableListenerFocus=true
  bEnableOcclusion=true
  distanceAlgorithm=Logarithmic
  falloffDistance=1300
  focusVolumeAttenuation=1.1
  nonFocusVolumeAttenuation=0.6
  occlusionInterpolationTime=1
  occlusionLowPassFilterFrequency=750
  occlusionTraceChannel=ECC_GameTraceChannel11
  occlusionVolumeAttenuation=0.5
  reverbWetLevelMin=0.6905
  stereoSpread=100
  attenuationShapeExtents {y=400, z=400}
 }
 concurrencyOverrides {maxCount=3, volumeScaleMode=Distance}
}
