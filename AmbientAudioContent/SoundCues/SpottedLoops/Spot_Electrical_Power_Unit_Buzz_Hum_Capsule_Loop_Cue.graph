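# Looping spot ambient cue for an electrical power unit buzz/hum in the Fort Ambient Athena sound class.
# Every quality level uses one random branch that starts immediately or after a short fixed delay, staggering instances.
# Both paths share a volume fade-in envelope and a modulator over a random pick of three looping waves.
# Overrides set capsule attenuation with listener focus and tight concurrency, alongside a distance-based reverb submix send.
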SoundCue /AmbientAudioContent/SoundCues/SpottedLoops/Spot_Electrical_Power_Unit_Buzz_Hum_Capsule_Loop_Cue
root=QualityLevel_0
nodes [
 {id=Delay_0, children [Enveloper_1], properties {delayMax=0.3, delayMin=0.3}}
 {
  id=Enveloper_1
  children [Modulator_0]
  properties {pitchCurve {editorCurveData {keys [{time=0}]}}, volumeCurve {editorCurveData {keys [<at 0 0>,<at 1 1>]}}}
 }
 {id=Modulator_0, children [Random_2]}
 {id=QualityLevel_0, children [Random_1,Random_1,Random_1,Random_1]}
 {id=Random_1, children [Delay_0,Enveloper_1], properties {weights [1,1]}}
 {id=Random_2, children [WavePlayer_0,WavePlayer_3,WavePlayer_4], properties {weights [1,1,1]}}
 {id=WavePlayer_0, properties {bLooping=true, soundWaveAssetPtr=/AmbientAudioContent/Waves/SpottedLoops/Spot_Electrical_Power_Unit_Loop_03}}
 {id=WavePlayer_3, properties {bLooping=true, soundWaveAssetPtr=/AmbientAudioContent/Waves/SpottedLoops/Spot_Electrical_Power_Unit_Loop_02}}
 {id=WavePlayer_4, properties {bLooping=true, soundWaveAssetPtr=/AmbientAudioContent/Waves/SpottedLoops/Spot_Electrical_Power_Unit_Loop_01}}
]
properties {
 bEnableBusSends=false
 bOverrideAttenuation=true
 bOverrideConcurrency=true
 maxDistance=1300
 soundClassObject=&Fort_Ambient_Athena
 volumeMultiplier=0.5
 attenuationOverrides {
  attenuationShape=Capsule
  bEnableListenerFocus=true
  distanceAlgorithm=Logarithmic
  falloffDistance=850
  focusAzimuth=25
  nonFocusAzimuth=55
  nonFocusVolumeAttenuation=0.75
  stereoSpread=100
  attenuationShapeExtents=<450 150>
 }
 concurrencyOverrides {maxCount=2, voiceStealReleaseTime=0.2}
 concurrencySet [/AmbientAudioBase/Concurrency/AmbientConc_3Max_Spot_Default_Concurrency]
 soundSubmixSends [{
   maxSendDistance=2000
   maxSendLevel=0.6
   minSendDistance=250
   minSendLevel=0.25
   soundSubmix=&Fortnite_Reverb_Default
   customSendLevelCurve {editorCurveData {keys [<at 0 0>,<at 1 1>]}}
  }]
}
