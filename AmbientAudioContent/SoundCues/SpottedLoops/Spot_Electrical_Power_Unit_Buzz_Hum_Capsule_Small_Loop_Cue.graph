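# Looping ambient spot cue for a small electrical power unit buzz and hum, in the Fort_Ambient_Athena sound class.
# It layers one of three random power unit loops over a half-volume, randomly up-pitched transformer buzz that slowly swells on a looping envelope.
# All quality levels share a random start, immediate or briefly delayed, with a fade-in.
# Overrides add capsule attenuation with occlusion and listener focus, tight concurrency and distance-scaled reverb send.
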SoundCue /AmbientAudioContent/SoundCues/SpottedLoops/Spot_Electrical_Power_Unit_Buzz_Hum_Capsule_Small_Loop_Cue
root=QualityLevel_0
nodes [
 {id=Delay_0, children [Enveloper_1], properties {delayMax=0.3, delayMin=0.3}}
 {
  id=Enveloper_0
  children [Modulator_1]
  properties {
   bLoop=true
   bLoopIndefinitely=true
   loopEnd=20
   pitchCurve {editorCurveData {keys [{time=0}]}}
   volumeCurve {editorCurveData {keys [
      {arriveTangent=-0.01, interpMode=RCIM_Cubic, time=0, value=1}
      <at 10 0.7 RCIM_Cubic>
      {interpMode=RCIM_Cubic, leaveTangent=0.01, time=20, value=1}
     ]}}
  }
 }
 {
  id=Enveloper_1
  children [Modulator_0]
  properties {pitchCurve {editorCurveData {keys [{time=0}]}}, volumeCurve {editorCurveData {keys [<at 0 0>,<at 1 1>]}}}
 }
 {id=Mixer_0, children [Enveloper_0,Random_2], properties {inputVolume [0.5,1]}}
 {id=Modulator_0, children [Mixer_0]}
 {id=Modulator_1, children [WavePlayer_1], properties {pitchMax=1.5, pitchMin=1}}
 {id=QualityLevel_0, children [Random_1,Random_1,Random_1,Random_1]}
 {id=Random_1, children [Delay_0,Enveloper_1], properties {weights [1,1]}}
 {id=Random_2, children [WavePlayer_0,WavePlayer_3,WavePlayer_4], properties {weights [1,1,1]}}
 {id=WavePlayer_0, properties {bLooping=true, soundWaveAssetPtr=/AmbientAudioContent/Waves/SpottedLoops/Spot_Electrical_Power_Unit_Loop_03}}
 {id=WavePlayer_1, properties {bLooping=true, soundWaveAssetPtr=/AmbientAudioContent/Waves/SpottedLoops/Spot_Transformer_Large_Buzz_01}}
 {id=WavePlayer_3, properties {bLooping=true, soundWaveAssetPtr=/AmbientAudioContent/Waves/SpottedLoops/Spot_Electrical_Power_Unit_Loop_02}}
 {id=WavePlayer_4, properties {bLooping=true, soundWaveAssetPtr=/AmbientAudioContent/Waves/SpottedLoops/Spot_Electrical_Power_Unit_Loop_01}}
]
properties {
 bEnableBusSends=false
 bOverrideAttenuation=true
 bOverrideConcurrency=true
 maxDistance=950
 soundClassObject=&Fort_Ambient_Athena
 volumeMultiplier=0.33
 attenuationOverrides {
  attenuationShape=Capsule
  bEnableListenerFocus=true
  bEnableOcclusion=true
  falloffDistance=750
  focusAzimuth=25
  nonFocusAzimuth=55
  nonFocusVolumeAttenuation=0.75
  occlusionInterpolationTime=0.3
  occlusionLowPassFilterFrequency=850
  occlusionTraceChannel=ECC_GameTraceChannel11
  occlusionVolumeAttenuation=0.5
  stereoSpread=100
  attenuationShapeExtents=<200 150>
 }
 concurrencyOverrides {maxCount=2, voiceStealReleaseTime=0.2}
 concurrencySet [/AmbientAudioBase/Concurrency/AmbientConc_3Max_Spot_Default_Concurrency]
 soundSubmixSends [{
   maxSendDistance=2000
   maxSendLevel=0.6
   minSendDistance=250
   minSendLevel=0.25
   soundSubmix=&Fortnite_Reverb_Default
   customSendLevelCurve {editorCurveData {keys [<at 0 0>,<at 1 1>]}}
  }]
}
