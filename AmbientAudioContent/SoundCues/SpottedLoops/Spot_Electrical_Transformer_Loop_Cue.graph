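# Spotted ambient loop cue for an electrical transformer hum, in the outdoor ambient sound class with a send to the default reverb submix.
# A Random node picks one of five equally weighted looping hum waves, with one turned down through a Mixer.
# The pick feeds a Modulator with pitch held fixed, then an Attenuation node that borrows a neon sign buzz attenuation asset.
# A concurrency override allows one instance, stopping the quietest.
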SoundCue /AmbientAudioContent/SoundCues/SpottedLoops/Spot_Electrical_Transformer_Loop_Cue
root=Attenuation_0
nodes [
 {
  id=Attenuation_0
  children [Modulator_0]
  properties {attenuationSettings=/Game/Sounds/Helios/Environment_Actors/NeonLightSignBuzz/Spot_Neon_Light_Buzz_Sign_Loop_Narrow_Att}
 }
 {id=Mixer_2, children [WavePlayer_8], properties {inputVolume [0.4]}}
 {id=Modulator_0, children [Random_2], properties {pitchMax=1, pitchMin=1}}
 {id=Random_2, children [Mixer_2, WavePlayer_9, WavePlayer_10, WavePlayer_11, WavePlayer_12], properties {weights [1,1,1,1,1]}}
 {
  id=WavePlayer_10
  properties {bLooping=true, soundWaveAssetPtr=/AmbientAudioContent/Waves/SpottedLoops/Spot_Electric_Transformer_Hum_03_Loop}
 }
 {
  id=WavePlayer_11
  properties {bLooping=true, soundWaveAssetPtr=/AmbientAudioContent/Waves/SpottedLoops/Spot_Electric_Transformer_Hum_02_Loop}
 }
 {
  id=WavePlayer_12
  properties {bLooping=true, soundWaveAssetPtr=/AmbientAudioContent/Waves/SpottedLoops/Spot_Electric_Transformer_Hum_01_Loop}
 }
 {id=WavePlayer_8, properties {bLooping=true, soundWaveAssetPtr=/AmbientAudioContent/Waves/SpottedLoops/Spot_Electric_Transformer_Hum_05_Loop}}
 {id=WavePlayer_9, properties {bLooping=true, soundWaveAssetPtr=/AmbientAudioContent/Waves/SpottedLoops/Spot_Electric_Transformer_Hum_04_Loop}}
]
properties {
 bOverrideConcurrency=true
 maxDistance=3000
 soundClassObject=&Fort_Ambient_Outdoors_Athena
 volumeMultiplier=0.8
 attenuationOverrides {
  bEnableOcclusion=true
  falloffDistance=925
  nonFocusVolumeAttenuation=0
  nonSpatializedRadiusStart=900
  occlusionLowPassFilterFrequency=2000
  occlusionTraceChannel=ECC_GameTraceChannel6
  occlusionVolumeAttenuation=0.7
  stereoSpread=1500
  attenuationShapeExtents {x=900}
 }
 concurrencyOverrides {bVolumeScaleCanRelease=true, maxCount=1, resolutionRule=StopQuietest}
 soundSubmixSends [{minSendLevel=0.5, soundSubmix=&Fortnite_Reverb_Default}]
}
