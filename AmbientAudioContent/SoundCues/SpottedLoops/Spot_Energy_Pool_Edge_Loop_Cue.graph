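# Looping spot ambience Sound Cue for an energy pool edge, classed under Fort_Gadgets.
# Four looping layers (sizzle, whine, humming rocks, radioactive energy) are mixed at differing volumes into a Modulator.
# A Random node picks either a short delay or immediate start, feeding two quality slots while the other two are empty.
# Overrides add HRTF spatialization, low-pass distance filtering, listener focus and a tight concurrency cap.
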SoundCue /AmbientAudioContent/SoundCues/SpottedLoops/Spot_Energy_Pool_Edge_Loop_Cue
root=QualityLevel_0
nodes [
 {id=Delay_0, children [Modulator_0], properties {delayMax=0.3, delayMin=0.3}}
 {id=Mixer_0, children [WavePlayer_2,WavePlayer_3,WavePlayer_4,WavePlayer_0], properties {inputVolume [1,0.75,5,0.5]}}
 {id=Modulator_0, children [Mixer_0]}
 {id=QualityLevel_0, children [Random_0,Random_0,null,null]}
 {id=Random_0, children [Delay_0,Modulator_0], properties {weights [1,1]}}
 {id=WavePlayer_0, properties {bLooping=true, soundWaveAssetPtr=/AmbientAudioContent/Waves/SpottedLoops/Spot_RadioActive_Energy_Loop}}
 {id=WavePlayer_2, properties {bLooping=true, soundWaveAssetPtr=/AmbientAudioContent/Waves/SpottedLoops/Spot_Energy_Sizzle_Loop}}
 {id=WavePlayer_3, properties {bLooping=true, soundWaveAssetPtr=/AmbientAudioContent/Waves/SpottedLoops/Spot_Energy_Whine_Loop}}
 {
  id=WavePlayer_4
  properties {bLooping=true, soundWaveAssetPtr=/AmbientAudioContent/Waves/SpottedLoops/Spot_Amb_Tower_Aftermath_Humming_Rocks_Loop}
 }
]
properties {
 bEnableSubmixSends=false
 bOverrideAttenuation=true
 bOverrideConcurrency=true
 maxDistance=1250
 soundClassObject=&Fort_Gadgets
 volumeMultiplier=0.6
 attenuationOverrides {
  bAttenuateWithLPF=true
  bEnableListenerFocus=true
  bEnableLogFrequencyScaling=true
  falloffDistance=1000
  focusReleaseInterpSpeed=0.8
  hPFFrequencyAtMax=500
  hPFFrequencyAtMin=500
  lPFRadiusMax=1000
  lPFRadiusMin=500
  nonFocusVolumeAttenuation=0.4
  spatializationAlgorithm=SPATIALIZATION_HRTF
  attenuationShapeExtents {x=250}
  customAttenuationCurve {editorCurveData {keys [{arriveTangent=-2.3812, interpMode=RCIM_Cubic, leaveTangent=-2.3812, tangentMode=RCTM_User, time=0, value=1}, <at 1 0.01 RCIM_Cubic>]}}
 }
 concurrencyOverrides {maxCount=2, voiceStealReleaseTime=2}
}
