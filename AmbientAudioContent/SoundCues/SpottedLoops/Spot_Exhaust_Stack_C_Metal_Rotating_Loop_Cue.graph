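# Ambient spot-loop Sound Cue for a rotating metal exhaust stack, playing one of two looping "ExhaustStackA" metal waves chosen at random.
# The chosen wave passes through a Modulator and an Enveloper volume fade-in.
# An outer Random node starts that shared branch immediately or after a short fixed delay, likely staggering nearby instances.
# A root attenuation override adds logarithmic falloff with listener focus, alongside concurrency limits and a reverb send.
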SoundCue /AmbientAudioContent/SoundCues/SpottedLoops/Spot_Exhaust_Stack_C_Metal_Rotating_Loop_Cue
root=Attenuation_0
nodes [
 {
  id=Attenuation_0
  children [Random_1]
  properties {
   bOverrideAttenuation=true
   attenuationOverrides {
    bEnableListenerFocus=true
    distanceAlgorithm=Logarithmic
    falloffDistance=1500
    focusVolumeAttenuation=1.2
    nonFocusVolumeAttenuation=0.75
    attenuationShapeExtents {x=100}
    customAttenuationCurve {editorCurveData {keys [<at 0 1>,<at 1 0>]}}
    submixSendSettings [{
      bEnableHPFCutoff=false
      bEnableLPFCutoff=false
      disableManualSendClamp=false
      hPFCutoff=20
      lPFCutoff=20000
      maxSendDistance=6000
      maxSendLevel=1
      minSendDistance=400
      minSendLevel=0
      sendLevel=0.2
      sendLevelControlMethod=Linear
      soundSubmix=null
     }]
   }
  }
 }
 {id=Delay_0, children [Enveloper_1], properties {delayMax=0.3, delayMin=0.3}}
 {
  id=Enveloper_1
  children [Modulator_0]
  properties {pitchCurve {editorCurveData {keys [{time=0}]}}, volumeCurve {editorCurveData {keys [<at 0 0>,<at 1 1>]}}}
 }
 {id=Modulator_0, children [Random_0]}
 {id=Random_0, children [WavePlayer_0,WavePlayer_1], properties {weights [1,1]}}
 {id=Random_1, children [Delay_0,Enveloper_1], properties {weights [1,1]}}
 {
  id=WavePlayer_0
  properties {bLooping=true, soundWaveAssetPtr=/AmbientAudioContent/Waves/SpottedLoops/Spot_ExhaustStackA_Rotating_Metal_Loop_03}
 }
 {
  id=WavePlayer_1
  properties {bLooping=true, soundWaveAssetPtr=/AmbientAudioContent/Waves/SpottedLoops/Spot_ExhaustStackA_Rotating_Metal_Loop_05}
 }
]
properties {
 bEnableBusSends=false
 bOverrideConcurrency=true
 maxDistance=1600
 soundClassObject=&Fort_Ambient_Athena
 volumeMultiplier=0.2
 attenuationOverrides {
  bEnableListenerFocus=true
  distanceAlgorithm=Logarithmic
  falloffDistance=1000
  nonFocusVolumeAttenuation=0.75
  stereoSpread=100
  attenuationShapeExtents {x=250}
 }
 concurrencyOverrides {maxCount=2, voiceStealReleaseTime=0.2}
 concurrencySet [/AmbientAudioBase/Concurrency/AmbientConc_3Max_Spot_Default_Concurrency]
 soundSubmixSends [{
   maxSendDistance=2000
   minSendLevel=0
   soundSubmix=&Fortnite_Reverb_Default
   customSendLevelCurve {editorCurveData {keys [<at 0 0>,<at 1 1>]}}
  }]
}
