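# Ambient sound cue for a spot-placed, looping fire torch crackle, with positional attenuation and a cap of three concurrent voices.
# A short random delay and one-second fade-in stagger and soften each start.
# A random node picks one of three looping torch waves and a modulator varies its pitch.
# Two endlessly looping envelopes of different lengths slowly drift volume and pitch, making the loop less repetitive.
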SoundCue /AmbientAudioContent/SoundCues/SpottedLoops/Spot_FireTorch_Crackle_Loop_Cue
root=Delay_0
nodes [
 {id=Delay_0, children [Enveloper_1], properties {delayMax=0.25}}
 {
  id=Enveloper_0
  children [Modulator_1]
  properties {
   bLoop=true
   bLoopIndefinitely=true
   loopEnd=30
   pitchCurve {editorCurveData {keys [
      {arriveTangent=-0.0013, interpMode=RCIM_Cubic, leaveTangent=-0.0013, time=0}
      <at 15 0.98 RCIM_Cubic>
      {arriveTangent=0.0013, interpMode=RCIM_Cubic, leaveTangent=0.0013, time=30}
     ]}}
   volumeCurve {editorCurveData {keys [<at 0 1>]}}
  }
 }
 {
  id=Enveloper_1
  children [Enveloper_3]
  properties {pitchCurve {editorCurveData {keys [{time=0}]}}, volumeCurve {editorCurveData {keys [<at 0 0>,<at 1 1>]}}}
 }
 {
  id=Enveloper_3
  children [Enveloper_0]
  properties {
   bLoop=true
   bLoopIndefinitely=true
   loopEnd=20
   pitchCurve {editorCurveData {keys [{time=0}]}}
   volumeCurve {editorCurveData {keys [
      {arriveTangent=-0.01, interpMode=RCIM_Cubic, leaveTangent=-0.01, time=0, value=1}
      <at 10 0.9 RCIM_Cubic>
      {arriveTangent=0.01, interpMode=RCIM_Cubic, leaveTangent=0.01, time=20, value=1}
     ]}}
  }
 }
 {id=Modulator_1, children [Random_1], properties {pitchMax=1.1, pitchMin=0.9}}
 {id=Random_1, children [WavePlayer_0,WavePlayer_4,WavePlayer_5], properties {weights [1,1,1]}}
 {id=WavePlayer_0, properties {bLooping=true, soundWaveAssetPtr=/AmbientAudioContent/Waves/SpottedLoops/Spot_Fire_Torch_01}}
 {id=WavePlayer_4, properties {bLooping=true, soundWaveAssetPtr=/AmbientAudioContent/Waves/SpottedLoops/Spot_Fire_Torch_02}}
 {id=WavePlayer_5, properties {bLooping=true, soundWaveAssetPtr=/AmbientAudioContent/Waves/SpottedLoops/Spot_Fire_Torch_03}}
]
properties {
 attenuationSettings=/Game/Sounds/Apollo/Biomes/UFO_Crash_Sites/FireLoops/Fire_Small_Close_MorePositional_Attenuation
 bBypassVolumeScaleForPriority=true
 bOverrideConcurrency=true
 maxDistance=600
 priority=0
 soundClassObject=/Game/Sounds/FortSoundClassesAndModes/Fort_Ambient_NonLoop
 volumeMultiplier=0.6
 attenuationOverrides {attenuationShape=Cone, attenuationShapeExtents {x=171.3182}}
 concurrencyOverrides {maxCount=3, voiceStealReleaseTime=3}
}
