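# Looping spot ambience cue for a large fire, built from brazier and urn fire recordings.
# One of three urn loops, picked at random, plays at reduced volume beneath close mono and stereo brazier loops, each brazier layer with its own attenuation.
# A modulator slightly randomizes the pitch of the whole mix.
# The cue overrides attenuation with logarithmic falloff, focus and occlusion tuning, and overrides concurrency with a per-owner limit.
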SoundCue /AmbientAudioContent/SoundCues/SpottedLoops/Spot_Fire_Large_Loop_Cue
root=Modulator_0
nodes [
 {
  id=Attenuation_0
  children [WavePlayer_1]
  properties {attenuationSettings=/Game/Sounds/Asteria/Environment_Actors/Brazier_LightSource/Spot_Bazier_LightSource_Fire_Close_Att}
 }
 {
  id=Attenuation_1
  children [WavePlayer_2]
  properties {attenuationSettings=/Game/Sounds/Asteria/Environment_Actors/Brazier_LightSource/Spot_Bazier_LightSource_Fire_Att}
 }
 {id=Mixer_0, children [Random_0,Mixer_1], properties {inputVolume [0.5,1]}}
 {id=Mixer_1, children [Attenuation_0, Attenuation_1], properties {inputVolume [1,1]}}
 {id=Modulator_0, children [Mixer_0], properties {pitchMax=1.1, pitchMin=0.9}}
 {id=Random_0, children [WavePlayer_3,WavePlayer_6,WavePlayer_7], properties {weights [1,1,1]}}
 {id=WavePlayer_1, properties {bLooping=true, soundWaveAssetPtr=/AmbientAudioContent/Waves/OneShots/Spot_Coliseum_Brazier_Fire_Close_Mono}}
 {id=WavePlayer_2, properties {bLooping=true, soundWaveAssetPtr=/AmbientAudioContent/Waves/OneShots/Spot_Coliseum_Brazier_Fire_Stereo}}
 {id=WavePlayer_3, properties {bLooping=true, soundWaveAssetPtr=/AmbientAudioContent/Waves/SpottedLoops/Spot_Fire_Urn_03}}
 {id=WavePlayer_6, properties {bLooping=true, soundWaveAssetPtr=/AmbientAudioContent/Waves/SpottedLoops/Spot_Fire_Urn_02}}
 {id=WavePlayer_7, properties {bLooping=true, soundWaveAssetPtr=/AmbientAudioContent/Waves/SpottedLoops/Spot_Fire_Urn_01}}
]
properties {
 attenuationSettings=/Game/Sounds/Helios/Environment_Actors/Brazier/Spot_Brazier_LightSource_Fire_Large_With_Focus_Att
 bOverrideAttenuation=true
 bOverrideConcurrency=true
 maxDistance=2900
 soundClassObject=&Fort_Ambient_Athena
 subtitlePriority=5000
 volumeMultiplier=0.802
 attenuationOverrides {
  bEnableFocusInterpolation=true
  dBAttenuationAtMax=-40
  distanceAlgorithm=Logarithmic
  falloffDistance=2500
  focusAzimuth=15
  focusVolumeAttenuation=1.5
  nonFocusAzimuth=45
  nonFocusVolumeAttenuation=0.5
  occlusionInterpolationTime=1
  occlusionLowPassFilterFrequency=10000
  occlusionTraceChannel=ECC_GameTraceChannel11
  occlusionVolumeAttenuation=0.5
  stereoSpread=512
  attenuationShapeExtents {y=64.5942, z=90}
 }
 concurrencyOverrides {bLimitToOwner=true, maxCount=6}
 concurrencySet [/AmbientAudioBase/Concurrency/AmbientConc_3Max_Spot_Default_Concurrency]
}
