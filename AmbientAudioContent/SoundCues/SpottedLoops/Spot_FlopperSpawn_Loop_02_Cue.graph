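# Spotted ambient loop cue for a Flopper spawn, played quietly on the water ambient sound class.
# A mixer layers looping fish-tank bubble hum and lava bubbles with a Sakura Springs water-flow wave not set to loop, boosting the first and last inputs.
# Overridden attenuation adds distance low-pass filtering, listener focus and occlusion.
# Overridden concurrency caps simultaneous voices with a timed steal release.
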SoundCue /AmbientAudioContent/SoundCues/SpottedLoops/Spot_FlopperSpawn_Loop_02_Cue
root=Mixer_0
nodes [
 {id=Mixer_0, children [WavePlayer_1,WavePlayer_2,WavePlayer_3], properties {inputVolume [2,1,2]}}
 {id=WavePlayer_1, properties {bLooping=true, soundWaveAssetPtr=/AmbientAudioContent/Waves/SpottedLoops/Spot_FishTank_Bubble_Hum_Loop}}
 {id=WavePlayer_2, properties {bLooping=true, soundWaveAssetPtr=/AmbientAudioContent/Waves/SpottedLoops/Spot_Lava_Bubbles_Loop}}
 {id=WavePlayer_3, properties {bLooping=true, soundWaveAssetPtr=/AmbientAudioContent/Waves/Water/Spot_Sakura_Springs_Water_Flow_01_LQ}}
]
properties {
 bOverrideAttenuation=true
 bOverrideConcurrency=true
 maxDistance=3790.1655
 soundClassObject=/Game/Sounds/FortSoundClassesAndModes/Fort_Water_Ambient
 volumeMultiplier=0.3
 attenuationOverrides {
  bAttenuateWithLPF=true
  bEnableListenerFocus=true
  bEnableOcclusion=true
  falloffDistance=2710.8213
  focusVolumeAttenuation=1.0775
  lPFFrequencyAtMax=200
  lPFRadiusMax=1281.8871
  lPFRadiusMin=442.5163
  nonFocusVolumeAttenuation=0.25
  occlusionInterpolationTime=2
  occlusionLowPassFilterFrequency=8000
  occlusionTraceChannel=ECC_GameTraceChannel11
  occlusionVolumeAttenuation=0.5
  stereoSpread=1707.1663
  attenuationShapeExtents {x=1079.3441}
 }
 concurrencyOverrides {maxCount=3, voiceStealReleaseTime=2}
}
