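# Looping spot ambience cue that plays frog chirps and croaks at randomly delayed intervals of a few seconds.
# A quality switch feeds the delay: the first two slots randomly pick pitched-up swamp frogs or a frog-insect call pool, the last two use only two calls.
# Overridden attenuation uses a custom curve that is silent very close, peaks nearby, then fades out with distance.
# A concurrency set limits simultaneous spot voices.
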SoundCue /AmbientAudioContent/SoundCues/SpottedLoops/Spot_Frog_Chirps_Croaks_Loop_Cue
root=Looping_0
nodes [
 {id=Delay_0, children [QualityLevel_0], properties {delayMax=6, delayMin=1}}
 {id=Looping_0, children [Delay_0]}
 {id=Modulator_0, children [Random_1]}
 {id=Modulator_1, children [Random_2]}
 {id=Modulator_2, children [Random_0], properties {pitchMax=1.3, pitchMin=1.2}}
 {id=QualityLevel_0, children [Random_3,Random_3,Modulator_1,Modulator_1]}
 {
  id=Random_0
  children [WavePlayer_1, WavePlayer_2, WavePlayer_3, WavePlayer_4, WavePlayer_8, WavePlayer_11]
  properties {weights [1,1,1,1,1,1]}
 }
 {
  id=Random_1
  children [
   WavePlayer_0
   WavePlayer_5
   WavePlayer_6
   WavePlayer_7
   WavePlayer_9
   WavePlayer_10
   WavePlayer_18
   WavePlayer_19
   WavePlayer_20
   WavePlayer_21
  ]
  properties {weights [1,1,1,1,1,1,1,1,1,1]}
 }
 {id=Random_2, children [WavePlayer_16, WavePlayer_17], properties {weights [1,1]}}
 {id=Random_3, children [Modulator_2,Modulator_0], properties {weights [1,1]}}
 {id=WavePlayer_0, properties {soundWaveAssetPtr=/AmbientAudioContent/Waves/OneShots/Spot_Frog_Insects_Call_01}}
 {id=WavePlayer_1, properties {soundWaveAssetPtr=/AmbientAudioContent/Waves/OneShots/Spot_Biome_Swamp_01_Frog_01}}
 {id=WavePlayer_10, properties {soundWaveAssetPtr=/AmbientAudioContent/Waves/OneShots/Spot_Frog_Insects_Call_06}}
 {id=WavePlayer_11, properties {soundWaveAssetPtr=/AmbientAudioContent/Waves/OneShots/Spot_Biome_Swamp_01_Frog_06}}
 {id=WavePlayer_16, properties {soundWaveAssetPtr=&Spot_Frog_Insects_Call_01}}
 {id=WavePlayer_17, properties {soundWaveAssetPtr=/AmbientAudioContent/Waves/OneShots/Spot_Frog_Insects_Call_02}}
 {id=WavePlayer_18, properties {soundWaveAssetPtr=/AmbientAudioContent/Waves/OneShots/Spot_Frog_Insects_Call_07}}
 {id=WavePlayer_19, properties {soundWaveAssetPtr=/AmbientAudioContent/Waves/OneShots/Spot_Frog_Insects_Call_08}}
 {id=WavePlayer_2, properties {soundWaveAssetPtr=/AmbientAudioContent/Waves/OneShots/Spot_Biome_Swamp_01_Frog_02}}
 {id=WavePlayer_20, properties {soundWaveAssetPtr=/AmbientAudioContent/Waves/OneShots/Spot_Frog_Insects_Call_09}}
 {id=WavePlayer_21, properties {soundWaveAssetPtr=/AmbientAudioContent/Waves/OneShots/Spot_Frog_Insects_Call_10}}
 {id=WavePlayer_3, properties {soundWaveAssetPtr=/AmbientAudioContent/Waves/OneShots/Spot_Biome_Swamp_01_Frog_03}}
 {id=WavePlayer_4, properties {soundWaveAssetPtr=/AmbientAudioContent/Waves/OneShots/Spot_Biome_Swamp_01_Frog_04}}
 {id=WavePlayer_5, properties {soundWaveAssetPtr=&Spot_Frog_Insects_Call_02}}
 {id=WavePlayer_6, properties {soundWaveAssetPtr=/AmbientAudioContent/Waves/OneShots/Spot_Frog_Insects_Call_03}}
 {id=WavePlayer_7, properties {soundWaveAssetPtr=/AmbientAudioContent/Waves/OneShots/Spot_Frog_Insects_Call_04}}
 {id=WavePlayer_8, properties {soundWaveAssetPtr=/AmbientAudioContent/Waves/OneShots/Spot_Biome_Swamp_01_Frog_05}}
 {id=WavePlayer_9, properties {soundWaveAssetPtr=/AmbientAudioContent/Waves/OneShots/Spot_Frog_Insects_Call_05}}
]
properties {
 attenuationSettings=/Game/Sounds/Attenuation/Fort_TileAmbient_OneShot_Att
 bOverrideAttenuation=true
 maxDistance=6000
 soundClassObject=&Fort_Ambient_OneShots_Outdoors
 volumeMultiplier=0.501
 attenuationOverrides {
  distanceAlgorithm=Custom
  falloffDistance=6000
  stereoSpread=50
  attenuationShapeExtents {x=0}
  customAttenuationCurve {editorCurveData {keys [<at 0.05 0>,<at 0.2 1>,<at 1 0>]}}
 }
 concurrencyOverrides {maxCount=2, voiceStealReleaseTime=1}
 concurrencySet [/AmbientAudioBase/Concurrency/AmbientConc_2Max_Spot_Concurrency]
 preEffectBusSends [{}]
}
