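# Sound cue for a looping outdoor spot ambience of geothermal chimney smoke.
# One looping wave feeds two capsule attenuation overrides, a short-falloff layer and a long-falloff layer with listener focus, blended in a mixer.
# A modulator randomizes pitch, then two nested looping envelopers add slow pitch and volume drift on 30 and 20 second cycles.
# The cue routes to the outdoor ambient sound class and exterior bus, with a concurrency set limiting voices.
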SoundCue /AmbientAudioContent/SoundCues/SpottedLoops/Spot_GeoThermal_Chimney_Smoke_Loop_02_Cue
root=Enveloper_3
nodes [
 {
  id=Attenuation_0
  children [WavePlayer_0]
  properties {
   bOverrideAttenuation=true
   attenuationOverrides {
    attenuationShape=Capsule
    bEnableFocusInterpolation=true
    bEnableListenerFocus=true
    falloffDistance=7000
    focusAttackInterpSpeed=0.5
    focusReleaseInterpSpeed=0.01
    lPFFrequencyAtMax=1000
    lPFRadiusMax=3000
    lPFRadiusMin=1000
    nonFocusAzimuth=45
    nonFocusVolumeAttenuation=0.4
    stereoSpread=512
    attenuationShapeExtents=<2000 3000>
   }
  }
 }
 {
  id=Attenuation_1
  children [WavePlayer_0]
  properties {
   bOverrideAttenuation=true
   attenuationOverrides {
    attenuationShape=Capsule
    bEnableFocusInterpolation=true
    falloffDistance=500
    lPFFrequencyAtMax=1000
    lPFRadiusMax=3000
    lPFRadiusMin=1000
    nonFocusVolumeAttenuation=0.4
    nonSpatializedRadiusEnd=750
    nonSpatializedRadiusStart=1000
    stereoSpread=1000
    attenuationShapeExtents=<2000 500>
   }
  }
 }
 {
  id=Enveloper_0
  children [Modulator_0]
  properties {
   bLoop=true
   bLoopIndefinitely=true
   loopEnd=30
   pitchCurve {editorCurveData {keys [
      {arriveTangent=-0.0013, interpMode=RCIM_Cubic, leaveTangent=-0.0013, time=0}
      <at 15 0.98 RCIM_Cubic>
      {arriveTangent=0.0013, interpMode=RCIM_Cubic, leaveTangent=0.0013, time=30}
     ]}}
   volumeCurve {editorCurveData {keys [<at 0 1>]}}
  }
 }
 {
  id=Enveloper_3
  children [Enveloper_0]
  properties {
   bLoop=true
   bLoopIndefinitely=true
   loopEnd=20
   pitchCurve {editorCurveData {keys [{time=0}]}}
   volumeCurve {editorCurveData {keys [
      {arriveTangent=-0.01, interpMode=RCIM_Cubic, leaveTangent=-0.01, time=0, value=1}
      <at 10 0.9 RCIM_Cubic>
      {arriveTangent=0.01, interpMode=RCIM_Cubic, leaveTangent=0.01, time=20, value=1}
     ]}}
  }
 }
 {id=Mixer_2, children [Attenuation_1, Attenuation_0], properties {inputVolume [1,1.1]}}
 {id=Modulator_0, children [Mixer_2], properties {pitchMax=1.1, pitchMin=0.9}}
 {id=WavePlayer_0, properties {bLooping=true, soundWaveAssetPtr=/AmbientAudioContent/Waves/SpottedLoops/Spot_Smoke_GeoThermal_Loop_02}}
]
properties {
 maxDistance=10000
 soundClassObject=&Fort_Ambient_Outdoors_Athena
 subtitlePriority=5000
 volumeMultiplier=0.2
 attenuationOverrides {
  attenuationShape=Capsule
  bAttenuateWithLPF=true
  bEnableFocusInterpolation=true
  bEnableListenerFocus=true
  dBAttenuationAtMax=-40
  falloffDistance=5000
  lPFFrequencyAtMax=1000
  lPFRadiusMax=3000
  lPFRadiusMin=1000
  nonFocusVolumeAttenuation=0.4
  occlusionInterpolationTime=1
  occlusionLowPassFilterFrequency=10000
  occlusionTraceChannel=ECC_GameTraceChannel11
  occlusionVolumeAttenuation=0.5
  stereoSpread=512
  attenuationShapeExtents=<2031.3811 553.7532 90>
 }
 busSends [{
   audioBus=null
   maxSendDistance=5000
   minSendDistance=1000
   minSendLevel=0.4
   sendLevel=1
   soundSourceBus=&AmbientAudio_Exterior_SourceBus
   customSendLevelCurve {editorCurveData {keys [<at 0 0.4>,<at 1 1>]}}
  }]
 concurrencyOverrides {maxCount=1, voiceStealReleaseTime=2}
 concurrencySet [/AmbientAudioBase/Concurrency/AmbientConc_3Max_Spot_Default_Concurrency]
}
