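# Looping outdoor ambient cue for a distant geothermal plant spot sound, limited to one instance and sent to the exterior ambient source bus.
# One looping wave with slight random pitch variation feeds two mixed branches, each with its own borrowed Gold Mine attenuation asset.
# The close, no-focus branch plays directly at reduced level.
# The distant, max-focus branch runs through an enveloper that fades it in from silence to partial volume.
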SoundCue /AmbientAudioContent/SoundCues/SpottedLoops/Spot_GeoThermal_Plant_Distant_Max_Focus_Loop_02_Cue
root=Mixer_0
nodes [
 {id=Attenuation_0, children [Modulator_0], properties {attenuationSettings=/AmbientAudioBase/Attenuations/Spot_Gold_Mine_Close_No_Focus_Att}}
 {
  id=Attenuation_1
  children [Enveloper_0]
  properties {attenuationSettings=/AmbientAudioBase/Attenuations/Spot_Gold_Mine_Distant_Max_Focus_Att}
 }
 {
  id=Enveloper_0
  children [Modulator_0]
  properties {pitchCurve {editorCurveData {keys [{time=0}]}}, volumeCurve {editorCurveData {keys [<at 0 0>,<at 0.2 0.0769>,<at 0.75 0.5>]}}}
 }
 {id=Mixer_0, children [Attenuation_0, Attenuation_1], properties {inputVolume [0.75,1]}}
 {id=Modulator_0, children [WavePlayer_0], properties {pitchMax=1.1, pitchMin=0.9}}
 {
  id=WavePlayer_0
  properties {bLooping=true, soundWaveAssetPtr=/AmbientAudioContent/Waves/SpottedLoops/Spot_GeoThermal_Plant_Distant_Focus_Loop}
 }
]
properties {
 bEnableSubmixSends=false
 bOverrideConcurrency=true
 maxDistance=15000
 soundClassObject=&Fort_Ambient_Outdoors_Athena
 subtitlePriority=5000
 volumeMultiplier=0.4
 attenuationOverrides {
  attenuationShape=Capsule
  bAttenuateWithLPF=true
  bEnableFocusInterpolation=true
  bEnableListenerFocus=true
  dBAttenuationAtMax=-40
  falloffDistance=2638.2017
  lPFFrequencyAtMax=1000
  lPFRadiusMax=3000
  lPFRadiusMin=1000
  nonFocusVolumeAttenuation=0
  occlusionInterpolationTime=1
  occlusionLowPassFilterFrequency=10000
  occlusionTraceChannel=ECC_GameTraceChannel11
  occlusionVolumeAttenuation=0.5
  stereoSpread=512
  attenuationShapeExtents=<3720.1087 1500 90>
 }
 busSends [{audioBus=null, minSendLevel=0, sendLevel=1, soundSourceBus=&AmbientAudio_Exterior_SourceBus}]
 concurrencyOverrides {maxCount=1, resolutionRule=PreventNew, voiceStealReleaseTime=2}
 soundSubmixSends [{minSendLevel=0, soundSubmix=null}]
}
