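# Sound cue for a spotted ambient loop of an idling Black Hawk helicopter, in the exterior vehicles sound class.
# A single looping wave player plays at reduced volume with no randomization or mixing.
# Overridden attenuation adds logarithmic distance falloff, distance-based low-pass filtering and listener focus that ducks it when out of focus.
# It sends to the exterior ambient source bus and uses a limit-to-one, prevent-new concurrency rule.
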SoundCue /AmbientAudioContent/SoundCues/SpottedLoops/Spot_Helicopter_Idle_Loop
root=WavePlayer_2
nodes [{id=WavePlayer_2, properties {bLooping=true, soundWaveAssetPtr=/AmbientAudioContent/Waves/SpottedLoops/Spot_Helicopter_BlackHawk_Idle_Loop}}]
properties {
 bOverrideAttenuation=true
 maxDistance=4600
 soundClassObject=&Fort_Vehicles_Exterior
 volumeMultiplier=0.3
 attenuationOverrides {
  bAttenuateWithLPF=true
  bEnableFocusInterpolation=true
  bEnableListenerFocus=true
  distanceAlgorithm=Logarithmic
  focusReleaseInterpSpeed=0.2
  lPFFrequencyAtMax=2000
  lPFRadiusMin=2000
  nonFocusVolumeAttenuation=0.4
  attenuationShapeExtents {x=1000}
 }
 busSends [{
   audioBus=null
   maxSendDistance=4000
   minSendLevel=0.6
   sendLevel=1
   soundSourceBus=&AmbientAudio_Exterior_SourceBus
  }]
 concurrencySet [/Game/Sounds/Concurrency_Rules/LimitToOnePreventNew]
}
