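# Sound cue for a spotted wind ambience loop, named for windows in Iceberg Isles.
# A single looping wave player plays the reused Winterfest exterior wind loop, with no randomization, modulation or mixing nodes.
# The cue lowers the volume and routes to the Athena indoor ambient sound class, although the source wave is an exterior recording by name.
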SoundCue /AmbientAudioContent/SoundCues/SpottedLoops/Spot_IcebergIsles_Window_Wind_Cue
root=WavePlayer_0
nodes [{id=WavePlayer_0, properties {bLooping=true, soundWaveAssetPtr=/AmbientAudioContent/Waves/SpottedLoops/Spot_Winterfest_Exterior_Wind_Loop}}]
properties {maxDistance=2097152, soundClassObject=&Fort_Ambient_Indoors_Athena, volumeMultiplier=0.35}
