SoundCue /AmbientAudioContent/SoundCues/SpottedLoops/Spot_Industrial_AirVent_Tunnel_Fan_Loop_Cue
root=QualityLevel_0
nodes [
 {
  id=Attenuation_0
  children [Mixer_0]
  properties {
   bOverrideAttenuation=true
   attenuationOverrides {
    bAttenuateWithLPF=true
    bEnableFocusInterpolation=true
    bEnableListenerFocus=true
    bEnableOcclusion=true
    bEnableReverbSend=false
    falloffDistance=2000
    focusReleaseInterpSpeed=0.25
    lPFFrequencyAtMax=2000
    lPFRadiusMax=3000
    lPFRadiusMin=1000
    nonFocusVolumeAttenuation=0.5
    occlusionInterpolationTime=1
    occlusionLowPassFilterFrequency=1500
    occlusionTraceChannel=ECC_GameTraceChannel11
    occlusionVolumeAttenuation=0.5
    stereoSpread=100
    attenuationShapeExtents {x=200}
   }
  }
 }
 {
  id=Attenuation_1
  children [Modulator_0]
  properties {
   bOverrideAttenuation=true
   attenuationOverrides {
    bAttenuateWithLPF=true
    bEnableFocusInterpolation=true
    bEnableListenerFocus=true
    bEnableOcclusion=true
    bEnableReverbSend=false
    falloffDistance=1000
    focusReleaseInterpSpeed=0.25
    lPFFrequencyAtMax=2000
    lPFRadiusMax=3000
    lPFRadiusMin=500
    nonFocusVolumeAttenuation=0.5
    occlusionInterpolationTime=1
    occlusionLowPassFilterFrequency=1500
    occlusionTraceChannel=ECC_GameTraceChannel11
    occlusionVolumeAttenuation=0.5
    stereoSpread=50
    attenuationShapeExtents {x=0}
   }
  }
 }
 {id=Delay_0, children [Modulator_2], properties {delayMax=18, delayMin=3}}
 {id=Looping_0, children [Delay_0]}
 {id=Mixer_0, children [Looping_0], properties {inputVolume [0.75]}}
 {id=Mixer_1, children [WavePlayer_1,Modulator_1], properties {inputVolume [1.2,0.25]}}
 {id=Mixer_2, children [Attenuation_1, Attenuation_0], properties {inputVolume [0.5,0.5]}}
 {id=Modulator_0, children [Mixer_1], properties {pitchMax=1.1, pitchMin=1}}
 {id=Modulator_1, children [WavePlayer_0], properties {pitchMax=0.8, pitchMin=0.8, volumeMax=1, volumeMin=1}}
 {id=Modulator_2, children [Random_0], properties {pitchMax=1.1, pitchMin=0.9, volumeMin=0.9}}
 {id=QualityLevel_0, children [Mixer_2,Mixer_2,Mixer_2,Mixer_2]}
 {id=Random_0, children [WavePlayer_3,WavePlayer_4,WavePlayer_5], properties {weights [1,1,1]}}
 {id=WavePlayer_0, properties {bLooping=true, soundWaveAssetPtr=/AmbientAudioContent/Waves/SpottedLoops/Spot_AC_Duct_Air_03}}
 {
  id=WavePlayer_1
  properties {bLooping=true, soundWaveAssetPtr=/AmbientAudioContent/Waves/SpottedLoops/Spot_AirVent_Industrial_Tunnel_Fan_Loop}
 }
 {id=WavePlayer_3, properties {soundWaveAssetPtr=/AmbientAudioContent/Waves/SpottedLoops/Spot_AirVent_Industrial_Tunnel_Fan_Squeak_01}}
 {id=WavePlayer_4, properties {soundWaveAssetPtr=/AmbientAudioContent/Waves/SpottedLoops/Spot_AirVent_Industrial_Tunnel_Fan_Squeak_02}}
 {id=WavePlayer_5, properties {soundWaveAssetPtr=/AmbientAudioContent/Waves/SpottedLoops/Spot_AirVent_Industrial_Tunnel_Fan_Squeak_03}}
]
properties {
 bOverrideConcurrency=true
 maxDistance=2200
 soundClassObject=&Fort_Ambient_Indoors_Athena
 volumeMultiplier=1
 attenuationOverrides {attenuationShapeExtents {x=0}}
 concurrencyOverrides {maxCount=1, voiceStealReleaseTime=2}
}
