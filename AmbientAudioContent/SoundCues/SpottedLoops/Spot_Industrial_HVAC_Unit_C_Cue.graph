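# Sound cue for a quiet, spotted industrial HVAC unit ambient loop in the Athena ambient sound class, capped by a concurrency override.
# A looping wave passes through a default Modulator into an Attenuation node with distance low-pass, occlusion and listener focus, with reverb send off.
# A quality-level root routes the first two levels to that shared chain and leaves the other two empty, so those play nothing.
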SoundCue /AmbientAudioContent/SoundCues/SpottedLoops/Spot_Industrial_HVAC_Unit_C_Cue
root=QualityLevel_0
nodes [
 {
  id=Attenuation_1
  children [Modulator_0]
  properties {
   bOverrideAttenuation=true
   attenuationOverrides {
    bAttenuateWithLPF=true
    bEnableFocusInterpolation=true
    bEnableListenerFocus=true
    bEnableOcclusion=true
    bEnableReverbSend=false
    falloffDistance=700
    focusReleaseInterpSpeed=0.25
    lPFFrequencyAtMax=2000
    lPFRadiusMax=3000
    lPFRadiusMin=500
    nonFocusVolumeAttenuation=0.5
    occlusionInterpolationTime=1
    occlusionLowPassFilterFrequency=2000
    occlusionTraceChannel=ECC_GameTraceChannel11
    occlusionVolumeAttenuation=0.5
    attenuationShapeExtents {x=300}
   }
  }
 }
 {id=Modulator_0, children [WavePlayer_1]}
 {id=QualityLevel_0, children [Attenuation_1, Attenuation_1, null, null]}
 {id=WavePlayer_1, properties {bLooping=true, soundWaveAssetPtr=/AmbientAudioContent/Waves/OneShots/Spot_HVAC_Industrial_Unit_C}}
]
properties {
 bOverrideConcurrency=true
 maxDistance=1000
 pitchMultiplier=1.1
 soundClassObject=&Fort_Ambient_Athena
 volumeMultiplier=0.15
 attenuationOverrides {attenuationShapeExtents {x=0}}
 concurrencyOverrides {maxCount=3}
}
