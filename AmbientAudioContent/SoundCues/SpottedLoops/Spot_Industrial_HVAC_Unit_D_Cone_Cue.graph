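# Spot ambient SoundCue for an industrial HVAC unit, playing a looping AC duct air sound in the outdoor ambience sound class.
# A Random node picks one of two equally weighted looping waves and feeds a Modulator, which all four inputs of the root Quality Level node share.
# An HVAC cone attenuation asset shapes falloff, while a concurrency override caps simultaneous instances and the cue plays at reduced volume.
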SoundCue /AmbientAudioContent/SoundCues/SpottedLoops/Spot_Industrial_HVAC_Unit_D_Cone_Cue
root=QualityLevel_0
nodes [
 {id=Modulator_0, children [Random_1]}
 {id=QualityLevel_0, children [Modulator_0,Modulator_0,Modulator_0,Modulator_0]}
 {id=Random_1, children [WavePlayer_0,WavePlayer_2], properties {weights [1,1]}}
 {id=WavePlayer_0, properties {bLooping=true, soundWaveAssetPtr=/AmbientAudioContent/Waves/SpottedLoops/Spot_AC_Duct_Air_03}}
 {id=WavePlayer_2, properties {bLooping=true, soundWaveAssetPtr=/AmbientAudioContent/Waves/SpottedLoops/Spot_AC_Duct_Air_02}}
]
properties {
 attenuationSettings=/AmbientAudioBase/Attenuations/ATT_Spot_City_HVAC_Aircon_Cone
 bOverrideConcurrency=true
 maxDistance=1375
 soundClassObject=&Fort_Ambient_Outdoors_Athena
 volumeMultiplier=0.5
 attenuationOverrides {attenuationShapeExtents {x=0}}
 concurrencyOverrides {maxCount=3}
}
