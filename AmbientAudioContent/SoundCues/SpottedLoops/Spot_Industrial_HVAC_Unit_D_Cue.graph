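# Looping spot ambience cue for an industrial HVAC unit, classed as outdoor ambient sound and limited by its own concurrency override.
# A Random node picks one of two equally weighted, looping AC duct air waves, then a default Modulator varies it.
# An Attenuation node adds distance low-pass, listener focus and traced occlusion, with reverb send off.
# The root Quality Level node feeds that chain to its first two slots and leaves the other two empty.
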SoundCue /AmbientAudioContent/SoundCues/SpottedLoops/Spot_Industrial_HVAC_Unit_D_Cue
root=QualityLevel_0
nodes [
 {
  id=Attenuation_1
  children [Modulator_0]
  properties {
   bOverrideAttenuation=true
   attenuationOverrides {
    bAttenuateWithLPF=true
    bEnableFocusInterpolation=true
    bEnableListenerFocus=true
    bEnableOcclusion=true
    bEnableReverbSend=false
    falloffDistance=700
    focusReleaseInterpSpeed=0.25
    lPFFrequencyAtMax=2000
    lPFRadiusMax=3000
    lPFRadiusMin=500
    nonFocusVolumeAttenuation=0.5
    occlusionInterpolationTime=1
    occlusionLowPassFilterFrequency=2000
    occlusionTraceChannel=ECC_GameTraceChannel11
    occlusionVolumeAttenuation=0.5
    attenuationShapeExtents {x=200}
   }
  }
 }
 {id=Modulator_0, children [Random_1]}
 {id=QualityLevel_0, children [Attenuation_1, Attenuation_1, null, null]}
 {id=Random_1, children [WavePlayer_0,WavePlayer_2], properties {weights [1,1]}}
 {id=WavePlayer_0, properties {bLooping=true, soundWaveAssetPtr=/AmbientAudioContent/Waves/SpottedLoops/Spot_AC_Duct_Air_03}}
 {id=WavePlayer_2, properties {bLooping=true, soundWaveAssetPtr=/AmbientAudioContent/Waves/SpottedLoops/Spot_AC_Duct_Air_02}}
]
properties {
 bOverrideConcurrency=true
 maxDistance=900
 soundClassObject=&Fort_Ambient_Outdoors_Athena
 volumeMultiplier=0.5
 attenuationOverrides {attenuationShapeExtents {x=0}}
 concurrencyOverrides {maxCount=3}
}
