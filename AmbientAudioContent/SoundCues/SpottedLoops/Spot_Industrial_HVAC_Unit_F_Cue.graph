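# Spot ambient sound cue for an industrial HVAC unit, a looping indoor emitter built on an AC duct air wave played at lowered pitch and volume.
# The looping wave player runs through a modulator into an attenuation node, which feeds every branch of the quality-level root.
# Attenuation overrides add distance low-pass filtering, listener focus and occlusion on a custom trace channel.
# A concurrency override limits instances, and reverb goes through a submix send.
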SoundCue /AmbientAudioContent/SoundCues/SpottedLoops/Spot_Industrial_HVAC_Unit_F_Cue
root=QualityLevel_0
nodes [
 {
  id=Attenuation_1
  children [Modulator_0]
  properties {
   bOverrideAttenuation=true
   attenuationOverrides {
    bAttenuateWithLPF=true
    bEnableFocusInterpolation=true
    bEnableListenerFocus=true
    bEnableReverbSend=false
    falloffDistance=592.9207
    focusReleaseInterpSpeed=0.25
    lPFFrequencyAtMax=2000
    lPFRadiusMax=3000
    lPFRadiusMin=500
    nonFocusVolumeAttenuation=0.5
    occlusionInterpolationTime=1
    occlusionLowPassFilterFrequency=2000
    occlusionTraceChannel=ECC_GameTraceChannel11
    occlusionVolumeAttenuation=0.5
    attenuationShapeExtents {x=688.6242}
   }
  }
 }
 {id=Modulator_0, children [WavePlayer_2]}
 {id=QualityLevel_0, children [Attenuation_1, Attenuation_1, Attenuation_1, Attenuation_1]}
 {id=WavePlayer_2, properties {bLooping=true, soundWaveAssetPtr=/AmbientAudioContent/Waves/SpottedLoops/Spot_AC_Duct_Air_02}}
]
properties {
 bOverrideConcurrency=true
 maxDistance=1281.5449
 pitchMultiplier=0.8
 soundClassObject=&Fort_Ambient_Indoors_Athena
 volumeMultiplier=0.25
 attenuationOverrides {attenuationShapeExtents {x=0}}
 concurrencyOverrides {maxCount=3}
 soundSubmixSends [{minSendLevel=0.4, soundSubmix=&Fortnite_Reverb_Default}]
}
