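# Sound Cue for a distant industrial presence spot ambience, looping a cooling tower wave in the outdoor ambient sound class.
# One looping wave player feeds two occluded Attenuation nodes with opposing custom curves, one for near range and one for far range with listener focus.
# A Mixer sums both branches into a root Modulator that randomizes pitch.
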SoundCue /AmbientAudioContent/SoundCues/SpottedLoops/Spot_Industrial_Pressence_Distant_Cue
root=Modulator_0
nodes [
 {
  id=Attenuation_0
  children [WavePlayer_2]
  properties {
   bOverrideAttenuation=true
   attenuationOverrides {
    bEnableOcclusion=true
    distanceAlgorithm=Custom
    falloffDistance=0
    focusAzimuth=50
    nonFocusAzimuth=70
    nonFocusVolumeAttenuation=0.5
    nonSpatializedRadiusEnd=4000
    nonSpatializedRadiusStart=4000
    occlusionInterpolationTime=1
    occlusionLowPassFilterFrequency=750
    stereoSpread=500
    attenuationShapeExtents {x=4000}
    customAttenuationCurve {editorCurveData {keys [<at 0 1>,<at 0.25 1>,<at 0.3 0>,<at 1 0>]}}
   }
  }
 }
 {
  id=Attenuation_2
  children [WavePlayer_2]
  properties {
   bOverrideAttenuation=true
   attenuationOverrides {
    bEnableListenerFocus=true
    bEnableOcclusion=true
    distanceAlgorithm=Custom
    falloffDistance=16000
    focusAzimuth=50
    nonFocusAzimuth=70
    nonFocusVolumeAttenuation=0.5
    nonSpatializedRadiusEnd=4000
    nonSpatializedRadiusStart=6000
    occlusionInterpolationTime=1
    occlusionLowPassFilterFrequency=750
    stereoSpread=500
    attenuationShapeExtents {x=9500}
    customAttenuationCurve {editorCurveData {keys [<at 0.25 0>,<at 0.3 1>,<at 1 0>]}}
   }
  }
 }
 {id=Mixer_0, children [Attenuation_2, Attenuation_0], properties {inputVolume [1,1]}}
 {id=Modulator_0, children [Mixer_0], properties {pitchMax=1.2, pitchMin=0.8}}
 {id=WavePlayer_2, properties {bLooping=true, soundWaveAssetPtr=/AmbientAudioContent/Waves/SpottedLoops/Spot_Cooling_Tower_Distant_01}}
]
properties {
 bDebug=true
 maxDistance=25500
 soundClassObject=&Fort_Ambient_Outdoors_Athena
 volumeMultiplier=0.3
 attenuationOverrides {bSpatialize=false, stereoSpread=1147.9967, attenuationShapeExtents {x=6826.1992}}
 busSends [{
   audioBus=null
   maxSendDistance=15000
   minSendDistance=8000
   minSendLevel=0.4
   sendLevel=1
   soundSourceBus=null
   customSendLevelCurve {editorCurveData {keys [<at 0 0.4>,<at 1 1>]}}
  }]
}
