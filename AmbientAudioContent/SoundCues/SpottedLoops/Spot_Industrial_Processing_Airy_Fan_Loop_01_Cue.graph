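# Looping spot-ambience cue for an airy industrial processing fan, in the Fort_Ambient_Athena sound class.
# One looping wave feeds two listener-focused attenuation nodes, one stereo-spread without occlusion and the other zero-spread, narrower-focus and occluded.
# A mixer layers both at equal volume, then feeds every quality level and a slight upward pitch modulator.
# The cue adds a cone attenuation override, a low concurrency cap and a default reverb submix send.
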SoundCue /AmbientAudioContent/SoundCues/SpottedLoops/Spot_Industrial_Processing_Airy_Fan_Loop_01_Cue
root=Modulator_0
nodes [
 {
  id=Attenuation_1
  children [WavePlayer_1]
  properties {
   bOverrideAttenuation=true
   attenuationOverrides {
    bEnableListenerFocus=true
    falloffDistance=1750
    focusAzimuth=25
    nonFocusAzimuth=55
    nonFocusVolumeAttenuation=0.75
    stereoSpread=50
    attenuationShapeExtents=<250 60 30>
    customAttenuationCurve {editorCurveData {keys [<at 0 1>,<at 0.4985 0.2>,<at 1 0>]}}
    submixSendSettings [{
      bEnableHPFCutoff=false
      bEnableLPFCutoff=false
      disableManualSendClamp=false
      hPFCutoff=20
      lPFCutoff=20000
      maxSendDistance=6000
      maxSendLevel=1
      minSendDistance=400
      minSendLevel=0
      sendLevel=0.2
      sendLevelControlMethod=Linear
      soundSubmix=null
     }]
   }
  }
 }
 {
  id=Attenuation_3
  children [WavePlayer_1]
  properties {
   bOverrideAttenuation=true
   attenuationOverrides {
    bEnableListenerFocus=true
    bEnableOcclusion=true
    falloffDistance=1750
    focusAzimuth=15
    nonFocusAzimuth=45
    nonFocusVolumeAttenuation=0.75
    occlusionLowPassFilterFrequency=2000
    occlusionTraceChannel=ECC_GameTraceChannel11
    occlusionVolumeAttenuation=0.75
    stereoSpread=0
    attenuationShapeExtents=<250 90>
    customAttenuationCurve {editorCurveData {keys [<at 0 1>,<at 0.4985 0.2>,<at 1 0>]}}
    submixSendSettings [{
      bEnableHPFCutoff=false
      bEnableLPFCutoff=false
      disableManualSendClamp=false
      hPFCutoff=20
      lPFCutoff=20000
      maxSendDistance=6000
      maxSendLevel=1
      minSendDistance=400
      minSendLevel=0
      sendLevel=0.2
      sendLevelControlMethod=Linear
      soundSubmix=null
     }]
   }
  }
 }
 {id=Mixer_0, children [Attenuation_1, Attenuation_3], properties {inputVolume [1,1]}}
 {id=Modulator_0, children [QualityLevel_0], properties {pitchMax=1.1, pitchMin=1}}
 {id=QualityLevel_0, children [Mixer_0,Mixer_0,Mixer_0,Mixer_0]}
 {
  id=WavePlayer_1
  properties {bLooping=true, soundWaveAssetPtr=/AmbientAudioContent/Waves/SpottedLoops/Spot_Industrial_Processing_Airy_Fan_Loop_01}
 }
]
properties {
 bOverrideAttenuation=true
 bOverrideConcurrency=true
 maxDistance=2000
 soundClassObject=&Fort_Ambient_Athena
 volumeMultiplier=1
 attenuationOverrides {
  attenuationShape=Cone
  bEnableListenerFocus=true
  coneSphereRadius=250
  falloffDistance=850
  nonFocusVolumeAttenuation=0.75
  attenuationShapeExtents=<250 90>
 }
 concurrencyOverrides {maxCount=2}
 soundSubmixSends [{
   maxSendLevel=0.6
   minSendDistance=250
   minSendLevel=0.25
   soundSubmix=&Fortnite_Reverb_Default
   customSendLevelCurve {editorCurveData {keys [<at 0 0>,<at 1 1>]}}
  }]
}
