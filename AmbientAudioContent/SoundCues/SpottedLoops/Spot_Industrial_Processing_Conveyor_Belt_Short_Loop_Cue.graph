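# Ambient sound cue for a looping industrial processing conveyor belt, layering close and distant recordings by distance.
# Each layer randomly picks one of two looping waves with slight pitch variation, then gets its own capsule attenuation with reverb send.
# Custom curves fade close out as distant peaks mid-range; only distant adds low-pass and occlusion.
# A mixer combines both, feeding identical quality-level inputs and a fade-in envelope, with overridden concurrency limiting instances.
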SoundCue /AmbientAudioContent/SoundCues/SpottedLoops/Spot_Industrial_Processing_Conveyor_Belt_Short_Loop_Cue
root=Enveloper_0
nodes [
 {
  id=Attenuation_0
  children [Modulator_1]
  properties {
   bOverrideAttenuation=true
   attenuationOverrides {
    attenuationShape=Capsule
    bEnableListenerFocus=true
    bEnableSubmixSends=true
    distanceAlgorithm=Custom
    falloffDistance=3000
    nonFocusVolumeAttenuation=0.4
    stereoSpread=100
    attenuationShapeExtents=<200 100>
    customAttenuationCurve {editorCurveData {keys [<at 0 1>,<at 0.125 1>,<at 0.4937 0.2>,<at 1 0>]}}
    submixSendSettings [{
      bEnableHPFCutoff=false
      bEnableLPFCutoff=false
      disableManualSendClamp=false
      hPFCutoff=20
      lPFCutoff=20000
      maxSendDistance=2000
      maxSendLevel=0.6
      minSendDistance=1000
      minSendLevel=0.25
      sendLevel=0.2
      sendLevelControlMethod=Linear
      soundSubmix=&Fortnite_Reverb_Default
      customSendLevelCurve {editorCurveData {keys [<at 0 0>,<at 1 1>]}}
     }]
   }
  }
 }
 {
  id=Attenuation_2
  children [Modulator_0]
  properties {
   bOverrideAttenuation=true
   attenuationOverrides {
    attenuationShape=Capsule
    bAttenuateWithLPF=true
    bEnableFocusInterpolation=true
    bEnableListenerFocus=true
    bEnableOcclusion=true
    bEnableSubmixSends=true
    distanceAlgorithm=Custom
    falloffDistance=3000
    focusAttackInterpSpeed=0.1
    focusReleaseInterpSpeed=0.01
    lPFFrequencyAtMax=750
    lPFRadiusMax=4000
    lPFRadiusMin=2000
    nonFocusVolumeAttenuation=0.4
    occlusionInterpolationTime=5
    occlusionLowPassFilterFrequency=1000
    occlusionTraceChannel=ECC_GameTraceChannel11
    occlusionVolumeAttenuation=0.5
    attenuationShapeExtents {x=250}
    customAttenuationCurve {editorCurveData {keys [<at 0 0>,<at 0.2 0>,<at 0.4985 1 RCIM_Cubic>,<at 1 0 RCIM_Cubic>]}}
    submixSendSettings [{
      bEnableHPFCutoff=false
      bEnableLPFCutoff=false
      disableManualSendClamp=false
      hPFCutoff=20
      lPFCutoff=20000
      maxSendDistance=2000
      maxSendLevel=0.6
      minSendDistance=1000
      minSendLevel=0.25
      sendLevel=0.2
      sendLevelControlMethod=Linear
      soundSubmix=&Fortnite_Reverb_Default
      customSendLevelCurve {editorCurveData {keys [<at 0 0>,<at 1 1>]}}
     }]
   }
  }
 }
 {
  id=Enveloper_0
  children [QualityLevel_0]
  properties {pitchCurve {editorCurveData {keys [{time=0}]}}, volumeCurve {editorCurveData {keys [<at 0 0>,<at 0.5 1>]}}}
 }
 {id=Mixer_0, children [Attenuation_0, Attenuation_2], properties {inputVolume [1,0.6]}}
 {id=Modulator_0, children [Random_0], properties {pitchMax=1.1, pitchMin=1}}
 {id=Modulator_1, children [Random_1], properties {pitchMax=1.1, pitchMin=1}}
 {id=QualityLevel_0, children [Mixer_0,Mixer_0,Mixer_0,Mixer_0]}
 {id=Random_0, children [WavePlayer_2,WavePlayer_3], properties {weights [1,1]}}
 {id=Random_1, children [WavePlayer_4,WavePlayer_5], properties {weights [1,1]}}
 {
  id=WavePlayer_2
  properties {bLooping=true, soundWaveAssetPtr=/AmbientAudioContent/Waves/SpottedLoops/Spot_Industrial_Processing_Coveyor_Distant_Loop_02}
 }
 {
  id=WavePlayer_3
  properties {bLooping=true, soundWaveAssetPtr=/AmbientAudioContent/Waves/SpottedLoops/Spot_Industrial_Processing_Coveyor_Distant_Loop_01}
 }
 {
  id=WavePlayer_4
  properties {bLooping=true, soundWaveAssetPtr=/AmbientAudioContent/Waves/SpottedLoops/Spot_Industrial_Processing_Coveyor_Loop_01}
 }
 {
  id=WavePlayer_5
  properties {bLooping=true, soundWaveAssetPtr=/AmbientAudioContent/Waves/SpottedLoops/Spot_Industrial_Processing_Coveyor_Loop_02}
 }
]
properties {
 bOverrideConcurrency=true
 maxDistance=3250
 soundClassObject=&Fort_Ambient_Athena
 volumeMultiplier=0.3
 attenuationOverrides {
  attenuationShape=Capsule
  bEnableListenerFocus=true
  bEnableSubmixSends=true
  coneSphereRadius=250
  falloffDistance=750
  nonFocusVolumeAttenuation=0.75
  attenuationShapeExtents=<250 90 90>
 }
 concurrencyOverrides {maxCount=2, voiceStealReleaseTime=2}
 soundSubmixSends [{
   maxSendLevel=0.6
   minSendDistance=250
   minSendLevel=0.25
   soundSubmix=&Fortnite_Reverb_Default
   customSendLevelCurve {editorCurveData {keys [<at 0 0>,<at 1 1>]}}
  }]
}
