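# Looping ambient cue for a spotted belt-driven industrial processing machine, routed to the Athena ambient sound class.
# A Random node picks a delayed or immediate start for the single looping wave, which gets slight pitch variation and a fade-in envelope.
# All quality levels share this chain through a one-input Mixer.
# Overrides add distance low-pass, occlusion, listener focus, a reverb submix send and a two-instance concurrency limit.
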SoundCue /AmbientAudioContent/SoundCues/SpottedLoops/Spot_Industrial_Processing_Machine_Belt_Driven_Loop_Cue
root=QualityLevel_0
nodes [
 {id=Delay_0, children [Enveloper_1], properties {delayMax=0.3, delayMin=0.3}}
 {
  id=Enveloper_1
  children [Modulator_1]
  properties {pitchCurve {editorCurveData {keys [{time=0}]}}, volumeCurve {editorCurveData {keys [<at 0 0>,<at 1 1>]}}}
 }
 {id=Mixer_0, children [Random_1], properties {inputVolume [1]}}
 {id=Modulator_1, children [WavePlayer_0], properties {pitchMax=1.1, pitchMin=1}}
 {id=QualityLevel_0, children [Mixer_0,Mixer_0,Mixer_0,Mixer_0]}
 {id=Random_1, children [Delay_0,Enveloper_1], properties {weights [1,1]}}
 {
  id=WavePlayer_0
  properties {bLooping=true, soundWaveAssetPtr=/AmbientAudioContent/Waves/SpottedLoops/Spot_Industrial_Processing_MachineBelt_Driven_Machine_Loop}
 }
]
properties {
 bOverrideAttenuation=true
 bOverrideConcurrency=true
 maxDistance=2750
 soundClassObject=&Fort_Ambient_Athena
 volumeMultiplier=0.25
 attenuationOverrides {
  bAttenuateWithLPF=true
  bEnableListenerFocus=true
  bEnableLogFrequencyScaling=true
  bEnableOcclusion=true
  bEnableSubmixSends=true
  coneSphereRadius=250
  falloffDistance=2000
  lPFFrequencyAtMax=500
  lPFRadiusMax=3000
  lPFRadiusMin=1000
  nonFocusVolumeAttenuation=0.4
  occlusionLowPassFilterFrequency=3000
  occlusionTraceChannel=ECC_GameTraceChannel11
  attenuationShapeExtents=<750 90 90>
 }
 concurrencyOverrides {maxCount=2}
 soundSubmixSends [{
   maxSendLevel=0.6
   minSendDistance=250
   minSendLevel=0.25
   soundSubmix=&Fortnite_Reverb_Default
   customSendLevelCurve {editorCurveData {keys [<at 0 0>,<at 1 1>]}}
  }]
}
