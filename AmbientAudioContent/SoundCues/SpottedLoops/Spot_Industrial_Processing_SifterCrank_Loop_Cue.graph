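# Spotted ambient Sound Cue that loops an industrial sand sifter crank sound at low volume in the Fort_Ambient_Athena sound class.
# A looping Wave Player feeds a Modulator root that slightly randomizes pitch upward.
# Overridden attenuation uses a capsule with logarithmic falloff, distance low-pass filtering, slowly interpolated occlusion and listener focus.
# It caps concurrency at two instances and sends to the default reverb submix by distance.
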SoundCue /AmbientAudioContent/SoundCues/SpottedLoops/Spot_Industrial_Processing_SifterCrank_Loop_Cue
root=Modulator_1
nodes [
 {id=Modulator_1, children [WavePlayer_0], properties {pitchMax=1.1, pitchMin=1}}
 {
  id=WavePlayer_0
  properties {bLooping=true, soundWaveAssetPtr=/AmbientAudioContent/Waves/SpottedLoops/Spot_Industrial_Processing_Sand_Sifter_Crank_Loop}
 }
]
properties {
 bOverrideAttenuation=true
 bOverrideConcurrency=true
 maxDistance=1750
 soundClassObject=&Fort_Ambient_Athena
 volumeMultiplier=0.2
 attenuationOverrides {
  attenuationShape=Capsule
  bAttenuateWithLPF=true
  bEnableListenerFocus=true
  bEnableLogFrequencyScaling=true
  bEnableOcclusion=true
  bEnableSubmixSends=true
  coneSphereRadius=250
  distanceAlgorithm=Logarithmic
  falloffDistance=1500
  lPFFrequencyAtMax=500
  lPFRadiusMax=1500
  lPFRadiusMin=850
  nonFocusVolumeAttenuation=0.4
  occlusionInterpolationTime=5
  occlusionLowPassFilterFrequency=5000
  occlusionTraceChannel=ECC_GameTraceChannel11
  attenuationShapeExtents=<250 90 90>
 }
 concurrencyOverrides {maxCount=2}
 soundSubmixSends [{
   maxSendLevel=0.6
   minSendDistance=250
   minSendLevel=0.25
   soundSubmix=&Fortnite_Reverb_Default
   customSendLevelCurve {editorCurveData {keys [<at 0 0>,<at 1 1>]}}
  }]
}
