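# Spot ambient cue that plays a Japanese crossing signal loop outdoors as an intermittent, repeating sound.
# A looping node repeats a fixed delay followed by a concatenator, which plays the same wave player twice back to back.
# It borrows a radar dish attenuation asset, uses the outdoor ambient sound class and a shared spot concurrency set, and sends to the exterior ambient source bus.
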SoundCue /AmbientAudioContent/SoundCues/SpottedLoops/Spot_Japanese_Crossing_Signal_Loop_Cue
root=Looping_0
nodes [
 {id=Concatenator_0, children [WavePlayer_2,WavePlayer_2], properties {inputVolume [1,1]}}
 {id=Delay_0, children [Concatenator_0], properties {delayMax=45, delayMin=45}}
 {id=Looping_0, children [Delay_0]}
 {id=WavePlayer_2, properties {soundWaveAssetPtr=/AmbientAudioContent/Waves/SpottedLoops/Spot_Japanese_Crossing_Signal_Loop}}
]
properties {
 attenuationSettings=/Game/Sounds/Asteria/POI/ExplorerBase/Spots/Spot_ExplorerBase_Radar_Dish_Smaller_Att
 maxDistance=1500
 soundClassObject=&Fort_Ambient_Outdoors_Athena
 volumeMultiplier=0.301
 attenuationOverrides {
  bAttenuateWithLPF=true
  bEnableReverbSend=false
  falloffDistance=7000
  lPFFrequencyAtMax=500
  lPFRadiusMax=5000
  lPFRadiusMin=2000
  stereoSpread=5000
  attenuationShapeExtents {x=1000}
 }
 busSends [{audioBus=null, minSendLevel=0.6, sendLevel=1, soundSourceBus=&AmbientAudio_Exterior_SourceBus}]
 concurrencyOverrides {maxCount=1}
 concurrencySet [/AmbientAudioBase/Concurrency/AmbientConc_3Max_Spot_Default_Concurrency]
}
