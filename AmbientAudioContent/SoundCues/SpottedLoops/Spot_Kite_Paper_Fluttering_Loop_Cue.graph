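# Spotted ambient loop cue for a fluttering paper kite, in the outdoor ambience sound class with custom attenuation, listener focus, occlusion and reverb send.
# One looping wave runs through nested envelopes: a repeating pitch drift, a longer repeating volume swell and a short fade-in.
# A random root plays that shared chain either directly or delayed and pitched up, varying each start.
# Concurrency overrides cap voices per owner with distance-based volume scaling.
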SoundCue /AmbientAudioContent/SoundCues/SpottedLoops/Spot_Kite_Paper_Fluttering_Loop_Cue
root=Random_0
nodes [
 {id=Delay_0, children [Modulator_0], properties {delayMax=1, delayMin=1}}
 {
  id=Enveloper_0
  children [Enveloper_2]
  properties {pitchCurve {editorCurveData {keys [{time=0}]}}, volumeCurve {editorCurveData {keys [<at 0 0>,<at 1 1>]}}}
 }
 {
  id=Enveloper_1
  children [WavePlayer_2]
  properties {
   bLoop=true
   bLoopIndefinitely=true
   loopEnd=30
   pitchCurve {editorCurveData {keys [{interpMode=RCIM_Cubic, time=0}, <at 15.9468 0.9181 RCIM_Cubic>, {interpMode=RCIM_Cubic, time=30}]}}
   volumeCurve {editorCurveData {keys [<at 0 1>]}}
  }
 }
 {
  id=Enveloper_2
  children [Enveloper_1]
  properties {
   bLoop=true
   bLoopIndefinitely=true
   loopEnd=60
   pitchCurve {editorCurveData {keys [{time=0}]}}
   volumeCurve {editorCurveData {keys [
      {arriveTangent=-0.01, interpMode=RCIM_Cubic, time=0, value=1}
      <at 10 0.4 RCIM_Cubic>
      {arriveTangent=0.0037, interpMode=RCIM_Cubic, leaveTangent=0.0037, tangentMode=RCTM_User, time=20, value=1}
      <at 30 0.3 RCIM_Cubic>
      <at 40 1>
      <at 60 1>
     ]}}
  }
 }
 {id=Modulator_0, children [Enveloper_0], properties {pitchMax=1.1, pitchMin=1.1}}
 {id=Random_0, children [Delay_0,Enveloper_0], properties {weights [1,1]}}
 {id=WavePlayer_2, properties {bLooping=true, soundWaveAssetPtr=/AmbientAudioContent/Waves/SpottedLoops/Spot_Kite_Paper_Flutter_Loop}}
]
properties {
 bDebug=true
 bOverrideAttenuation=true
 bOverrideConcurrency=true
 maxDistance=7000
 soundClassObject=&Fort_Ambient_Outdoors_Athena
 volumeMultiplier=0.4
 attenuationOverrides {
  bEnableFocusInterpolation=true
  bEnableListenerFocus=true
  coneSphereRadius=600
  distanceAlgorithm=Custom
  falloffDistance=7000
  nonFocusVolumeAttenuation=0.2
  occlusionLowPassFilterFrequency=10000
  occlusionTraceChannel=ECC_GameTraceChannel11
  occlusionVolumeAttenuation=0.5
  stereoSpread=50
  attenuationShapeExtents=<0 8 10>
  customAttenuationCurve {editorCurveData {keys [
     <at 0.15 0.9897>
     {arriveTangent=-0.7842, interpMode=RCIM_Cubic, leaveTangent=-0.7842, tangentMode=RCTM_User, time=0.5322, value=0.3563}
     {arriveTangent=-0.0339, interpMode=RCIM_Cubic, leaveTangent=-0.0339, tangentMode=RCTM_User, time=1, value=0}
    ]}}
 }
 concurrencyOverrides {bLimitToOwner=true, maxCount=2, voiceStealReleaseTime=2, volumeScaleAttackTime=2, volumeScaleMode=Distance}
 soundSubmixSends [{
   maxSendDistance=2500
   minSendDistance=500
   minSendLevel=0.4
   soundSubmix=&Fortnite_Reverb_Default
   customSendLevelCurve {editorCurveData {keys [<at 0 0.4>,<at 1 1>]}}
  }]
}
