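# Looping ambient SoundCue that plays occasional lake ice pings as a spot sound in the outdoor ambience class.
# A Looping node wraps a randomized delay of several seconds, so pings repeat at irregular intervals.
# Each pass picks one of five equally weighted ice ping one-shots through a Random node, with a Modulator varying pitch and volume slightly.
# An external attenuation asset handles spatial falloff.
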SoundCue /AmbientAudioContent/SoundCues/SpottedLoops/Spot_Lake_Ice_Ping_Loop_Cue
root=Looping_0
nodes [
 {id=Delay_0, children [Modulator_1], properties {delayMax=12, delayMin=6}}
 {id=Looping_0, children [Delay_0]}
 {id=Modulator_1, children [Random_0], properties {pitchMax=1.1, pitchMin=0.9, volumeMin=0.85}}
 {id=Random_0, children [WavePlayer_1,WavePlayer_2,WavePlayer_3,WavePlayer_4,WavePlayer_5], properties {weights [1,1,1,1,1]}}
 {id=WavePlayer_1, properties {soundWaveAssetPtr=/AmbientAudioContent/Waves/OneShots/Spot_RealityWarrior_Ice_Ping_01}}
 {id=WavePlayer_2, properties {soundWaveAssetPtr=/AmbientAudioContent/Waves/OneShots/Spot_RealityWarrior_Ice_Ping_02}}
 {id=WavePlayer_3, properties {soundWaveAssetPtr=/AmbientAudioContent/Waves/OneShots/Spot_RealityWarrior_Ice_Ping_03}}
 {id=WavePlayer_4, properties {soundWaveAssetPtr=/AmbientAudioContent/Waves/OneShots/Spot_RealityWarrior_Ice_Ping_04}}
 {id=WavePlayer_5, properties {soundWaveAssetPtr=/AmbientAudioContent/Waves/OneShots/Spot_RealityWarrior_Ice_Ping_05}}
]
properties {
 attenuationSettings=/Game/Sounds/Asteria/Biomes/RealityWarrior/Spots/Spot_Ice_Lake_Ping_Smaller_Att
 maxDistance=4000
 soundClassObject=&Fort_Ambient_Outdoors_Athena
 volumeMultiplier=0.626
 preEffectBusSends [{}]
}
