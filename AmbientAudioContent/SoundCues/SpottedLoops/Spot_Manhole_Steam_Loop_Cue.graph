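# Looping spot ambience cue for manhole steam, in the outdoor ambient sound class.
# A mixer layers one of two randomly picked AC duct air loops with a heavily boosted pipe fitting loop that has its own slight pitch variation.
# A root modulator varies the pitch of the whole mix, and the cue plays at reduced volume.
# It uses a dedicated attenuation asset and overrides concurrency to one voice, stopping the quietest.
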SoundCue /AmbientAudioContent/SoundCues/SpottedLoops/Spot_Manhole_Steam_Loop_Cue
root=Modulator_0
nodes [
 {id=Mixer_0, children [Random_0,Modulator_1], properties {inputVolume [1,1]}}
 {id=Modulator_0, children [Mixer_0], properties {pitchMax=1.1, pitchMin=0.9, volumeMax=1}}
 {id=Modulator_1, children [WavePlayer_5], properties {pitchMax=1, pitchMin=0.9, volumeMax=11, volumeMin=11}}
 {id=Random_0, children [WavePlayer_2,WavePlayer_3], properties {weights [1,1]}}
 {id=WavePlayer_2, properties {bLooping=true, soundWaveAssetPtr=/AmbientAudioContent/Waves/SpottedLoops/Spot_AC_Duct_Air_03}}
 {id=WavePlayer_3, properties {bLooping=true, soundWaveAssetPtr=/AmbientAudioContent/Waves/SpottedLoops/Spot_AC_Duct_Air_01}}
 {id=WavePlayer_5, properties {bLooping=true, soundWaveAssetPtr=/AmbientAudioContent/Waves/SpottedLoops/Spot_Outpost_Pipe_Fitting_Loop}}
]
properties {
 attenuationSettings=/Game/Sounds/Asteria/POI/IndustrialTown/Spots/ATT_Spot_Manhole_Steam_Loop
 bOverrideConcurrency=true
 maxDistance=2000
 soundClassObject=&Fort_Ambient_Outdoors_Athena
 volumeMultiplier=0.175
 attenuationOverrides {attenuationShape=Capsule, bEnableListenerFocus=true, falloffDistance=1250, nonFocusVolumeAttenuation=0.4, attenuationShapeExtents=<1200 250>}
 concurrencyOverrides {maxCount=1, resolutionRule=StopQuietest, voiceStealReleaseTime=0.5, volumeScaleAttackTime=0.5}
}
