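# Looping spot ambience sound cue for a metal air duct fan.
# A single looping wave player plays one air duct fan wave at reduced volume, with no mixing or randomization.
# It uses the air duct fan attenuation asset, routes to the Fort_Doors sound class, and overrides concurrency to prevent new instances once the limit is reached.
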SoundCue /AmbientAudioContent/SoundCues/SpottedLoops/Spot_Metal_AirDuct_FanLoop_Cue
root=WavePlayer_0
nodes [{id=WavePlayer_0, properties {bLooping=true, soundWaveAssetPtr=/AmbientAudioContent/Waves/SpottedLoops/Spot_AMB_Air_Duct_Fan_Loop_01}}]
properties {
 attenuationSettings=/Game/Sounds/Apollo/Props/AirDucts/AirDuct_Fan_Att
 bOverrideConcurrency=true
 maxDistance=512
 soundClassObject=&Fort_Doors
 volumeMultiplier=0.4
 concurrencyOverrides {resolutionRule=PreventNew}
}
