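# Looping ambient spot cue that plays intermittent light metal-stress creaks, built from Tilted Towers one-shot waves.
# An endless loop waits a random delay of several seconds, then plays one of ten equally weighted waves with randomized volume and pitch, raised overall by the cue's pitch multiplier.
# The cue overrides attenuation with a custom falloff curve and listener focus, caps concurrency, and sends to the default reverb submix by distance.
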SoundCue /AmbientAudioContent/SoundCues/SpottedLoops/Spot_Metal_Stress_Light_Loop_Cue
root=Looping_0
nodes [
 {id=Delay_0, children [Modulator_0], properties {delayMax=14, delayMin=4}}
 {id=Looping_0, children [Delay_0]}
 {id=Modulator_0, children [Random_1], properties {pitchMin=0.99, volumeMin=0.5}}
 {
  id=Random_1
  children [
   WavePlayer_0
   WavePlayer_1
   WavePlayer_2
   WavePlayer_13
   WavePlayer_14
   WavePlayer_15
   WavePlayer_16
   WavePlayer_17
   WavePlayer_18
   WavePlayer_19
  ]
  properties {weights [1,1,1,1,1,1,1,1,1,1]}
 }
 {id=WavePlayer_0, properties {soundWaveAssetPtr=/AmbientAudioContent/Waves/OneShots/Spot_Tilted_Towers_Metal_Stress_Light_01}}
 {id=WavePlayer_1, properties {soundWaveAssetPtr=/AmbientAudioContent/Waves/OneShots/Spot_Tilted_Towers_Metal_Stress_Light_02}}
 {id=WavePlayer_13, properties {soundWaveAssetPtr=/AmbientAudioContent/Waves/OneShots/Spot_Tilted_Towers_Metal_Stress_Light_04}}
 {id=WavePlayer_14, properties {soundWaveAssetPtr=/AmbientAudioContent/Waves/OneShots/Spot_Tilted_Towers_Metal_Stress_Light_05}}
 {id=WavePlayer_15, properties {soundWaveAssetPtr=/AmbientAudioContent/Waves/OneShots/Spot_Tilted_Towers_Metal_Stress_Light_06}}
 {id=WavePlayer_16, properties {soundWaveAssetPtr=/AmbientAudioContent/Waves/OneShots/Spot_Tilted_Towers_Metal_Stress_Light_07}}
 {id=WavePlayer_17, properties {soundWaveAssetPtr=/AmbientAudioContent/Waves/OneShots/Spot_Tilted_Towers_Metal_Stress_Light_08}}
 {id=WavePlayer_18, properties {soundWaveAssetPtr=/AmbientAudioContent/Waves/OneShots/Spot_Tilted_Towers_Metal_Stress_Light_09}}
 {id=WavePlayer_19, properties {soundWaveAssetPtr=/AmbientAudioContent/Waves/OneShots/Spot_Tilted_Towers_Metal_Stress_Light_10}}
 {id=WavePlayer_2, properties {soundWaveAssetPtr=/AmbientAudioContent/Waves/OneShots/Spot_Tilted_Towers_Metal_Stress_Light_03}}
]
properties {
 bOverrideAttenuation=true
 bOverrideConcurrency=true
 maxDistance=3600.0002
 pitchMultiplier=2.4
 soundClassObject=&Fort_Ambient_OneShots
 volumeMultiplier=0.6
 attenuationOverrides {
  bEnableListenerFocus=true
  distanceAlgorithm=Custom
  falloffDistance=3000
  nonFocusVolumeAttenuation=0.75
  attenuationShapeExtents {x=0}
  customAttenuationCurve {editorCurveData {keys [
     {arriveTangent=-0.6475, interpMode=RCIM_Cubic, leaveTangent=-0.6475, tangentMode=RCTM_User, time=0, value=1}
     {arriveTangent=-1.2105, interpMode=RCIM_Cubic, leaveTangent=-1.2105, tangentMode=RCTM_User, time=0.3306, value=0.5982}
     {arriveTangent=-0.5571, interpMode=RCIM_Cubic, leaveTangent=-0.5571, tangentMode=RCTM_User, time=1.2, value=0}
    ]}}
 }
 concurrencyOverrides {maxCount=2}
 soundSubmixSends [{
   maxSendDistance=2000
   maxSendLevel=0.6
   minSendDistance=250
   minSendLevel=0.25
   soundSubmix=&Fortnite_Reverb_Default
   customSendLevelCurve {editorCurveData {keys [<at 0 0>,<at 1 1>]}}
  }]
}
