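# Sound cue for an exterior ambient spot loop of light metal stress sounds, using Tilted Towers one-shot waves.
# A looping node repeats a random delay of a few seconds, then a modulator varies pitch and volume of one of ten equally weighted waves.
# The cue pitches the waves well above their recorded pitch at reduced volume.
# It plays on the outdoor ambient one-shot sound class with a smaller metal-stress attenuation asset.
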SoundCue /AmbientAudioContent/SoundCues/SpottedLoops/Spot_Metal_Stress_Light_Loop_Ext_Cue
root=Looping_0
nodes [
 {id=Delay_0, children [Modulator_0], properties {delayMax=8, delayMin=3}}
 {id=Looping_0, children [Delay_0]}
 {id=Modulator_0, children [Random_1], properties {pitchMax=1.1, pitchMin=0.9, volumeMin=0.7}}
 {
  id=Random_1
  children [
   WavePlayer_0
   WavePlayer_1
   WavePlayer_2
   WavePlayer_13
   WavePlayer_14
   WavePlayer_15
   WavePlayer_16
   WavePlayer_17
   WavePlayer_18
   WavePlayer_19
  ]
  properties {weights [1,1,1,1,1,1,1,1,1,1]}
 }
 {id=WavePlayer_0, properties {soundWaveAssetPtr=/AmbientAudioContent/Waves/OneShots/Spot_Tilted_Towers_Metal_Stress_Light_01}}
 {id=WavePlayer_1, properties {soundWaveAssetPtr=/AmbientAudioContent/Waves/OneShots/Spot_Tilted_Towers_Metal_Stress_Light_02}}
 {id=WavePlayer_13, properties {soundWaveAssetPtr=/AmbientAudioContent/Waves/OneShots/Spot_Tilted_Towers_Metal_Stress_Light_04}}
 {id=WavePlayer_14, properties {soundWaveAssetPtr=/AmbientAudioContent/Waves/OneShots/Spot_Tilted_Towers_Metal_Stress_Light_05}}
 {id=WavePlayer_15, properties {soundWaveAssetPtr=/AmbientAudioContent/Waves/OneShots/Spot_Tilted_Towers_Metal_Stress_Light_06}}
 {id=WavePlayer_16, properties {soundWaveAssetPtr=/AmbientAudioContent/Waves/OneShots/Spot_Tilted_Towers_Metal_Stress_Light_07}}
 {id=WavePlayer_17, properties {soundWaveAssetPtr=/AmbientAudioContent/Waves/OneShots/Spot_Tilted_Towers_Metal_Stress_Light_08}}
 {id=WavePlayer_18, properties {soundWaveAssetPtr=/AmbientAudioContent/Waves/OneShots/Spot_Tilted_Towers_Metal_Stress_Light_09}}
 {id=WavePlayer_19, properties {soundWaveAssetPtr=/AmbientAudioContent/Waves/OneShots/Spot_Tilted_Towers_Metal_Stress_Light_10}}
 {id=WavePlayer_2, properties {soundWaveAssetPtr=/AmbientAudioContent/Waves/OneShots/Spot_Tilted_Towers_Metal_Stress_Light_03}}
]
properties {
 attenuationSettings=/AmbientAudioBase/Attenuations/Fort_Metal_Stress_Oneshot_Smaller_Att
 maxDistance=5200
 pitchMultiplier=2.4
 soundClassObject=&Fort_Ambient_OneShots_Outdoors
 volumeMultiplier=0.7
 attenuationOverrides {falloffDistance=6000, stereoSpread=10000, attenuationShapeExtents {x=2000}}
}
