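# Looping exterior ambient spot cue that scatters occasional metal stairwell stress creaks.
# A loop waits a random few seconds, then plays one of five equally weighted creak waves through pitch and volume modulation.
# Overridden attenuation uses a custom curve that is silent up close, peaks at mid-near range and fades out with distance.
# It limits concurrency and sends to default and distant reverb submixes plus the exterior ambient bus.
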SoundCue /AmbientAudioContent/SoundCues/SpottedLoops/Spot_Metal_Stress_Stair_Creaks_01_Ext_Cue
root=Looping_0
nodes [
 {id=Delay_0, children [Modulator_0], properties {delayMax=12, delayMin=3}}
 {id=Looping_0, children [Delay_0]}
 {id=Modulator_0, children [Random_1], properties {pitchMin=0.8}}
 {id=Random_1, children [WavePlayer_0,WavePlayer_1,WavePlayer_2,WavePlayer_3,WavePlayer_4], properties {weights [1,1,1,1,1]}}
 {id=WavePlayer_0, properties {soundWaveAssetPtr=/AmbientAudioContent/Waves/OneShots/Spot_Metal_Stress_Stairwell_Creak_02}}
 {id=WavePlayer_1, properties {soundWaveAssetPtr=/AmbientAudioContent/Waves/OneShots/Spot_Metal_Stress_Stairwell_Creak_03}}
 {id=WavePlayer_2, properties {soundWaveAssetPtr=/AmbientAudioContent/Waves/OneShots/Spot_Metal_Stress_Stairwell_Creak_05}}
 {id=WavePlayer_3, properties {soundWaveAssetPtr=/AmbientAudioContent/Waves/OneShots/Spot_Metal_Stress_Stairwell_Creak_07}}
 {id=WavePlayer_4, properties {soundWaveAssetPtr=/AmbientAudioContent/Waves/OneShots/Spot_Metal_Stress_Stairwell_Creak_08}}
]
properties {
 bOverrideAttenuation=true
 bOverrideConcurrency=true
 maxDistance=1750
 pitchMultiplier=0.9
 soundClassObject=&Fort_Ambient_OneShots_Outdoors
 volumeMultiplier=0.3
 attenuationOverrides {
  bEnableSubmixSends=true
  distanceAlgorithm=Custom
  falloffDistance=1750
  reverbDistanceMax=2000
  reverbWetLevelMin=0.6
  attenuationShapeExtents {x=0}
  customAttenuationCurve {editorCurveData {keys [
     {arriveTangent=4.4763, interpMode=RCIM_Cubic, leaveTangent=4.4763, tangentMode=RCTM_User, time=0.1, value=0}
     {arriveTangent=3.3333, interpMode=RCIM_Cubic, leaveTangent=3.3333, time=0.3, value=1}
     <at 0.4 1>
     <at 1 0>
    ]}}
  submixSendSettings [{
    bEnableHPFCutoff=false
    bEnableLPFCutoff=false
    disableManualSendClamp=false
    hPFCutoff=20
    lPFCutoff=20000
    maxSendDistance=6000
    maxSendLevel=1
    minSendDistance=400
    minSendLevel=0.6
    sendLevel=0.2
    sendLevelControlMethod=Linear
    soundSubmix=&Fortnite_Reverb_Distant
   }]
 }
 busSends [{
   audioBus=null
   maxSendDistance=2000
   minSendLevel=1
   sendLevel=1
   soundSourceBus=&AmbientAudio_Exterior_SourceBus
  }]
 concurrencyOverrides {maxCount=2}
 preEffectBusSends [{soundSourceBus=null}]
 soundSubmixSends [{minSendLevel=1, soundSubmix=&Fortnite_Reverb_Default}]
}
