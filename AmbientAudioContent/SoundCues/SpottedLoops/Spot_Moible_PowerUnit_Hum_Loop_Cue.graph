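# Ambient sound cue for a looping mobile power unit hum, placed as a spot sound in the Fort Athena ambient props sound class.
# It mixes a computer server loop with a quieter, pitched-down transformer hum, then applies slight random pitch variation and a volume fade-in envelope.
# The Random root starts that shared chain either immediately or after a short delay, varying start timing.
# Overrides set custom-curve attenuation with low-pass, occlusion and listener focus, plus stop-quietest concurrency.
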SoundCue /AmbientAudioContent/SoundCues/SpottedLoops/Spot_Moible_PowerUnit_Hum_Loop_Cue
root=Random_1
nodes [
 {id=Delay_0, children [Enveloper_1], properties {delayMax=0.3, delayMin=0.3}}
 {
  id=Enveloper_1
  children [Modulator_1]
  properties {pitchCurve {editorCurveData {keys [{time=0}]}}, volumeCurve {editorCurveData {keys [<at 0 0>,<at 1 1>]}}}
 }
 {id=Mixer_1, children [WavePlayer_2,Modulator_0], properties {inputVolume [1,0.5]}}
 {id=Modulator_0, children [WavePlayer_3], properties {pitchMax=0.8, pitchMin=0.8}}
 {id=Modulator_1, children [Mixer_1], properties {pitchMax=1.1, pitchMin=0.9}}
 {id=Random_1, children [Delay_0,Enveloper_1], properties {weights [1,1]}}
 {id=WavePlayer_2, properties {bLooping=true, soundWaveAssetPtr=/AmbientAudioContent/Waves/SpottedLoops/Spot_Agency_Computer_Server_Loop}}
 {id=WavePlayer_3, properties {bLooping=true, soundWaveAssetPtr=/AmbientAudioContent/Waves/SpottedLoops/Spot_Electric_Transformer_Hum_05_Loop}}
]
properties {
 bOverrideAttenuation=true
 bOverrideConcurrency=true
 maxDistance=850
 priority=50
 soundClassObject=&Fort_Ambient_Props_Athena
 volumeMultiplier=0.5
 attenuationOverrides {
  bEnableListenerFocus=true
  falloffDistance=750
  focusAzimuth=25
  lPFFrequencyAtMax=10000
  lPFRadiusMax=1000
  lPFRadiusMin=500
  nonFocusAzimuth=55
  nonFocusVolumeAttenuation=0.75
  occlusionInterpolationTime=0.3
  occlusionLowPassFilterFrequency=3000
  occlusionVolumeAttenuation=0.75
  stereoSpread=50
  attenuationShapeExtents {x=100}
  customAttenuationCurve {editorCurveData {keys [<at 0 1 RCIM_Cubic>, {arriveTangent=-1, interpMode=RCIM_Cubic, leaveTangent=-1, time=0.5, value=0.1}, <at 1 0 RCIM_Cubic>]}}
 }
 concurrencyOverrides {resolutionRule=StopQuietest}
}
