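# Looping ambient sound cue for small metal truss creaks at spotted NitroDome locations, classed as outdoor ambience.
# After each random 8–18 second delay, a Mixer layers one random stress groan or knock with one random bridge creak.
# The loop feeds only the first two quality levels, leaving the other two empty.
# It uses "Large" capsule attenuation despite its name, overrides concurrency, and sends to exterior bus and reverb.
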SoundCue /AmbientAudioContent/SoundCues/SpottedLoops/Spot_NitroDome_Metal_Truss_Creaks_Small_Loop_Cue
root=QualityLevel_0
nodes [
 {id=Delay_0, children [Mixer_0], properties {delayMax=18, delayMin=8}}
 {id=Looping_0, children [Delay_0]}
 {id=Mixer_0, children [Random_2,Random_3], properties {inputVolume [1,1]}}
 {id=QualityLevel_0, children [Looping_0,Looping_0,null,null]}
 {
  id=Random_2
  children [
   WavePlayer_28
   WavePlayer_29
   WavePlayer_30
   WavePlayer_31
   WavePlayer_32
   WavePlayer_33
   WavePlayer_37
   WavePlayer_38
   WavePlayer_39
   WavePlayer_34
   WavePlayer_35
   WavePlayer_36
  ]
  properties {weights [1,1,1,1,1,1,1,1,1,1,1,1]}
 }
 {
  id=Random_3
  children [WavePlayer_20, WavePlayer_21, WavePlayer_22, WavePlayer_23, WavePlayer_24, WavePlayer_25, WavePlayer_26, WavePlayer_27]
  properties {weights [1,1,1,1,1,1,1,1]}
 }
 {id=WavePlayer_20, properties {soundWaveAssetPtr=/AmbientAudioContent/Waves/OneShots/Spot_Big_Bridge_Metal_Creak_Large_03}}
 {id=WavePlayer_21, properties {soundWaveAssetPtr=/AmbientAudioContent/Waves/OneShots/Spot_Big_Bridge_Metal_Creak_01}}
 {id=WavePlayer_22, properties {soundWaveAssetPtr=/AmbientAudioContent/Waves/OneShots/Spot_Big_Bridge_Metal_Creak_02}}
 {id=WavePlayer_23, properties {soundWaveAssetPtr=/AmbientAudioContent/Waves/OneShots/Spot_Big_Bridge_Metal_Creak_03}}
 {id=WavePlayer_24, properties {soundWaveAssetPtr=/AmbientAudioContent/Waves/OneShots/Spot_Big_Bridge_Metal_Creak_04}}
 {id=WavePlayer_25, properties {soundWaveAssetPtr=/AmbientAudioContent/Waves/OneShots/Spot_Big_Bridge_Metal_Creak_05}}
 {id=WavePlayer_26, properties {soundWaveAssetPtr=/AmbientAudioContent/Waves/OneShots/Spot_Big_Bridge_Metal_Creak_Large_01}}
 {id=WavePlayer_27, properties {soundWaveAssetPtr=/AmbientAudioContent/Waves/OneShots/Spot_Big_Bridge_Metal_Creak_Large_02}}
 {id=WavePlayer_28, properties {soundWaveAssetPtr=/AmbientAudioContent/Waves/OneShots/Spot_Metal_Stress_Wronk_Groan_Knock_01}}
 {id=WavePlayer_29, properties {soundWaveAssetPtr=/AmbientAudioContent/Waves/OneShots/Spot_Metal_Stress_Wronk_Groan_Knock_02}}
 {id=WavePlayer_30, properties {soundWaveAssetPtr=/AmbientAudioContent/Waves/OneShots/Spot_Metal_Stress_Wronk_Groan_Knock_03}}
 {id=WavePlayer_31, properties {soundWaveAssetPtr=/AmbientAudioContent/Waves/OneShots/Spot_Metal_Stress_Wronk_Groan_Knock_04}}
 {id=WavePlayer_32, properties {soundWaveAssetPtr=/AmbientAudioContent/Waves/OneShots/Spot_Metal_Stress_Wronk_Groan_Knock_05}}
 {id=WavePlayer_33, properties {soundWaveAssetPtr=/AmbientAudioContent/Waves/OneShots/Spot_Metal_Stress_Wronk_Groan_Knock_06}}
 {id=WavePlayer_34, properties {soundWaveAssetPtr=/AmbientAudioContent/Waves/OneShots/Spot_Metal_Stress_Wronk_Groan_Knock_07}}
 {id=WavePlayer_35, properties {soundWaveAssetPtr=/AmbientAudioContent/Waves/OneShots/Spot_Metal_Stress_Wronk_Groan_Knock_08}}
 {id=WavePlayer_36, properties {soundWaveAssetPtr=/AmbientAudioContent/Waves/OneShots/Spot_Metal_Stress_Wronk_Groan_Knock_09}}
 {id=WavePlayer_37, properties {soundWaveAssetPtr=/AmbientAudioContent/Waves/OneShots/Spot_Metal_Stress_Wronk_Groan_Stretch_02}}
 {id=WavePlayer_38, properties {soundWaveAssetPtr=/AmbientAudioContent/Waves/OneShots/Spot_Metal_Stress_Wronk_Groan_Stretch_03}}
 {id=WavePlayer_39, properties {soundWaveAssetPtr=/AmbientAudioContent/Waves/OneShots/Spot_Metal_Stress_Wronk_Groan_Stretch_04}}
]
properties {
 attenuationSettings=/Game/Sounds/Helios/POI/NitroDrome/Spots/Spot_Metal_Truss_Creaks_Large_Linear_Capsule_Att
 bOverrideConcurrency=true
 maxDistance=3446.1375
 soundClassObject=&Fort_Ambient_Outdoors_Athena
 volumeMultiplier=0.501
 busSends [
  {audioBus=null, minSendLevel=1, sendLevel=1, soundSourceBus=&AmbientAudio_Exterior_SourceBus}
  {audioBus=null, minSendLevel=0, sendLevel=1, soundSourceBus=null}
 ]
 concurrencyOverrides {maxCount=2, voiceStealReleaseTime=2}
 soundSubmixSends [{minSendLevel=0.6, soundSubmix=&Fortnite_Reverb_Default}]
}
