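# Spotted ambient sound cue for a looping neon sign buzz, named for NitroDome, built from a large transformer buzz wave in the outdoor ambient sound class.
# A short random start delay and random pitch offset vary each instance, while a looping envelope slowly dips and restores volume.
# It uses a shared cone attenuation asset, overrides concurrency to two voices, and sends to the exterior ambient bus and default reverb.
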SoundCue /AmbientAudioContent/SoundCues/SpottedLoops/Spot_NitroDome_Neon_Light_Buzz_Sign_Loop_Cone_Cue
root=Delay_0
nodes [
 {id=Delay_0, children [Enveloper_0], properties {delayMax=0.25}}
 {
  id=Enveloper_0
  children [Modulator_0]
  properties {
   bLoop=true
   bLoopIndefinitely=true
   loopEnd=20
   pitchCurve {editorCurveData {keys [{time=0}]}}
   volumeCurve {editorCurveData {keys [
      {arriveTangent=-0.01, interpMode=RCIM_Cubic, time=0, value=1}
      <at 10 0.7 RCIM_Cubic>
      {interpMode=RCIM_Cubic, leaveTangent=0.01, time=20, value=1}
     ]}}
  }
 }
 {id=Modulator_0, children [WavePlayer_1], properties {pitchMax=1.5, pitchMin=1}}
 {id=WavePlayer_1, properties {bLooping=true, soundWaveAssetPtr=/AmbientAudioContent/Waves/SpottedLoops/Spot_Transformer_Large_Buzz_01}}
]
properties {
 attenuationSettings=/Game/Sounds/Helios/Environment_Actors/NeonLightSignBuzz/Spot_Neon_Light_Buzz_Sign_Loop_Cone_Att
 bOverrideConcurrency=true
 maxDistance=2100
 pitchMultiplier=0.8
 soundClassObject=&Fort_Ambient_Outdoors_Athena
 volumeMultiplier=0.2
 attenuationOverrides {
  bEnableFocusInterpolation=true
  bEnableListenerFocus=true
  bEnableOcclusion=true
  dBAttenuationAtMax=-40
  distanceAlgorithm=Logarithmic
  falloffDistance=1500
  focusAzimuth=15
  focusVolumeAttenuation=1.1
  nonFocusAzimuth=45
  nonFocusVolumeAttenuation=0.4
  nonSpatializedRadiusStart=100
  occlusionTraceChannel=ECC_GameTraceChannel11
  occlusionVolumeAttenuation=0.7
  stereoSpread=2000
  attenuationShapeExtents {x=100}
 }
 busSends [{audioBus=null, minSendLevel=1, sendLevel=1, soundSourceBus=&AmbientAudio_Exterior_SourceBus}]
 concurrencyOverrides {maxCount=2, voiceStealReleaseTime=1}
 concurrencySet [/AmbientAudioBase/Concurrency/AmbientConc_3Max_Spot_Default_Concurrency]
 soundSubmixSends [{
   minSendLevel=0.2
   soundSubmix=&Fortnite_Reverb_Default
   customSendLevelCurve {editorCurveData {keys [<at 0 0.4>,<at 1 1>]}}
  }]
}
