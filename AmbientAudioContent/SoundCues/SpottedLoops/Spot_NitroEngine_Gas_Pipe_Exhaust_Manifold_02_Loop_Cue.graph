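# Looping ambient spot sound cue for a nitro engine gas pipe exhaust manifold, sent to the exterior ambient audio source bus.
# A random node picks one of three looping waves, then a modulator and an endlessly looping enveloper slowly vary pitch and volume.
# Only the first two quality levels play this chain; the other two are empty.
# Overrides add capsule attenuation with listener focus and tight concurrency, while the sound class is Fort_Doors.
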SoundCue /AmbientAudioContent/SoundCues/SpottedLoops/Spot_NitroEngine_Gas_Pipe_Exhaust_Manifold_02_Loop_Cue
root=QualityLevel_0
nodes [
 {
  id=Enveloper_0
  children [Modulator_2]
  properties {
   bLoopIndefinitely=true
   loopEnd=30
   pitchCurve {editorCurveData {keys [{interpMode=RCIM_Cubic, time=0}, <at 15 1.1 RCIM_Cubic>, {interpMode=RCIM_Cubic, time=30}]}}
   volumeCurve {editorCurveData {keys [<at 0 1 RCIM_Cubic>,<at 15 0.7 RCIM_Cubic>,<at 30 1 RCIM_Cubic>]}}
  }
 }
 {id=Modulator_2, children [Random_2]}
 {id=QualityLevel_0, children [Enveloper_0,Enveloper_0,null,null]}
 {id=Random_2, children [WavePlayer_3,WavePlayer_4,WavePlayer_5], properties {weights [1,1,1]}}
 {
  id=WavePlayer_3
  properties {bLooping=true, soundWaveAssetPtr=/AmbientAudioContent/Waves/SpottedLoops/Spot_NitroEngine_Gas_Pipe_Exhaust_Manifold_Loop_05}
 }
 {
  id=WavePlayer_4
  properties {bLooping=true, soundWaveAssetPtr=/AmbientAudioContent/Waves/SpottedLoops/Spot_NitroEngine_Gas_Pipe_Exhaust_Manifold_Loop_04}
 }
 {
  id=WavePlayer_5
  properties {bLooping=true, soundWaveAssetPtr=/AmbientAudioContent/Waves/SpottedLoops/Spot_NitroEngine_Gas_Pipe_Exhaust_Manifold_Loop_03}
 }
]
properties {
 bOverrideAttenuation=true
 bOverrideConcurrency=true
 maxDistance=3500
 soundClassObject=&Fort_Doors
 volumeMultiplier=0.6
 attenuationOverrides {
  attenuationShape=Capsule
  bEnableFocusInterpolation=true
  bEnableListenerFocus=true
  falloffDistance=1000
  focusAzimuth=15
  focusReleaseInterpSpeed=0.9
  focusVolumeAttenuation=1.2
  nonFocusAzimuth=45
  nonFocusVolumeAttenuation=0.4
  nonSpatializedRadiusEnd=125
  nonSpatializedRadiusStart=150
  stereoSpread=150
  attenuationShapeExtents=<1600 2500>
 }
 busSends [{audioBus=null, minSendLevel=0.6, sendLevel=1, soundSourceBus=&AmbientAudio_Exterior_SourceBus}]
 concurrencyOverrides {maxCount=2, voiceStealReleaseTime=1}
 concurrencySet [&Weapon_Drop_StopOldest]
}
