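# Spotted ambient loop cue for a small nitro engine gas pipe exhaust manifold, in the outdoor ambient sound class with an exterior bus send.
# It randomly picks one of three looping waves, adds random modulation and a repeating 30-second volume and pitch swell.
# A top-level random choice starts it with a short fade-in, either immediately or slightly delayed.
# Only two of four quality slots are connected; capsule attenuation adds occlusion and listener focus.
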SoundCue /AmbientAudioContent/SoundCues/SpottedLoops/Spot_NitroEngine_Gas_Pipe_Exhaust_Manifold_02_Small_Short_Loop_Cue
root=Random_0
nodes [
 {id=Delay_0, children [Enveloper_1], properties {delayMax=0.35, delayMin=0.35}}
 {
  id=Enveloper_0
  children [Modulator_2]
  properties {
   bLoopIndefinitely=true
   loopEnd=30
   pitchCurve {editorCurveData {keys [{interpMode=RCIM_Cubic, time=0}, <at 15 1.1 RCIM_Cubic>, {interpMode=RCIM_Cubic, time=30}]}}
   volumeCurve {editorCurveData {keys [<at 0 1 RCIM_Cubic>,<at 15 0.7 RCIM_Cubic>,<at 30 1 RCIM_Cubic>]}}
  }
 }
 {
  id=Enveloper_1
  children [QualityLevel_0]
  properties {pitchCurve {editorCurveData {keys [{time=0}]}}, volumeCurve {editorCurveData {keys [<at 0 0>,<at 0.2 1>]}}}
 }
 {
  id=Enveloper_2
  children [QualityLevel_0]
  properties {pitchCurve {editorCurveData {keys [{time=0}]}}, volumeCurve {editorCurveData {keys [<at 0 0>,<at 0.2 1>]}}}
 }
 {id=Modulator_2, children [Random_2], properties {pitchMax=1.1}}
 {id=QualityLevel_0, children [Enveloper_0,Enveloper_0,null,null]}
 {id=Random_0, children [Delay_0,Enveloper_2], properties {weights [1,1]}}
 {id=Random_2, children [WavePlayer_3,WavePlayer_4,WavePlayer_5], properties {weights [1,1,1]}}
 {
  id=WavePlayer_3
  properties {bLooping=true, soundWaveAssetPtr=/AmbientAudioContent/Waves/SpottedLoops/Spot_NitroEngine_Gas_Pipe_Exhaust_Manifold_Loop_05}
 }
 {
  id=WavePlayer_4
  properties {bLooping=true, soundWaveAssetPtr=/AmbientAudioContent/Waves/SpottedLoops/Spot_NitroEngine_Gas_Pipe_Exhaust_Manifold_Loop_04}
 }
 {
  id=WavePlayer_5
  properties {bLooping=true, soundWaveAssetPtr=/AmbientAudioContent/Waves/SpottedLoops/Spot_NitroEngine_Gas_Pipe_Exhaust_Manifold_Loop_03}
 }
]
properties {
 bOverrideAttenuation=true
 bOverrideConcurrency=true
 maxDistance=1100
 soundClassObject=&Fort_Ambient_Outdoors_Athena
 volumeMultiplier=0.6
 attenuationOverrides {
  attenuationShape=Capsule
  bEnableFocusInterpolation=true
  bEnableListenerFocus=true
  bEnableOcclusion=true
  falloffDistance=750
  focusAzimuth=15
  focusReleaseInterpSpeed=0.9
  nonFocusAzimuth=45
  nonFocusVolumeAttenuation=0.4
  occlusionTraceChannel=ECC_GameTraceChannel11
  stereoSpread=100
  attenuationShapeExtents=<350 200>
 }
 busSends [{audioBus=null, minSendLevel=0.6, sendLevel=1, soundSourceBus=&AmbientAudio_Exterior_SourceBus}]
 concurrencyOverrides {maxCount=2, voiceStealReleaseTime=2, volumeScaleAttackTime=2, volumeScaleMode=Distance}
 concurrencySet [&Weapon_Drop_StopOldest]
}
