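# Spotted ambient loop cue for large NitroOrbSilo storage tanks, layering tube rattle and flowing liquid metal with capsule attenuation and listener focus.
# A random root starts it immediately or after a fixed delay, through a quality switch filling only the first two of four slots.
# Below that, a looping envelope slowly drifts volume and pitch over a pitch-randomized choice between two mixes of the same waves, balanced differently.
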SoundCue /AmbientAudioContent/SoundCues/SpottedLoops/Spot_NitroOrbSilo_Storage_Tank_Rattles_Large_Capsule_Loop_Cue
root=Random_0
nodes [
 {id=Delay_0, children [QualityLevel_0], properties {delayMax=1, delayMin=1}}
 {
  id=Enveloper_1
  children [Modulator_2]
  properties {
   bLoopIndefinitely=true
   loopEnd=30
   pitchCurve {editorCurveData {keys [{interpMode=RCIM_Cubic, time=0}, <at 15 1.1 RCIM_Cubic>, {interpMode=RCIM_Cubic, time=30}]}}
   volumeCurve {editorCurveData {keys [<at 0 1 RCIM_Cubic>,<at 15 0.7 RCIM_Cubic>,<at 30 1 RCIM_Cubic>]}}
  }
 }
 {id=Mixer_0, children [WavePlayer_3,WavePlayer_1], properties {inputVolume [2,1.25]}}
 {id=Mixer_1, children [WavePlayer_2,WavePlayer_0], properties {inputVolume [1,2]}}
 {id=Modulator_2, children [Random_1], properties {pitchMin=0.8}}
 {id=QualityLevel_0, children [Enveloper_1,Enveloper_1,null,null]}
 {id=Random_0, children [Delay_0, QualityLevel_0], properties {weights [1,1]}}
 {id=Random_1, children [Mixer_1,Mixer_0], properties {weights [1,1]}}
 {
  id=WavePlayer_0
  properties {bLooping=true, soundWaveAssetPtr=/AmbientAudioContent/Waves/SpottedLoops/Spot_NitroSilo_Liquid_Metal_Flowing_Inside_Loop}
 }
 {
  id=WavePlayer_1
  properties {bLooping=true, soundWaveAssetPtr=&Spot_NitroSilo_Liquid_Metal_Flowing_Inside_Loop}
 }
 {id=WavePlayer_2, properties {bLooping=true, soundWaveAssetPtr=/AmbientAudioContent/Waves/SpottedLoops/Spot_Nitro_Chimney_Tube_Rattle_Loop}}
 {id=WavePlayer_3, properties {bLooping=true, soundWaveAssetPtr=&Spot_Nitro_Chimney_Tube_Rattle_Loop}}
]
properties {
 attenuationSettings=/Game/Sounds/Artemis/Environment_Actors/OilRigStorageTank/Spot_OilRig_Storage_Tank_Att
 bEnableBusSends=false
 bEnableSubmixSends=false
 bOverrideAttenuation=true
 maxDistance=1503.1504
 soundClassObject=&Fort_Ambient_Outdoors_Athena
 volumeMultiplier=1
 attenuationOverrides {
  attenuationShape=Capsule
  bEnableListenerFocus=true
  falloffDistance=650
  focusVolumeAttenuation=1.2
  nonFocusVolumeAttenuation=0.75
  attenuationShapeExtents=<853.1504 500>
 }
 concurrencyOverrides {maxCount=2, voiceStealReleaseTime=0.2}
 concurrencySet [/AmbientAudioBase/Concurrency/AmbientConc_3Max_Spot_Default_Concurrency]
}
