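# Looping outdoor ambient SoundCue for a Nitro orb silo storage tank, layering a tube rattle with louder flowing liquid metal.
# A looping 30-second envelope slowly swells pitch while dipping volume on the mixed layers.
# A random node picks a pitched-up, delayed variant or a default-modulated one, both sharing that chain, under a fade-in envelope.
# It reuses oil rig tank attenuation, limits concurrency, and leaves two quality-level slots empty.
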SoundCue /AmbientAudioContent/SoundCues/SpottedLoops/Spot_NitroOrbSilo_Storage_Tank_Rattles_Loop_Cue
root=QualityLevel_0
nodes [
 {id=Delay_1, children [Enveloper_1], properties {delayMax=0.5, delayMin=0.5}}
 {
  id=Enveloper_0
  children [Random_0]
  properties {pitchCurve {editorCurveData {keys [{time=0}]}}, volumeCurve {editorCurveData {keys [<at 0 0>,<at 1 1>,<at 2 1>]}}}
 }
 {
  id=Enveloper_1
  children [Mixer_1]
  properties {
   bLoopIndefinitely=true
   loopEnd=30
   pitchCurve {editorCurveData {keys [{interpMode=RCIM_Cubic, time=0}, <at 15 1.1 RCIM_Cubic>, {interpMode=RCIM_Cubic, time=30}]}}
   volumeCurve {editorCurveData {keys [<at 0 1 RCIM_Cubic>,<at 15 0.7 RCIM_Cubic>,<at 30 1 RCIM_Cubic>]}}
  }
 }
 {id=Mixer_1, children [WavePlayer_2,WavePlayer_0], properties {inputVolume [1,2]}}
 {id=Modulator_0, children [Delay_1], properties {pitchMax=1.2, pitchMin=1.2}}
 {id=Modulator_1, children [Enveloper_1]}
 {id=QualityLevel_0, children [Enveloper_0,Enveloper_0,null,null]}
 {id=Random_0, children [Modulator_0,Modulator_1], properties {weights [1,1]}}
 {
  id=WavePlayer_0
  properties {bLooping=true, soundWaveAssetPtr=/AmbientAudioContent/Waves/SpottedLoops/Spot_NitroSilo_Liquid_Metal_Flowing_Inside_Loop}
 }
 {id=WavePlayer_2, properties {bLooping=true, soundWaveAssetPtr=/AmbientAudioContent/Waves/SpottedLoops/Spot_Nitro_Chimney_Tube_Rattle_Loop}}
]
properties {
 attenuationSettings=/Game/Sounds/Artemis/Environment_Actors/OilRigStorageTank/Spot_OilRig_Storage_Tank_Att
 bEnableBusSends=false
 bEnableSubmixSends=false
 maxDistance=650
 soundClassObject=&Fort_Ambient_Outdoors_Athena
 volumeMultiplier=0.2
 concurrencyOverrides {maxCount=2, voiceStealReleaseTime=0.2}
 concurrencySet [/AmbientAudioBase/Concurrency/AmbientConc_3Max_Spot_Default_Concurrency]
}
