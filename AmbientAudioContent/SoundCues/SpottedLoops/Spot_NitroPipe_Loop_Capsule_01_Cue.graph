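# Ambient spot SoundCue for a nitro pipe, layering a steady underground pipe loop with intermittent rattles.
# The loop runs through a repeating envelope that slowly shifts pitch and volume, and is mixed well above the rattles.
# Rattles are picked at random after random delays, then looped and modulated to a lower pitch.
# All quality levels share one mixer; capsule attenuation with listener focus shapes playback, which feeds an exterior ambient bus.
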SoundCue /AmbientAudioContent/SoundCues/SpottedLoops/Spot_NitroPipe_Loop_Capsule_01_Cue
root=QualityLevel_0
nodes [
 {id=Delay_0, children [Random_1], properties {delayMax=6, delayMin=1}}
 {
  id=Enveloper_1
  children [WavePlayer_0]
  properties {
   bLoopIndefinitely=true
   loopEnd=30
   pitchCurve {editorCurveData {keys [{interpMode=RCIM_Cubic, time=0}, <at 15 0.8 RCIM_Cubic>, {interpMode=RCIM_Cubic, time=30}]}}
   volumeCurve {editorCurveData {keys [<at 0 1 RCIM_Cubic>,<at 15 1.2 RCIM_Cubic>,<at 30 1 RCIM_Cubic>]}}
  }
 }
 {id=Looping_0, children [Delay_0]}
 {id=Mixer_1, children [Modulator_0,Enveloper_1], properties {inputVolume [1,4]}}
 {id=Modulator_0, children [Looping_0], properties {pitchMax=0.9, pitchMin=0.8, volumeMax=1.2}}
 {id=QualityLevel_0, children [Mixer_1,Mixer_1,Mixer_1,Mixer_1]}
 {id=Random_1, children [WavePlayer_5, WavePlayer_6, WavePlayer_7, WavePlayer_10, WavePlayer_11], properties {weights [1,1,1,1,1]}}
 {id=WavePlayer_0, properties {bLooping=true, soundWaveAssetPtr=/AmbientAudioContent/Waves/SpottedLoops/Spot_UndergroundPipe_Loop_02}}
 {id=WavePlayer_10, properties {soundWaveAssetPtr=/AmbientAudioContent/Waves/OneShots/Spot_UndergroundPipe_Rattle_05}}
 {id=WavePlayer_11, properties {soundWaveAssetPtr=/AmbientAudioContent/Waves/OneShots/Spot_UndergroundPipe_Rattle_06}}
 {id=WavePlayer_5, properties {soundWaveAssetPtr=/AmbientAudioContent/Waves/OneShots/Spot_UndergroundPipe_Rattle_01}}
 {id=WavePlayer_6, properties {soundWaveAssetPtr=/AmbientAudioContent/Waves/OneShots/Spot_UndergroundPipe_Rattle_03}}
 {id=WavePlayer_7, properties {soundWaveAssetPtr=/AmbientAudioContent/Waves/OneShots/Spot_UndergroundPipe_Rattle_04}}
]
properties {
 attenuationSettings=/Game/Sounds/Helios/Environment_Actors/SewerPipes/Spot_UndergroundPipe_Loop_Capsule_Att
 bOverrideAttenuation=true
 maxDistance=1600
 soundClassObject=&Fort_Ambient_Athena
 volumeMultiplier=0.501
 attenuationOverrides {
  attenuationShape=Capsule
  bEnableListenerFocus=true
  falloffDistance=1250
  focusVolumeAttenuation=1.25
  nonSpatializedRadiusEnd=125
  nonSpatializedRadiusStart=150
  stereoSpread=150
  attenuationShapeExtents=<350 250>
 }
 busSends [{audioBus=null, minSendLevel=0.6, sendLevel=1, soundSourceBus=&AmbientAudio_Exterior_SourceBus}]
 concurrencyOverrides {maxCount=2}
 concurrencySet [/AmbientAudioBase/Concurrency/AmbientConc_3Max_Spot_Default_Concurrency]
}
