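# Ambient looping spot sound cue for indoor office consoles and server racks, randomly picking one hum variation.
# The first two quality slots share a large pool: quiet mixer-attenuated server hums or full-level console and server loops.
# The last two share a small pool: one console loop, or a quiet branch holding one server hum and an empty entry.
# Overridden attenuation adds occlusion and listener focus, and concurrency is limited per owner.
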SoundCue /AmbientAudioContent/SoundCues/SpottedLoops/Spot_Office_Console_ServerRack_Various_Cue
root=QualityLevel_0
nodes [
 {id=Mixer_0, children [Random_1], properties {inputVolume [0.1]}}
 {id=Mixer_1, children [Random_4], properties {inputVolume [0.1]}}
 {id=QualityLevel_0, children [Random_0,Random_0,Random_5,Random_5]}
 {id=Random_0, children [Mixer_0,Random_2], properties {weights [1,1]}}
 {
  id=Random_1
  children [WavePlayer_9, WavePlayer_10, WavePlayer_0, WavePlayer_3, WavePlayer_5, WavePlayer_11]
  properties {weights [1,1,1,1,1,1]}
 }
 {id=Random_2, children [WavePlayer_1, WavePlayer_6, WavePlayer_8, WavePlayer_7, WavePlayer_12], properties {weights [1,1,1,1,1]}}
 {id=Random_4, children [WavePlayer_19, null], properties {weights [1,1]}}
 {id=Random_5, children [Mixer_1, WavePlayer_20], properties {weights [1,1]}}
 {id=WavePlayer_0, properties {bLooping=true, soundWaveAssetPtr=/AmbientAudioContent/Waves/SpottedLoops/Spot_Server_Stack_Hum_02}}
 {
  id=WavePlayer_1
  properties {bLooping=true, soundWaveAssetPtr=/AmbientAudioContent/Waves/SpottedLoops/Spot_Console_Terminal_Communication_Hum_Loop_04}
 }
 {id=WavePlayer_10, properties {bLooping=true, soundWaveAssetPtr=/AmbientAudioContent/Waves/SpottedLoops/Spot_Server_Stack_Hum_06}}
 {
  id=WavePlayer_11
  properties {bLooping=true, soundWaveAssetPtr=/AmbientAudioContent/Waves/SpottedLoops/Spot_Computer_Server_Rack_Tower_Hum_Loop_01}
 }
 {id=WavePlayer_12, properties {bLooping=true, soundWaveAssetPtr=/AmbientAudioContent/Waves/SpottedLoops/Spot_Agency_Computer_Server_Loop}}
 {id=WavePlayer_19, properties {bLooping=true, soundWaveAssetPtr=/AmbientAudioContent/Waves/SpottedLoops/Spot_Server_Stack_Hum_07}}
 {
  id=WavePlayer_20
  properties {bLooping=true, soundWaveAssetPtr=/AmbientAudioContent/Waves/SpottedLoops/Spot_Console_Terminal_Communication_Hum_Loop_02}
 }
 {id=WavePlayer_3, properties {bLooping=true, soundWaveAssetPtr=/AmbientAudioContent/Waves/SpottedLoops/Spot_Server_Stack_Hum_04}}
 {id=WavePlayer_5, properties {bLooping=true, soundWaveAssetPtr=/AmbientAudioContent/Waves/SpottedLoops/Spot_Server_Stack_Hum_05}}
 {
  id=WavePlayer_6
  properties {bLooping=true, soundWaveAssetPtr=/AmbientAudioContent/Waves/SpottedLoops/Spot_Console_Terminal_Communication_Hum_Loop_01}
 }
 {
  id=WavePlayer_7
  properties {bLooping=true, soundWaveAssetPtr=&Spot_Console_Terminal_Communication_Hum_Loop_02}
 }
 {
  id=WavePlayer_8
  properties {bLooping=true, soundWaveAssetPtr=/AmbientAudioContent/Waves/SpottedLoops/Spot_Console_Terminal_Communication_Hum_Loop_03}
 }
 {id=WavePlayer_9, properties {bLooping=true, soundWaveAssetPtr=&Spot_Server_Stack_Hum_07}}
]
properties {
 attenuationSettings=/Game/Sounds/Helios/Environment_Actors/ConsoleTerminals/Spot_Console_Terminal_Communication_Hum_Att
 bOverrideAttenuation=true
 bOverrideConcurrency=true
 maxDistance=2500
 soundClassObject=&Fort_Ambient_Indoors_Athena
 subtitlePriority=5000
 volumeMultiplier=1.1
 attenuationOverrides {
  bEnableFocusInterpolation=true
  bEnableListenerFocus=true
  bEnableOcclusion=true
  dBAttenuationAtMax=-40
  falloffDistance=1700
  focusAzimuth=15
  focusVolumeAttenuation=1.5
  nonFocusAzimuth=45
  nonFocusVolumeAttenuation=0.5
  occlusionInterpolationTime=1
  occlusionLowPassFilterFrequency=10000
  occlusionTraceChannel=ECC_GameTraceChannel11
  occlusionVolumeAttenuation=0.5
  stereoSpread=512
  attenuationShapeExtents=<800 64.5942 90>
 }
 concurrencyOverrides {bLimitToOwner=true, maxCount=2}
}
