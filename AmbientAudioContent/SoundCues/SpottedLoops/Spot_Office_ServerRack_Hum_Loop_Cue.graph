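# Looping ambient Sound Cue for an office server rack hum, by name a spotted prop sound, in the Athena ambient props sound class.
# A Mixer blends two looping computer transmitter waves, weighting the regular layer above the low one.
# Spatialization references a shared oil rig console screen-loop attenuation asset; the cue also stores its own attenuation override values.
# Overridden concurrency stops the quietest instance when the limit is reached.
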SoundCue /AmbientAudioContent/SoundCues/SpottedLoops/Spot_Office_ServerRack_Hum_Loop_Cue
root=Mixer_0
nodes [
 {id=Mixer_0, children [WavePlayer_0,WavePlayer_1], properties {inputVolume [1.2,0.8]}}
 {
  id=WavePlayer_0
  properties {bLooping=true, soundWaveAssetPtr=/AmbientAudioContent/Waves/SpottedLoops/Spot_ComputerTransimtter_Running_Loop_01}
 }
 {
  id=WavePlayer_1
  properties {bLooping=true, soundWaveAssetPtr=/AmbientAudioContent/Waves/SpottedLoops/Spot_ComputerTransimtter_Running_Loop_Low_01}
 }
]
properties {
 attenuationSettings=/Game/Sounds/Artemis/Environment_Actors/OilRigControlConsole/Spot_Console_Screen_Loop_Att
 bOverrideConcurrency=true
 maxDistance=650
 priority=50
 soundClassObject=&Fort_Ambient_Props_Athena
 volumeMultiplier=1
 attenuationOverrides {
  bEnableFocusInterpolation=true
  bEnableListenerFocus=true
  distanceAlgorithm=Logarithmic
  falloffDistance=1100
  focusAttackInterpSpeed=20
  focusAzimuth=20
  focusReleaseInterpSpeed=10
  focusVolumeAttenuation=1.75
  lPFFrequencyAtMax=10000
  lPFRadiusMax=1000
  lPFRadiusMin=500
  nonFocusAzimuth=40
  nonFocusVolumeAttenuation=0.8
  occlusionInterpolationTime=0.3
  occlusionLowPassFilterFrequency=3000
  occlusionVolumeAttenuation=0.75
  stereoSpread=10000
  attenuationShapeExtents {x=0}
  customAttenuationCurve {editorCurveData {keys [<at 0 1 RCIM_Cubic>, {arriveTangent=-1, interpMode=RCIM_Cubic, leaveTangent=-1, time=0.5, value=0.1}, <at 1 0 RCIM_Cubic>]}}
 }
 concurrencyOverrides {resolutionRule=StopQuietest}
}
