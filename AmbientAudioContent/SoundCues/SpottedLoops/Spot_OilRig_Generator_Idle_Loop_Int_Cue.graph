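# Looping indoor ambient spot cue for an oil rig generator idle, built on a reused refueling cart idle wave and attenuation.
# The looping wave runs through a short random start delay and an enveloper that fades volume in.
# A quality node feeds that chain to its first two levels and leaves the other two inputs empty.
# Low volume, short max distance and a spot concurrency set keep it localized.
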SoundCue /AmbientAudioContent/SoundCues/SpottedLoops/Spot_OilRig_Generator_Idle_Loop_Int_Cue
root=QualityLevel_0
nodes [
 {id=Delay_0, children [WavePlayer_0], properties {delayMax=0.3}}
 {
  id=Enveloper_1
  children [Delay_0]
  properties {pitchCurve {editorCurveData {keys [{time=0}]}}, volumeCurve {editorCurveData {keys [<at 0 0>,<at 1 1>]}}}
 }
 {id=QualityLevel_0, children [Enveloper_1,Enveloper_1,null,null]}
 {id=WavePlayer_0, properties {bLooping=true, soundWaveAssetPtr=/AmbientAudioContent/Waves/SpottedLoops/Spot_RefuelingCart_Ilde_Loop}}
]
properties {
 attenuationSettings=/Game/Sounds/Artemis/Environment_Actors/RefuelingCart/Spot_RefuelingCart_Att
 bEnableBusSends=false
 bEnableSubmixSends=false
 maxDistance=750
 soundClassObject=&Fort_Ambient_Indoors_Athena
 volumeMultiplier=0.15
 concurrencyOverrides {maxCount=2, voiceStealReleaseTime=0.2}
 concurrencySet [/AmbientAudioBase/Concurrency/AmbientConc_3Max_Spot_Default_Concurrency]
}
