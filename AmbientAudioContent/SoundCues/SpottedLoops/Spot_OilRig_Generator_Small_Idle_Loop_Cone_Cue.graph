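# Spotted ambient loop cue for a small oil rig generator at idle, built on a reused refueling cart idle wave with raised pitch and lowered volume.
# The looping wave passes through a Modulator, a short random Delay and an Enveloper fade-in, feeding all quality levels identically.
# Cone-shaped attenuation override keeps it local and directional.
# It plays in the Athena ambient sound class with a tight concurrency override.
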SoundCue /AmbientAudioContent/SoundCues/SpottedLoops/Spot_OilRig_Generator_Small_Idle_Loop_Cone_Cue
root=QualityLevel_0
nodes [
 {id=Delay_0, children [Modulator_0], properties {delayMax=0.3}}
 {
  id=Enveloper_1
  children [Delay_0]
  properties {pitchCurve {editorCurveData {keys [{time=0}]}}, volumeCurve {editorCurveData {keys [<at 0 0>,<at 1 1>]}}}
 }
 {id=Modulator_0, children [WavePlayer_0]}
 {id=QualityLevel_0, children [Enveloper_1,Enveloper_1,Enveloper_1,Enveloper_1]}
 {id=WavePlayer_0, properties {bLooping=true, soundWaveAssetPtr=/AmbientAudioContent/Waves/SpottedLoops/Spot_RefuelingCart_Ilde_Loop}}
]
properties {
 bEnableBusSends=false
 bEnableSubmixSends=false
 bOverrideAttenuation=true
 bOverrideConcurrency=true
 maxDistance=875
 pitchMultiplier=1.1
 soundClassObject=&Fort_Ambient_Athena
 volumeMultiplier=0.4
 attenuationOverrides {attenuationShape=Cone, falloffDistance=750, attenuationShapeExtents=<125 90 35.9794>}
 concurrencyOverrides {maxCount=2, voiceStealReleaseTime=0.2}
 concurrencySet [/AmbientAudioBase/Concurrency/AmbientConc_3Max_Spot_Default_Concurrency]
}
