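# Spotted ambient sound cue that plays a looping refrigerator hum for the OilRig environment.
# A single looping wave player feeds the output at reduced volume, routed to the indoor Athena ambient sound class.
# Spatialization comes from a dedicated attenuation asset; inline attenuation values for occlusion and listener focus are stored, but their override flag is not set.
# A concurrency override caps simultaneous instances per owner.
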SoundCue /AmbientAudioContent/SoundCues/SpottedLoops/Spot_OilRig_Refridgerator_Hum_Loop_01_Cue
root=WavePlayer_3
nodes [{id=WavePlayer_3, properties {bLooping=true, soundWaveAssetPtr=/AmbientAudioContent/Waves/SpottedLoops/Spot_OilRig_Refridgerator_Hum_Loop}}]
properties {
 attenuationSettings=/Game/Sounds/Asteria/Environment_Actors/OilRig_Refridgerator/Spot_OilRig_Refridgerator_Hum_Loop_Att
 bOverrideConcurrency=true
 maxDistance=850
 soundClassObject=&Fort_Ambient_Indoors_Athena
 subtitlePriority=5000
 volumeMultiplier=0.301
 attenuationOverrides {
  bEnableFocusInterpolation=true
  bEnableListenerFocus=true
  bEnableOcclusion=true
  dBAttenuationAtMax=-40
  falloffDistance=512
  focusAzimuth=15
  focusVolumeAttenuation=1.5
  nonFocusAzimuth=45
  nonFocusVolumeAttenuation=0.5
  occlusionInterpolationTime=1
  occlusionLowPassFilterFrequency=10000
  occlusionTraceChannel=ECC_GameTraceChannel11
  occlusionVolumeAttenuation=0.5
  stereoSpread=512
  attenuationShapeExtents=<256 64.5942 90>
 }
 concurrencyOverrides {bLimitToOwner=true, maxCount=2}
}
