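# Sound cue for a looping oil rig refrigerator hum, a quiet indoor ambient spot sound in the Athena indoor ambient sound class.
# A single looping wave player feeds the output, shaped by attenuation and concurrency settings overridden on the cue itself.
# An elongated capsule falloff adds distance low-pass filtering, listener focus, slow-blending occlusion and a send to a distant reverb submix.
# Concurrency allows two instances per owner.
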SoundCue /AmbientAudioContent/SoundCues/SpottedLoops/Spot_OilRig_Refridgerator_Hum_Loop_03_Long_Capsule_Cue
root=WavePlayer_3
nodes [{id=WavePlayer_3, properties {bLooping=true, soundWaveAssetPtr=/AmbientAudioContent/Waves/SpottedLoops/Spot_OilRig_Refridgerator_Hum_Loop}}]
properties {
 bOverrideAttenuation=true
 bOverrideConcurrency=true
 maxDistance=1400
 soundClassObject=&Fort_Ambient_Indoors_Athena
 subtitlePriority=5000
 volumeMultiplier=0.301
 attenuationOverrides {
  attenuationShape=Capsule
  bAttenuateWithLPF=true
  bEnableFocusInterpolation=true
  bEnableListenerFocus=true
  bEnableLogFrequencyScaling=true
  bEnableOcclusion=true
  bEnableSubmixSends=true
  dBAttenuationAtMax=-40
  falloffDistance=850
  focusAttackInterpSpeed=0.01
  focusAzimuth=15
  focusReleaseInterpSpeed=0.01
  hPFFrequencyAtMax=250
  lPFFrequencyAtMax=2000
  lPFRadiusMax=1600
  lPFRadiusMin=500
  nonFocusAzimuth=45
  nonFocusVolumeAttenuation=0.75
  occlusionInterpolationTime=5
  occlusionLowPassFilterFrequency=1000
  occlusionTraceChannel=ECC_GameTraceChannel11
  occlusionVolumeAttenuation=0.5
  stereoSpread=300
  attenuationShapeExtents=<550 100 90>
  submixSendSettings [{
    bEnableHPFCutoff=false
    bEnableLPFCutoff=false
    disableManualSendClamp=false
    hPFCutoff=20
    lPFCutoff=20000
    maxSendDistance=6000
    maxSendLevel=0.25
    minSendDistance=0
    minSendLevel=0
    sendLevel=0.2
    sendLevelControlMethod=Linear
    soundSubmix=&Fortnite_Reverb_Distant
   }]
 }
 concurrencyOverrides {bLimitToOwner=true, maxCount=2}
}
