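# Spotted ambient loop cue for wind-vibrated wire cable on the oil rig, with its own attenuation and a two-voice concurrency override.
# A looping metal vibration bed runs through nested envelopes for fade-in and slowly cycling pitch and volume.
# Looped random delays mix in pitch-varied cable creaks, randomly plain, faded-in or faded-out, plus quieter metal stress one-shots.
# A quality switch leaves the creaks out of its last two levels.
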SoundCue /AmbientAudioContent/SoundCues/SpottedLoops/Spot_OilRig_Wire_Cable_Wind_Vibrating_Metal_Loop_Cue
root=Mixer_0
nodes [
 {id=Delay_1, children [QualityLevel_1], properties {delayMax=3, delayMin=1}}
 {id=Delay_2, children [Modulator_2], properties {delayMax=3, delayMin=1}}
 {
  id=Enveloper_0
  children [Modulator_1]
  properties {
   bLoop=true
   bLoopIndefinitely=true
   loopEnd=30
   pitchCurve {editorCurveData {keys [{interpMode=RCIM_Cubic, time=0}, <at 15.9468 0.9 RCIM_Cubic>, {interpMode=RCIM_Cubic, time=30}]}}
   volumeCurve {editorCurveData {keys [<at 0 1>]}}
  }
 }
 {
  id=Enveloper_1
  children [Modulator_0]
  properties {
   pitchCurve {editorCurveData {keys [{time=0}]}}
   volumeCurve {editorCurveData {keys [<at 0 1>, {arriveTangent=-0.2, interpMode=RCIM_Cubic, leaveTangent=-0.2, time=3, value=1}, <at 5 0>]}}
  }
 }
 {
  id=Enveloper_2
  children [Modulator_0]
  properties {
   pitchCurve {editorCurveData {keys [{time=0}]}}
   volumeCurve {editorCurveData {keys [<at 0 0 RCIM_Cubic>, {arriveTangent=0.0151, interpMode=RCIM_Cubic, leaveTangent=0.0151, tangentMode=RCTM_User, time=3, value=0}, <at 5 1>]}}
  }
 }
 {
  id=Enveloper_29
  children [Enveloper_3]
  properties {
   pitchCurve {editorCurveData {keys [{time=0}]}}
   volumeCurve {editorCurveData {keys [{arriveTangent=3.4293, interpMode=RCIM_Cubic, leaveTangent=3.4293, tangentMode=RCTM_User, time=0, value=0}, <at 0.4 1 RCIM_Cubic>]}}
  }
 }
 {
  id=Enveloper_3
  children [Enveloper_0]
  properties {
   bLoop=true
   bLoopIndefinitely=true
   loopEnd=120
   pitchCurve {editorCurveData {keys [{time=0}]}}
   volumeCurve {editorCurveData {keys [
      {arriveTangent=0.0001, interpMode=RCIM_Cubic, leaveTangent=0.0082, tangentMode=RCTM_User, time=0.5, value=0.3803}
      {arriveTangent=0.0318, interpMode=RCIM_Cubic, leaveTangent=0.0318, time=10, value=0.6}
      {arriveTangent=-0.0028, interpMode=RCIM_Cubic, leaveTangent=-0.0028, time=20, value=1}
      <at 45.8196 0.5 RCIM_Cubic>
      {arriveTangent=0.0151, interpMode=RCIM_Cubic, leaveTangent=0.0151, time=60, value=1}
      {arriveTangent=-0.0202, interpMode=RCIM_Cubic, leaveTangent=-0.0202, tangentMode=RCTM_User, time=82.1697, value=1.05}
      <at 120 0.5>
     ]}}
  }
 }
 {id=Looping_0, children [Delay_1]}
 {id=Looping_1, children [Delay_2]}
 {id=Mixer_0, children [QualityLevel_0, Looping_0, Looping_1], properties {inputVolume [1,1,0.25]}}
 {id=Modulator_0, children [Random_5], properties {pitchMax=0.9, pitchMin=0.7}}
 {id=Modulator_1, children [WavePlayer_1], properties {pitchMax=1.2, pitchMin=1}}
 {id=Modulator_2, children [Random_4], properties {pitchMin=0.9}}
 {id=QualityLevel_0, children [Enveloper_29,Enveloper_29,Enveloper_29,Enveloper_29]}
 {id=QualityLevel_1, children [Random_2,Random_2,null,null]}
 {id=Random_2, children [Enveloper_1,Modulator_0,Enveloper_2], properties {weights [1,1,1]}}
 {
  id=Random_4
  children [
   WavePlayer_30
   WavePlayer_31
   WavePlayer_32
   WavePlayer_33
   WavePlayer_34
   WavePlayer_35
   WavePlayer_36
   WavePlayer_37
   WavePlayer_38
   WavePlayer_39
  ]
  properties {weights [1,1,1,1,1,1,1,1,1,1]}
 }
 {
  id=Random_5
  children [WavePlayer_40, WavePlayer_41, WavePlayer_42, WavePlayer_43, WavePlayer_44, WavePlayer_45, WavePlayer_46, WavePlayer_47, WavePlayer_48]
  properties {weights [1,1,1,1,1,1,1,1,1]}
 }
 {
  id=WavePlayer_1
  properties {bLooping=true, soundWaveAssetPtr=/AmbientAudioContent/Waves/SpottedLoops/Spot_Wind_Excited_Metal_Vibrating_Loop}
 }
 {id=WavePlayer_30, properties {soundWaveAssetPtr=/AmbientAudioContent/Waves/OneShots/Spot_Tilted_Towers_Metal_Stress_Light_01}}
 {id=WavePlayer_31, properties {soundWaveAssetPtr=/AmbientAudioContent/Waves/OneShots/Spot_Tilted_Towers_Metal_Stress_Light_02}}
 {id=WavePlayer_32, properties {soundWaveAssetPtr=/AmbientAudioContent/Waves/OneShots/Spot_Tilted_Towers_Metal_Stress_Light_03}}
 {id=WavePlayer_33, properties {soundWaveAssetPtr=/AmbientAudioContent/Waves/OneShots/Spot_Tilted_Towers_Metal_Stress_Light_04}}
 {id=WavePlayer_34, properties {soundWaveAssetPtr=/AmbientAudioContent/Waves/OneShots/Spot_Tilted_Towers_Metal_Stress_Light_05}}
 {id=WavePlayer_35, properties {soundWaveAssetPtr=/AmbientAudioContent/Waves/OneShots/Spot_Tilted_Towers_Metal_Stress_Light_06}}
 {id=WavePlayer_36, properties {soundWaveAssetPtr=/AmbientAudioContent/Waves/OneShots/Spot_Tilted_Towers_Metal_Stress_Light_07}}
 {id=WavePlayer_37, properties {soundWaveAssetPtr=/AmbientAudioContent/Waves/OneShots/Spot_Tilted_Towers_Metal_Stress_Light_08}}
 {id=WavePlayer_38, properties {soundWaveAssetPtr=/AmbientAudioContent/Waves/OneShots/Spot_Tilted_Towers_Metal_Stress_Light_09}}
 {id=WavePlayer_39, properties {soundWaveAssetPtr=/AmbientAudioContent/Waves/OneShots/Spot_Tilted_Towers_Metal_Stress_Light_10}}
 {id=WavePlayer_40, properties {soundWaveAssetPtr=/AmbientAudioContent/Waves/OneShots/Spot_Blimp_Crashsite_Cable_Tension_Creak_01}}
 {id=WavePlayer_41, properties {soundWaveAssetPtr=/AmbientAudioContent/Waves/OneShots/Spot_Blimp_Crashsite_Cable_Tension_Creak_02}}
 {id=WavePlayer_42, properties {soundWaveAssetPtr=/AmbientAudioContent/Waves/OneShots/Spot_Blimp_Crashsite_Cable_Tension_Creak_03}}
 {id=WavePlayer_43, properties {soundWaveAssetPtr=/AmbientAudioContent/Waves/OneShots/Spot_Blimp_Crashsite_Cable_Tension_Creak_04}}
 {id=WavePlayer_44, properties {soundWaveAssetPtr=/AmbientAudioContent/Waves/OneShots/Spot_Blimp_Crashsite_Cable_Tension_Creak_05}}
 {id=WavePlayer_45, properties {soundWaveAssetPtr=/AmbientAudioContent/Waves/OneShots/Spot_Blimp_Crashsite_Cable_Tension_Creak_06}}
 {id=WavePlayer_46, properties {soundWaveAssetPtr=/AmbientAudioContent/Waves/OneShots/Spot_Blimp_Crashsite_Cable_Tension_Creak_07}}
 {id=WavePlayer_47, properties {soundWaveAssetPtr=/AmbientAudioContent/Waves/OneShots/Spot_Blimp_Crashsite_Cable_Tension_Creak_08}}
 {id=WavePlayer_48, properties {soundWaveAssetPtr=/AmbientAudioContent/Waves/OneShots/Spot_Blimp_Crashsite_Cable_Tension_Creak_09}}
]
properties {
 attenuationSettings=/Game/Sounds/Helios/POI/OilRigPlatform/Spots/Spot_OilRig_Wire_Cable_Wind_Vibrating_Metal_Att
 bEnableBusSends=false
 bEnableSubmixSends=false
 bOverrideConcurrency=true
 maxDistance=2003.4948
 soundClassObject=&Fort_Gadgets
 volumeMultiplier=0.6
 concurrencyOverrides {maxCount=2, voiceStealReleaseTime=2}
}
