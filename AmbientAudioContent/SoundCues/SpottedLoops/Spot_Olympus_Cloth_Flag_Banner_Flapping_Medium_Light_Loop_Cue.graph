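# Spotted ambient Sound Cue that plays a quiet looping cloth flag or banner flapping sound for outdoor environments.
# A looping wave feeds a mixer, then two endlessly looping envelopes: a slow pitch curve and a shorter volume swell curve, behind a randomized start delay.
# Spatialization references flag capsule attenuation, with concurrency limits, an exterior bus send and a reverb send.
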SoundCue /AmbientAudioContent/SoundCues/SpottedLoops/Spot_Olympus_Cloth_Flag_Banner_Flapping_Medium_Light_Loop_Cue
root=Delay_0
nodes [
 {id=Delay_0, children [Enveloper_3], properties {delayMax=2}}
 {
  id=Enveloper_0
  children [Mixer_0]
  properties {
   bLoop=true
   bLoopIndefinitely=true
   loopEnd=60
   pitchCurve {editorCurveData {keys [
      {interpMode=RCIM_Cubic, time=0}
      <at 15 0.95 RCIM_Cubic>
      {arriveTangent=0.0072, interpMode=RCIM_Cubic, leaveTangent=0.0072, tangentMode=RCTM_User, time=30}
      {arriveTangent=-0.0082, interpMode=RCIM_Cubic, leaveTangent=-0.0082, tangentMode=RCTM_User, time=45}
      {time=60}
     ]}}
   volumeCurve {editorCurveData {keys [<at 0 1>]}}
  }
 }
 {
  id=Enveloper_3
  children [Enveloper_0]
  properties {
   bLoop=true
   bLoopIndefinitely=true
   loopEnd=20
   pitchCurve {editorCurveData {keys [{time=0}]}}
   volumeCurve {editorCurveData {keys [
      {arriveTangent=-0.01, interpMode=RCIM_Cubic, time=0, value=1}
      {arriveTangent=-0.1116, interpMode=RCIM_Cubic, leaveTangent=-0.1116, tangentMode=RCTM_User, time=5, value=0.8}
      {arriveTangent=0.0002, interpMode=RCIM_Cubic, leaveTangent=0.0002, tangentMode=RCTM_User, time=7.5, value=0.2}
      {arriveTangent=-0.0014, interpMode=RCIM_Cubic, leaveTangent=0.0091, tangentMode=RCTM_User, time=12.2755, value=0.7965}
      {arriveTangent=0.0117, interpMode=RCIM_Cubic, leaveTangent=0.0117, time=16.2222, value=0.2468}
      {arriveTangent=0.0099, interpMode=RCIM_Cubic, leaveTangent=0.0099, time=19.9681, value=0.8869}
      {arriveTangent=0.0175, interpMode=RCIM_Cubic, leaveTangent=0.0175, time=23.3402, value=0.3172}
      <at 26.3735 0.9991 RCIM_Cubic>
     ]}}
  }
 }
 {id=Mixer_0, children [WavePlayer_2], properties {inputVolume [3]}}
 {
  id=WavePlayer_2
  properties {bLooping=true, soundWaveAssetPtr=/AmbientAudioContent/Waves/SpottedLoops/Spot_Cloth_Flag_Banner_Flapping_Medium_Slow_Loop_02}
 }
]
properties {
 attenuationSettings=/Game/Sounds/Helios/Environment_Actors/FlagBanner/Spot_Cloth_Flag_Banner_Flapping_Medium_Capsule_Att
 maxDistance=1850
 soundClassObject=&Fort_Ambient_Outdoors_Athena
 volumeMultiplier=0.2
 attenuationOverrides {
  bEnableFocusInterpolation=true
  bEnableListenerFocus=true
  bEnableOcclusion=true
  dBAttenuationAtMax=-40
  distanceAlgorithm=Logarithmic
  falloffDistance=1500
  focusAzimuth=15
  focusVolumeAttenuation=1.1
  nonFocusAzimuth=45
  nonFocusVolumeAttenuation=0.4
  nonSpatializedRadiusStart=100
  occlusionTraceChannel=ECC_GameTraceChannel11
  occlusionVolumeAttenuation=0.7
  stereoSpread=2000
  attenuationShapeExtents {x=100}
 }
 busSends [{audioBus=null, minSendLevel=1, sendLevel=1, soundSourceBus=&AmbientAudio_Exterior_SourceBus}]
 concurrencyOverrides {maxCount=2, voiceStealReleaseTime=1}
 concurrencySet [/AmbientAudioBase/Concurrency/AmbientConc_3Max_Spot_Default_Concurrency]
 soundSubmixSends [{
   minSendLevel=0.4
   soundSubmix=&Fortnite_Reverb_Default
   customSendLevelCurve {editorCurveData {keys [<at 0 0.4>,<at 1 1>]}}
  }]
}
