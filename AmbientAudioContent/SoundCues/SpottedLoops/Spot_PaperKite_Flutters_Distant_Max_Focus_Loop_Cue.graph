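# Looping outdoor ambient cue for distant paper kite flutters, as a spot sound that favors listener focus.
# Two looping waves, the second at half level, feed a random pitch Modulator and two nested endlessly looping Envelopers with 30-second pitch and 20-second volume cycles.
# Overrides add natural falloff, distance low-pass, occlusion and focus interpolation, limit concurrency to two voices, and send to the exterior ambient source bus by distance.
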SoundCue /AmbientAudioContent/SoundCues/SpottedLoops/Spot_PaperKite_Flutters_Distant_Max_Focus_Loop_Cue
root=Enveloper_3
nodes [
 {
  id=Enveloper_0
  children [Modulator_0]
  properties {
   bLoop=true
   bLoopIndefinitely=true
   loopEnd=30
   pitchCurve {editorCurveData {keys [
      {arriveTangent=-0.0013, interpMode=RCIM_Cubic, leaveTangent=-0.0013, time=0}
      <at 15 0.98 RCIM_Cubic>
      {arriveTangent=0.0013, interpMode=RCIM_Cubic, leaveTangent=0.0013, time=30}
     ]}}
   volumeCurve {editorCurveData {keys [<at 0 1>]}}
  }
 }
 {
  id=Enveloper_3
  children [Enveloper_0]
  properties {
   bLoop=true
   bLoopIndefinitely=true
   loopEnd=20
   pitchCurve {editorCurveData {keys [{time=0}]}}
   volumeCurve {editorCurveData {keys [
      {arriveTangent=-0.01, interpMode=RCIM_Cubic, time=0, value=1}
      <at 10 0.5 RCIM_Cubic>
      {interpMode=RCIM_Cubic, leaveTangent=0.01, time=20, value=1}
     ]}}
  }
 }
 {id=Mixer_0, children [WavePlayer_1,WavePlayer_0], properties {inputVolume [1,0.5]}}
 {id=Modulator_0, children [Mixer_0], properties {pitchMax=1.1, pitchMin=0.9}}
 {
  id=WavePlayer_0
  properties {bLooping=true, soundWaveAssetPtr=/AmbientAudioContent/Waves/SpottedLoops/Spot_Paper_Kites_Distant_Max_Focus_Loop_02}
 }
 {
  id=WavePlayer_1
  properties {bLooping=true, soundWaveAssetPtr=/AmbientAudioContent/Waves/SpottedLoops/Spot_Paper_Kites_Distant_Max_Focus_Loop}
 }
]
properties {
 bOverrideAttenuation=true
 bOverrideConcurrency=true
 maxDistance=15500
 soundClassObject=&Fort_Ambient_Outdoors_Athena
 subtitlePriority=5000
 volumeMultiplier=0.5
 attenuationOverrides {
  bAttenuateWithLPF=true
  bEnableFocusInterpolation=true
  bEnableListenerFocus=true
  bEnableOcclusion=true
  dBAttenuationAtMax=-40
  distanceAlgorithm=NaturalSound
  falloffDistance=9000
  focusAttackInterpSpeed=0.4
  focusDistanceScale=0.5
  focusReleaseInterpSpeed=0.4
  lPFFrequencyAtMax=500
  lPFRadiusMax=10000
  lPFRadiusMin=2000
  nonFocusVolumeAttenuation=0.05
  nonSpatializedRadiusEnd=2000
  nonSpatializedRadiusStart=3000
  occlusionInterpolationTime=2
  occlusionLowPassFilterFrequency=1000
  occlusionTraceChannel=ECC_GameTraceChannel11
  occlusionVolumeAttenuation=0.5
  reverbDistanceMax=10000
  stereoSpread=50
  attenuationShapeExtents=<2000 1500 90>
  customAttenuationCurve {editorCurveData {keys [
     <at 0.1 0 RCIM_Cubic>
     {arriveTangent=1.0161, interpMode=RCIM_Cubic, leaveTangent=1.0161, time=0.2, value=1}
     {arriveTangent=-1.25, interpMode=RCIM_Cubic, leaveTangent=-1.25, time=0.4892, value=0.3955}
     {arriveTangent=-2.1297, interpMode=RCIM_Cubic, leaveTangent=-2.1297, tangentMode=RCTM_User, time=1, value=0}
    ]}}
 }
 busSends [{
   audioBus=null
   maxSendDistance=9500
   minSendDistance=2000
   minSendLevel=0.4
   sendLevel=1
   soundSourceBus=&AmbientAudio_Exterior_SourceBus
   customSendLevelCurve {editorCurveData {keys [<at 0 0.4>,<at 1 1>]}}
  }]
 concurrencyOverrides {maxCount=2, resolutionRule=StopFarthestThenPreventNew, voiceStealReleaseTime=2}
}
