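# Looping ambient spot sound cue for closed pickle factory barrels, in the Fort_Ambient_Athena sound class.
# A Random node picks one of two equally weighted looping barrel waves and feeds a root Modulator left at default settings.
# Spatialization references the borrowed large oil rig storage tank attenuation asset, and a concurrency override resolves conflicts by stopping the quietest sound.
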SoundCue /AmbientAudioContent/SoundCues/SpottedLoops/Spot_Pickle_Barrel_01_Loop_Cue
root=Modulator_0
nodes [
 {id=Modulator_0, children [Random_0]}
 {id=Random_0, children [WavePlayer_0,WavePlayer_1], properties {weights [1,1]}}
 {id=WavePlayer_0, properties {bLooping=true, soundWaveAssetPtr=/AmbientAudioContent/Waves/SpottedLoops/Spot_PickleFactory_Barrels_Closed_01}}
 {id=WavePlayer_1, properties {bLooping=true, soundWaveAssetPtr=/AmbientAudioContent/Waves/SpottedLoops/Spot_PickleFactory_Barrels_Closed_02}}
]
properties {
 attenuationSettings=/AmbientAudioBase/Attenuations/Spot_OilRig_Storage_Tank_Large_Att
 bOverrideConcurrency=true
 maxDistance=2150
 soundClassObject=&Fort_Ambient_Athena
 volumeMultiplier=1
 attenuationOverrides {
  falloffDistance=2500
  focusDistanceScale=0.9
  focusVolumeAttenuation=1.2
  nonFocusVolumeAttenuation=0.4
  stereoSpread=180
  attenuationShapeExtents {x=500}
  customAttenuationCurve {editorCurveData {keys [{arriveTangent=-2.3812, interpMode=RCIM_Cubic, leaveTangent=-2.3812, tangentMode=RCTM_User, time=0, value=1}, <at 1 0.01 RCIM_Cubic>]}}
 }
 concurrencyOverrides {resolutionRule=StopQuietest}
}
