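# Looping spot ambience cue for an open pickle barrel, assigned to the Athena ambient sound class.
# A single looping wave player plays the pickle factory open-barrel wave through a Modulator left at default settings, at reduced cue volume.
# Inline attenuation overrides its referenced oil rig storage tank attenuation asset, adding a custom falloff curve, wide stereo spread, focus-based volume scaling and a narrow reverb send range.
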SoundCue /AmbientAudioContent/SoundCues/SpottedLoops/Spot_Pickle_Barrel_Open_01_Loop_Cue
root=Modulator_0
nodes [
 {id=Modulator_0, children [WavePlayer_0]}
 {
  id=WavePlayer_0
  properties {bLooping=true, soundWaveAssetPtr=/AmbientAudioContent/Waves/SpottedLoops/Spot_PickleFactory_Barrels_Open_Top_01}
 }
]
properties {
 attenuationSettings=/Game/Sounds/Artemis/Environment_Actors/OilRigStorageTank/Spot_OilRig_Storage_Tank_Att
 bOverrideAttenuation=true
 bOverrideConcurrency=true
 maxDistance=2800
 soundClassObject=&Fort_Ambient_Athena
 volumeMultiplier=0.6
 attenuationOverrides {
  falloffDistance=2400
  focusDistanceScale=0.9
  focusVolumeAttenuation=1.2
  nonFocusVolumeAttenuation=0.4
  reverbWetLevelMax=0.3
  reverbWetLevelMin=0.2
  stereoSpread=180
  customAttenuationCurve {editorCurveData {keys [{arriveTangent=-2.3812, interpMode=RCIM_Cubic, leaveTangent=-2.3812, tangentMode=RCTM_User, time=0, value=1}, <at 1 0.01 RCIM_Cubic>]}}
 }
}
