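# Spotted ambient loop cue for a small, occludable pickle pipe sound in the Athena ambient sound class, using closed-barrel pickle factory waves.
# A root Random either delays briefly or plays immediately, both paths sharing a pitch-randomizing Modulator that feeds a Random choice of two looping mono waves, so overlapping instances vary.
# Overrides add capsule attenuation with occlusion, distance filtering and listener focus, plus tight, distance-scaled concurrency.
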SoundCue /AmbientAudioContent/SoundCues/SpottedLoops/Spot_Pickle_Pipe_Loop_Occ_Loop_Small_Cue
root=Random_1
nodes [
 {id=Delay_0, children [Modulator_0], properties {delayMax=0.25, delayMin=0.25}}
 {id=Modulator_0, children [Random_2], properties {pitchMax=1.1, pitchMin=0.9}}
 {id=Random_1, children [Delay_0,Modulator_0], properties {weights [1,1]}}
 {id=Random_2, children [WavePlayer_2,WavePlayer_3], properties {weights [1,1]}}
 {
  id=WavePlayer_2
  properties {bLooping=true, soundWaveAssetPtr=/AmbientAudioContent/Waves/SpottedLoops/Spot_PickleFactory_Barrels_Closed_01_Mono}
 }
 {
  id=WavePlayer_3
  properties {bLooping=true, soundWaveAssetPtr=/AmbientAudioContent/Waves/SpottedLoops/Spot_PickleFactory_Barrels_Closed_02_Mono}
 }
]
properties {
 bOverrideAttenuation=true
 bOverrideConcurrency=true
 maxDistance=1050
 pitchMultiplier=0.9
 soundClassObject=&Fort_Ambient_Athena
 attenuationOverrides {
  attenuationShape=Capsule
  bAttenuateWithLPF=true
  bEnableFocusInterpolation=true
  bEnableListenerFocus=true
  bEnableOcclusion=true
  distanceAlgorithm=Logarithmic
  falloffDistance=850
  focusAzimuth=15
  focusDistanceScale=0.9
  focusReleaseInterpSpeed=2
  hPFFrequencyAtMax=200
  lPFFrequencyAtMax=600
  lPFRadiusMax=1400
  lPFRadiusMin=1000
  nonFocusAzimuth=30
  nonFocusVolumeAttenuation=0.4
  occlusionInterpolationTime=0.3
  occlusionLowPassFilterFrequency=3000
  occlusionTraceChannel=ECC_GameTraceChannel11
  occlusionVolumeAttenuation=0.75
  stereoSpread=0
  attenuationShapeExtents=<200 50>
  customAttenuationCurve {editorCurveData {keys [{arriveTangent=-2.3812, interpMode=RCIM_Cubic, leaveTangent=-2.3812, tangentMode=RCTM_User, time=0, value=1}, <at 1 0.01 RCIM_Cubic>]}}
 }
 concurrencyOverrides {maxCount=2, voiceStealReleaseTime=2, volumeScaleMode=Distance}
}
