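# Ambient spot-loop sound cue for a pickle silo, playing looping closed-barrel factory recordings under the Athena ambient sound class.
# The root random node either starts immediately or passes through a short fixed delay, both feeding one shared slight pitch modulator.
# That modulator draws from an equal-weight random pick of two looping waves.
# It references an oil-rig storage tank attenuation asset and overrides concurrency to stop the quietest instance.
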SoundCue /AmbientAudioContent/SoundCues/SpottedLoops/Spot_Pickle_Silo_Loop_Occ_Cue
root=Random_1
nodes [
 {id=Delay_0, children [Modulator_1], properties {delayMax=0.25, delayMin=0.25}}
 {id=Modulator_1, children [Random_0], properties {pitchMax=1.1, pitchMin=0.9}}
 {id=Random_0, children [WavePlayer_0,WavePlayer_1], properties {weights [1,1]}}
 {id=Random_1, children [Delay_0,Modulator_1], properties {weights [1,1]}}
 {id=WavePlayer_0, properties {bLooping=true, soundWaveAssetPtr=/AmbientAudioContent/Waves/SpottedLoops/Spot_PickleFactory_Barrels_Closed_01}}
 {id=WavePlayer_1, properties {bLooping=true, soundWaveAssetPtr=/AmbientAudioContent/Waves/SpottedLoops/Spot_PickleFactory_Barrels_Closed_02}}
]
properties {
 attenuationSettings=/Game/Sounds/Artemis/Environment_Actors/OilRigStorageTank/Spot_OilRig_Storage_Tank_Att
 bOverrideConcurrency=true
 maxDistance=650
 soundClassObject=&Fort_Ambient_Athena
 volumeMultiplier=1
 attenuationOverrides {
  falloffDistance=2500
  focusDistanceScale=0.9
  focusVolumeAttenuation=1.2
  nonFocusVolumeAttenuation=0.4
  stereoSpread=180
  attenuationShapeExtents {x=500}
  customAttenuationCurve {editorCurveData {keys [{arriveTangent=-2.3812, interpMode=RCIM_Cubic, leaveTangent=-2.3812, tangentMode=RCTM_User, time=0, value=1}, <at 1 0.01 RCIM_Cubic>]}}
 }
 concurrencyOverrides {resolutionRule=StopQuietest}
}
