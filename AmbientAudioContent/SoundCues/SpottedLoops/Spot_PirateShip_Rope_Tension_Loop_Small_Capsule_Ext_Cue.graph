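# Ambient outdoor sound cue for pirate ship rope tension, placed as a small spot source with capsule attenuation, HRTF spatialization and limited concurrency.
# It loops a random delay into one of two branches, each picking between two rope waves with pitch variation.
# One branch layers looping envelopes for irregular volume swells and slow pitch drift; the other fades out quickly as a short creak.
# The last two quality-level slots are empty.
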SoundCue /AmbientAudioContent/SoundCues/SpottedLoops/Spot_PirateShip_Rope_Tension_Loop_Small_Capsule_Ext_Cue
root=QualityLevel_0
nodes [
 {id=Delay_1, children [Random_3], properties {delayMax=6, delayMin=2}}
 {
  id=Enveloper_0
  children [Random_1]
  properties {
   bLoop=true
   bLoopIndefinitely=true
   loopEnd=60
   pitchCurve {editorCurveData {keys [
      {interpMode=RCIM_Cubic, time=0}
      <at 15 0.85 RCIM_Cubic>
      {arriveTangent=0.0091, interpMode=RCIM_Cubic, leaveTangent=0.0091, tangentMode=RCTM_User, time=30}
      <at 45 1.2 RCIM_Cubic>
      {time=60}
     ]}}
   volumeCurve {editorCurveData {keys [<at 0 1>]}}
  }
 }
 {
  id=Enveloper_1
  children [Modulator_1]
  properties {pitchCurve {editorCurveData {keys [{time=0}]}}, volumeCurve {editorCurveData {keys [<at 0 1>,<at 0.5 1>,<at 1.5 0>]}}}
 }
 {
  id=Enveloper_2
  children [Enveloper_3]
  properties {
   bLoop=true
   bLoopIndefinitely=true
   loopEnd=20
   pitchCurve {editorCurveData {keys [{time=0}]}}
   volumeCurve {editorCurveData {keys [
      {arriveTangent=-0.01, interpMode=RCIM_Cubic, time=0, value=1}
      <at 10 0.8 RCIM_Cubic>
      {interpMode=RCIM_Cubic, leaveTangent=0.01, time=20, value=1}
     ]}}
  }
 }
 {
  id=Enveloper_3
  children [Enveloper_0]
  properties {
   bLoop=true
   bLoopIndefinitely=true
   loopEnd=60
   pitchCurve {editorCurveData {keys [{time=0}]}}
   volumeCurve {editorCurveData {keys [
      {arriveTangent=-0.01, interpMode=RCIM_Cubic, time=0, value=1}
      {arriveTangent=-0.0035, interpMode=RCIM_Cubic, leaveTangent=-0.0035, tangentMode=RCTM_User, time=3.9149, value=0.3102}
      {arriveTangent=0.0016, interpMode=RCIM_Cubic, leaveTangent=0.0016, tangentMode=RCTM_User, time=7.2345, value=0.7106}
      {arriveTangent=0.0008, interpMode=RCIM_Cubic, leaveTangent=0.0008, tangentMode=RCTM_User, time=10.4798, value=0.3237}
      {arriveTangent=0.0016, interpMode=RCIM_Cubic, leaveTangent=0.0016, tangentMode=RCTM_User, time=12.3496, value=0.3093}
      {arriveTangent=0.0009, interpMode=RCIM_Cubic, leaveTangent=0.0009, time=14.812, value=0.593}
      {arriveTangent=-0.0011, interpMode=RCIM_Cubic, leaveTangent=-0.0011, tangentMode=RCTM_User, time=17.9589, value=0.3143}
      {arriveTangent=0.0002, interpMode=RCIM_Cubic, leaveTangent=0.0002, time=23.8657, value=0.8146}
      {arriveTangent=0.0016, interpMode=RCIM_Cubic, leaveTangent=0.0016, tangentMode=RCTM_User, time=27.4062, value=0.3165}
      {arriveTangent=-0.0003, interpMode=RCIM_Cubic, leaveTangent=-0.0003, time=31.0495, value=0.6077}
      {arriveTangent=0.0009, interpMode=RCIM_Cubic, leaveTangent=0.0009, tangentMode=RCTM_User, time=34.2948, value=0.3145}
      {arriveTangent=-0.0008, interpMode=RCIM_Cubic, leaveTangent=-0.0008, time=37.8398, value=0.8146}
      <at 41.3802 0.3092>
      {arriveTangent=0.0024, interpMode=RCIM_Cubic, leaveTangent=0.0024, tangentMode=RCTM_User, time=43.9389, value=0.3143}
      {arriveTangent=-0.0006, interpMode=RCIM_Cubic, leaveTangent=-0.0006, time=47.7791, value=0.7106}
      {arriveTangent=-0.0056, interpMode=RCIM_Cubic, leaveTangent=-0.0056, tangentMode=RCTM_User, time=52.5064, value=0.3093}
      <at 54.7639 0.6891>
      {arriveTangent=-0.0003, interpMode=RCIM_Cubic, leaveTangent=-0.0003, tangentMode=RCTM_User, time=57.1257, value=0.309}
      <at 60 1 RCIM_Cubic>
     ]}}
  }
 }
 {id=Looping_0, children [Delay_1]}
 {id=Modulator_0, children [Enveloper_2], properties {pitchMax=1.1, pitchMin=0.9}}
 {id=Modulator_1, children [Random_2], properties {pitchMax=1.2}}
 {id=QualityLevel_0, children [Looping_0,Looping_0,null,null]}
 {id=Random_1, children [WavePlayer_1,WavePlayer_3], properties {weights [1,1]}}
 {id=Random_2, children [WavePlayer_0,WavePlayer_2], properties {weights [1,1]}}
 {id=Random_3, children [Modulator_0,Enveloper_1], properties {weights [1,1]}}
 {id=WavePlayer_0, properties {soundWaveAssetPtr=/AmbientAudioContent/Waves/SpottedLoops/Spot_PirateShip_Rope_Tension_Loop_01}}
 {id=WavePlayer_1, properties {soundWaveAssetPtr=&Spot_PirateShip_Rope_Tension_Loop_01}}
 {id=WavePlayer_2, properties {soundWaveAssetPtr=/AmbientAudioContent/Waves/SpottedLoops/Spot_PirateShip_Rope_Tension_Loop_02}}
 {id=WavePlayer_3, properties {soundWaveAssetPtr=&Spot_PirateShip_Rope_Tension_Loop_02}}
]
properties {
 bEnableSubmixSends=false
 bOverrideAttenuation=true
 bOverrideConcurrency=true
 maxDistance=1250
 soundClassObject=&Fort_Ambient_Outdoors_Athena
 volumeMultiplier=0.501
 attenuationOverrides {
  attenuationShape=Capsule
  bEnableListenerFocus=true
  falloffDistance=850
  focusVolumeAttenuation=1.1
  lPFFrequencyAtMax=1000
  lPFRadiusMax=2000
  lPFRadiusMin=512
  nonFocusAzimuth=90
  nonFocusVolumeAttenuation=0.75
  spatializationAlgorithm=SPATIALIZATION_HRTF
  attenuationShapeExtents {y=150}
 }
 concurrencyOverrides {maxCount=2, voiceStealReleaseTime=3}
}
