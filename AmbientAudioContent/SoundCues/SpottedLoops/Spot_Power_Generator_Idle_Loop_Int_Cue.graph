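# Indoor ambient Sound Cue for an idling power generator, played as a looping spot emitter with medium prop attenuation.
# It mixes a dominant refueling-cart idle loop, delayed and faded in by an Enveloper, with a quieter power-receiver loop varied by a Modulator.
# A root Delay staggers the start, and every quality level feeds the same mix.
# Concurrency limits cap simultaneous instances, and bus and submix sends are disabled.
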SoundCue /AmbientAudioContent/SoundCues/SpottedLoops/Spot_Power_Generator_Idle_Loop_Int_Cue
root=Delay_1
nodes [
 {id=Delay_0, children [WavePlayer_0], properties {delayMax=0.3}}
 {id=Delay_1, children [QualityLevel_0], properties {delayMax=1}}
 {
  id=Enveloper_1
  children [Delay_0]
  properties {pitchCurve {editorCurveData {keys [{time=0}]}}, volumeCurve {editorCurveData {keys [<at 0 0>,<at 1 1>]}}}
 }
 {id=Mixer_0, children [Enveloper_1,Modulator_0], properties {inputVolume [1.2,0.33]}}
 {id=Modulator_0, children [WavePlayer_1]}
 {id=QualityLevel_0, children [Mixer_0,Mixer_0,Mixer_0,Mixer_0]}
 {id=WavePlayer_0, properties {bLooping=true, soundWaveAssetPtr=/AmbientAudioContent/Waves/SpottedLoops/Spot_RefuelingCart_Ilde_Loop}}
 {id=WavePlayer_1, properties {bLooping=true, soundWaveAssetPtr=/AmbientAudioContent/Waves/POI/Spot_Mole_AmbientLoop_Object_PowerReceivers_03}}
]
properties {
 attenuationSettings=/Game/Sounds/Attenuation/Fort_Amb_Emitter_Prop_Med_Att
 bEnableBusSends=false
 bEnableSubmixSends=false
 maxDistance=900
 soundClassObject=&Fort_Ambient_Indoors_Athena
 volumeMultiplier=0.6
 concurrencyOverrides {maxCount=2, voiceStealReleaseTime=0.2}
 concurrencySet [/AmbientAudioBase/Concurrency/AmbientConc_3Max_Spot_Default_Concurrency]
}
