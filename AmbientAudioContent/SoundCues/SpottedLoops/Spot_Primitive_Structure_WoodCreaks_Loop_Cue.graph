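# Looping Sound Cue for spotted ambient wood creaks from primitive structures, built from one-shot waves.
# A Looping node repeats a pitch Modulator and a random Delay of several seconds, which feed an equal-weight Random pick among ten creak waves, giving sparse, varied creaks.
# It plays quiet and pitched down in the ambient one-shots class, with overridden attenuation using listener focus and a shared primitive-structure concurrency set.
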SoundCue /AmbientAudioContent/SoundCues/SpottedLoops/Spot_Primitive_Structure_WoodCreaks_Loop_Cue
root=Looping_0
nodes [
 {id=Delay_0, children [Random_1], properties {delayMax=12, delayMin=3}}
 {id=Looping_0, children [Modulator_0]}
 {id=Modulator_0, children [Delay_0], properties {pitchMax=1, pitchMin=0.8}}
 {
  id=Random_1
  children [
   WavePlayer_10
   WavePlayer_11
   WavePlayer_12
   WavePlayer_13
   WavePlayer_14
   WavePlayer_15
   WavePlayer_16
   WavePlayer_17
   WavePlayer_18
   WavePlayer_19
  ]
  properties {weights [1,1,1,1,1,1,1,1,1,1]}
 }
 {id=WavePlayer_10, properties {soundWaveAssetPtr=/AmbientAudioContent/Waves/OneShots/Spot_Primitive_Structure_WoodCreaks_10}}
 {id=WavePlayer_11, properties {soundWaveAssetPtr=/AmbientAudioContent/Waves/OneShots/Spot_Primitive_Structure_WoodCreaks_09}}
 {id=WavePlayer_12, properties {soundWaveAssetPtr=/AmbientAudioContent/Waves/OneShots/Spot_Primitive_Structure_WoodCreaks_08}}
 {id=WavePlayer_13, properties {soundWaveAssetPtr=/AmbientAudioContent/Waves/OneShots/Spot_Primitive_Structure_WoodCreaks_07}}
 {id=WavePlayer_14, properties {soundWaveAssetPtr=/AmbientAudioContent/Waves/OneShots/Spot_Primitive_Structure_WoodCreaks_06}}
 {id=WavePlayer_15, properties {soundWaveAssetPtr=/AmbientAudioContent/Waves/OneShots/Spot_Primitive_Structure_WoodCreaks_05}}
 {id=WavePlayer_16, properties {soundWaveAssetPtr=/AmbientAudioContent/Waves/OneShots/Spot_Primitive_Structure_WoodCreaks_04}}
 {id=WavePlayer_17, properties {soundWaveAssetPtr=/AmbientAudioContent/Waves/OneShots/Spot_Primitive_Structure_WoodCreaks_03}}
 {id=WavePlayer_18, properties {soundWaveAssetPtr=/AmbientAudioContent/Waves/OneShots/Spot_Primitive_Structure_WoodCreaks_02}}
 {id=WavePlayer_19, properties {soundWaveAssetPtr=/AmbientAudioContent/Waves/OneShots/Spot_Primitive_Structure_WoodCreaks_01}}
]
properties {
 bOverrideAttenuation=true
 maxDistance=1350
 pitchMultiplier=0.8
 soundClassObject=&Fort_Ambient_OneShots
 volumeMultiplier=0.25
 attenuationOverrides {bEnableListenerFocus=true, falloffDistance=1000, nonFocusVolumeAttenuation=0.75, attenuationShapeExtents {x=350}}
 concurrencySet [/Game/Sounds/Fort_Ambient_NonLoop/Primitive/Primitive_Structure_GLobal_Concurrency]
}
