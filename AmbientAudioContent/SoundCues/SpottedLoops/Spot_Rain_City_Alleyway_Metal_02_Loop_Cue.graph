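# Looping spot ambience cue for rain on metal in a city alleyway, classed as Athena outdoor ambient sound.
# A quality-level node feeds one looping rain wave into its first two slots and leaves the other two empty.
# It plays at reduced volume with a shared spline-surface rain attenuation asset.
# A city rain surface concurrency set is assigned to manage simultaneous instances.
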SoundCue /AmbientAudioContent/SoundCues/SpottedLoops/Spot_Rain_City_Alleyway_Metal_02_Loop_Cue
root=QualityLevel_0
nodes [
 {id=QualityLevel_0, children [WavePlayer_0,WavePlayer_0,null,null]}
 {id=WavePlayer_0, properties {bLooping=true, soundWaveAssetPtr=/AmbientAudioContent/Waves/SpottedLoops/Spot_City_Rain06_Loop}}
]
properties {
 attenuationSettings=/Game/Athena/Sounds/Ambient/Fort_Spot_Spline_Surface_Rain_Att
 maxDistance=3500
 soundClassObject=&Fort_Ambient_Outdoors_Athena
 volumeMultiplier=0.3
 attenuationOverrides {falloffDistance=2000, nonSpatializedRadiusStart=700, stereoSpread=1000, attenuationShapeExtents {x=925}}
 concurrencyOverrides {maxCount=1, voiceStealReleaseTime=1, volumeScaleAttackTime=1}
 concurrencySet [&/Spot_City_Rain_Surface_Concurrency]
}
