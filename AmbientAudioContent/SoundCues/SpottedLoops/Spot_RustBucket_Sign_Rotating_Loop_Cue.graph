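# Sound cue for the looping ambient spot sound of the rotating RustBucket sign prop.
# A mixer blends two looping waves borrowed by name from other props, with the Slurpy Swamp keg topper loop dominant and the SlapFactory sign loop much quieter.
# It plays in the ambient props sound class with a dedicated attenuation asset and a default spot concurrency set.
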SoundCue /AmbientAudioContent/SoundCues/SpottedLoops/Spot_RustBucket_Sign_Rotating_Loop_Cue
root=Mixer_0
nodes [
 {id=Mixer_0, children [WavePlayer_0,WavePlayer_1], properties {inputVolume [1.2,0.2]}}
 {
  id=WavePlayer_0
  properties {bLooping=true, soundWaveAssetPtr=/AmbientAudioContent/Waves/SpottedLoops/Spot_Slurpy_Swamp_Keg_Building_Topper_Loop}
 }
 {id=WavePlayer_1, properties {bLooping=true, soundWaveAssetPtr=/AmbientAudioContent/Waves/SpottedLoops/Spot_SlapFactory_Sign01}}
]
properties {
 attenuationSettings=/Game/Sounds/Helios/Environment_Actors/RustBucketSignRotating/Spot_RustBucket_Sign_Rotating_Att
 maxDistance=950
 soundClassObject=&Fort_Ambient_Props_Athena
 volumeMultiplier=0.501
 attenuationOverrides {
  attenuationShape=Box
  bEnableOcclusion=true
  falloffDistance=740
  occlusionInterpolationTime=2
  occlusionLowPassFilterFrequency=600
  occlusionTraceChannel=ECC_GameTraceChannel11
  reverbWetLevelMin=0.6905
  stereoSpread=5000
  attenuationShapeExtents=<600 600 800>
 }
 concurrencyOverrides {maxCount=3, volumeScaleMode=Distance}
 concurrencySet [/AmbientAudioBase/Concurrency/AmbientConc_3Max_Spot_Default_Concurrency]
}
