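# Ambient spot sound cue that loops scary speaker one-shots, classed as Athena prop ambience.
# A Looping node repeats a random delay, then a Quality Level switch selects one of two overlapping random wave pools.
# The first two quality slots share the larger pool, and the last two share the smaller one.
# Base submix and submix sends are disabled; output feeds the SpottedLoops_FNM audio bus, with a spot concurrency set limiting voices.
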SoundCue /AmbientAudioContent/SoundCues/SpottedLoops/Spot_SC_Scary_Speaker_Large_Loop
root=Looping_0
nodes [
 {id=Delay_0, children [QualityLevel_0], properties {delayMax=30, delayMin=10}}
 {id=Looping_0, children [Delay_0]}
 {id=QualityLevel_0, children [Random_0,Random_0,Random_4,Random_4]}
 {
  id=Random_0
  children [
   WavePlayer_37
   WavePlayer_40
   WavePlayer_49
   WavePlayer_50
   WavePlayer_51
   WavePlayer_52
   WavePlayer_53
   WavePlayer_54
   WavePlayer_55
   WavePlayer_56
   WavePlayer_57
   WavePlayer_68
   WavePlayer_69
   WavePlayer_70
   WavePlayer_71
   WavePlayer_72
   WavePlayer_73
   WavePlayer_74
   WavePlayer_75
   WavePlayer_76
   WavePlayer_77
   WavePlayer_78
   WavePlayer_79
   WavePlayer_80
   WavePlayer_81
   WavePlayer_82
   WavePlayer_83
   WavePlayer_84
   WavePlayer_85
   WavePlayer_86
   WavePlayer_87
   WavePlayer_88
  ]
  properties {weights [1,1,1,1,1,1,1,1,1,1,1,1,1,1,1,1,1,1,1,1,1,1,1,1,1,1,1,1,1,1,1,1]}
 }
 {
  id=Random_4
  children [
   WavePlayer_113
   WavePlayer_114
   WavePlayer_115
   WavePlayer_116
   WavePlayer_117
   WavePlayer_118
   WavePlayer_119
   WavePlayer_120
   WavePlayer_121
   WavePlayer_122
   WavePlayer_123
   WavePlayer_124
   WavePlayer_125
   WavePlayer_126
   WavePlayer_127
   WavePlayer_128
   WavePlayer_129
   WavePlayer_130
   WavePlayer_131
   WavePlayer_132
   WavePlayer_133
   WavePlayer_134
   WavePlayer_135
   WavePlayer_136
  ]
  properties {weights [1,1,1,1,1,1,1,1,1,1,1,1,1,1,1,1,1,1,1,1,1,1,1,1]}
 }
 {id=WavePlayer_113, properties {soundWaveAssetPtr=/AmbientAudioContent/Waves/OneShots/Spot_FNM_Scary_Speaker_Sounds_19}}
 {id=WavePlayer_114, properties {soundWaveAssetPtr=/AmbientAudioContent/Waves/OneShots/Spot_FNM_Scary_Speaker_Sounds_18}}
 {id=WavePlayer_115, properties {soundWaveAssetPtr=/AmbientAudioContent/Waves/OneShots/Spot_FNM_Scary_Speaker_Sounds_17}}
 {id=WavePlayer_116, properties {soundWaveAssetPtr=/AmbientAudioContent/Waves/OneShots/Spot_FNM_Scary_Speaker_Sounds_16}}
 {id=WavePlayer_117, properties {soundWaveAssetPtr=/AmbientAudioContent/Waves/OneShots/Spot_FNM_Scary_Speaker_Sounds_15}}
 {id=WavePlayer_118, properties {soundWaveAssetPtr=/AmbientAudioContent/Waves/OneShots/Spot_FNM_Scary_Speaker_Sounds_14}}
 {id=WavePlayer_119, properties {soundWaveAssetPtr=/AmbientAudioContent/Waves/OneShots/Spot_FNM_Scary_Speaker_Sounds_13}}
 {id=WavePlayer_120, properties {soundWaveAssetPtr=/AmbientAudioContent/Waves/OneShots/Spot_FNM_Scary_Speaker_Sounds_12}}
 {id=WavePlayer_121, properties {soundWaveAssetPtr=/AmbientAudioContent/Waves/OneShots/Spot_FNM_Scary_Speaker_Sounds_11}}
 {id=WavePlayer_122, properties {soundWaveAssetPtr=/AmbientAudioContent/Waves/OneShots/Spot_FNM_Scary_Speaker_Sounds_10}}
 {id=WavePlayer_123, properties {soundWaveAssetPtr=/AmbientAudioContent/Waves/OneShots/Spot_FNM_Scary_Speaker_Sounds_09}}
 {id=WavePlayer_124, properties {soundWaveAssetPtr=/AmbientAudioContent/Waves/OneShots/Spot_FNM_Scary_Speaker_Sounds_07}}
 {id=WavePlayer_125, properties {soundWaveAssetPtr=/AmbientAudioContent/Waves/OneShots/Spot_FNM_Scary_Speaker_Sounds_06}}
 {id=WavePlayer_126, properties {soundWaveAssetPtr=/AmbientAudioContent/Waves/OneShots/Spot_FNM_Scary_Speaker_Sounds_05}}
 {id=WavePlayer_127, properties {soundWaveAssetPtr=/AmbientAudioContent/Waves/OneShots/Spot_FNM_Scary_Speaker_Sounds_JB_01}}
 {id=WavePlayer_128, properties {soundWaveAssetPtr=/AmbientAudioContent/Waves/OneShots/Spot_FNM_Scary_Speaker_Sounds_JB_02}}
 {id=WavePlayer_129, properties {soundWaveAssetPtr=/AmbientAudioContent/Waves/OneShots/Spot_FNM_Scary_Speaker_Sounds_JB_03}}
 {id=WavePlayer_130, properties {soundWaveAssetPtr=/AmbientAudioContent/Waves/OneShots/Spot_FNM_Scary_Speaker_Sounds_JB_04}}
 {id=WavePlayer_131, properties {soundWaveAssetPtr=/AmbientAudioContent/Waves/OneShots/Spot_FNM_Scary_Speaker_Sounds_JB_05}}
 {id=WavePlayer_132, properties {soundWaveAssetPtr=/AmbientAudioContent/Waves/OneShots/Spot_FNM_Scary_Speaker_Sounds_JB_06}}
 {id=WavePlayer_133, properties {soundWaveAssetPtr=/AmbientAudioContent/Waves/OneShots/Spot_FNM_Scary_Speaker_Sounds_JB_07}}
 {id=WavePlayer_134, properties {soundWaveAssetPtr=/AmbientAudioContent/Waves/OneShots/Spot_FNM_Scary_Speaker_Sounds_JB_08}}
 {id=WavePlayer_135, properties {soundWaveAssetPtr=/AmbientAudioContent/Waves/OneShots/Spot_FNM_Scary_Speaker_Sounds_JB_09}}
 {id=WavePlayer_136, properties {soundWaveAssetPtr=/AmbientAudioContent/Waves/OneShots/Spot_FNM_Scary_Speaker_Sounds_JB_10}}
 {id=WavePlayer_37, properties {soundWaveAssetPtr=/AmbientAudioContent/Waves/OneShots/Spot_FNM_Scary_Speaker_Sounds_01}}
 {id=WavePlayer_40, properties {soundWaveAssetPtr=/AmbientAudioContent/Waves/OneShots/Spot_FNM_Scary_Speaker_Sounds_02}}
 {id=WavePlayer_49, properties {soundWaveAssetPtr=/AmbientAudioContent/Waves/OneShots/Spot_FNM_Scary_Speaker_Sounds_03}}
 {id=WavePlayer_50, properties {soundWaveAssetPtr=/AmbientAudioContent/Waves/OneShots/Spot_FNM_Scary_Speaker_Sounds_04}}
 {id=WavePlayer_51, properties {soundWaveAssetPtr=&Spot_FNM_Scary_Speaker_Sounds_05}}
 {id=WavePlayer_52, properties {soundWaveAssetPtr=&Spot_FNM_Scary_Speaker_Sounds_06}}
 {id=WavePlayer_53, properties {soundWaveAssetPtr=&Spot_FNM_Scary_Speaker_Sounds_07}}
 {id=WavePlayer_54, properties {soundWaveAssetPtr=/AmbientAudioContent/Waves/OneShots/Spot_FNM_Scary_Speaker_Sounds_08}}
 {id=WavePlayer_55, properties {soundWaveAssetPtr=&Spot_FNM_Scary_Speaker_Sounds_09}}
 {id=WavePlayer_56, properties {soundWaveAssetPtr=&Spot_FNM_Scary_Speaker_Sounds_10}}
 {id=WavePlayer_57, properties {soundWaveAssetPtr=&Spot_FNM_Scary_Speaker_Sounds_11}}
 {id=WavePlayer_68, properties {soundWaveAssetPtr=&Spot_FNM_Scary_Speaker_Sounds_12}}
 {id=WavePlayer_69, properties {soundWaveAssetPtr=&Spot_FNM_Scary_Speaker_Sounds_13}}
 {id=WavePlayer_70, properties {soundWaveAssetPtr=&Spot_FNM_Scary_Speaker_Sounds_14}}
 {id=WavePlayer_71, properties {soundWaveAssetPtr=&Spot_FNM_Scary_Speaker_Sounds_15}}
 {id=WavePlayer_72, properties {soundWaveAssetPtr=&Spot_FNM_Scary_Speaker_Sounds_16}}
 {id=WavePlayer_73, properties {soundWaveAssetPtr=&Spot_FNM_Scary_Speaker_Sounds_17}}
 {id=WavePlayer_74, properties {soundWaveAssetPtr=&Spot_FNM_Scary_Speaker_Sounds_18}}
 {id=WavePlayer_75, properties {soundWaveAssetPtr=&Spot_FNM_Scary_Speaker_Sounds_19}}
 {id=WavePlayer_76, properties {soundWaveAssetPtr=/AmbientAudioContent/Waves/OneShots/Spot_FNM_Scary_Speaker_Sounds_20}}
 {id=WavePlayer_77, properties {soundWaveAssetPtr=/AmbientAudioContent/Waves/OneShots/Spot_FNM_Scary_Speaker_Sounds_21}}
 {id=WavePlayer_78, properties {soundWaveAssetPtr=/AmbientAudioContent/Waves/OneShots/Spot_FNM_Scary_Speaker_Sounds_22}}
 {id=WavePlayer_79, properties {soundWaveAssetPtr=/AmbientAudioContent/Waves/OneShots/Spot_FNM_Scary_Speaker_Sounds_23}}
 {id=WavePlayer_80, properties {soundWaveAssetPtr=/AmbientAudioContent/Waves/OneShots/Spot_FNM_Scary_Speaker_Sounds_24}}
 {id=WavePlayer_81, properties {soundWaveAssetPtr=&Spot_FNM_Scary_Speaker_Sounds_JB_01}}
 {id=WavePlayer_82, properties {soundWaveAssetPtr=&Spot_FNM_Scary_Speaker_Sounds_JB_02}}
 {id=WavePlayer_83, properties {soundWaveAssetPtr=&Spot_FNM_Scary_Speaker_Sounds_JB_03}}
 {id=WavePlayer_84, properties {soundWaveAssetPtr=&Spot_FNM_Scary_Speaker_Sounds_JB_04}}
 {id=WavePlayer_85, properties {soundWaveAssetPtr=&Spot_FNM_Scary_Speaker_Sounds_JB_05}}
 {id=WavePlayer_86, properties {soundWaveAssetPtr=&Spot_FNM_Scary_Speaker_Sounds_JB_06}}
 {id=WavePlayer_87, properties {soundWaveAssetPtr=&Spot_FNM_Scary_Speaker_Sounds_JB_07}}
 {id=WavePlayer_88, properties {soundWaveAssetPtr=&Spot_FNM_Scary_Speaker_Sounds_JB_08}}
]
properties {
 bEnableBaseSubmix=false
 bEnableSubmixSends=false
 maxDistance=2097152
 soundClassObject=&Fort_Ambient_Props_Athena
 volumeMultiplier=1.7
 concurrencyOverrides {maxCount=4, resolutionRule=StopFarthestThenPreventNew}
 concurrencySet [/AmbientAudioBase/Concurrency/AmbientConc_3Max_Spot_Default_Concurrency]
 preEffectBusSends [{audioBus=&/AB_SpottedLoops_FNM, sendLevel=1.7, soundSourceBus=null}]
}
