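# Spotted ambient loop cue for sailboat sail flapping, in the outdoor ambient sound class with capsule attenuation, occlusion and a concurrency set.
# The first two quality levels mix the looping flap bed with randomly picked large flap one-shots, repeated after random delays with pitch and volume modulation.
# The last two play only the bed after a short delay, through three stacked looping envelopers that vary volume and pitch over time.
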SoundCue /AmbientAudioContent/SoundCues/SpottedLoops/Spot_Sailboat_Sail_Flag_Flap_Halyard_Impact_loop_Cue
root=QualityLevel_1
nodes [
 {id=Delay_0, children [Enveloper_4], properties {delayMax=0.75, delayMin=0.2}}
 {id=Delay_2, children [Random_0], properties {delayMax=5, delayMin=3}}
 {
  id=Enveloper_0
  children [WavePlayer_0]
  properties {
   bLoop=true
   bLoopIndefinitely=true
   loopEnd=60
   pitchCurve {editorCurveData {keys [
      {interpMode=RCIM_Cubic, time=0}
      <at 15 0.85 RCIM_Cubic>
      {arriveTangent=0.0091, interpMode=RCIM_Cubic, leaveTangent=0.0091, tangentMode=RCTM_User, time=30}
      <at 45 1.1 RCIM_Cubic>
      {time=60}
     ]}}
   volumeCurve {editorCurveData {keys [<at 0 1>]}}
  }
 }
 {
  id=Enveloper_3
  children [Enveloper_0]
  properties {
   bLoop=true
   bLoopIndefinitely=true
   loopEnd=60
   pitchCurve {editorCurveData {keys [{time=0}]}}
   volumeCurve {editorCurveData {keys [
      {arriveTangent=-0.01, interpMode=RCIM_Cubic, time=0, value=1}
      {arriveTangent=-0.0035, interpMode=RCIM_Cubic, leaveTangent=-0.0035, tangentMode=RCTM_User, time=5.0017, value=0.3634}
      {arriveTangent=0.0016, interpMode=RCIM_Cubic, leaveTangent=0.0016, tangentMode=RCTM_User, time=7.591, value=0.8382}
      {arriveTangent=0.0008, interpMode=RCIM_Cubic, leaveTangent=0.0008, tangentMode=RCTM_User, time=11.5666, value=0.3768}
      {arriveTangent=0.0016, interpMode=RCIM_Cubic, leaveTangent=0.0016, tangentMode=RCTM_User, time=13.4364, value=0.3624}
      {arriveTangent=0.0009, interpMode=RCIM_Cubic, leaveTangent=0.0009, time=15.1685, value=0.7207}
      {arriveTangent=-0.0011, interpMode=RCIM_Cubic, leaveTangent=-0.0011, tangentMode=RCTM_User, time=19.0457, value=0.3674}
      {arriveTangent=0.0002, interpMode=RCIM_Cubic, leaveTangent=0.0002, time=24.1334, value=0.8568}
      {arriveTangent=0.0016, interpMode=RCIM_Cubic, leaveTangent=0.0016, tangentMode=RCTM_User, time=28.493, value=0.3696}
      {arriveTangent=-0.0003, interpMode=RCIM_Cubic, leaveTangent=-0.0003, time=31.406, value=0.7353}
      {arriveTangent=0.0009, interpMode=RCIM_Cubic, leaveTangent=0.0009, tangentMode=RCTM_User, time=35.3816, value=0.3676}
      {arriveTangent=-0.0008, interpMode=RCIM_Cubic, leaveTangent=-0.0008, time=38.1075, value=0.8568}
      <at 42.4671 0.3623>
      {arriveTangent=0.0024, interpMode=RCIM_Cubic, leaveTangent=0.0024, tangentMode=RCTM_User, time=45.0257, value=0.3674}
      {arriveTangent=-0.0006, interpMode=RCIM_Cubic, leaveTangent=-0.0006, time=48.1356, value=0.8382}
      {arriveTangent=-0.0056, interpMode=RCIM_Cubic, leaveTangent=-0.0056, tangentMode=RCTM_User, time=53.5932, value=0.3624}
      <at 55.5199 0.7454>
      {arriveTangent=-0.0003, interpMode=RCIM_Cubic, leaveTangent=-0.0003, tangentMode=RCTM_User, time=58.2125, value=0.3621}
      <at 60 1 RCIM_Cubic>
     ]}}
  }
 }
 {
  id=Enveloper_4
  children [Enveloper_3]
  properties {
   bLoop=true
   bLoopIndefinitely=true
   loopEnd=40
   pitchCurve {editorCurveData {keys [{time=0}]}}
   volumeCurve {editorCurveData {keys [
      {arriveTangent=-0.01, interpMode=RCIM_Cubic, time=0, value=1}
      <at 10 0.7 RCIM_Cubic>
      {arriveTangent=-0.0173, interpMode=RCIM_Cubic, leaveTangent=-0.0173, time=20, value=1}
      {arriveTangent=-0.0749, interpMode=RCIM_Cubic, leaveTangent=-0.0749, tangentMode=RCTM_User, time=25.9279, value=0.4252}
      {arriveTangent=0.129, interpMode=RCIM_Cubic, leaveTangent=0.129, tangentMode=RCTM_User, time=36.1629, value=0.5685}
      <at 40 1 RCIM_Cubic>
     ]}}
  }
 }
 {id=Looping_1, children [Modulator_1]}
 {id=Mixer_0, children [Looping_1,WavePlayer_0], properties {inputVolume [1.1,1.1]}}
 {id=Modulator_1, children [Delay_2], properties {pitchMin=0.85, volumeMax=1.2, volumeMin=0.9}}
 {id=QualityLevel_1, children [Mixer_0,Mixer_0,Delay_0,Delay_0]}
 {
  id=Random_0
  children [
   WavePlayer_12
   WavePlayer_13
   WavePlayer_14
   WavePlayer_15
   WavePlayer_16
   WavePlayer_17
   WavePlayer_18
   WavePlayer_19
   WavePlayer_20
   WavePlayer_21
   WavePlayer_22
  ]
  properties {weights [1,1,1,1,1,1,1,1,1,1,1]}
 }
 {id=WavePlayer_0, properties {bLooping=true, soundWaveAssetPtr=/AmbientAudioContent/Waves/SpottedLoops/Spot_Amb_PirateCove_SailFlap_01}}
 {id=WavePlayer_12, properties {soundWaveAssetPtr=/AmbientAudioContent/Waves/OneShots/Spot_PirateCove_SailFlap_Large_01}}
 {id=WavePlayer_13, properties {soundWaveAssetPtr=/AmbientAudioContent/Waves/OneShots/Spot_PirateCove_SailFlap_Large_02}}
 {id=WavePlayer_14, properties {soundWaveAssetPtr=/AmbientAudioContent/Waves/OneShots/Spot_PirateCove_SailFlap_Large_03}}
 {id=WavePlayer_15, properties {soundWaveAssetPtr=/AmbientAudioContent/Waves/OneShots/Spot_PirateCove_SailFlap_Large_04}}
 {id=WavePlayer_16, properties {soundWaveAssetPtr=/AmbientAudioContent/Waves/OneShots/Spot_PirateCove_SailFlap_Large_05}}
 {id=WavePlayer_17, properties {soundWaveAssetPtr=/AmbientAudioContent/Waves/OneShots/Spot_PirateCove_SailFlap_Large_06}}
 {id=WavePlayer_18, properties {soundWaveAssetPtr=/AmbientAudioContent/Waves/OneShots/Spot_PirateCove_SailFlap_Large_07}}
 {id=WavePlayer_19, properties {soundWaveAssetPtr=/AmbientAudioContent/Waves/OneShots/Spot_PirateCove_SailFlap_Large_08}}
 {id=WavePlayer_20, properties {soundWaveAssetPtr=/AmbientAudioContent/Waves/OneShots/Spot_PirateCove_SailFlap_Large_09}}
 {id=WavePlayer_21, properties {soundWaveAssetPtr=/AmbientAudioContent/Waves/OneShots/Spot_PirateCove_SailFlap_Large_10}}
 {id=WavePlayer_22, properties {soundWaveAssetPtr=/AmbientAudioContent/Waves/OneShots/Spot_PirateCove_SailFlap_Large_11}}
]
properties {
 bOverrideAttenuation=true
 maxDistance=2500
 soundClassObject=&Fort_Ambient_Outdoors_Athena
 volumeMultiplier=0.601
 attenuationOverrides {
  attenuationShape=Capsule
  bEnableFocusInterpolation=true
  bEnableOcclusion=true
  falloffDistance=2500
  focusAttackInterpSpeed=3
  focusReleaseInterpSpeed=3
  manualReverbSendLevel=10
  nonFocusVolumeAttenuation=0.5
  occlusionLowPassFilterFrequency=3000
  occlusionTraceChannel=ECC_GameTraceChannel11
  reverbDistanceMax=2000
  reverbSendMethod=Manual
  reverbWetLevelMax=5
  attenuationShapeExtents {x=0}
 }
 concurrencySet [/AmbientAudioBase/Concurrency/AmbientConc_3Max_Spot_Default_Concurrency]
}
