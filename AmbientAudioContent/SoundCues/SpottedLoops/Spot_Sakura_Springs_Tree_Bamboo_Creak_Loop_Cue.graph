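# Spotted ambient Sound Cue that plays intermittent bamboo creaks on a loop for the Sakura Springs tree, in the outdoor ambient sound class.
# Each loop waits a random delay, then plays a randomly chosen creak with varied pitch through a Modulator.
# A quality switch gives the first two levels nine creak variations and the last two only three.
# Attenuation and concurrency overrides shape falloff and limit voices, with sends to the exterior bus and reverb submix.
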SoundCue /AmbientAudioContent/SoundCues/SpottedLoops/Spot_Sakura_Springs_Tree_Bamboo_Creak_Loop_Cue
root=QualityLevel_0
nodes [
 {id=Delay_0, children [Random_2], properties {delayMax=12, delayMin=2}}
 {id=Delay_1, children [Random_1], properties {delayMax=12, delayMin=2}}
 {id=Looping_0, children [Modulator_0]}
 {id=Looping_1, children [Modulator_1]}
 {id=Modulator_0, children [Delay_0], properties {pitchMin=0.7}}
 {id=Modulator_1, children [Delay_1], properties {pitchMin=0.7}}
 {id=QualityLevel_0, children [Looping_0,Looping_0,Looping_1,Looping_1]}
 {id=Random_1, children [WavePlayer_3, WavePlayer_10, WavePlayer_11], properties {weights [1,1,1]}}
 {
  id=Random_2
  children [WavePlayer_12, WavePlayer_13, WavePlayer_14, WavePlayer_15, WavePlayer_16, WavePlayer_17, WavePlayer_18, WavePlayer_19, WavePlayer_20]
  properties {weights [1,1,1,1,1,1,1,1,1]}
 }
 {id=WavePlayer_10, properties {soundWaveAssetPtr=/AmbientAudioContent/Waves/OneShots/Spot_Bamboo_Creak_2}}
 {id=WavePlayer_11, properties {soundWaveAssetPtr=/AmbientAudioContent/Waves/OneShots/Spot_Bamboo_Creak_3}}
 {id=WavePlayer_12, properties {soundWaveAssetPtr=/AmbientAudioContent/Waves/OneShots/Spot_Bamboo_Creak_1}}
 {id=WavePlayer_13, properties {soundWaveAssetPtr=&Spot_Bamboo_Creak_2}}
 {id=WavePlayer_14, properties {soundWaveAssetPtr=&Spot_Bamboo_Creak_3}}
 {id=WavePlayer_15, properties {soundWaveAssetPtr=/AmbientAudioContent/Waves/OneShots/Spot_Bamboo_Creak_4}}
 {id=WavePlayer_16, properties {soundWaveAssetPtr=/AmbientAudioContent/Waves/OneShots/Spot_Bamboo_Creak_5}}
 {id=WavePlayer_17, properties {soundWaveAssetPtr=/AmbientAudioContent/Waves/OneShots/Spot_Bamboo_Creak_6}}
 {id=WavePlayer_18, properties {soundWaveAssetPtr=/AmbientAudioContent/Waves/OneShots/Spot_Bamboo_Creak_7}}
 {id=WavePlayer_19, properties {soundWaveAssetPtr=/AmbientAudioContent/Waves/OneShots/Spot_Bamboo_Creak_8}}
 {id=WavePlayer_20, properties {soundWaveAssetPtr=/AmbientAudioContent/Waves/OneShots/Spot_Bamboo_Creak_9}}
 {id=WavePlayer_3, properties {soundWaveAssetPtr=&Spot_Bamboo_Creak_1}}
]
properties {
 attenuationSettings=/Game/Sounds/Attenuation/Fort_TileAmbient_OneShot_Wider_Att
 bOverrideConcurrency=true
 maxDistance=8000
 soundClassObject=&Fort_Ambient_Outdoors_Athena
 volumeMultiplier=0.752
 attenuationOverrides {falloffDistance=2000, nonSpatializedRadiusStart=700, stereoSpread=1000, attenuationShapeExtents {x=925}}
 busSends [{audioBus=null, minSendLevel=1, sendLevel=1, soundSourceBus=&AmbientAudio_Exterior_SourceBus}]
 concurrencyOverrides {maxCount=2, voiceStealReleaseTime=2}
 soundSubmixSends [{minSendLevel=0.6, soundSubmix=&Fortnite_Reverb_Default}]
}
