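# Spotted ambient cue for sauna stove steam, played on the indoor ambient sound class with a tight concurrency limit.
# All quality levels share a mixer layering two attenuation branches, one cone-shaped and one non-cone with occlusion, both using listener focus.
# Both branches feed from one enveloper that fades volume in.
# Beneath it, a random node picks between the looping sauna wave and a quieter NitroDrome steam wave, whose player is not set to loop.
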SoundCue /AmbientAudioContent/SoundCues/SpottedLoops/Spot_Sauna_Stove_Steam_Loop_Cue
root=QualityLevel_0
nodes [
 {
  id=Attenuation_0
  children [Enveloper_1]
  properties {
   bOverrideAttenuation=true
   attenuationOverrides {
    attenuationShape=Cone
    bEnableListenerFocus=true
    falloffDistance=850
    nonFocusVolumeAttenuation=0.75
    nonSpatializedRadiusStart=500
    attenuationShapeExtents {y=90, z=30}
   }
  }
 }
 {
  id=Attenuation_2
  children [Enveloper_1]
  properties {
   bOverrideAttenuation=true
   attenuationOverrides {
    bEnableListenerFocus=true
    bEnableOcclusion=true
    falloffDistance=1000
    nonFocusVolumeAttenuation=0.75
    nonSpatializedRadiusStart=500
    occlusionLowPassFilterFrequency=2000
    occlusionTraceChannel=ECC_GameTraceChannel11
    occlusionVolumeAttenuation=0.5
    attenuationShapeExtents {x=500}
   }
  }
 }
 {
  id=Enveloper_1
  children [Random_1]
  properties {pitchCurve {editorCurveData {keys [{time=0}]}}, volumeCurve {editorCurveData {keys [<at 0 0>,<at 1 1>]}}}
 }
 {id=Mixer_1, children [WavePlayer_0], properties {inputVolume [0.6]}}
 {id=Mixer_2, children [Attenuation_0, Attenuation_2], properties {inputVolume [1,1]}}
 {id=QualityLevel_0, children [Mixer_2,Mixer_2,Mixer_2,Mixer_2]}
 {id=Random_1, children [WavePlayer_1,Mixer_1], properties {weights [1,1]}}
 {id=WavePlayer_0, properties {soundWaveAssetPtr=/AmbientAudioContent/Waves/SpottedLoops/Spot_NitroDrome_EntranceHead_Steam_Loop}}
 {id=WavePlayer_1, properties {bLooping=true, soundWaveAssetPtr=/AmbientAudioContent/Waves/SpottedLoops/Spot_Sauna_Stove_Steam_Loop}}
]
properties {
 bEnableBusSends=false
 bEnableSubmixSends=false
 bOverrideConcurrency=true
 maxDistance=1500
 soundClassObject=&Fort_Ambient_Indoors_Athena
 volumeMultiplier=0.4
 attenuationOverrides {
  attenuationShape=Cone
  bAttenuateWithLPF=true
  bEnableListenerFocus=true
  falloffDistance=850
  lPFFrequencyAtMax=12000
  lPFRadiusMax=1000
  lPFRadiusMin=500
  nonFocusVolumeAttenuation=0.75
  nonSpatializedRadiusStart=500
  occlusionLowPassFilterFrequency=12000
  occlusionTraceChannel=ECC_GameTraceChannel11
  occlusionVolumeAttenuation=0.5
  attenuationShapeExtents=<500 90>
 }
 concurrencyOverrides {maxCount=2, voiceStealReleaseTime=0.2}
 concurrencySet [/AmbientAudioBase/Concurrency/AmbientConc_3Max_Spot_Default_Concurrency]
}
