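# Spotted looping ambient cue for a science lab kiosk, playing a console terminal communication hum under the indoor ambient sound class.
# A quality switch sends its first two levels to a random pick of two looping hum waves and its last two to a single wave.
# Each branch varies pitch and volume through a modulator and a mixer with an empty second input.
# An attenuation asset shapes falloff, and a concurrency override allows one instance at a time.
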SoundCue /AmbientAudioContent/SoundCues/SpottedLoops/Spot_ScienceLab_LabKiosk_Loop_Cue
root=QualityLevel_0
nodes [
 {id=Mixer_1, children [Modulator_1,null], properties {inputVolume [1,1]}}
 {id=Mixer_2, children [Modulator_0,null], properties {inputVolume [1,1]}}
 {id=Modulator_0, children [WavePlayer_2], properties {pitchMax=1.15, volumeMax=0.9, volumeMin=0.8}}
 {id=Modulator_1, children [Random_0], properties {pitchMax=1.15, volumeMax=0.9, volumeMin=0.8}}
 {id=QualityLevel_0, children [Mixer_1,Mixer_1,Mixer_2,Mixer_2]}
 {id=Random_0, children [WavePlayer_0,WavePlayer_1], properties {weights [1,1]}}
 {
  id=WavePlayer_0
  properties {bLooping=true, soundWaveAssetPtr=/AmbientAudioContent/Waves/SpottedLoops/Spot_Console_Terminal_Communication_Hum_Loop_03}
 }
 {
  id=WavePlayer_1
  properties {bLooping=true, soundWaveAssetPtr=/AmbientAudioContent/Waves/SpottedLoops/Spot_Console_Terminal_Communication_Hum_Loop_04}
 }
 {
  id=WavePlayer_2
  properties {bLooping=true, soundWaveAssetPtr=&Spot_Console_Terminal_Communication_Hum_Loop_03}
 }
]
properties {
 attenuationSettings=/Game/Sounds/Helios/Environment_Actors/ConsoleTerminals/Spot_Console_Terminal_Communication_Hum_Att
 bOverrideConcurrency=true
 maxDistance=1000
 soundClassObject=&Fort_Ambient_Indoors_Athena
 volumeMultiplier=1.1
 concurrencyOverrides {maxCount=1}
}
