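# Spot ambient cue for a large scrap heap fan, a looping prop sound with large air duct fan attenuation.
# A mixer layers two looping fan waves, with the outpost wall fan spin much louder than the duct loop.
# A random node picks either a default modulation or a higher-pitched, slightly delayed variant of that shared mix, under an enveloper volume fade-in.
# Concurrency blocks new instances at the limit, and listener focus attenuates off-focus volume.
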SoundCue /AmbientAudioContent/SoundCues/SpottedLoops/Spot_ScrapHeap_Fan_Large_Loop_Cue
root=Enveloper_0
nodes [
 {id=Delay_0, children [Mixer_0], properties {delayMax=0.3, delayMin=0.3}}
 {
  id=Enveloper_0
  children [Random_0]
  properties {pitchCurve {editorCurveData {keys [{time=0}]}}, volumeCurve {editorCurveData {keys [<at 0 0>,<at 1 1>,<at 2 1>]}}}
 }
 {id=Mixer_0, children [WavePlayer_2,WavePlayer_1], properties {inputVolume [0.5,2.5]}}
 {id=Modulator_0, children [Mixer_0]}
 {id=Modulator_1, children [Delay_0], properties {pitchMax=1.2, pitchMin=1.2}}
 {id=Random_0, children [Modulator_1,Modulator_0], properties {weights [1,1]}}
 {id=WavePlayer_1, properties {bLooping=true, soundWaveAssetPtr=/AmbientAudioContent/Waves/SpottedLoops/Spot_Outpost_Wall_Fan_Spin_Loop}}
 {id=WavePlayer_2, properties {bLooping=true, soundWaveAssetPtr=/AmbientAudioContent/Waves/SpottedLoops/Spot_AMB_Air_Duct_Fan_Loop_01}}
]
properties {
 attenuationSettings=/Game/Sounds/Apollo/Props/AirDucts/AirDuct_Fan_Large_Att
 bOverrideConcurrency=true
 maxDistance=1500
 soundClassObject=&Fort_Ambient_Props_Athena
 volumeMultiplier=0.502
 attenuationOverrides {bEnableListenerFocus=true, nonFocusVolumeAttenuation=0.4}
 concurrencyOverrides {resolutionRule=PreventNew}
}
