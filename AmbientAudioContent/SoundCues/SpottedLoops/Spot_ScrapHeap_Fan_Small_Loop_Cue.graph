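# Sound cue for a small scrap heap fan ambient loop, a spotted prop sound built from a looping air duct fan wave.
# A random node picks evenly between two modulated branches sharing that wave, one pitched up and briefly delayed, the other with default modulation.
# An enveloper above them fades volume in at the start.
# Playback uses borrowed outpost wall fan attenuation with listener focus, plus prop concurrency rules.
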SoundCue /AmbientAudioContent/SoundCues/SpottedLoops/Spot_ScrapHeap_Fan_Small_Loop_Cue
root=Enveloper_0
nodes [
 {id=Delay_0, children [WavePlayer_1], properties {delayMax=0.3, delayMin=0.3}}
 {
  id=Enveloper_0
  children [Random_0]
  properties {pitchCurve {editorCurveData {keys [{time=0}]}}, volumeCurve {editorCurveData {keys [<at 0 0>,<at 1 1>,<at 2 1>]}}}
 }
 {id=Modulator_1, children [WavePlayer_1]}
 {id=Modulator_2, children [Delay_0], properties {pitchMax=1.2, pitchMin=1.2}}
 {id=Random_0, children [Modulator_2,Modulator_1], properties {weights [1,1]}}
 {id=WavePlayer_1, properties {bLooping=true, soundWaveAssetPtr=/AmbientAudioContent/Waves/SpottedLoops/Spot_AMB_Air_Duct_Fan_Loop_01}}
]
properties {
 attenuationSettings=/Game/Sounds/Artemis/Biomes/Grasslands/Grassland_Spots/Spot_Outpost_Wall_Fan_Att
 bEnableBusSends=false
 maxDistance=3950
 soundClassObject=&Fort_Ambient_Props_Athena
 volumeMultiplier=0.501
 attenuationOverrides {bEnableListenerFocus=true, nonFocusVolumeAttenuation=0.4}
 concurrencyOverrides {maxCount=2}
 concurrencySet [/Game/Sounds/Concurrency_Rules/Athena_Prop]
 preEffectBusSends [{}]
}
