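# Sound cue for a spotted server rack ambience, a quiet, low-priority looping scan and hum in the Athena ambient props sound class.
# Two looping waves feed a mixer that boosts the scanning layer and lowers the tower hum.
# The mix passes through a default modulator for variation, then an enveloper whose volume curve ramps up from silence.
# A server rack attenuation asset and three-voice ambient concurrency limit its range and instance count.
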SoundCue /AmbientAudioContent/SoundCues/SpottedLoops/Spot_ServeRack_Scanning_Loop_01_Cue
root=Enveloper_0
nodes [
 {
  id=Enveloper_0
  children [Modulator_0]
  properties {pitchCurve {editorCurveData {keys [{time=0}]}}, volumeCurve {editorCurveData {keys [<at 0 0>,<at 1 1>]}}}
 }
 {id=Mixer_0, children [WavePlayer_1,WavePlayer_0], properties {inputVolume [2,0.5]}}
 {id=Modulator_0, children [Mixer_0]}
 {
  id=WavePlayer_0
  properties {bLooping=true, soundWaveAssetPtr=/AmbientAudioContent/Waves/SpottedLoops/Spot_Computer_Server_Rack_Tower_Hum_Loop_01}
 }
 {id=WavePlayer_1, properties {bLooping=true, soundWaveAssetPtr=/AmbientAudioContent/Waves/SpottedLoops/Spot_Server_Stack_Scanning_Loop_01}}
]
properties {
 attenuationSettings=/Game/Sounds/Helios/Environment_Actors/ServerRacks/Spot_ServerRack_Hum_Att
 maxDistance=850
 priority=0.25
 soundClassObject=&Fort_Ambient_Props_Athena
 volumeMultiplier=0.15
 concurrencyOverrides {maxCount=3}
 concurrencySet [/AmbientAudioBase/Concurrency/AmbientConc_3Max_Spot_Default_Concurrency]
}
