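# Looping spot ambience cue for server rack presence in the Athena ambient sound class, built from a pitched-down console terminal hum wave.
# One looping wave player feeds every quality level, then an endlessly looping 60-second Enveloper that gently swells the volume.
# Overridden attenuation adds a custom falloff curve, listener focus and low-pass occlusion, with distance-scaled sends to two reverb submixes.
# Concurrency allows one instance per owner and blocks new ones.
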SoundCue /AmbientAudioContent/SoundCues/SpottedLoops/Spot_ServerRack_Pressence_Box_Loop_Cue
root=Enveloper_3
nodes [
 {
  id=Enveloper_3
  children [QualityLevel_0]
  properties {
   bLoop=true
   bLoopIndefinitely=true
   loopEnd=60
   pitchCurve {editorCurveData {keys [{time=0}]}}
   volumeCurve {editorCurveData {keys [{arriveTangent=-0.01, interpMode=RCIM_Cubic, time=0, value=1}, <at 30 1.1431 RCIM_Cubic>, <at 60 1>]}}
  }
 }
 {id=QualityLevel_0, children [WavePlayer_1,WavePlayer_1,WavePlayer_1,WavePlayer_1]}
 {
  id=WavePlayer_1
  properties {bLooping=true, soundWaveAssetPtr=/AmbientAudioContent/Waves/SpottedLoops/Spot_Console_Terminal_Communication_Hum_Loop_04}
 }
]
properties {
 attenuationSettings=/Game/Sounds/Helios/Environment_Actors/ConsoleTerminals/Spot_Console_Terminal_Communication_Hum_Att
 bOverrideAttenuation=true
 bOverrideConcurrency=true
 maxDistance=4000
 pitchMultiplier=0.7
 soundClassObject=&Fort_Ambient_Athena
 subtitlePriority=5000
 volumeMultiplier=0.6
 attenuationOverrides {
  bEnableFocusInterpolation=true
  bEnableOcclusion=true
  dBAttenuationAtMax=-40
  distanceAlgorithm=Custom
  falloffDistance=2500
  focusAzimuth=15
  focusVolumeAttenuation=1.5
  nonFocusAzimuth=45
  nonFocusVolumeAttenuation=0.5
  nonSpatializedRadiusEnd=1500
  nonSpatializedRadiusStart=2500
  occlusionInterpolationTime=5
  occlusionLowPassFilterFrequency=750
  occlusionTraceChannel=ECC_GameTraceChannel11
  occlusionVolumeAttenuation=0
  stereoSpread=400
  attenuationShapeExtents=<1500 64.5942 90>
  customAttenuationCurve {editorCurveData {keys [<at 0 1 RCIM_Cubic>, {arriveTangent=-2.6702, interpMode=RCIM_Cubic, leaveTangent=-2.6702, tangentMode=RCTM_User, time=1, value=0}]}}
 }
 concurrencyOverrides {bLimitToOwner=true, maxCount=1, resolutionRule=PreventNew}
 soundSubmixSends [
  {
   maxSendDistance=2000
   maxSendLevel=0.25
   minSendDistance=1000
   minSendLevel=0
   soundSubmix=&Fortnite_Reverb_Distant
   customSendLevelCurve {editorCurveData {keys [<at 0 0>,<at 1 1>]}}
  }
  {
   maxSendDistance=2000
   maxSendLevel=0.6
   minSendDistance=1000
   minSendLevel=0.25
   soundSubmix=&Fortnite_Reverb_Default
   customSendLevelCurve {editorCurveData {keys [<at 0 0>,<at 1 1>]}}
  }
 ]
}
