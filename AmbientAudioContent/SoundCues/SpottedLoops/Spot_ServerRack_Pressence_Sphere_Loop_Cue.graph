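# Looping ambient cue for server rack presence, a spatialized spot sound in the Athena ambient sound class, played pitched down.
# Nested mixers blend three looping waves (scanning, rack hum, console hum) into a quality node that uses the same mix at every level.
# An enveloper at the root slowly swells and relaxes the volume on an endless loop.
# Overrides add HRTF, occlusion, focus and one voice per owner, with distance-based sends to two reverb submixes.
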SoundCue /AmbientAudioContent/SoundCues/SpottedLoops/Spot_ServerRack_Pressence_Sphere_Loop_Cue
root=Enveloper_3
nodes [
 {
  id=Enveloper_3
  children [QualityLevel_0]
  properties {
   bLoop=true
   bLoopIndefinitely=true
   loopEnd=60
   pitchCurve {editorCurveData {keys [{time=0}]}}
   volumeCurve {editorCurveData {keys [{arriveTangent=-0.01, interpMode=RCIM_Cubic, time=0, value=1}, <at 30 1.1431 RCIM_Cubic>, <at 60 1>]}}
  }
 }
 {id=Mixer_0, children [WavePlayer_2,WavePlayer_0], properties {inputVolume [2,0.5]}}
 {id=Mixer_1, children [Mixer_0,WavePlayer_1], properties {inputVolume [0.8,1]}}
 {id=QualityLevel_0, children [Mixer_1,Mixer_1,Mixer_1,Mixer_1]}
 {
  id=WavePlayer_0
  properties {bLooping=true, soundWaveAssetPtr=/AmbientAudioContent/Waves/SpottedLoops/Spot_Computer_Server_Rack_Tower_Hum_Loop_01}
 }
 {
  id=WavePlayer_1
  properties {bLooping=true, soundWaveAssetPtr=/AmbientAudioContent/Waves/SpottedLoops/Spot_Console_Terminal_Communication_Hum_Loop_04}
 }
 {id=WavePlayer_2, properties {bLooping=true, soundWaveAssetPtr=/AmbientAudioContent/Waves/SpottedLoops/Spot_Server_Stack_Scanning_Loop_02}}
]
properties {
 attenuationSettings=/Game/Sounds/Helios/Environment_Actors/ConsoleTerminals/Spot_Console_Terminal_Communication_Hum_Att
 bOverrideAttenuation=true
 bOverrideConcurrency=true
 maxDistance=3000
 pitchMultiplier=0.7
 soundClassObject=&Fort_Ambient_Athena
 subtitlePriority=5000
 volumeMultiplier=0.5
 attenuationOverrides {
  bEnableFocusInterpolation=true
  bEnableOcclusion=true
  dBAttenuationAtMax=-40
  falloffDistance=2000
  focusAzimuth=15
  focusVolumeAttenuation=1.5
  nonFocusAzimuth=45
  nonFocusVolumeAttenuation=0.5
  nonSpatializedRadiusEnd=1000
  nonSpatializedRadiusStart=2500
  occlusionInterpolationTime=2.5
  occlusionLowPassFilterFrequency=750
  occlusionTraceChannel=ECC_GameTraceChannel11
  occlusionVolumeAttenuation=0.5
  spatializationAlgorithm=SPATIALIZATION_HRTF
  stereoSpread=500
  attenuationShapeExtents=<1000 90 20>
  customAttenuationCurve {editorCurveData {keys [<at 0 1 RCIM_Cubic>, {arriveTangent=-2.6702, interpMode=RCIM_Cubic, leaveTangent=-2.6702, tangentMode=RCTM_User, time=1, value=0}]}}
 }
 concurrencyOverrides {bLimitToOwner=true, maxCount=1, resolutionRule=PreventNew, voiceStealReleaseTime=2}
 soundSubmixSends [
  {
   maxSendDistance=2000
   maxSendLevel=0.25
   minSendDistance=1000
   minSendLevel=0
   soundSubmix=&Fortnite_Reverb_Distant
   customSendLevelCurve {editorCurveData {keys [<at 0 0>,<at 1 1>]}}
  }
  {
   maxSendDistance=2000
   minSendDistance=1000
   minSendLevel=0.6
   sendLevel=1.5
   soundSubmix=&Fortnite_Reverb_Default
   customSendLevelCurve {editorCurveData {keys [<at 0 0>,<at 1 1>]}}
  }
 ]
}
