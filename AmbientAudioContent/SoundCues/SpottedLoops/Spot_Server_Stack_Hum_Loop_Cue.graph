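# Looping ambient spot cue for a server stack hum, built from a NeonCity UltraSam hologram loop rather than a dedicated hum wave.
# A Modulator fixes the looping Wave Player at double pitch, and the cue plays at reduced volume in the Athena ambient sound class.
# Attenuation and concurrency overrides give it a short falloff and a single-instance limit, superseding the referenced hologram attenuation and weapon-drop concurrency assets.
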SoundCue /AmbientAudioContent/SoundCues/SpottedLoops/Spot_Server_Stack_Hum_Loop_Cue
root=Modulator_0
nodes [
 {id=Modulator_0, children [WavePlayer_2], properties {pitchMax=2, pitchMin=2, volumeMin=1}}
 {id=WavePlayer_2, properties {bLooping=true, soundWaveAssetPtr=/AmbientAudioContent/Waves/POI/Spot_NeonCity_Hologram_UltraSam_Loop}}
]
properties {
 attenuationSettings=/Game/Sounds/Asteria/POI/NeonCity/Holograms/NeonCity_Hologram_Tiny_Att
 bOverrideAttenuation=true
 bOverrideConcurrency=true
 maxDistance=700
 soundClassObject=&Fort_Ambient_Athena
 volumeMultiplier=0.226
 attenuationOverrides {falloffDistance=600, attenuationShapeExtents {x=100}}
 concurrencyOverrides {maxCount=1}
 concurrencySet [&Weapon_Drop_StopOldest]
}
