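# Indoor ambient loop cue for a shipwreck window: wind, gentle lapping waves and wood knocks, with a short fade-in.
# A shared bed mixes delayed, pitch-varied random wood impacts with lake-shore lapping shaped by nested volume and pitch envelopes.
# A quality switch pairs that bed with a random wind loop from a four-wave or two-wave pool, modulated by looping envelopes.
# Overrides set cone attenuation with listener focus and limit concurrency.
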SoundCue /AmbientAudioContent/SoundCues/SpottedLoops/Spot_Shipwreck_Window_Wind_Gentle_Waves_Wood_Int_Loop_Cue
root=Enveloper_1
nodes [
 {id=Delay_0, children [Modulator_0], properties {delayMax=6, delayMin=3}}
 {
  id=Enveloper_0
  children [Random_4]
  properties {
   bLoop=true
   bLoopIndefinitely=true
   loopEnd=30
   pitchCurve {editorCurveData {keys [{interpMode=RCIM_Cubic, time=0}, <at 15.9468 0.9181 RCIM_Cubic>, {interpMode=RCIM_Cubic, time=30}]}}
   volumeCurve {editorCurveData {keys [<at 0 1>]}}
  }
 }
 {
  id=Enveloper_1
  children [QualityLevel_0]
  properties {pitchCurve {editorCurveData {keys [{time=0}]}}, volumeCurve {editorCurveData {keys [<at 0 0 RCIM_Cubic>,<at 0.5 1 RCIM_Cubic>]}}}
 }
 {
  id=Enveloper_2
  children [Random_0]
  properties {
   bLoop=true
   bLoopIndefinitely=true
   loopEnd=30
   pitchCurve {editorCurveData {keys [{interpMode=RCIM_Cubic, time=0}, <at 15.9468 0.9181 RCIM_Cubic>, {interpMode=RCIM_Cubic, time=30}]}}
   volumeCurve {editorCurveData {keys [<at 0 1>]}}
  }
 }
 {
  id=Enveloper_3
  children [Enveloper_2]
  properties {
   bLoop=true
   bLoopIndefinitely=true
   loopEnd=60
   pitchCurve {editorCurveData {keys [{time=0}]}}
   volumeCurve {editorCurveData {keys [
      {arriveTangent=-0.01, interpMode=RCIM_Cubic, time=0, value=1}
      <at 15 0.8 RCIM_Cubic>
      {arriveTangent=0.0001, interpMode=RCIM_Cubic, leaveTangent=0.0001, tangentMode=RCTM_User, time=30, value=1}
      <at 45 0.7 RCIM_Cubic>
      <at 60 1 RCIM_Cubic>
     ]}}
  }
 }
 {
  id=Enveloper_4
  children [Enveloper_0]
  properties {
   bLoop=true
   bLoopIndefinitely=true
   loopEnd=60
   pitchCurve {editorCurveData {keys [{time=0}]}}
   volumeCurve {editorCurveData {keys [
      {arriveTangent=-0.01, interpMode=RCIM_Cubic, time=0, value=1}
      <at 15 0.8 RCIM_Cubic>
      {arriveTangent=0.0001, interpMode=RCIM_Cubic, leaveTangent=0.0001, tangentMode=RCTM_User, time=30, value=1}
      <at 45 0.7 RCIM_Cubic>
      <at 60 1 RCIM_Cubic>
     ]}}
  }
 }
 {
  id=Enveloper_6
  children [WavePlayer_7]
  properties {
   bLoop=true
   bLoopIndefinitely=true
   loopEnd=60
   pitchCurve {editorCurveData {keys [
      {interpMode=RCIM_Cubic, time=0}
      <at 15 0.85 RCIM_Cubic>
      {arriveTangent=0.0091, interpMode=RCIM_Cubic, leaveTangent=0.0091, tangentMode=RCTM_User, time=30}
      {interpMode=RCIM_Cubic, time=45}
      {time=60}
     ]}}
   volumeCurve {editorCurveData {keys [<at 0 1>]}}
  }
 }
 {
  id=Enveloper_7
  children [Enveloper_6]
  properties {
   bLoop=true
   bLoopIndefinitely=true
   loopEnd=60
   pitchCurve {editorCurveData {keys [{time=0}]}}
   volumeCurve {editorCurveData {keys [
      {arriveTangent=-0.01, interpMode=RCIM_Cubic, time=0, value=1}
      {arriveTangent=-0.0035, interpMode=RCIM_Cubic, leaveTangent=-0.0035, tangentMode=RCTM_User, time=4.9208, value=0.7335}
      {arriveTangent=0.0016, interpMode=RCIM_Cubic, leaveTangent=0.0016, tangentMode=RCTM_User, time=7.5626, value=0.9613}
      {arriveTangent=0.0008, interpMode=RCIM_Cubic, leaveTangent=0.0008, tangentMode=RCTM_User, time=11.4857, value=0.747}
      {arriveTangent=0.0016, interpMode=RCIM_Cubic, leaveTangent=0.0016, tangentMode=RCTM_User, time=13.3554, value=0.7326}
      {arriveTangent=0.0009, interpMode=RCIM_Cubic, leaveTangent=0.0009, time=15.1401, value=0.8437}
      {arriveTangent=-0.0011, interpMode=RCIM_Cubic, leaveTangent=-0.0011, tangentMode=RCTM_User, time=18.9647, value=0.7376}
      {arriveTangent=0.0002, interpMode=RCIM_Cubic, leaveTangent=0.0002, time=24.105, value=0.9798}
      {arriveTangent=0.0016, interpMode=RCIM_Cubic, leaveTangent=0.0016, tangentMode=RCTM_User, time=28.412, value=0.7398}
      {arriveTangent=-0.0003, interpMode=RCIM_Cubic, leaveTangent=-0.0003, time=31.3775, value=0.8584}
      {arriveTangent=0.0009, interpMode=RCIM_Cubic, leaveTangent=0.0009, tangentMode=RCTM_User, time=35.3006, value=0.7378}
      {arriveTangent=-0.0008, interpMode=RCIM_Cubic, leaveTangent=-0.0008, time=38.0791, value=0.9798}
      <at 42.3861 0.7324>
      {arriveTangent=0.0024, interpMode=RCIM_Cubic, leaveTangent=0.0024, tangentMode=RCTM_User, time=44.9447, value=0.7376}
      {arriveTangent=-0.0006, interpMode=RCIM_Cubic, leaveTangent=-0.0006, time=48.1071, value=0.9613}
      {arriveTangent=-0.0056, interpMode=RCIM_Cubic, leaveTangent=-0.0056, tangentMode=RCTM_User, time=53.5122, value=0.7326}
      <at 55.4914 0.8685>
      {arriveTangent=-0.0003, interpMode=RCIM_Cubic, leaveTangent=-0.0003, tangentMode=RCTM_User, time=58.1315, value=0.7323}
      <at 60 1 RCIM_Cubic>
     ]}}
  }
 }
 {
  id=Enveloper_8
  children [Enveloper_7]
  properties {
   bLoop=true
   bLoopIndefinitely=true
   loopEnd=20
   pitchCurve {editorCurveData {keys [{time=0}]}}
   volumeCurve {editorCurveData {keys [
      {arriveTangent=-0.01, interpMode=RCIM_Cubic, time=0, value=1}
      <at 10 1.2 RCIM_Cubic>
      {interpMode=RCIM_Cubic, leaveTangent=0.01, time=20, value=1}
     ]}}
  }
 }
 {id=Looping_0, children [Delay_0]}
 {id=Mixer_0, children [Enveloper_4,Mixer_4], properties {inputVolume [0.9,1]}}
 {id=Mixer_1, children [Mixer_4,Enveloper_3], properties {inputVolume [1,1]}}
 {id=Mixer_4, children [Looping_0,Enveloper_8], properties {inputVolume [0.3,0.4]}}
 {id=Mixer_5, children [WavePlayer_14], properties {inputVolume [0.5]}}
 {id=Modulator_0, children [Random_3], properties {pitchMax=1.1, pitchMin=1}}
 {id=QualityLevel_0, children [Mixer_0,Mixer_0,Mixer_1,Mixer_1]}
 {id=Random_0, children [WavePlayer_6, WavePlayer_18], properties {weights [1,1]}}
 {
  id=Random_3
  children [WavePlayer_21, WavePlayer_22, WavePlayer_23, WavePlayer_24, WavePlayer_25, WavePlayer_26]
  properties {weights [1,1,1,1,1,1]}
 }
 {id=Random_4, children [Mixer_5, WavePlayer_15, WavePlayer_16, WavePlayer_17], properties {weights [1,1,1,1]}}
 {
  id=WavePlayer_14
  properties {bLooping=true, soundWaveAssetPtr=/AmbientAudioContent/Waves/SpottedLoops/Spot_Mountain_Wind_Window_Loop_Stereo}
 }
 {id=WavePlayer_15, properties {bLooping=true, soundWaveAssetPtr=/AmbientAudioContent/Waves/SpottedLoops/Spot_Olympus_Window_Wind_01}}
 {id=WavePlayer_16, properties {bLooping=true, soundWaveAssetPtr=/AmbientAudioContent/Waves/SpottedLoops/Spot_Olympus_Window_Wind_02}}
 {id=WavePlayer_17, properties {bLooping=true, soundWaveAssetPtr=/AmbientAudioContent/Waves/SpottedLoops/Spot_Olympus_Window_Wind_03}}
 {id=WavePlayer_18, properties {bLooping=true, soundWaveAssetPtr=/AmbientAudioContent/Waves/SpottedLoops/Spot_Winterfest_Exterior_Wind_Loop}}
 {id=WavePlayer_21, properties {soundWaveAssetPtr=/AmbientAudioContent/Waves/Water/Spot_Water_Pontoon_Wood_Impact_01}}
 {id=WavePlayer_22, properties {soundWaveAssetPtr=/AmbientAudioContent/Waves/Water/Spot_Water_Pontoon_Wood_Impact_02}}
 {id=WavePlayer_23, properties {soundWaveAssetPtr=/AmbientAudioContent/Waves/Water/Spot_Water_Pontoon_Wood_Impact_03}}
 {id=WavePlayer_24, properties {soundWaveAssetPtr=/AmbientAudioContent/Waves/Water/Spot_Water_Pontoon_Wood_Impact_04}}
 {id=WavePlayer_25, properties {soundWaveAssetPtr=/AmbientAudioContent/Waves/Water/Spot_Water_Pontoon_Wood_Impact_05}}
 {id=WavePlayer_26, properties {soundWaveAssetPtr=/AmbientAudioContent/Waves/Water/Spot_Water_Pontoon_Wood_Impact_06}}
 {id=WavePlayer_6, properties {bLooping=true, soundWaveAssetPtr=&Spot_Mountain_Wind_Window_Loop_Stereo}}
 {id=WavePlayer_7, properties {bLooping=true, soundWaveAssetPtr=/AmbientAudioContent/Waves/Water/Spot_Water_Lake_Shore_Lapping_02}}
]
properties {
 attenuationSettings=/Game/Sounds/Artemis/POI/LogJam/Spots/Spot_Mountain_Window_Wind_Sphere_Att
 bOverrideAttenuation=true
 bOverrideConcurrency=true
 maxDistance=312.5335
 soundClassObject=&Fort_Ambient_Indoors_Athena
 volumeMultiplier=0.401
 attenuationOverrides {
  attenuationShape=Cone
  bEnableListenerFocus=true
  falloffDistance=112.5335
  nonFocusVolumeAttenuation=0.75
  nonSpatializedRadiusEnd=100
  nonSpatializedRadiusStart=400
  stereoSpread=100
  attenuationShapeExtents=<200 90>
 }
 concurrencyOverrides {maxCount=2, voiceStealReleaseTime=2}
}
