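# Looping indoor ambience cue for a shipwreck window, layering wind, water slapping and wood impacts, with a cone attenuation override and its own voice limit.
# After a short fade-in, a quality switch picks a five-option random wind bed for the first two levels or a two-option bed for the others.
# Looping envelopes swell the wind and water beds.
# Both branches share the water bed and randomly delayed, pitch-varied wood impacts.
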SoundCue /AmbientAudioContent/SoundCues/SpottedLoops/Spot_Shipwreck_Window_Wind_Waves_Wood_Int_Loop_Cue
root=Enveloper_1
nodes [
 {id=Delay_0, children [Modulator_0], properties {delayMax=6, delayMin=3}}
 {
  id=Enveloper_0
  children [Random_1]
  properties {
   bLoop=true
   bLoopIndefinitely=true
   loopEnd=30
   pitchCurve {editorCurveData {keys [{interpMode=RCIM_Cubic, time=0}, <at 15.9468 0.9181 RCIM_Cubic>, {interpMode=RCIM_Cubic, time=30}]}}
   volumeCurve {editorCurveData {keys [<at 0 1>]}}
  }
 }
 {
  id=Enveloper_1
  children [QualityLevel_0]
  properties {pitchCurve {editorCurveData {keys [{time=0}]}}, volumeCurve {editorCurveData {keys [<at 0 0 RCIM_Cubic>,<at 0.5 1 RCIM_Cubic>]}}}
 }
 {
  id=Enveloper_2
  children [Random_0]
  properties {
   bLoop=true
   bLoopIndefinitely=true
   loopEnd=30
   pitchCurve {editorCurveData {keys [{interpMode=RCIM_Cubic, time=0}, <at 15.9468 0.9181 RCIM_Cubic>, {interpMode=RCIM_Cubic, time=30}]}}
   volumeCurve {editorCurveData {keys [<at 0 1>]}}
  }
 }
 {
  id=Enveloper_3
  children [Enveloper_2]
  properties {
   bLoop=true
   bLoopIndefinitely=true
   loopEnd=60
   pitchCurve {editorCurveData {keys [{time=0}]}}
   volumeCurve {editorCurveData {keys [
      {arriveTangent=-0.01, interpMode=RCIM_Cubic, time=0, value=1}
      <at 15 0.8 RCIM_Cubic>
      {arriveTangent=0.0001, interpMode=RCIM_Cubic, leaveTangent=0.0001, tangentMode=RCTM_User, time=30, value=1}
      <at 45 0.7 RCIM_Cubic>
      <at 60 1 RCIM_Cubic>
     ]}}
  }
 }
 {
  id=Enveloper_4
  children [Enveloper_0]
  properties {
   bLoop=true
   bLoopIndefinitely=true
   loopEnd=60
   pitchCurve {editorCurveData {keys [{time=0}]}}
   volumeCurve {editorCurveData {keys [
      {arriveTangent=-0.01, interpMode=RCIM_Cubic, time=0, value=1}
      <at 15 0.8 RCIM_Cubic>
      {arriveTangent=0.0001, interpMode=RCIM_Cubic, leaveTangent=0.0001, tangentMode=RCTM_User, time=30, value=1}
      <at 45 0.7 RCIM_Cubic>
      <at 60 1 RCIM_Cubic>
     ]}}
  }
 }
 {
  id=Enveloper_5
  children [WavePlayer_14]
  properties {
   bLoop=true
   bLoopIndefinitely=true
   loopEnd=10
   pitchCurve {editorCurveData {keys [{time=0}]}}
   volumeCurve {editorCurveData {keys [
      <at 0 1 RCIM_Cubic>
      <at 5 0.25 RCIM_Cubic>
      {arriveTangent=0.011, interpMode=RCIM_Cubic, leaveTangent=0.011, tangentMode=RCTM_User, time=10, value=1}
     ]}}
  }
 }
 {id=Looping_0, children [Delay_0]}
 {id=Mixer_0, children [Enveloper_4,Mixer_4], properties {inputVolume [1,1]}}
 {id=Mixer_1, children [Mixer_4,Enveloper_3], properties {inputVolume [1,1]}}
 {id=Mixer_2, children [WavePlayer_6], properties {inputVolume [0.5]}}
 {id=Mixer_3, children [WavePlayer_1], properties {inputVolume [0.5]}}
 {id=Mixer_4, children [Looping_0,Enveloper_5], properties {inputVolume [0.4,0.8]}}
 {id=Modulator_0, children [Random_3], properties {pitchMax=1.1, pitchMin=0.9}}
 {id=QualityLevel_0, children [Mixer_0,Mixer_0,Mixer_1,Mixer_1]}
 {id=Random_0, children [WavePlayer_4,WavePlayer_0], properties {weights [1,1]}}
 {id=Random_1, children [Mixer_3,WavePlayer_2,WavePlayer_3,WavePlayer_5,Mixer_2], properties {weights [1,1,1,1,1]}}
 {
  id=Random_3
  children [WavePlayer_7, WavePlayer_15, WavePlayer_16, WavePlayer_17, WavePlayer_18, WavePlayer_19]
  properties {weights [1,1,1,1,1,1]}
 }
 {id=WavePlayer_0, properties {bLooping=true, soundWaveAssetPtr=/AmbientAudioContent/Waves/SpottedLoops/Spot_Winterfest_Exterior_Wind_Loop}}
 {id=WavePlayer_1, properties {bLooping=true, soundWaveAssetPtr=/AmbientAudioContent/Waves/SpottedLoops/Spot_Mountain_Wind_Window_Loop_Stereo}}
 {
  id=WavePlayer_14
  properties {bLooping=true, soundWaveAssetPtr=/AmbientAudioContent/Waves/SpottedLoops/Spot_Water_Slapping_Against_Pontoon_Loop}
 }
 {id=WavePlayer_15, properties {soundWaveAssetPtr=/AmbientAudioContent/Waves/Water/Spot_Water_Pontoon_Wood_Impact_02}}
 {id=WavePlayer_16, properties {soundWaveAssetPtr=/AmbientAudioContent/Waves/Water/Spot_Water_Pontoon_Wood_Impact_03}}
 {id=WavePlayer_17, properties {soundWaveAssetPtr=/AmbientAudioContent/Waves/Water/Spot_Water_Pontoon_Wood_Impact_04}}
 {id=WavePlayer_18, properties {soundWaveAssetPtr=/AmbientAudioContent/Waves/Water/Spot_Water_Pontoon_Wood_Impact_05}}
 {id=WavePlayer_19, properties {soundWaveAssetPtr=/AmbientAudioContent/Waves/Water/Spot_Water_Pontoon_Wood_Impact_06}}
 {id=WavePlayer_2, properties {bLooping=true, soundWaveAssetPtr=/AmbientAudioContent/Waves/SpottedLoops/Spot_Olympus_Window_Wind_01}}
 {id=WavePlayer_3, properties {bLooping=true, soundWaveAssetPtr=/AmbientAudioContent/Waves/SpottedLoops/Spot_Olympus_Window_Wind_02}}
 {id=WavePlayer_4, properties {bLooping=true, soundWaveAssetPtr=&Spot_Mountain_Wind_Window_Loop_Stereo}}
 {id=WavePlayer_5, properties {bLooping=true, soundWaveAssetPtr=/AmbientAudioContent/Waves/SpottedLoops/Spot_Olympus_Window_Wind_03}}
 {id=WavePlayer_6, properties {bLooping=true, soundWaveAssetPtr=&Spot_Winterfest_Exterior_Wind_Loop}}
 {id=WavePlayer_7, properties {soundWaveAssetPtr=/AmbientAudioContent/Waves/Water/Spot_Water_Pontoon_Wood_Impact_01}}
]
properties {
 attenuationSettings=/Game/Sounds/Artemis/POI/LogJam/Spots/Spot_Mountain_Window_Wind_Sphere_Att
 bOverrideAttenuation=true
 bOverrideConcurrency=true
 maxDistance=312.5335
 soundClassObject=&Fort_Ambient_Indoors_Athena
 volumeMultiplier=0.401
 attenuationOverrides {
  attenuationShape=Cone
  bEnableListenerFocus=true
  falloffDistance=112.5335
  nonFocusVolumeAttenuation=0.75
  nonSpatializedRadiusEnd=100
  nonSpatializedRadiusStart=400
  stereoSpread=100
  attenuationShapeExtents=<200 90>
 }
 concurrencyOverrides {maxCount=2, voiceStealReleaseTime=2}
}
