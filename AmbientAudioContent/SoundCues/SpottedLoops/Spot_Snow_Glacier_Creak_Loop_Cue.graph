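# Looping spot ambience cue of glacier creaks for snowy outdoor areas, capped at two concurrent instances.
# Each loop waits a random delay, then a quality-level node feeds a mixer on the first two levels and nothing on the other two.
# The mixer layers a random ice lake one-shot with a quieter, pitched-down close ice pressure detail.
# Each layer has its own pitch and volume modulation and attenuation asset.
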SoundCue /AmbientAudioContent/SoundCues/SpottedLoops/Spot_Snow_Glacier_Creak_Loop_Cue
root=Looping_0
nodes [
 {
  id=Attenuation_0
  children [Random_2]
  properties {attenuationSettings=/Game/Sounds/Asteria/Biomes/RealityWarrior/Spots/Spot_Glacier_Creak_Capsule_Linear_Att}
 }
 {
  id=Attenuation_1
  children [Random_3]
  properties {attenuationSettings=/Game/Sounds/Asteria/Biomes/RealityWarrior/Spots/Spot_Glacier_Creak_Capsule_Linear_Details_Att}
 }
 {id=Delay_0, children [QualityLevel_0], properties {delayMax=25, delayMin=12}}
 {id=Looping_0, children [Delay_0]}
 {id=Mixer_0, children [Modulator_1,Modulator_0], properties {inputVolume [1,0.5]}}
 {id=Modulator_0, children [Attenuation_1], properties {pitchMax=0.5, pitchMin=0.4, volumeMax=0.5, volumeMin=0.4}}
 {id=Modulator_1, children [Attenuation_0], properties {pitchMax=1.1, pitchMin=0.9, volumeMin=0.85}}
 {id=QualityLevel_0, children [Mixer_0,Mixer_0,null,null]}
 {
  id=Random_2
  children [WavePlayer_15, WavePlayer_16, WavePlayer_17, WavePlayer_18, WavePlayer_19, WavePlayer_20, WavePlayer_21]
  properties {weights [1,1,1,1,1,1,1]}
 }
 {
  id=Random_3
  children [WavePlayer_22, WavePlayer_23, WavePlayer_24, WavePlayer_25, WavePlayer_26, WavePlayer_27, WavePlayer_28, WavePlayer_29]
  properties {weights [1,1,1,1,1,1,1,1]}
 }
 {id=WavePlayer_15, properties {soundWaveAssetPtr=/AmbientAudioContent/Waves/OneShots/Spot_Snow_Ice_Lake_01}}
 {id=WavePlayer_16, properties {soundWaveAssetPtr=/AmbientAudioContent/Waves/OneShots/Spot_Snow_Ice_Lake_02}}
 {id=WavePlayer_17, properties {soundWaveAssetPtr=/AmbientAudioContent/Waves/OneShots/Spot_Snow_Ice_Lake_03}}
 {id=WavePlayer_18, properties {soundWaveAssetPtr=/AmbientAudioContent/Waves/OneShots/Spot_Snow_Ice_Lake_04}}
 {id=WavePlayer_19, properties {soundWaveAssetPtr=/AmbientAudioContent/Waves/OneShots/Spot_Snow_Ice_Lake_05}}
 {id=WavePlayer_20, properties {soundWaveAssetPtr=/AmbientAudioContent/Waves/OneShots/Spot_Snow_Ice_Lake_06}}
 {id=WavePlayer_21, properties {soundWaveAssetPtr=/AmbientAudioContent/Waves/OneShots/Spot_Snow_Ice_Lake_07}}
 {id=WavePlayer_22, properties {soundWaveAssetPtr=/AmbientAudioContent/Waves/OneShots/Spot_Ice_Pressure_Close_01}}
 {id=WavePlayer_23, properties {soundWaveAssetPtr=/AmbientAudioContent/Waves/OneShots/Spot_Ice_Pressure_Close_02}}
 {id=WavePlayer_24, properties {soundWaveAssetPtr=/AmbientAudioContent/Waves/OneShots/Spot_Ice_Pressure_Close_03}}
 {id=WavePlayer_25, properties {soundWaveAssetPtr=/AmbientAudioContent/Waves/OneShots/Spot_Ice_Pressure_Close_04}}
 {id=WavePlayer_26, properties {soundWaveAssetPtr=/AmbientAudioContent/Waves/OneShots/Spot_Ice_Pressure_Close_05}}
 {id=WavePlayer_27, properties {soundWaveAssetPtr=/AmbientAudioContent/Waves/OneShots/Spot_Ice_Pressure_Close_06}}
 {id=WavePlayer_28, properties {soundWaveAssetPtr=/AmbientAudioContent/Waves/OneShots/Spot_Ice_Pressure_Close_07}}
 {id=WavePlayer_29, properties {soundWaveAssetPtr=/AmbientAudioContent/Waves/OneShots/Spot_Ice_Pressure_Close_08}}
]
properties {
 attenuationSettings=&Spot_Glacier_Creak_Capsule_Linear_Att
 bOverrideConcurrency=true
 maxDistance=5000
 soundClassObject=&Fort_Ambient_Outdoors_Athena
 volumeMultiplier=0.4
 concurrencyOverrides {maxCount=2}
 preEffectBusSends [{}]
}
