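# Spotted ambient loop cue that plays randomized gusty wind-blown sand one-shots on metal at intervals.
# A looping random delay triggers a pitch and volume modulator over an equal pick between two wave pools, the smaller one slightly boosted by a mixer.
# All quality levels share this chain.
# Overrides add capsule attenuation with filtering and listener focus, a tight concurrency cap, and a send to the exterior ambient source bus.
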SoundCue /AmbientAudioContent/SoundCues/SpottedLoops/Spot_Snow_Wind_Sand_Debris_On_Metal_Loop_02_Cue
root=QualityLevel_2
nodes [
 {id=Delay_1, children [Modulator_5], properties {delayMax=6, delayMin=2}}
 {id=Looping_3, children [Delay_1]}
 {id=Mixer_0, children [Random_3], properties {inputVolume [1.1]}}
 {id=Mixer_1, children [Looping_3], properties {inputVolume [1]}}
 {id=Modulator_5, children [Random_4], properties {pitchMax=1, pitchMin=0.9, volumeMax=1.1, volumeMin=0.9}}
 {id=QualityLevel_2, children [Mixer_1,Mixer_1,Mixer_1,Mixer_1]}
 {
  id=Random_2
  children [
   WavePlayer_60
   WavePlayer_61
   WavePlayer_62
   WavePlayer_63
   WavePlayer_64
   WavePlayer_65
   WavePlayer_66
   WavePlayer_67
   WavePlayer_68
   WavePlayer_69
   WavePlayer_70
   WavePlayer_71
   WavePlayer_72
   WavePlayer_73
   WavePlayer_74
   WavePlayer_75
   WavePlayer_76
   WavePlayer_83
   WavePlayer_84
   WavePlayer_85
   WavePlayer_86
   WavePlayer_87
   WavePlayer_88
   WavePlayer_89
  ]
  properties {weights [1,1,1,1,1,1,1,1,1,1,1,1,1,1,1,1,1,1,1,1,1,1,1,1]}
 }
 {
  id=Random_3
  children [
   WavePlayer_90
   WavePlayer_91
   WavePlayer_92
   WavePlayer_93
   WavePlayer_94
   WavePlayer_95
   WavePlayer_96
   WavePlayer_97
   WavePlayer_98
   WavePlayer_99
  ]
  properties {weights [1,1,1,1,1,1,1,1,1,1]}
 }
 {id=Random_4, children [Random_2,Mixer_0], properties {weights [1,1]}}
 {id=WavePlayer_60, properties {soundWaveAssetPtr=/AmbientAudioContent/Waves/OneShots/Spot_Wind_Sand_On_Metal_Gusty_01}}
 {id=WavePlayer_61, properties {soundWaveAssetPtr=/AmbientAudioContent/Waves/OneShots/Spot_Wind_Sand_On_Metal_Gusty_02}}
 {id=WavePlayer_62, properties {soundWaveAssetPtr=/AmbientAudioContent/Waves/OneShots/Spot_Wind_Sand_On_Metal_Gusty_03}}
 {id=WavePlayer_63, properties {soundWaveAssetPtr=/AmbientAudioContent/Waves/OneShots/Spot_Wind_Sand_On_Metal_Gusty_04}}
 {id=WavePlayer_64, properties {soundWaveAssetPtr=/AmbientAudioContent/Waves/OneShots/Spot_Wind_Sand_On_Metal_Gusty_05}}
 {id=WavePlayer_65, properties {soundWaveAssetPtr=/AmbientAudioContent/Waves/OneShots/Spot_Wind_Sand_On_Metal_Gusty_06}}
 {id=WavePlayer_66, properties {soundWaveAssetPtr=/AmbientAudioContent/Waves/OneShots/Spot_Wind_Sand_On_Metal_Gusty_07}}
 {id=WavePlayer_67, properties {soundWaveAssetPtr=/AmbientAudioContent/Waves/OneShots/Spot_Wind_Sand_On_Metal_Gusty_08}}
 {id=WavePlayer_68, properties {soundWaveAssetPtr=/AmbientAudioContent/Waves/OneShots/Spot_Wind_Sand_On_Metal_Gusty_09}}
 {id=WavePlayer_69, properties {soundWaveAssetPtr=/AmbientAudioContent/Waves/OneShots/Spot_Wind_Sand_On_Metal_Gusty_10}}
 {id=WavePlayer_70, properties {soundWaveAssetPtr=/AmbientAudioContent/Waves/OneShots/Spot_Wind_Sand_On_Metal_Gusty_11}}
 {id=WavePlayer_71, properties {soundWaveAssetPtr=/AmbientAudioContent/Waves/OneShots/Spot_Wind_Sand_On_Metal_Gusty_12}}
 {id=WavePlayer_72, properties {soundWaveAssetPtr=/AmbientAudioContent/Waves/OneShots/Spot_Wind_Sand_On_Metal_Gusty_13}}
 {id=WavePlayer_73, properties {soundWaveAssetPtr=/AmbientAudioContent/Waves/OneShots/Spot_Wind_Sand_On_Metal_Gusty_14}}
 {id=WavePlayer_74, properties {soundWaveAssetPtr=/AmbientAudioContent/Waves/OneShots/Spot_Wind_Sand_On_Metal_Gusty_15}}
 {id=WavePlayer_75, properties {soundWaveAssetPtr=/AmbientAudioContent/Waves/OneShots/Spot_Wind_Sand_On_Metal_Gusty_16}}
 {id=WavePlayer_76, properties {soundWaveAssetPtr=/AmbientAudioContent/Waves/OneShots/Spot_Wind_Sand_On_Metal_Gusty_17}}
 {id=WavePlayer_83, properties {soundWaveAssetPtr=/AmbientAudioContent/Waves/OneShots/Spot_Wind_Sand_On_Metal_Gusty_24}}
 {id=WavePlayer_84, properties {soundWaveAssetPtr=/AmbientAudioContent/Waves/OneShots/Spot_Wind_Sand_On_Metal_Gusty_25}}
 {id=WavePlayer_85, properties {soundWaveAssetPtr=/AmbientAudioContent/Waves/OneShots/Spot_Wind_Sand_On_Metal_Gusty_26}}
 {id=WavePlayer_86, properties {soundWaveAssetPtr=/AmbientAudioContent/Waves/OneShots/Spot_Wind_Sand_On_Metal_Gusty_27}}
 {id=WavePlayer_87, properties {soundWaveAssetPtr=/AmbientAudioContent/Waves/OneShots/Spot_Wind_Sand_On_Metal_Gusty_28}}
 {id=WavePlayer_88, properties {soundWaveAssetPtr=/AmbientAudioContent/Waves/OneShots/Spot_Wind_Sand_On_Metal_Gusty_29}}
 {id=WavePlayer_89, properties {soundWaveAssetPtr=/AmbientAudioContent/Waves/OneShots/Spot_Wind_Sand_On_Metal_Gusty_30}}
 {id=WavePlayer_90, properties {soundWaveAssetPtr=/AmbientAudioContent/Waves/OneShots/Spot_Wind_Sand_On_Metal_Gusty_31}}
 {id=WavePlayer_91, properties {soundWaveAssetPtr=/AmbientAudioContent/Waves/OneShots/Spot_Wind_Sand_On_Metal_Gusty_32}}
 {id=WavePlayer_92, properties {soundWaveAssetPtr=/AmbientAudioContent/Waves/OneShots/Spot_Wind_Sand_On_Metal_Gusty_33}}
 {id=WavePlayer_93, properties {soundWaveAssetPtr=/AmbientAudioContent/Waves/OneShots/Spot_Wind_Sand_On_Metal_Gusty_34}}
 {id=WavePlayer_94, properties {soundWaveAssetPtr=/AmbientAudioContent/Waves/OneShots/Spot_Wind_Sand_On_Metal_Gusty_35}}
 {id=WavePlayer_95, properties {soundWaveAssetPtr=/AmbientAudioContent/Waves/OneShots/Spot_Wind_Sand_On_Metal_Gusty_36}}
 {id=WavePlayer_96, properties {soundWaveAssetPtr=/AmbientAudioContent/Waves/OneShots/Spot_Wind_Sand_On_Metal_Gusty_37}}
 {id=WavePlayer_97, properties {soundWaveAssetPtr=/AmbientAudioContent/Waves/OneShots/Spot_Wind_Sand_On_Metal_Gusty_38}}
 {id=WavePlayer_98, properties {soundWaveAssetPtr=/AmbientAudioContent/Waves/OneShots/Spot_Wind_Sand_On_Metal_Gusty_39}}
 {id=WavePlayer_99, properties {soundWaveAssetPtr=/AmbientAudioContent/Waves/OneShots/Spot_Wind_Sand_On_Metal_Gusty_40}}
]
properties {
 bEnableSubmixSends=false
 bOverrideAttenuation=true
 bOverrideConcurrency=true
 maxDistance=2550
 soundClassObject=&Fort_Ambient_OneShots
 volumeMultiplier=0.8
 attenuationOverrides {
  attenuationShape=Capsule
  bAttenuateWithLPF=true
  bEnableListenerFocus=true
  falloffDistance=2000
  focusVolumeAttenuation=1.1
  hPFFrequencyAtMax=350
  hPFFrequencyAtMin=100
  lPFFrequencyAtMax=2000
  lPFRadiusMax=2000
  lPFRadiusMin=100
  nonFocusVolumeAttenuation=0.6
  stereoSpread=10
  attenuationShapeExtents=<550 100>
  customAttenuationCurve {editorCurveData {keys [
     <at 0 0 RCIM_Cubic>
     {arriveTangent=-0.0416, interpMode=RCIM_Cubic, leaveTangent=-0.0416, tangentMode=RCTM_User, time=0.0887, value=0}
     <at 0.3 1 RCIM_Cubic>
     <at 0.6894 0>
    ]}}
 }
 busSends [{
   audioBus=null
   maxSendDistance=2000
   maxSendLevel=1.2
   minSendLevel=1
   sendLevel=1.2
   soundSourceBus=&AmbientAudio_Exterior_SourceBus
  }]
 concurrencyOverrides {maxCount=2}
}
