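# Ambient one-shot cue for distant metal spark bursts, layering Indoors and Outdoors sound-class branches through one mixer at every quality level.
# Both branches draw from one random pool of twelve interior spark waves, with the outdoor branch pitched lower.
# Custom listener-focus attenuation with reverb sends keeps the indoor layer shorter-range and more focus-dependent.
# The occluded outdoor layer fades in away from the source and carries farther.
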SoundCue /AmbientAudioContent/SoundCues/SpottedLoops/Spot_Sparks_Burst_Metal_Distant_Focus_Ext_Cue
root=QualityLevel_0
nodes [
 {
  id=Attenuation_1
  children [SoundClass_1]
  properties {
   bOverrideAttenuation=true
   attenuationOverrides {
    bEnableListenerFocus=true
    bEnableOcclusion=true
    bEnableSubmixSends=true
    distanceAlgorithm=Custom
    falloffDistance=10000
    nonFocusVolumeAttenuation=0.4
    attenuationShapeExtents {x=0}
    customAttenuationCurve {editorCurveData {keys [
       {arriveTangent=3.2291, interpMode=RCIM_Cubic, leaveTangent=3.2292, tangentMode=RCTM_User, time=0, value=0}
       {arriveTangent=-0.0168, interpMode=RCIM_Cubic, leaveTangent=-0.0168, tangentMode=RCTM_User, time=0.2, value=1.0007}
       {arriveTangent=-0.02, interpMode=RCIM_Cubic, leaveTangent=-0.02, tangentMode=RCTM_User, time=0.4, value=1}
       {arriveTangent=-1.6667, interpMode=RCIM_Cubic, leaveTangent=-1.6667, time=0.598, value=0.3876}
       {arriveTangent=-1.6591, interpMode=RCIM_Cubic, leaveTangent=-1.6591, tangentMode=RCTM_User, time=1, value=0}
      ]}}
    submixSendSettings [{
      bEnableHPFCutoff=false
      bEnableLPFCutoff=false
      disableManualSendClamp=false
      hPFCutoff=20
      lPFCutoff=20000
      maxSendDistance=6000
      maxSendLevel=1
      minSendDistance=400
      minSendLevel=1
      sendLevel=1
      sendLevelControlMethod=Linear
      soundSubmix=&Fortnite_Reverb_Default
     }]
   }
  }
 }
 {
  id=Attenuation_2
  children [SoundClass_0]
  properties {
   bOverrideAttenuation=true
   attenuationOverrides {
    bEnableListenerFocus=true
    bEnableSubmixSends=true
    distanceAlgorithm=Custom
    falloffDistance=4000
    nonFocusVolumeAttenuation=0.1
    stereoSpread=3000
    attenuationShapeExtents {x=0}
    customAttenuationCurve {editorCurveData {keys [
       {arriveTangent=-0.0131, interpMode=RCIM_Cubic, leaveTangent=-0.0131, tangentMode=RCTM_User, time=0.2522, value=1.0124}
       {arriveTangent=-5.1404, interpMode=RCIM_Cubic, leaveTangent=-5.1404, tangentMode=RCTM_User, time=0.6, value=0}
      ]}}
    submixSendSettings [{
      bEnableHPFCutoff=false
      bEnableLPFCutoff=false
      disableManualSendClamp=false
      hPFCutoff=20
      lPFCutoff=20000
      maxSendDistance=6000
      maxSendLevel=1
      minSendDistance=400
      minSendLevel=1
      sendLevel=0.4
      sendLevelControlMethod=Linear
      soundSubmix=&Fortnite_Reverb_Default
     }]
   }
  }
 }
 {id=Mixer_0, children [Attenuation_2, Attenuation_1], properties {inputVolume [1,1]}}
 {id=Modulator_0, children [Random_1], properties {pitchMin=0.8}}
 {id=Modulator_1, children [Random_1], properties {pitchMax=0.8, pitchMin=0.7}}
 {id=QualityLevel_0, children [Mixer_0,Mixer_0,Mixer_0,Mixer_0]}
 {
  id=Random_1
  children [
   WavePlayer_8
   WavePlayer_9
   WavePlayer_10
   WavePlayer_11
   WavePlayer_12
   WavePlayer_13
   WavePlayer_14
   WavePlayer_15
   WavePlayer_20
   WavePlayer_21
   WavePlayer_22
   WavePlayer_23
  ]
  properties {weights [1,1,1,1,1,1,1,1,1,1,1,1]}
 }
 {id=SoundClass_0, children [Modulator_0], properties {soundClassOverride=/Game/Sounds/FortSoundClassesAndModes/Fort_Ambient_OneShots_Indoors}}
 {
  id=SoundClass_1
  children [Modulator_1]
  properties {soundClassOverride=&Fort_Ambient_OneShots_Outdoors}
 }
 {id=WavePlayer_10, properties {soundWaveAssetPtr=/AmbientAudioContent/Waves/OneShots/Spot_Spark_Burst_Metal_Int_10}}
 {id=WavePlayer_11, properties {soundWaveAssetPtr=/AmbientAudioContent/Waves/OneShots/Spot_Spark_Burst_Metal_Int_09}}
 {id=WavePlayer_12, properties {soundWaveAssetPtr=/AmbientAudioContent/Waves/OneShots/Spot_Spark_Burst_Metal_Int_08}}
 {id=WavePlayer_13, properties {soundWaveAssetPtr=/AmbientAudioContent/Waves/OneShots/Spot_Spark_Burst_Metal_Int_07}}
 {id=WavePlayer_14, properties {soundWaveAssetPtr=/AmbientAudioContent/Waves/OneShots/Spot_Spark_Burst_Metal_Int_06}}
 {id=WavePlayer_15, properties {soundWaveAssetPtr=/AmbientAudioContent/Waves/OneShots/Spot_Spark_Burst_Metal_Int_05}}
 {id=WavePlayer_20, properties {soundWaveAssetPtr=/AmbientAudioContent/Waves/OneShots/Spot_Spark_Burst_Metal_Int_04}}
 {id=WavePlayer_21, properties {soundWaveAssetPtr=/AmbientAudioContent/Waves/OneShots/Spot_Spark_Burst_Metal_Int_03}}
 {id=WavePlayer_22, properties {soundWaveAssetPtr=/AmbientAudioContent/Waves/OneShots/Spot_Spark_Burst_Metal_Int_02}}
 {id=WavePlayer_23, properties {soundWaveAssetPtr=/AmbientAudioContent/Waves/OneShots/Spot_Spark_Burst_Metal_Int_01}}
 {id=WavePlayer_8, properties {soundWaveAssetPtr=/AmbientAudioContent/Waves/OneShots/Spot_Spark_Burst_Metal_Int_12}}
 {id=WavePlayer_9, properties {soundWaveAssetPtr=/AmbientAudioContent/Waves/OneShots/Spot_Spark_Burst_Metal_Int_11}}
]
properties {
 bEnableBusSends=false
 maxDistance=10000
 soundClassObject=&Fort_Ambient_OneShots
 volumeMultiplier=0.126
 attenuationOverrides {falloffDistance=6000, stereoSpread=10000, attenuationShapeExtents {x=2000}}
 soundSubmixSends [{minSendLevel=1, soundSubmix=&Fortnite_Reverb_Default}]
}
