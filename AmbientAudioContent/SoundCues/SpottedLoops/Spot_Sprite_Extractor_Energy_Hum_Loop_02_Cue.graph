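# Looping spatialized ambient cue for a Sprite Extractor energy hum, routed to the ambient props sound class.
# Mixes low-energy and mech loops, pitched-down sizzle and slime loops, and random bubble swells retriggered after short delays.
# Nested looping Envelopers cycle pitch over 30 seconds and dip volume to half over 20 seconds, so the bed drifts.
# Only the first two quality levels are connected; attenuation overrides add HRTF and listener focus.
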SoundCue /AmbientAudioContent/SoundCues/SpottedLoops/Spot_Sprite_Extractor_Energy_Hum_Loop_02_Cue
root=QualityLevel_0
nodes [
 {id=Delay_1, children [Random_0], properties {delayMax=2, delayMin=1}}
 {
  id=Enveloper_0
  children [Mixer_0]
  properties {
   bLoop=true
   bLoopIndefinitely=true
   loopEnd=30
   pitchCurve {editorCurveData {keys [{arriveTangent=-0.0013, interpMode=RCIM_Cubic, time=0}, <at 15 0.95 RCIM_Cubic>, {interpMode=RCIM_Cubic, leaveTangent=0.0013, time=30}]}}
   volumeCurve {editorCurveData {keys [<at 0 1>]}}
  }
 }
 {
  id=Enveloper_3
  children [Enveloper_0]
  properties {
   bLoop=true
   bLoopIndefinitely=true
   loopEnd=20
   pitchCurve {editorCurveData {keys [{time=0}]}}
   volumeCurve {editorCurveData {keys [
      {arriveTangent=-0.01, interpMode=RCIM_Cubic, time=0, value=1}
      <at 10 0.5 RCIM_Cubic>
      {interpMode=RCIM_Cubic, leaveTangent=0.01, time=20, value=1}
     ]}}
  }
 }
 {id=Looping_1, children [Delay_1]}
 {id=Mixer_0, children [WavePlayer_9,Modulator_1,Looping_1,Modulator_0,WavePlayer_8], properties {inputVolume [0.5,1,0.4,0.9,1]}}
 {id=Modulator_0, children [WavePlayer_7], properties {pitchMax=0.7, pitchMin=0.7}}
 {id=Modulator_1, children [WavePlayer_2], properties {pitchMax=0.7, pitchMin=0.7}}
 {id=QualityLevel_0, children [Enveloper_3,Enveloper_3,null,null]}
 {id=Random_0, children [WavePlayer_1,WavePlayer_3,WavePlayer_4,WavePlayer_5,WavePlayer_6], properties {weights [1,1,1,1,1]}}
 {id=WavePlayer_1, properties {soundWaveAssetPtr=/AmbientAudioContent/Waves/OneShots/Spot_FishTank_Bubble_Swell_01}}
 {id=WavePlayer_2, properties {bLooping=true, soundWaveAssetPtr=/AmbientAudioContent/Waves/SpottedLoops/Spot_Energy_Sizzle_Loop}}
 {id=WavePlayer_3, properties {soundWaveAssetPtr=/AmbientAudioContent/Waves/OneShots/Spot_FishTank_Bubble_Swell_02}}
 {id=WavePlayer_4, properties {soundWaveAssetPtr=/AmbientAudioContent/Waves/OneShots/Spot_FishTank_Bubble_Swell_03}}
 {id=WavePlayer_5, properties {soundWaveAssetPtr=/AmbientAudioContent/Waves/OneShots/Spot_FishTank_Bubble_Swell_04}}
 {id=WavePlayer_6, properties {soundWaveAssetPtr=/AmbientAudioContent/Waves/OneShots/Spot_FishTank_Bubble_Swell_05}}
 {
  id=WavePlayer_7
  properties {bLooping=true, soundWaveAssetPtr=/AmbientAudioContent/Waves/SpottedLoops/Spot_Water_Toxic_Sewer_Goop_Slime_Loop}
 }
 {
  id=WavePlayer_8
  properties {bLooping=true, soundWaveAssetPtr=/AmbientAudioContent/Waves/SpottedLoops/Spot_HollowBase_Coaster_Energy_Mech_Loop1}
 }
 {
  id=WavePlayer_9
  properties {bLooping=true, soundWaveAssetPtr=/AmbientAudioContent/Waves/SpottedLoops/Spot_Zeropoint_Petrified_Low_Energy_Very_Close_Loop}
 }
]
properties {
 attenuationSettings=/Game/Sounds/Attenuation/Fort_Traps_Loop_Att
 bOverrideAttenuation=true
 maxDistance=500
 soundClassObject=&Fort_Ambient_Props_Athena
 volumeMultiplier=1
 attenuationOverrides {
  bEnableListenerFocus=true
  distanceAlgorithm=Logarithmic
  falloffDistance=250
  focusAzimuth=15
  focusVolumeAttenuation=1.2
  nonFocusAzimuth=35
  nonFocusVolumeAttenuation=0.4
  spatializationAlgorithm=SPATIALIZATION_HRTF
  stereoSpread=50
  attenuationShapeExtents=<250 100>
 }
 concurrencyOverrides {resolutionRule=PreventNew}
 concurrencySet [/AmbientAudioBase/Concurrency/AmbientConc_3Max_Spot_Default_Concurrency]
}
