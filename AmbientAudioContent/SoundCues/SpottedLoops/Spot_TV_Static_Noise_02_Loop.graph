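# Spot ambient loop SoundCue for TV static noise, played quietly in the Fort_Gadgets_Items sound class.
# A Mixer layers two looping waves: a TV static loop, pitch-randomised through a Modulator, and a CubeShard idle static loop.
# Overridden attenuation uses a directional cone with air-absorption filtering, occlusion and listener focus, superseding the referenced screen-static attenuation asset.
# It sends to the default reverb submix and uses a shared ambient spot concurrency set.
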SoundCue /AmbientAudioContent/SoundCues/SpottedLoops/Spot_TV_Static_Noise_02_Loop
root=Mixer_1
nodes [
 {id=Mixer_1, children [Modulator_1,WavePlayer_8], properties {inputVolume [1,1]}}
 {id=Modulator_1, children [WavePlayer_4], properties {pitchMax=1.2}}
 {id=WavePlayer_4, properties {bLooping=true, soundWaveAssetPtr=/AmbientAudioContent/Waves/SpottedLoops/Spot_FNM_TV_Static_loop}}
 {id=WavePlayer_8, properties {bLooping=true, soundWaveAssetPtr=/AmbientAudioContent/Waves/SpottedLoops/Spot_MASH_CubeShard_IdleStatic}}
]
properties {
 attenuationSettings=/Game/Sounds/Helios/Environment_Actors/Screens/Spot_TV_Screen_Static_Sphere_Att
 bOverrideAttenuation=true
 maxDistance=1100
 soundClassObject=&Fort_Gadgets_Items
 volumeMultiplier=0.2
 attenuationOverrides {
  attenuationShape=Cone
  bAttenuateWithLPF=true
  bEnableFocusInterpolation=true
  bEnableListenerFocus=true
  bEnableOcclusion=true
  coneSphereFalloffDistance=750
  coneSphereRadius=1
  falloffDistance=850
  focusAttackInterpSpeed=2
  focusAzimuth=15
  focusReleaseInterpSpeed=0.5
  hPFFrequencyAtMax=200
  lPFFrequencyAtMax=600
  lPFRadiusMax=1400
  lPFRadiusMin=500
  nonFocusAzimuth=45
  nonFocusVolumeAttenuation=0.4
  occlusionLowPassFilterFrequency=10000
  occlusionTraceChannel=ECC_GameTraceChannel11
  occlusionVolumeAttenuation=0.75
  stereoSpread=0
  attenuationShapeExtents=<250 90>
 }
 concurrencyOverrides {maxCount=2}
 concurrencySet [/AmbientAudioBase/Concurrency/AmbientConc_3Max_Spot_Default_Concurrency]
 soundSubmixSends [{
   minSendLevel=0.25
   soundSubmix=&Fortnite_Reverb_Default
   customSendLevelCurve {editorCurveData {keys [<at 0 0>,<at 1 1>]}}
  }]
}
