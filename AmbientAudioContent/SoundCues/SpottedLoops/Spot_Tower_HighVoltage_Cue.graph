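# Looping ambient sound cue for a spotted high-voltage tower emitter.
# A single looping wave player feeds the output directly at reduced volume, routed to the Fort_Ambient_TileMix sound class.
# An attenuation override disables spatialization, so the sound is not panned by position and fades with distance beyond an inner radius.
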SoundCue /AmbientAudioContent/SoundCues/SpottedLoops/Spot_Tower_HighVoltage_Cue
root=WavePlayer_1
nodes [{id=WavePlayer_1, properties {bLooping=true, soundWaveAssetPtr=/AmbientAudioContent/Waves/SpottedLoops/Spot_Tower_HighVoltage_01}}]
properties {
 bOverrideAttenuation=true
 maxDistance=3450.9424
 soundClassObject=/Game/Sounds/FortSoundClassesAndModes/Fort_Ambient_TileMix
 volumeMultiplier=0.5
 attenuationOverrides {bSpatialize=false, falloffDistance=1961.4901, stereoSpread=1000, attenuationShapeExtents {x=1489.4523}}
}
