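# Spotted-loop ambience cue for an idling train, in the exterior vehicles sound class.
# A looping idle wave is mixed with a quieter prototype idle wave not flagged as looping.
# Over this, a Looping node repeats randomly delayed, modulated picks from distant blimp air-release one-shots.
# Overrides add capsule attenuation with occlusion, distance filtering and listener focus, plus a concurrency limit; a send feeds the exterior ambient bus.
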SoundCue /AmbientAudioContent/SoundCues/SpottedLoops/Spot_Train_Idle_Loop_Cue
root=Mixer_1
nodes [
 {id=Delay_0, children [Modulator_0], properties {delayMax=18, delayMin=6}}
 {id=Looping_0, children [Delay_0]}
 {id=Mixer_0, children [WavePlayer_0,WavePlayer_1], properties {inputVolume [0.75,0.25]}}
 {id=Mixer_1, children [Looping_0,Mixer_0], properties {inputVolume [0.75,1]}}
 {id=Modulator_0, children [Random_0], properties {pitchMin=0.9}}
 {
  id=Random_0
  children [WavePlayer_2, WavePlayer_3, WavePlayer_4, WavePlayer_5, WavePlayer_6, WavePlayer_7, WavePlayer_8, WavePlayer_9, WavePlayer_18, WavePlayer_19]
  properties {weights [1,1,1,1,1,1,1,1,1,1]}
 }
 {id=WavePlayer_0, properties {bLooping=true, soundWaveAssetPtr=/AmbientAudioContent/Waves/SpottedLoops/Train_Movement_Idle}}
 {id=WavePlayer_1, properties {soundWaveAssetPtr=/AmbientAudioContent/Waves/SpottedLoops/Train_Proto_Movement_Idle}}
 {id=WavePlayer_18, properties {soundWaveAssetPtr=/AmbientAudioContent/Waves/OneShots/Spot_Blimp_Air_Release_Distant_09_Stereo}}
 {id=WavePlayer_19, properties {soundWaveAssetPtr=/AmbientAudioContent/Waves/OneShots/Spot_Blimp_Air_Release_Distant_10_Stereo}}
 {id=WavePlayer_2, properties {soundWaveAssetPtr=/AmbientAudioContent/Waves/OneShots/Spot_Blimp_Air_Release_Distant_01_Stereo}}
 {id=WavePlayer_3, properties {soundWaveAssetPtr=/AmbientAudioContent/Waves/OneShots/Spot_Blimp_Air_Release_Distant_02_Stereo}}
 {id=WavePlayer_4, properties {soundWaveAssetPtr=/AmbientAudioContent/Waves/OneShots/Spot_Blimp_Air_Release_Distant_03_Stereo}}
 {id=WavePlayer_5, properties {soundWaveAssetPtr=/AmbientAudioContent/Waves/OneShots/Spot_Blimp_Air_Release_Distant_04_Stereo}}
 {id=WavePlayer_6, properties {soundWaveAssetPtr=/AmbientAudioContent/Waves/OneShots/Spot_Blimp_Air_Release_Distant_05_Stereo}}
 {id=WavePlayer_7, properties {soundWaveAssetPtr=/AmbientAudioContent/Waves/OneShots/Spot_Blimp_Air_Release_Distant_06_Stereo}}
 {id=WavePlayer_8, properties {soundWaveAssetPtr=/AmbientAudioContent/Waves/OneShots/Spot_Blimp_Air_Release_Distant_07_Stereo}}
 {id=WavePlayer_9, properties {soundWaveAssetPtr=/AmbientAudioContent/Waves/OneShots/Spot_Blimp_Air_Release_Distant_08_Stereo}}
]
properties {
 bOverrideAttenuation=true
 bOverrideConcurrency=true
 maxDistance=7000
 soundClassObject=&Fort_Vehicles_Exterior
 volumeMultiplier=0.65
 attenuationOverrides {
  attenuationShape=Capsule
  bAttenuateWithLPF=true
  bEnableListenerFocus=true
  bEnableOcclusion=true
  distanceAlgorithm=Logarithmic
  falloffDistance=4000
  lPFFrequencyAtMax=2000
  lPFRadiusMin=2000
  nonFocusVolumeAttenuation=0.4
  occlusionInterpolationTime=2
  occlusionLowPassFilterFrequency=850
  occlusionTraceChannel=ECC_GameTraceChannel11
  occlusionVolumeAttenuation=0.5
  attenuationShapeExtents=<3000 2000>
 }
 busSends [{
   audioBus=null
   maxSendDistance=4000
   minSendLevel=0.6
   sendLevel=1
   soundSourceBus=&AmbientAudio_Exterior_SourceBus
  }]
 concurrencyOverrides {maxCount=2}
}
