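# Spotted ambient Sound Cue for a moving train, limited to one instance and sent to the exterior ambient bus.
# The first two quality levels mix looping wheels with randomly delayed track squeaks, distant rumbles and horns; the last two play only wheels.
# Horns fire near and distant waves together, using custom attenuation curves to crossfade by distance, with occlusion on the distant one.
# Distant rumbles likewise peak at mid range.
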SoundCue /AmbientAudioContent/SoundCues/SpottedLoops/Spot_Train_Moving_Loop_Cue
root=QualityLevel_0
nodes [
 {
  id=Attenuation_0
  children [Looping_0]
  properties {
   bOverrideAttenuation=true
   attenuationOverrides {
    bAttenuateWithLPF=true
    bEnableListenerFocus=true
    falloffDistance=6000
    lPFFrequencyAtMax=750
    lPFRadiusMax=7000
    nonFocusVolumeAttenuation=0.4
    stereoSpread=500
    attenuationShapeExtents {x=1000}
    customAttenuationCurve {editorCurveData {keys [<at 0.2 0 RCIM_Cubic>,<at 0.593 0.649 RCIM_Cubic>,<at 1 0>]}}
   }
  }
 }
 {
  id=Attenuation_1
  children [WavePlayer_0]
  properties {
   bOverrideAttenuation=true
   attenuationOverrides {
    bEnableListenerFocus=true
    bEnableOcclusion=true
    distanceAlgorithm=Custom
    falloffDistance=10000
    nonFocusVolumeAttenuation=0.75
    occlusionLowPassFilterFrequency=750
    occlusionTraceChannel=ECC_GameTraceChannel11
    occlusionVolumeAttenuation=0.5
    attenuationShapeExtents {x=0}
    customAttenuationCurve {editorCurveData {keys [<at 0.2 0 RCIM_Cubic>,<at 0.593 0.649 RCIM_Cubic>,<at 1 0>]}}
   }
  }
 }
 {
  id=Attenuation_2
  children [WavePlayer_13]
  properties {
   bOverrideAttenuation=true
   attenuationOverrides {
    bEnableListenerFocus=true
    distanceAlgorithm=Custom
    falloffDistance=10000
    nonFocusVolumeAttenuation=0.75
    attenuationShapeExtents {x=0}
    customAttenuationCurve {editorCurveData {keys [
       <at 0 1>
       {arriveTangent=-1.6863, interpMode=RCIM_Cubic, leaveTangent=-1.6863, time=0.2, value=1}
       {arriveTangent=-3.4272, interpMode=RCIM_Cubic, leaveTangent=-3.4272, tangentMode=RCTM_User, time=0.593, value=0}
      ]}}
   }
  }
 }
 {
  id=Attenuation_4
  children [WavePlayer_11]
  properties {
   bOverrideAttenuation=true
   attenuationOverrides {
    bAttenuateWithLPF=true
    bEnableListenerFocus=true
    falloffDistance=6000
    lPFFrequencyAtMax=750
    lPFRadiusMax=7000
    nonFocusVolumeAttenuation=0.4
    stereoSpread=500
    attenuationShapeExtents {x=1000}
    customAttenuationCurve {editorCurveData {keys [<at 0.2 0 RCIM_Cubic>,<at 0.593 0.649 RCIM_Cubic>,<at 1 0>]}}
   }
  }
 }
 {
  id=Attenuation_5
  children [Looping_2]
  properties {
   bOverrideAttenuation=true
   attenuationOverrides {
    bEnableListenerFocus=true
    distanceAlgorithm=Custom
    falloffDistance=10000
    nonFocusVolumeAttenuation=0.75
    attenuationShapeExtents {x=0}
    customAttenuationCurve {editorCurveData {keys [<at 0.2 0 RCIM_Cubic>,<at 0.593 0.649 RCIM_Cubic>,<at 1 0>]}}
   }
  }
 }
 {id=Delay_0, children [Modulator_0], properties {delayMax=8, delayMin=3}}
 {id=Delay_1, children [Modulator_2], properties {delayMax=60, delayMin=30}}
 {id=Delay_2, children [Modulator_3], properties {delayMax=8, delayMin=3}}
 {id=Looping_0, children [Delay_0]}
 {id=Looping_1, children [Delay_1]}
 {id=Looping_2, children [Delay_2]}
 {id=Mixer_0, children [Attenuation_0, Attenuation_4, Looping_1, Attenuation_5], properties {inputVolume [0.8,2,0.25,1]}}
 {id=Mixer_1, children [Attenuation_1, Attenuation_2], properties {inputVolume [1,1]}}
 {id=Modulator_0, children [Random_0]}
 {id=Modulator_2, children [Mixer_1]}
 {id=Modulator_3, children [Random_3]}
 {id=QualityLevel_0, children [Mixer_0, Mixer_0, WavePlayer_24, WavePlayer_24]}
 {
  id=Random_0
  children [WavePlayer_1, WavePlayer_2, WavePlayer_3, WavePlayer_4, WavePlayer_5, WavePlayer_6, WavePlayer_7, WavePlayer_8, WavePlayer_9, WavePlayer_10]
  properties {weights [1,1,1,1,1,1,1,1,1,1]}
 }
 {
  id=Random_3
  children [WavePlayer_16, WavePlayer_17, WavePlayer_18, WavePlayer_19, WavePlayer_20, WavePlayer_21, WavePlayer_22, WavePlayer_23]
  properties {weights [1,1,1,1,1,1,1,1]}
 }
 {id=WavePlayer_0, properties {soundWaveAssetPtr=/AmbientAudioContent/Waves/OneShots/Train_Horn_Distant_Japan}}
 {id=WavePlayer_1, properties {soundWaveAssetPtr=/AmbientAudioContent/Waves/OneShots/Train_TrackSqueak_02}}
 {id=WavePlayer_10, properties {soundWaveAssetPtr=/AmbientAudioContent/Waves/OneShots/Train_TrackSqueak_11}}
 {id=WavePlayer_11, properties {bLooping=true, soundWaveAssetPtr=/AmbientAudioContent/Waves/SpottedLoops/Train_Proto_Movement_Wheels_09}}
 {id=WavePlayer_13, properties {soundWaveAssetPtr=/AmbientAudioContent/Waves/OneShots/Train_Horn_Japan}}
 {id=WavePlayer_16, properties {soundWaveAssetPtr=/AmbientAudioContent/Waves/OneShots/Train_Distant_Movement_01}}
 {id=WavePlayer_17, properties {soundWaveAssetPtr=/AmbientAudioContent/Waves/OneShots/Train_Distant_Movement_02}}
 {id=WavePlayer_18, properties {soundWaveAssetPtr=/AmbientAudioContent/Waves/OneShots/Train_Distant_Movement_03}}
 {id=WavePlayer_19, properties {soundWaveAssetPtr=/AmbientAudioContent/Waves/OneShots/Train_Distant_Movement_04}}
 {id=WavePlayer_2, properties {soundWaveAssetPtr=/AmbientAudioContent/Waves/OneShots/Train_TrackSqueak_03}}
 {id=WavePlayer_20, properties {soundWaveAssetPtr=/AmbientAudioContent/Waves/OneShots/Train_Distant_Movement_05}}
 {id=WavePlayer_21, properties {soundWaveAssetPtr=/AmbientAudioContent/Waves/OneShots/Train_Distant_Movement_06}}
 {id=WavePlayer_22, properties {soundWaveAssetPtr=/AmbientAudioContent/Waves/OneShots/Train_Distant_Movement_07}}
 {id=WavePlayer_23, properties {soundWaveAssetPtr=/AmbientAudioContent/Waves/OneShots/Train_Distant_Movement_08}}
 {id=WavePlayer_24, properties {bLooping=true, soundWaveAssetPtr=&Train_Proto_Movement_Wheels_09}}
 {id=WavePlayer_3, properties {soundWaveAssetPtr=/AmbientAudioContent/Waves/OneShots/Train_TrackSqueak_04}}
 {id=WavePlayer_4, properties {soundWaveAssetPtr=/AmbientAudioContent/Waves/OneShots/Train_TrackSqueak_05}}
 {id=WavePlayer_5, properties {soundWaveAssetPtr=/AmbientAudioContent/Waves/OneShots/Train_TrackSqueak_06}}
 {id=WavePlayer_6, properties {soundWaveAssetPtr=/AmbientAudioContent/Waves/OneShots/Train_TrackSqueak_07}}
 {id=WavePlayer_7, properties {soundWaveAssetPtr=/AmbientAudioContent/Waves/OneShots/Train_TrackSqueak_08}}
 {id=WavePlayer_8, properties {soundWaveAssetPtr=/AmbientAudioContent/Waves/OneShots/Train_TrackSqueak_09}}
 {id=WavePlayer_9, properties {soundWaveAssetPtr=/AmbientAudioContent/Waves/OneShots/Train_TrackSqueak_10}}
]
properties {
 bOverrideConcurrency=true
 maxDistance=10000
 soundClassObject=&Fort_Vehicles_Exterior
 volumeMultiplier=1.2
 attenuationOverrides {
  bAttenuateWithLPF=true
  bEnableListenerFocus=true
  falloffDistance=6000
  lPFFrequencyAtMax=750
  nonFocusVolumeAttenuation=0.4
  stereoSpread=500
  attenuationShapeExtents {x=1000}
 }
 busSends [{
   audioBus=null
   maxSendDistance=6000
   minSendDistance=1000
   minSendLevel=0.6
   sendLevel=1
   soundSourceBus=&AmbientAudio_Exterior_SourceBus
   customSendLevelCurve {editorCurveData {keys [<at 0 0.6>,<at 1 1>]}}
  }]
 concurrencyOverrides {maxCount=1}
}
