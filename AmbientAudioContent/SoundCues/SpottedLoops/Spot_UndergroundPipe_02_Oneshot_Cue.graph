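# Sound cue for a one-shot underground pipe rattle, an ambient spot sound in the indoor ambience sound class.
# A Random node picks one of five equally weighted rattle waves, and a Modulator lowers its pitch by a varying amount.
# All four quality-level inputs share that same chain.
# It overrides concurrency to limit simultaneous plays, uses distant-creature one-shot attenuation, and sends to the exterior ambient source bus.
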SoundCue /AmbientAudioContent/SoundCues/SpottedLoops/Spot_UndergroundPipe_02_Oneshot_Cue
root=QualityLevel_0
nodes [
 {id=Modulator_0, children [Random_0], properties {pitchMax=0.9, pitchMin=0.8}}
 {id=QualityLevel_0, children [Modulator_0,Modulator_0,Modulator_0,Modulator_0]}
 {id=Random_0, children [WavePlayer_1,WavePlayer_3,WavePlayer_4,WavePlayer_8,WavePlayer_9], properties {weights [1,1,1,1,1]}}
 {id=WavePlayer_1, properties {soundWaveAssetPtr=/AmbientAudioContent/Waves/OneShots/Spot_UndergroundPipe_Rattle_01}}
 {id=WavePlayer_3, properties {soundWaveAssetPtr=/AmbientAudioContent/Waves/OneShots/Spot_UndergroundPipe_Rattle_03}}
 {id=WavePlayer_4, properties {soundWaveAssetPtr=/AmbientAudioContent/Waves/OneShots/Spot_UndergroundPipe_Rattle_04}}
 {id=WavePlayer_8, properties {soundWaveAssetPtr=/AmbientAudioContent/Waves/OneShots/Spot_UndergroundPipe_Rattle_05}}
 {id=WavePlayer_9, properties {soundWaveAssetPtr=/AmbientAudioContent/Waves/OneShots/Spot_UndergroundPipe_Rattle_06}}
]
properties {
 attenuationSettings=/AmbientAudioBase/Attenuations/Fort_OneShot_Distant_Creature_Att
 bOverrideConcurrency=true
 maxDistance=8500
 soundClassObject=&Fort_Ambient_Indoors_Athena
 volumeMultiplier=0.752
 attenuationOverrides {falloffDistance=750, nonSpatializedRadiusEnd=125, nonSpatializedRadiusStart=150, stereoSpread=150, attenuationShapeExtents {x=125}}
 busSends [{audioBus=null, minSendLevel=0.6, sendLevel=1, soundSourceBus=&AmbientAudio_Exterior_SourceBus}]
 concurrencyOverrides {maxCount=2}
 concurrencySet [/AmbientAudioBase/Concurrency/AmbientConc_3Max_Spot_Default_Concurrency]
}
