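# Looping ambient spot cue that plays Japanese train station PA (tannoy) announcements at intervals.
# A Looping node repeats a random 6 to 18 second delay, followed by a Modulator and an equal-weight pick from nine one-shot waves.
# It uses the Athena ambient sound class, with overridden attenuation adding occlusion filtering, a two-voice concurrency cap and a distance-based send to an ambient reflection source bus.
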SoundCue /AmbientAudioContent/SoundCues/SpottedLoops/Spot_VO_Trainstation_Tannoy_Loop_Cue
root=Looping_0
nodes [
 {id=Delay_0, children [Modulator_0], properties {delayMax=18, delayMin=6}}
 {id=Looping_0, children [Delay_0]}
 {id=Modulator_0, children [Random_1]}
 {
  id=Random_1
  children [WavePlayer_13, WavePlayer_14, WavePlayer_15, WavePlayer_16, WavePlayer_17, WavePlayer_18, WavePlayer_19, WavePlayer_20, WavePlayer_21]
  properties {weights [1,1,1,1,1,1,1,1,1]}
 }
 {id=WavePlayer_13, properties {soundWaveAssetPtr=/AmbientAudioContent/Waves/OneShots/Spot_Japan_Train_Station_PA_01}}
 {id=WavePlayer_14, properties {soundWaveAssetPtr=/AmbientAudioContent/Waves/OneShots/Spot_Japan_Train_Station_PA_02}}
 {id=WavePlayer_15, properties {soundWaveAssetPtr=/AmbientAudioContent/Waves/OneShots/Spot_Japan_Train_Station_PA_03}}
 {id=WavePlayer_16, properties {soundWaveAssetPtr=/AmbientAudioContent/Waves/OneShots/Spot_Japan_Train_Station_PA_04}}
 {id=WavePlayer_17, properties {soundWaveAssetPtr=/AmbientAudioContent/Waves/OneShots/Spot_Japan_Train_Station_PA_05}}
 {id=WavePlayer_18, properties {soundWaveAssetPtr=/AmbientAudioContent/Waves/OneShots/Spot_Japan_Train_Station_PA_06}}
 {id=WavePlayer_19, properties {soundWaveAssetPtr=/AmbientAudioContent/Waves/OneShots/Spot_Japan_Train_Station_PA_07}}
 {id=WavePlayer_20, properties {soundWaveAssetPtr=/AmbientAudioContent/Waves/OneShots/Spot_Japan_Train_Station_PA_08}}
 {id=WavePlayer_21, properties {soundWaveAssetPtr=/AmbientAudioContent/Waves/OneShots/Spot_Japan_Train_Station_PA_09}}
]
properties {
 attenuationSettings=/Game/Sounds/Attenuation/Fort_TileAmbient_OneShot_Att
 bEnableSubmixSends=false
 bOverrideAttenuation=true
 bOverrideConcurrency=true
 maxDistance=7000
 soundClassObject=&Fort_Ambient_Athena
 volumeMultiplier=0.6
 attenuationOverrides {
  bEnableOcclusion=true
  falloffDistance=5000
  nonSpatializedRadiusEnd=2000
  nonSpatializedRadiusStart=3000
  occlusionInterpolationTime=2
  occlusionLowPassFilterFrequency=1000
  occlusionTraceChannel=ECC_GameTraceChannel11
  occlusionVolumeAttenuation=0.5
  stereoSpread=500
  attenuationShapeExtents {x=2000}
 }
 busSends [{
   audioBus=null
   maxSendDistance=3500
   minSendDistance=3000
   minSendLevel=0
   sendLevel=1
   soundSourceBus=/Game/Sounds/Apollo/Biomes/Reflections/AmbientAudio_Reflection_SourceBus
  }]
 concurrencyOverrides {maxCount=2, voiceStealReleaseTime=0.5}
}
