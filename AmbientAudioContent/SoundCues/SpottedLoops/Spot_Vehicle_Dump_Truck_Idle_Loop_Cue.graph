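# Looping idle sound cue for a spotted dump truck in the ambient audio set.
# One looping wave runs through an Enveloper fade-in into two custom distance attenuations, crossfading a near layer into a far layer with stereo spread and listener focus.
# A Mixer blends both layers and feeds every quality level identically.
# Low-pass filtering with distance, a distance-based reverb submix send and a tight concurrency limit shape playback.
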SoundCue /AmbientAudioContent/SoundCues/SpottedLoops/Spot_Vehicle_Dump_Truck_Idle_Loop_Cue
root=QualityLevel_0
nodes [
 {
  id=Attenuation_0
  children [Enveloper_1]
  properties {
   bOverrideAttenuation=true
   attenuationOverrides {
    bAttenuateWithLPF=true
    distanceAlgorithm=Custom
    falloffDistance=4500
    lPFFrequencyAtMax=750
    lPFRadiusMax=4500
    nonFocusVolumeAttenuation=0.5
    attenuationShapeExtents {x=0}
    customAttenuationCurve {editorCurveData {keys [
       <at 0 1>
       {arriveTangent=-0.0325, interpMode=RCIM_Cubic, leaveTangent=-0.0325, tangentMode=RCTM_User, time=0.25, value=1}
       {arriveTangent=-8.5684, interpMode=RCIM_Cubic, leaveTangent=-8.5684, tangentMode=RCTM_User, time=0.5, value=0}
      ]}}
   }
  }
 }
 {
  id=Attenuation_1
  children [Enveloper_1]
  properties {
   bOverrideAttenuation=true
   attenuationOverrides {
    bAttenuateWithLPF=true
    bEnableFocusInterpolation=true
    bEnableListenerFocus=true
    distanceAlgorithm=Custom
    falloffDistance=4500
    focusAttackInterpSpeed=0.98
    focusDistanceScale=1.1
    lPFFrequencyAtMax=750
    lPFRadiusMax=4500
    nonFocusVolumeAttenuation=0.5
    stereoSpread=1000
    attenuationShapeExtents {x=0}
    customAttenuationCurve {editorCurveData {keys [
       <at 0 0>
       {arriveTangent=4.9184, interpMode=RCIM_Cubic, leaveTangent=4.9184, tangentMode=RCTM_User, time=0.25, value=0}
       {arriveTangent=-1.4995, interpMode=RCIM_Cubic, leaveTangent=-1.4995, tangentMode=RCTM_User, time=0.5, value=1}
       <at 1.2 0>
      ]}}
   }
  }
 }
 {
  id=Enveloper_1
  children [WavePlayer_1]
  properties {pitchCurve {editorCurveData {keys [{time=0}]}}, volumeCurve {editorCurveData {keys [<at 0 0>,<at 0.5 1>]}}}
 }
 {id=Mixer_0, children [Attenuation_0, Attenuation_1], properties {inputVolume [1,0.8]}}
 {id=QualityLevel_0, children [Mixer_0,Mixer_0,Mixer_0,Mixer_0]}
 {id=WavePlayer_1, properties {bLooping=true, soundWaveAssetPtr=/AmbientAudioContent/Waves/SpottedLoops/Spot_Vehicle_Dump_Truck_Idle_Loop}}
]
properties {
 bEnableBusSends=false
 bOverrideConcurrency=true
 maxDistance=5400
 soundClassObject=&Fort_Gadgets
 volumeMultiplier=0.2
 attenuationOverrides {
  bAttenuateWithLPF=true
  bEnableListenerFocus=true
  distanceAlgorithm=Custom
  falloffDistance=4500
  lPFFrequencyAtMax=750
  lPFRadiusMax=4500
  nonFocusAzimuth=70
  nonFocusVolumeAttenuation=0.5
  attenuationShapeExtents {x=0}
 }
 concurrencyOverrides {maxCount=2, voiceStealReleaseTime=0.2}
 soundSubmixSends [{
   maxSendDistance=4500
   minSendDistance=1000
   minSendLevel=0
   soundSubmix=&Fortnite_Reverb_Default
   customSendLevelCurve {editorCurveData {keys [<at 0 0.4>,<at 1 1>]}}
  }]
}
