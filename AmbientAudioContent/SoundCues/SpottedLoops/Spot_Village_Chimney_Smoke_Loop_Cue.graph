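# Spatialized ambient Sound Cue for a village chimney smoke loop, played as an outdoor spot sound.
# A Mixer layers a looping bed with five random smoke one-shots that repeat after random delays of a few seconds.
# The bed randomly takes one of two pitch-modulated branches sharing nested looping Envelopers, which slowly drift volume and pitch on different periods.
# Overridden capsule attenuation adds HRTF, distance low-pass and focus, with concurrency limiting voices.
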SoundCue /AmbientAudioContent/SoundCues/SpottedLoops/Spot_Village_Chimney_Smoke_Loop_Cue
root=Mixer_0
nodes [
 {id=Delay_0, children [Modulator_1]}
 {id=Delay_1, children [Random_1], properties {delayMax=8, delayMin=3}}
 {
  id=Enveloper_0
  children [WavePlayer_1]
  properties {
   bLoop=true
   bLoopIndefinitely=true
   loopEnd=30
   pitchCurve {editorCurveData {keys [
      {arriveTangent=-0.0013, interpMode=RCIM_Cubic, leaveTangent=-0.0013, time=0}
      <at 15 0.98 RCIM_Cubic>
      {arriveTangent=0.0013, interpMode=RCIM_Cubic, leaveTangent=0.0013, time=30}
     ]}}
   volumeCurve {editorCurveData {keys [<at 0 1>]}}
  }
 }
 {
  id=Enveloper_3
  children [Enveloper_0]
  properties {
   bLoop=true
   bLoopIndefinitely=true
   loopEnd=20
   pitchCurve {editorCurveData {keys [{time=0}]}}
   volumeCurve {editorCurveData {keys [
      {arriveTangent=-0.01, interpMode=RCIM_Cubic, time=0, value=1}
      <at 10 0.8 RCIM_Cubic>
      {interpMode=RCIM_Cubic, leaveTangent=0.01, time=20, value=1}
     ]}}
  }
 }
 {id=Looping_0, children [Delay_1]}
 {id=Mixer_0, children [Random_0,Looping_0], properties {inputVolume [1,1.2]}}
 {id=Modulator_1, children [Enveloper_3], properties {pitchMax=1.1, pitchMin=1}}
 {id=Modulator_2, children [Enveloper_3], properties {pitchMax=0.9, pitchMin=0.9}}
 {id=Random_0, children [Delay_0,Modulator_2], properties {weights [1,1]}}
 {id=Random_1, children [WavePlayer_2,WavePlayer_3,WavePlayer_4,WavePlayer_5,WavePlayer_6], properties {weights [1,1,1,1,1]}}
 {id=WavePlayer_1, properties {bLooping=true, soundWaveAssetPtr=/AmbientAudioContent/Waves/SpottedLoops/Spot_Village_Chimney_Smoke_Loop}}
 {id=WavePlayer_2, properties {soundWaveAssetPtr=/AmbientAudioContent/Waves/SpottedLoops/Spot_Village_Chimney_Smoke_05}}
 {id=WavePlayer_3, properties {soundWaveAssetPtr=/AmbientAudioContent/Waves/SpottedLoops/Spot_Village_Chimney_Smoke_04}}
 {id=WavePlayer_4, properties {soundWaveAssetPtr=/AmbientAudioContent/Waves/SpottedLoops/Spot_Village_Chimney_Smoke_03}}
 {id=WavePlayer_5, properties {soundWaveAssetPtr=/AmbientAudioContent/Waves/SpottedLoops/Spot_Village_Chimney_Smoke_02}}
 {id=WavePlayer_6, properties {soundWaveAssetPtr=/AmbientAudioContent/Waves/SpottedLoops/Spot_Village_Chimney_Smoke_01}}
]
properties {
 bEnableBusSends=false
 bOverrideAttenuation=true
 maxDistance=1100
 soundClassObject=&Fort_Ambient_Outdoors_Athena
 subtitlePriority=5000
 volumeMultiplier=0.8
 attenuationOverrides {
  attenuationShape=Capsule
  bAttenuateWithLPF=true
  bEnableFocusInterpolation=true
  bEnableLogFrequencyScaling=true
  dBAttenuationAtMax=-40
  distanceAlgorithm=Logarithmic
  falloffDistance=850
  focusAzimuth=26
  hPFFrequencyAtMax=100
  hPFFrequencyAtMin=100
  lPFFrequencyAtMax=750
  lPFFrequencyAtMin=12000
  lPFRadiusMax=2000
  lPFRadiusMin=750
  nonFocusAzimuth=55
  nonFocusVolumeAttenuation=0.4
  occlusionLowPassFilterFrequency=2000
  occlusionTraceChannel=ECC_GameTraceChannel11
  spatializationAlgorithm=SPATIALIZATION_HRTF
  stereoSpread=0
  attenuationShapeExtents=<250 166 90>
 }
 busSends [{audioBus=null, minSendLevel=0, sendLevel=1, soundSourceBus=&AmbientAudio_Exterior_SourceBus}]
 concurrencyOverrides {maxCount=1, voiceStealReleaseTime=2}
 concurrencySet [/AmbientAudioBase/Concurrency/AmbientConc_3Max_Spot_Default_Concurrency]
}
