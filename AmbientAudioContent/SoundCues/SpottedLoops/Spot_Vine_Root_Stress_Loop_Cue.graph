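# Ambient prop cue for creaking, stressed vines and roots, layering looping beds with randomly delayed, repeating one-shots.
# The first two quality slots pair five one-shots with thorn and vine loops, run through one of two randomly chosen envelope chains that slowly drift pitch and volume.
# The last two slots use a lighter mix: two one-shots over one thorn loop.
# Overrides set capsule attenuation with distance filtering, listener focus and a small concurrency cap.
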SoundCue /AmbientAudioContent/SoundCues/SpottedLoops/Spot_Vine_Root_Stress_Loop_Cue
root=QualityLevel_0
nodes [
 {id=Delay_0, children [Modulator_0], properties {delayMax=6, delayMin=3}}
 {id=Delay_1, children [Modulator_3], properties {delayMax=8, delayMin=3}}
 {
  id=Enveloper_0
  children [Mixer_1]
  properties {
   bLoop=true
   bLoopIndefinitely=true
   loopEnd=30
   pitchCurve {editorCurveData {keys [{interpMode=RCIM_Cubic, time=0}, <at 15 1.2 RCIM_Cubic>, {interpMode=RCIM_Cubic, time=30}]}}
   volumeCurve {editorCurveData {keys [<at 0 1>]}}
  }
 }
 {
  id=Enveloper_1
  children [Mixer_1]
  properties {
   bLoop=true
   bLoopIndefinitely=true
   loopEnd=30
   pitchCurve {editorCurveData {keys [{interpMode=RCIM_Cubic, time=0}, <at 15 0.85 RCIM_Cubic>, {interpMode=RCIM_Cubic, time=30}]}}
   volumeCurve {editorCurveData {keys [<at 0 1>]}}
  }
 }
 {
  id=Enveloper_2
  children [Enveloper_1]
  properties {
   bLoop=true
   bLoopIndefinitely=true
   loopEnd=10
   pitchCurve {editorCurveData {keys [{time=0}]}}
   volumeCurve {editorCurveData {keys [
      {arriveTangent=-0.01, interpMode=RCIM_Cubic, time=0, value=1}
      <at 5 0.9 RCIM_Cubic>
      {interpMode=RCIM_Cubic, leaveTangent=0.01, time=10, value=1}
     ]}}
  }
 }
 {
  id=Enveloper_3
  children [Enveloper_0]
  properties {
   bLoop=true
   bLoopIndefinitely=true
   loopEnd=10
   pitchCurve {editorCurveData {keys [{time=0}]}}
   volumeCurve {editorCurveData {keys [
      {arriveTangent=-0.01, interpMode=RCIM_Cubic, time=0, value=1}
      <at 5 1.2 RCIM_Cubic>
      {interpMode=RCIM_Cubic, leaveTangent=0.01, time=10, value=1}
     ]}}
  }
 }
 {id=Looping_0, children [Delay_0]}
 {id=Looping_1, children [Delay_1]}
 {id=Mixer_0, children [Looping_0,Random_5], properties {inputVolume [1.3,1]}}
 {id=Mixer_1, children [Modulator_1,Modulator_2], properties {inputVolume [1,1]}}
 {id=Mixer_2, children [Looping_1,Modulator_4], properties {inputVolume [1,1]}}
 {id=Modulator_0, children [Random_0], properties {pitchMax=1, pitchMin=0.8}}
 {id=Modulator_1, children [Random_1], properties {pitchMax=1, pitchMin=0.9}}
 {id=Modulator_2, children [Random_2], properties {pitchMax=1, pitchMin=0.9}}
 {id=Modulator_3, children [Random_3]}
 {id=Modulator_4, children [Random_4]}
 {id=QualityLevel_0, children [Mixer_0,Mixer_0,Mixer_2,Mixer_2]}
 {id=Random_0, children [WavePlayer_0,WavePlayer_1,WavePlayer_2,WavePlayer_3,WavePlayer_4], properties {weights [1,1,1,1,1]}}
 {id=Random_1, children [WavePlayer_5,WavePlayer_6], properties {weights [1,1]}}
 {id=Random_2, children [WavePlayer_7,WavePlayer_8], properties {weights [1,1]}}
 {id=Random_3, children [WavePlayer_9, WavePlayer_10], properties {weights [1,1]}}
 {id=Random_4, children [WavePlayer_11, WavePlayer_12], properties {weights [1,1]}}
 {id=Random_5, children [Enveloper_2,Enveloper_3], properties {weights [1,1]}}
 {id=WavePlayer_0, properties {soundWaveAssetPtr=/AmbientAudioContent/Waves/OneShots/Spot_Vine_Root_Stress_01}}
 {id=WavePlayer_1, properties {soundWaveAssetPtr=/AmbientAudioContent/Waves/OneShots/Spot_Vine_Root_Stress_02}}
 {id=WavePlayer_10, properties {soundWaveAssetPtr=&Spot_Vine_Root_Stress_02}}
 {id=WavePlayer_11, properties {bLooping=true, soundWaveAssetPtr=/AmbientAudioContent/Waves/SpottedLoops/Spot_Thorn_Root_Stress_Loop_04}}
 {id=WavePlayer_12, properties {bLooping=true, soundWaveAssetPtr=/AmbientAudioContent/Waves/SpottedLoops/Spot_Thorn_Root_Stress_Loop_03}}
 {id=WavePlayer_2, properties {soundWaveAssetPtr=/AmbientAudioContent/Waves/OneShots/Spot_Vine_Root_Stress_03}}
 {id=WavePlayer_3, properties {soundWaveAssetPtr=/AmbientAudioContent/Waves/OneShots/Spot_Vine_Root_Stress_04}}
 {id=WavePlayer_4, properties {soundWaveAssetPtr=/AmbientAudioContent/Waves/OneShots/Spot_Vine_Root_Stress_05}}
 {id=WavePlayer_5, properties {bLooping=true, soundWaveAssetPtr=&Spot_Thorn_Root_Stress_Loop_04}}
 {id=WavePlayer_6, properties {bLooping=true, soundWaveAssetPtr=&Spot_Thorn_Root_Stress_Loop_03}}
 {id=WavePlayer_7, properties {bLooping=true, soundWaveAssetPtr=/AmbientAudioContent/Waves/SpottedLoops/Spot_Vine_Root_Stress_Loop_01}}
 {id=WavePlayer_8, properties {bLooping=true, soundWaveAssetPtr=/AmbientAudioContent/Waves/SpottedLoops/Spot_Vine_Root_Stress_Loop_02}}
 {id=WavePlayer_9, properties {soundWaveAssetPtr=&Spot_Vine_Root_Stress_01}}
]
properties {
 bDebug=true
 bOverrideAttenuation=true
 bOverrideConcurrency=true
 maxDistance=1350
 soundClassObject=&Fort_Ambient_Props_Athena
 volumeMultiplier=1
 attenuationOverrides {
  attenuationShape=Capsule
  bAttenuateWithLPF=true
  bEnableListenerFocus=true
  falloffDistance=950
  focusAzimuth=15
  focusVolumeAttenuation=0.8
  hPFFrequencyAtMax=400
  lPFFrequencyAtMax=850
  lPFRadiusMax=2000
  lPFRadiusMin=850
  nonFocusAzimuth=30
  nonFocusVolumeAttenuation=0.4
  stereoSpread=50
  attenuationShapeExtents {y=200}
 }
 concurrencyOverrides {maxCount=2}
}
