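# Sound cue for a spotted ambient loop of a washing machine or dryer tumbling gold bricks, in the Fort_Ambient_Athena sound class.
# A Random node picks one of four heavy-brick looping waves, and a Modulator varies its pitch slightly.
# An Attenuation node referencing ATT_Washing_Machine feeds the first two quality-level inputs, leaving the other two empty.
# A concurrency override caps simultaneous instances at a few voices.
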SoundCue /AmbientAudioContent/SoundCues/SpottedLoops/Spot_Washing_Machine_Dryer_Gold_Bricks_Cue
root=QualityLevel_1
nodes [
 {
  id=Attenuation_1
  children [Modulator_1]
  properties {
   attenuationSettings=&/ATT_Washing_Machine
   attenuationOverrides {
    bAttenuateWithLPF=true
    bEnableListenerFocus=true
    bEnableOcclusion=true
    bEnableReverbSend=false
    distanceAlgorithm=Logarithmic
    falloffDistance=1000
    focusAzimuth=10
    focusReleaseInterpSpeed=0.5
    lPFFrequencyAtMax=2000
    lPFRadiusMax=3000
    lPFRadiusMin=500
    nonFocusAzimuth=40
    occlusionInterpolationTime=1
    occlusionLowPassFilterFrequency=2000
    occlusionTraceChannel=ECC_GameTraceChannel11
    occlusionVolumeAttenuation=0.5
    spatializationAlgorithm=SPATIALIZATION_HRTF
    stereoSpread=0
    attenuationShapeExtents {x=150}
   }
  }
 }
 {id=Modulator_1, children [Random_2], properties {pitchMax=1.1, pitchMin=0.85, volumeMax=1, volumeMin=1}}
 {id=QualityLevel_1, children [Attenuation_1, Attenuation_1, null, null]}
 {
  id=Random_2
  children [WavePlayer_1,WavePlayer_2,WavePlayer_3,WavePlayer_4]
  properties {bShouldExcludeFromBranchCulling=true, weights [1,1,1,1]}
 }
 {
  id=WavePlayer_1
  properties {bLooping=true, soundWaveAssetPtr=/AmbientAudioContent/Waves/SpottedLoops/Spot_Washer_Dryer_Bricks_Heavy_Loop_01}
 }
 {
  id=WavePlayer_2
  properties {bLooping=true, soundWaveAssetPtr=/AmbientAudioContent/Waves/SpottedLoops/Spot_Washer_Dryer_Bricks_Heavy_Loop_02}
 }
 {
  id=WavePlayer_3
  properties {bLooping=true, soundWaveAssetPtr=/AmbientAudioContent/Waves/SpottedLoops/Spot_Washer_Dryer_Bricks_Heavy_Loop_03}
 }
 {
  id=WavePlayer_4
  properties {bLooping=true, soundWaveAssetPtr=/AmbientAudioContent/Waves/SpottedLoops/Spot_Washer_Dryer_Bricks_Heavy_Loop_04}
 }
]
properties {
 bOverrideConcurrency=true
 maxDistance=750
 soundClassObject=&Fort_Ambient_Athena
 volumeMultiplier=0.4
 attenuationOverrides {attenuationShapeExtents {x=0}}
 concurrencyOverrides {maxCount=3, voiceStealReleaseTime=0.25, volumeScaleAttackTime=0.05}
}
