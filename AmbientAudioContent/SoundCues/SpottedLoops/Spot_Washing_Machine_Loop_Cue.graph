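# Spotted ambient sound cue that plays a looping washing machine sound in the Athena ambient sound class.
# A quality-level switch sends the first two levels to a random pick of four washer/dryer loops and the last two to a pick of two.
# Each branch adds modulation and its own attenuation with distance low-pass, listener focus and occlusion, differing in occlusion filter cutoff.
# Concurrency is capped, stopping the quietest voice.
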SoundCue /AmbientAudioContent/SoundCues/SpottedLoops/Spot_Washing_Machine_Loop_Cue
root=QualityLevel_0
nodes [
 {
  id=Attenuation_0
  children [Modulator_3]
  properties {
   bOverrideAttenuation=true
   attenuationOverrides {
    bAttenuateWithLPF=true
    bEnableFocusInterpolation=true
    bEnableListenerFocus=true
    bEnableOcclusion=true
    bEnableReverbSend=false
    falloffDistance=800
    focusReleaseInterpSpeed=0.5
    lPFFrequencyAtMax=2000
    lPFRadiusMax=3000
    lPFRadiusMin=500
    nonFocusVolumeAttenuation=0.5
    occlusionInterpolationTime=1
    occlusionLowPassFilterFrequency=500
    occlusionTraceChannel=ECC_GameTraceChannel11
    occlusionVolumeAttenuation=0.5
    stereoSpread=512
    attenuationShapeExtents {x=160}
   }
  }
 }
 {
  id=Attenuation_2
  children [Modulator_0]
  properties {
   bOverrideAttenuation=true
   attenuationOverrides {
    bAttenuateWithLPF=true
    bEnableFocusInterpolation=true
    bEnableListenerFocus=true
    bEnableOcclusion=true
    bEnableReverbSend=false
    falloffDistance=800
    focusReleaseInterpSpeed=0.5
    lPFFrequencyAtMax=2000
    lPFRadiusMax=3000
    lPFRadiusMin=500
    nonFocusVolumeAttenuation=0.5
    occlusionInterpolationTime=1
    occlusionLowPassFilterFrequency=1000
    occlusionTraceChannel=ECC_GameTraceChannel11
    occlusionVolumeAttenuation=0.5
    stereoSpread=512
    attenuationShapeExtents {x=160}
   }
  }
 }
 {id=Modulator_0, children [Random_0]}
 {id=Modulator_3, children [Random_1]}
 {id=QualityLevel_0, children [Attenuation_0, Attenuation_0, Attenuation_2, Attenuation_2]}
 {id=Random_0, children [WavePlayer_10, WavePlayer_0], properties {weights [1,1]}}
 {id=Random_1, children [WavePlayer_6,WavePlayer_7,WavePlayer_8,WavePlayer_9], properties {weights [1,1,1,1]}}
 {id=WavePlayer_0, properties {bLooping=true, soundWaveAssetPtr=/AmbientAudioContent/Waves/SpottedLoops/Spot_Washer_Dryer_Loop_04}}
 {id=WavePlayer_10, properties {bLooping=true, soundWaveAssetPtr=/AmbientAudioContent/Waves/SpottedLoops/Spot_Washer_Dryer_Loop_01}}
 {id=WavePlayer_6, properties {bLooping=true, soundWaveAssetPtr=&Spot_Washer_Dryer_Loop_04}}
 {id=WavePlayer_7, properties {bLooping=true, soundWaveAssetPtr=/AmbientAudioContent/Waves/SpottedLoops/Spot_Washer_Dryer_Loop_03}}
 {id=WavePlayer_8, properties {bLooping=true, soundWaveAssetPtr=/AmbientAudioContent/Waves/SpottedLoops/Spot_Washer_Dryer_Loop_02}}
 {id=WavePlayer_9, properties {bLooping=true, soundWaveAssetPtr=&Spot_Washer_Dryer_Loop_01}}
]
properties {
 bOverrideConcurrency=true
 maxDistance=960
 soundClassObject=&Fort_Ambient_Athena
 volumeMultiplier=0.45
 attenuationOverrides {attenuationShapeExtents {x=0}}
 concurrencyOverrides {maxCount=4, resolutionRule=StopQuietest, voiceStealReleaseTime=0.25, volumeScaleAttackTime=0.05}
}
