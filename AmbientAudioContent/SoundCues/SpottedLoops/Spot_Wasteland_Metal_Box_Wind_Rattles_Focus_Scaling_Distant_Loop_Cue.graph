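# Looping spot ambience cue of gusty wind rattling metal, routed to the Athena indoor ambient sound class.
# After a random delay of a few seconds, a Looping node repeatedly picks from six equally weighted interior rattle one-shots.
# The delay sits outside the loop, so it staggers only the start, and every quality level shares this branch.
# Overrides apply box-shaped attenuation with short falloff and a tight concurrency limit.
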SoundCue /AmbientAudioContent/SoundCues/SpottedLoops/Spot_Wasteland_Metal_Box_Wind_Rattles_Focus_Scaling_Distant_Loop_Cue
root=QualityLevel_0
nodes [
 {id=Delay_0, children [Looping_0], properties {delayMax=8, delayMin=3}}
 {id=Looping_0, children [Random_1]}
 {id=QualityLevel_0, children [Delay_0,Delay_0,Delay_0,Delay_0]}
 {
  id=Random_1
  children [WavePlayer_2, WavePlayer_3, WavePlayer_8, WavePlayer_9, WavePlayer_10, WavePlayer_11]
  properties {weights [1,1,1,1,1,1]}
 }
 {id=WavePlayer_10, properties {soundWaveAssetPtr=/AmbientAudioContent/Waves/OneShots/Spot_Wind_Metal_Ratte_Gusty_Interior_05}}
 {id=WavePlayer_11, properties {soundWaveAssetPtr=/AmbientAudioContent/Waves/OneShots/Spot_Wind_Metal_Ratte_Gusty_Interior_06}}
 {id=WavePlayer_2, properties {soundWaveAssetPtr=/AmbientAudioContent/Waves/OneShots/Spot_Wind_Metal_Ratte_Gusty_Interior_01}}
 {id=WavePlayer_3, properties {soundWaveAssetPtr=/AmbientAudioContent/Waves/OneShots/Spot_Wind_Metal_Ratte_Gusty_Interior_02}}
 {id=WavePlayer_8, properties {soundWaveAssetPtr=/AmbientAudioContent/Waves/OneShots/Spot_Wind_Metal_Ratte_Gusty_Interior_03}}
 {id=WavePlayer_9, properties {soundWaveAssetPtr=/AmbientAudioContent/Waves/OneShots/Spot_Wind_Metal_Ratte_Gusty_Interior_04}}
]
properties {
 bOverrideAttenuation=true
 bOverrideConcurrency=true
 maxDistance=900.4811
 soundClassObject=&Fort_Ambient_Indoors_Athena
 volumeMultiplier=0.501
 attenuationOverrides {attenuationShape=Box, falloffDistance=250, attenuationShapeExtents=<500.1809 507.2498 650.4811>}
 concurrencyOverrides {maxCount=2}
 preEffectBusSends [{}]
}
