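# Spotted ambient hot tub water loop cue on the Ambient Props sound class, used as a quiet positional sound.
# A looping bed mixes with randomly chosen bubble one-shots that repeat after random delays.
# Each play randomly picks a slightly raised-pitch branch with a short start delay or a lowered-pitch one, both fading in, across all quality levels.
# Overrides add custom falloff, low-pass, occlusion, listener focus and a concurrency cap, alongside a distance-based reverb send.
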SoundCue /AmbientAudioContent/SoundCues/SpottedLoops/Spot_Water_HotTub_Loop_Cue
root=QualityLevel_0
nodes [
 {id=Delay_0, children [Modulator_0], properties {delayMax=5, delayMin=2}}
 {id=Delay_1, children [Enveloper_6], properties {delayMax=0.5, delayMin=0.5}}
 {
  id=Enveloper_6
  children [Modulator_1]
  properties {pitchCurve {editorCurveData {keys [{time=0}]}}, volumeCurve {editorCurveData {keys [<at 0 0>,<at 0.5 1>]}}}
 }
 {
  id=Enveloper_8
  children [Modulator_2]
  properties {pitchCurve {editorCurveData {keys [{time=0}]}}, volumeCurve {editorCurveData {keys [<at 0 0>,<at 0.5 1>]}}}
 }
 {id=Looping_0, children [Delay_0]}
 {id=Mixer_0, children [WavePlayer_3,Looping_0], properties {inputVolume [1,0.8]}}
 {id=Modulator_0, children [Random_1]}
 {id=Modulator_1, children [Mixer_0], properties {pitchMax=1.1, pitchMin=1.05}}
 {id=Modulator_2, children [Mixer_0], properties {pitchMax=1, pitchMin=0.9}}
 {id=QualityLevel_0, children [Random_2,Random_2,Random_2,Random_2]}
 {id=Random_1, children [WavePlayer_4,WavePlayer_5,WavePlayer_6], properties {weights [1,1,1]}}
 {id=Random_2, children [Delay_1,Enveloper_8], properties {weights [1,1]}}
 {id=WavePlayer_3, properties {bLooping=true, soundWaveAssetPtr=/AmbientAudioContent/Waves/SpottedLoops/Spot_Hot_Tub_Actived_Loop}}
 {id=WavePlayer_4, properties {soundWaveAssetPtr=/AmbientAudioContent/Waves/OneShots/Spot_Hot_Tub_Bubble_Swt_01}}
 {id=WavePlayer_5, properties {soundWaveAssetPtr=/AmbientAudioContent/Waves/OneShots/Spot_Hot_Tub_Bubble_Swt_02}}
 {id=WavePlayer_6, properties {soundWaveAssetPtr=/AmbientAudioContent/Waves/OneShots/Spot_Hot_Tub_Bubble_Swt_03}}
]
properties {
 attenuationSettings=/Game/Sounds/Helios/POI/MountOlympus/Spots/Fort_Olympus_Amb_Water_Pool_Short_Att
 bOverrideAttenuation=true
 bOverrideConcurrency=true
 maxDistance=2000
 soundClassObject=&Fort_Ambient_Props_Athena
 volumeMultiplier=0.15
 attenuationOverrides {
  bAttenuateWithLPF=true
  bEnableFocusInterpolation=true
  bEnableListenerFocus=true
  bEnableOcclusion=true
  distanceAlgorithm=Custom
  falloffDistance=2000
  focusAttackInterpSpeed=0.95
  lPFFrequencyAtMax=750
  lPFRadiusMax=2000
  lPFRadiusMin=750
  nonFocusVolumeAttenuation=0.4
  nonSpatializedRadiusEnd=200
  nonSpatializedRadiusStart=500
  occlusionLowPassFilterFrequency=4000
  occlusionTraceChannel=ECC_GameTraceChannel11
  occlusionVolumeAttenuation=0.5
  attenuationShapeExtents {x=0}
  customAttenuationCurve {editorCurveData {keys [
     <at 0.2 1>
     {arriveTangent=-1.25, interpMode=RCIM_Cubic, leaveTangent=-1.25, time=0.5, value=0.5}
     {arriveTangent=-2.0974, interpMode=RCIM_Cubic, leaveTangent=-2.0974, tangentMode=RCTM_User, time=1, value=0}
    ]}}
 }
 concurrencyOverrides {maxCount=2, voiceStealReleaseTime=2}
 soundSubmixSends [{
   maxSendDistance=2000
   maxSendLevel=0.6
   minSendDistance=1000
   minSendLevel=0.25
   soundSubmix=&Fortnite_Reverb_Default
   customSendLevelCurve {editorCurveData {keys [<at 0 0>,<at 1 1>]}}
  }]
}
