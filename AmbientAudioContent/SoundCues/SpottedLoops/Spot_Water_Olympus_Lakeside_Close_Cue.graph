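# Ambient spot loop cue for close lakeside water at the Olympus location.
# A mixer layers two looping waves at equal volume: a close lake loop and a shore lapping loop.
# It plays in the Athena outdoor ambient sound class with a shared river water attenuation asset.
# Concurrency overrides cap it to a single instance, preventing new ones from starting.
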SoundCue /AmbientAudioContent/SoundCues/SpottedLoops/Spot_Water_Olympus_Lakeside_Close_Cue
root=Mixer_0
nodes [
 {id=Mixer_0, children [WavePlayer_0,WavePlayer_1], properties {inputVolume [1,1]}}
 {
  id=WavePlayer_0
  properties {bLooping=true, soundWaveAssetPtr=/AmbientAudioContent/Waves/SpottedLoops/Spot_Starlight_S1_Water_Lake_Close_Loop_01}
 }
 {id=WavePlayer_1, properties {bLooping=true, soundWaveAssetPtr=/AmbientAudioContent/Waves/Water/Spot_Water_Lake_Shore_Lapping_01}}
]
properties {
 attenuationSettings=/Game/Athena/Sounds/Ambient/Fort_Amb_Water_River_Att
 maxDistance=2925
 soundClassObject=&Fort_Ambient_Outdoors_Athena
 volumeMultiplier=0.501
 attenuationOverrides {falloffDistance=2000, nonSpatializedRadiusStart=700, stereoSpread=1000, attenuationShapeExtents {x=925}}
 concurrencyOverrides {maxCount=1, resolutionRule=PreventNew}
}
