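# Ambient Sound Cue for close, gentle lakeside water at Olympus, layering looping shore recordings in the Athena outdoor ambience sound class.
# A slightly pitched-down pair of ocean-shore loops uses a very close shore attenuation and is mixed below two lake loops on river attenuation.
# Each pair has its own attenuation node before the final mix, and concurrency overrides specify a single instance that blocks new ones.
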SoundCue /AmbientAudioContent/SoundCues/SpottedLoops/Spot_Water_Olympus_Lakeside_Close_Gentle_Crashing_Cue
root=Mixer_2
nodes [
 {id=Attenuation_0, children [Mixer_0], properties {attenuationSettings=/Game/Athena/Sounds/Ambient/Fort_Amb_Water_River_Att}}
 {
  id=Attenuation_9
  children [Modulator_2]
  properties {attenuationSettings=/AmbientAudioBase/Attenuations/Asteria_Water_Ocean_Shore_Very_Close_Att}
 }
 {id=Mixer_0, children [WavePlayer_0,WavePlayer_1], properties {inputVolume [1,1]}}
 {id=Mixer_1, children [WavePlayer_2,WavePlayer_4], properties {inputVolume [1,1]}}
 {id=Mixer_2, children [Attenuation_9, Attenuation_0], properties {inputVolume [0.7,1]}}
 {id=Modulator_2, children [Mixer_1], properties {pitchMax=0.9, pitchMin=0.9}}
 {
  id=WavePlayer_0
  properties {bLooping=true, soundWaveAssetPtr=/AmbientAudioContent/Waves/SpottedLoops/Spot_Starlight_S1_Water_Lake_Close_Loop_01}
 }
 {id=WavePlayer_1, properties {bLooping=true, soundWaveAssetPtr=/AmbientAudioContent/Waves/Water/Spot_Water_Lake_Shore_Lapping_01}}
 {id=WavePlayer_2, properties {bLooping=true, soundWaveAssetPtr=/AmbientAudioContent/Waves/Water/Spot_Water_Ocean_Shore_Gentle_Laps_02}}
 {id=WavePlayer_4, properties {bLooping=true, soundWaveAssetPtr=/AmbientAudioContent/Waves/Water/Spot_Water_Ocean_Shore_SmWaves_Loop_01}}
]
properties {
 maxDistance=2925
 soundClassObject=&Fort_Ambient_Outdoors_Athena
 volumeMultiplier=0.501
 attenuationOverrides {falloffDistance=2000, nonSpatializedRadiusStart=700, stereoSpread=1000, attenuationShapeExtents {x=925}}
 concurrencyOverrides {maxCount=1, resolutionRule=PreventNew}
}
