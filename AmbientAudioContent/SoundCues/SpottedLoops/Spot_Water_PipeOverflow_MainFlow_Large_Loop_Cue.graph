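# Looping spotted ambient cue for the main flow of a large overflowing water pipe, in the outdoor ambience sound class.
# A Random root starts the loop either immediately or after a fixed delay, likely staggering instances.
# Both branches share one wave player with a volume fade-in envelope and slight pitch variation, identical across quality levels.
# Capsule attenuation, an exterior source-bus send and a tight concurrency override govern playback.
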SoundCue /AmbientAudioContent/SoundCues/SpottedLoops/Spot_Water_PipeOverflow_MainFlow_Large_Loop_Cue
root=Random_0
nodes [
 {id=Delay_1, children [QualityLevel_0], properties {delayMax=1.5, delayMin=1.5}}
 {
  id=Enveloper_1
  children [WavePlayer_1]
  properties {
   pitchCurve {editorCurveData {keys [{time=0}]}}
   volumeCurve {editorCurveData {keys [<at -0.0001 0 RCIM_Cubic>, {arriveTangent=0.4839, interpMode=RCIM_Cubic, leaveTangent=0.4839, tangentMode=RCTM_User, time=3, value=1}]}}
  }
 }
 {id=Modulator_1, children [Enveloper_1], properties {pitchMax=1.1, pitchMin=0.9, volumeMax=1, volumeMin=1}}
 {id=QualityLevel_0, children [Modulator_1,Modulator_1,Modulator_1,Modulator_1]}
 {id=Random_0, children [Delay_1, QualityLevel_0], properties {weights [1,1]}}
 {
  id=WavePlayer_1
  properties {bLooping=true, soundWaveAssetPtr=/AmbientAudioContent/Waves/SpottedLoops/Spot_Water_PipeOverflow_MainFlow_Large_Loop}
 }
]
properties {
 attenuationSettings=/AmbientAudioBase/Attenuations/Spot_Water_PipeOverflow_MainFlow_Capsule_Att
 bOverrideConcurrency=true
 maxDistance=3150
 priority=10
 soundClassObject=&Fort_Ambient_Outdoors_Athena
 volumeMultiplier=0.501
 busSends [{audioBus=null, minSendLevel=0.6, sendLevel=1, soundSourceBus=&AmbientAudio_Exterior_SourceBus}]
 concurrencyOverrides {maxCount=2, resolutionRule=StopLowestPriority}
}
