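# Sound cue for a spotted ambient loop of water splash impact from an overflowing pipe, in the water ambience sound class.
# A root Random picks evenly between a briefly delayed branch with an extra fade-in envelope and a direct branch.
# Both share a quality node whose levels all feed the same pitch modulator, fade-in envelope and looping wave.
# Attenuation is external, and concurrency allows two instances, stopping the quietest.
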SoundCue /AmbientAudioContent/SoundCues/SpottedLoops/Spot_Water_PipeOverflow_SplashImpact_Loop_01_Cue
root=Random_0
nodes [
 {id=Delay_0, children [Enveloper_0], properties {delayMax=0.5, delayMin=0.3}}
 {
  id=Enveloper_0
  children [QualityLevel_0]
  properties {
   pitchCurve {editorCurveData {keys [{time=0}]}}
   volumeCurve {editorCurveData {keys [<at -0.0001 0 RCIM_Cubic>, {arriveTangent=0.4839, interpMode=RCIM_Cubic, leaveTangent=0.4839, tangentMode=RCTM_User, time=3, value=1}]}}
  }
 }
 {
  id=Enveloper_1
  children [WavePlayer_0]
  properties {
   pitchCurve {editorCurveData {keys [{time=0}]}}
   volumeCurve {editorCurveData {keys [<at -0.0001 0 RCIM_Cubic>, {arriveTangent=0.4839, interpMode=RCIM_Cubic, leaveTangent=0.4839, tangentMode=RCTM_User, time=3, value=1}]}}
  }
 }
 {id=Modulator_1, children [Enveloper_1], properties {pitchMax=1.1, pitchMin=0.9, volumeMax=1, volumeMin=1}}
 {id=QualityLevel_0, children [Modulator_1,Modulator_1,Modulator_1,Modulator_1]}
 {id=Random_0, children [Delay_0, QualityLevel_0], properties {weights [1,1]}}
 {
  id=WavePlayer_0
  properties {bLooping=true, soundWaveAssetPtr=/AmbientAudioContent/Waves/SpottedLoops/Spot_Water_PiperOverflow_SplashImpact_Loop_01}
 }
]
properties {
 attenuationSettings=/Game/Sounds/Asteria/Environment_Actors/WaterPipeOverflow/Spot_Water_PipeOverflow_Splash_Impact_Loop_Att
 bOverrideConcurrency=true
 maxDistance=1500
 priority=10
 soundClassObject=/Game/Sounds/FortSoundClassesAndModes/Fort_Water_Ambient
 volumeMultiplier=0.501
 concurrencyOverrides {maxCount=2, resolutionRule=StopQuietest}
}
