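# Ambient spot loop cue for a large water silo or storage tank prop, with capsule attenuation, listener focus and limited concurrency.
# A random root starts it immediately or after a short delay, into identical mixes for every quality level.
# A tank loop fades cyclically, mixed with randomly chosen, looped stop one-shots after an initial delay.
# Rattle and flowing-liquid loops add a modulated bed under a slow volume and pitch envelope.
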SoundCue /AmbientAudioContent/SoundCues/SpottedLoops/Spot_Water_Silo_Storage_Tank_Large_Capsule_Loop_Cue
root=Random_0
nodes [
 {id=Delay_0, children [QualityLevel_0], properties {delayMax=1, delayMin=1}}
 {id=Delay_1, children [Looping_2], properties {delayMax=12, delayMin=12}}
 {
  id=Enveloper_0
  children [WavePlayer_10]
  properties {
   bLoop=true
   bLoopIndefinitely=true
   loopEnd=24
   pitchCurve {editorCurveData {keys [{time=0}]}}
   volumeCurve {editorCurveData {keys [
      {arriveTangent=0.6828, interpMode=RCIM_Cubic, leaveTangent=0.6828, tangentMode=RCTM_User, time=0, value=0}
      {arriveTangent=-0.0017, interpMode=RCIM_Cubic, leaveTangent=-0.0017, tangentMode=RCTM_User, time=2, value=1}
      {arriveTangent=-0.0017, interpMode=RCIM_Cubic, leaveTangent=-0.0017, tangentMode=RCTM_User, time=12, value=1}
      <at 14 0>
      <at 24 0>
     ]}}
  }
 }
 {
  id=Enveloper_1
  children [Modulator_2]
  properties {
   bLoopIndefinitely=true
   loopEnd=30
   pitchCurve {editorCurveData {keys [{interpMode=RCIM_Cubic, time=0}, <at 15 1.1 RCIM_Cubic>, {interpMode=RCIM_Cubic, time=30}]}}
   volumeCurve {editorCurveData {keys [<at 0 1 RCIM_Cubic>,<at 15 0.7 RCIM_Cubic>,<at 30 1 RCIM_Cubic>]}}
  }
 }
 {id=Looping_2, children [Random_2]}
 {id=Mixer_0, children [WavePlayer_3,WavePlayer_1], properties {inputVolume [2,1.25]}}
 {id=Mixer_2, children [Modulator_0,Modulator_7], properties {inputVolume [3,2]}}
 {id=Mixer_3, children [Mixer_2,Enveloper_1], properties {inputVolume [1,1]}}
 {id=Modulator_0, children [Delay_1]}
 {id=Modulator_2, children [Mixer_0], properties {pitchMin=0.8}}
 {id=Modulator_7, children [Enveloper_0], properties {pitchMax=0.8, pitchMin=0.8}}
 {id=QualityLevel_0, children [Mixer_3,Mixer_3,Mixer_3,Mixer_3]}
 {id=Random_0, children [Delay_0, QualityLevel_0], properties {weights [1,1]}}
 {id=Random_2, children [WavePlayer_4,WavePlayer_8,WavePlayer_9], properties {weights [1,1,1]}}
 {
  id=WavePlayer_1
  properties {bLooping=true, soundWaveAssetPtr=/AmbientAudioContent/Waves/SpottedLoops/Spot_NitroSilo_Liquid_Metal_Flowing_Inside_Loop}
 }
 {id=WavePlayer_10, properties {bLooping=true, soundWaveAssetPtr=/AmbientAudioContent/Waves/SpottedLoops/Spot_OilRig_Storage_Tank_Loop}}
 {id=WavePlayer_3, properties {bLooping=true, soundWaveAssetPtr=/AmbientAudioContent/Waves/SpottedLoops/Spot_Nitro_Chimney_Tube_Rattle_Loop}}
 {id=WavePlayer_4, properties {soundWaveAssetPtr=/AmbientAudioContent/Waves/OneShots/Spot_OilRig_Storage_Tank_Stop_01}}
 {id=WavePlayer_8, properties {soundWaveAssetPtr=/AmbientAudioContent/Waves/OneShots/Spot_OilRig_Storage_Tank_Stop_02}}
 {id=WavePlayer_9, properties {soundWaveAssetPtr=/AmbientAudioContent/Waves/OneShots/Spot_OilRig_Storage_Tank_Stop_04}}
]
properties {
 attenuationSettings=/Game/Sounds/Artemis/Environment_Actors/OilRigStorageTank/Spot_OilRig_Storage_Tank_Att
 bEnableBusSends=false
 bEnableSubmixSends=false
 bOverrideAttenuation=true
 maxDistance=1503.1504
 soundClassObject=&Fort_Ambient_Props_Athena
 volumeMultiplier=0.3
 attenuationOverrides {
  attenuationShape=Capsule
  bEnableListenerFocus=true
  falloffDistance=650
  focusVolumeAttenuation=1.2
  nonFocusVolumeAttenuation=0.75
  attenuationShapeExtents=<853.1504 500>
 }
 concurrencyOverrides {maxCount=2, voiceStealReleaseTime=0.2}
 concurrencySet [/AmbientAudioBase/Concurrency/AmbientConc_3Max_Spot_Default_Concurrency]
}
