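# Sound cue for a looping interior water-splash-on-concrete spot ambience, routed to the Fort water sound class with occlusion-named pipe-overflow attenuation.
# A Random node picks evenly between starting immediately and starting after a short fixed delay, which can stagger instances.
# Both paths share one looping wave with slight pitch variation, identical across all quality levels.
# An overridden concurrency limit allows three instances, stopping the quietest.
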SoundCue /AmbientAudioContent/SoundCues/SpottedLoops/Spot_Water_Splashing_Concrete_Loop_Int_02_Cue
root=Random_0
nodes [
 {id=Delay_0, children [QualityLevel_0], properties {delayMax=0.3, delayMin=0.3}}
 {id=Modulator_1, children [WavePlayer_0], properties {pitchMax=1.1, pitchMin=0.9, volumeMax=1, volumeMin=1}}
 {id=QualityLevel_0, children [Modulator_1,Modulator_1,Modulator_1,Modulator_1]}
 {id=Random_0, children [Delay_0, QualityLevel_0], properties {weights [1,1]}}
 {
  id=WavePlayer_0
  properties {bLooping=true, soundWaveAssetPtr=/AmbientAudioContent/Waves/SpottedLoops/Spot_Water_Splashing_Concrete_Int_Loop_02}
 }
]
properties {
 attenuationSettings=/Game/Sounds/Asteria/Environment_Actors/WaterPipeOverflow/Spot_Water_PipeOverflow_Splash_Impact_Heavy_Occ_Att
 bOverrideConcurrency=true
 maxDistance=1250
 soundClassObject=/Game/Sounds/FortSoundClassesAndModes/Fort_Water
 volumeMultiplier=0.5
 concurrencyOverrides {maxCount=3, resolutionRule=StopQuietest}
 concurrencySet [/AmbientAudioBase/Concurrency/AmbientConc_3Max_Spot_Default_Concurrency]
}
