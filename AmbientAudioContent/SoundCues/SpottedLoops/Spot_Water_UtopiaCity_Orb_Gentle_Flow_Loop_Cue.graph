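# Ambient Sound Cue for a gentle, looping water flow, named for an orb in Utopia City.
# A Mixer layers two looping waves, a boosted close lake loop over a much quieter distant infinity-pool waterfall.
# Custom attenuation adds listener focus, stereo spread and a non-spatialized near radius.
# It routes to the ambient props sound class, and its concurrency override allows one instance, refusing new ones.
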SoundCue /AmbientAudioContent/SoundCues/SpottedLoops/Spot_Water_UtopiaCity_Orb_Gentle_Flow_Loop_Cue
root=Mixer_0
nodes [
 {id=Mixer_0, children [WavePlayer_0,WavePlayer_2], properties {inputVolume [1.2,0.3]}}
 {
  id=WavePlayer_0
  properties {bLooping=true, soundWaveAssetPtr=/AmbientAudioContent/Waves/SpottedLoops/Spot_Starlight_S1_Water_Lake_Close_Loop_01}
 }
 {id=WavePlayer_2, properties {bLooping=true, soundWaveAssetPtr=/AmbientAudioContent/Waves/Water/Spot_Waterfall_Infinitiy_Pool_Cascade_Dist}}
]
properties {
 bOverrideAttenuation=true
 bOverrideConcurrency=true
 maxDistance=2307.7676
 soundClassObject=&Fort_Ambient_Props_Athena
 volumeMultiplier=1
 attenuationOverrides {
  bEnableListenerFocus=true
  falloffDistance=1287.2159
  nonFocusVolumeAttenuation=0.4
  nonSpatializedRadiusStart=700
  stereoSpread=1000
  attenuationShapeExtents {x=1020.5517}
 }
 concurrencyOverrides {maxCount=1, resolutionRule=PreventNew}
}
