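# Sound cue for a looping spot ambience of a wall-mounted fountain stream, classed as ambient water.
# A mixer layers a coliseum fountain loop with a quieter stream loop, then a modulator randomizes pitch, feeding every quality level identically.
# A random node starts that chain immediately or after a short delay, and an enveloper fades it in.
# Overridden capsule attenuation adds distance filtering and occlusion, and concurrency limits simultaneous instances.
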SoundCue /AmbientAudioContent/SoundCues/SpottedLoops/Spot_Water_WallMounted_Fountain_Stream_Loop_Capsule_Cue
root=Enveloper_1
nodes [
 {id=Delay_0, children [QualityLevel_0], properties {delayMax=0.5, delayMin=0.35}}
 {
  id=Enveloper_1
  children [Random_0]
  properties {
   pitchCurve {editorCurveData {keys [{time=0}]}}
   volumeCurve {editorCurveData {keys [<at -0.0001 0 RCIM_Cubic>, {arriveTangent=0.4839, interpMode=RCIM_Cubic, leaveTangent=0.4839, tangentMode=RCTM_User, time=0.5, value=1}]}}
  }
 }
 {id=Mixer_0, children [WavePlayer_1,WavePlayer_2], properties {inputVolume [1,0.25]}}
 {id=Modulator_1, children [Mixer_0], properties {pitchMax=1.2, pitchMin=0.9, volumeMax=1, volumeMin=1}}
 {id=QualityLevel_0, children [Modulator_1,Modulator_1,Modulator_1,Modulator_1]}
 {id=Random_0, children [Delay_0, QualityLevel_0], properties {weights [1,1]}}
 {id=WavePlayer_1, properties {bLooping=true, soundWaveAssetPtr=/AmbientAudioContent/Waves/SpottedLoops/Spot_Coliseum_Fountain_Loop_01}}
 {id=WavePlayer_2, properties {bLooping=true, soundWaveAssetPtr=/AmbientAudioContent/Waves/Water/Spot_Water_Stream_Med_Close_Loop_02}}
]
properties {
 bEnableSubmixSends=false
 bOverrideAttenuation=true
 bOverrideConcurrency=true
 maxDistance=1853.118
 priority=10
 soundClassObject=/Game/Sounds/FortSoundClassesAndModes/Fort_Water_Ambient
 volumeMultiplier=2
 attenuationOverrides {
  attenuationShape=Capsule
  bAttenuateWithLPF=true
  bEnableLogFrequencyScaling=true
  bEnableOcclusion=true
  falloffDistance=1500
  hPFFrequencyAtMax=300
  lPFFrequencyAtMax=500
  lPFRadiusMax=1500
  lPFRadiusMin=950
  nonFocusVolumeAttenuation=0.75
  occlusionInterpolationTime=0.5
  occlusionLowPassFilterFrequency=7500
  occlusionTraceChannel=ECC_GameTraceChannel11
  occlusionVolumeAttenuation=0.4
  attenuationShapeExtents=<353.1181 71.9598>
 }
 concurrencyOverrides {maxCount=2, resolutionRule=StopLowestPriority}
 concurrencySet [/AmbientAudioBase/Concurrency/AmbientConc_3Max_Spot_Default_Concurrency]
}
