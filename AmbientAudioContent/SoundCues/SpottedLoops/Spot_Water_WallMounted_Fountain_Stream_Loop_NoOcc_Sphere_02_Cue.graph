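# Spotted ambient sound cue for a looping wall-mounted fountain water stream, with spherical attenuation and no occlusion.
# Two looping stream waves are mixed, pitch-randomized by a Modulator and routed identically through every quality level.
# A Random node picks an immediate or briefly delayed start, likely staggering instances, under an Enveloper volume fade-in.
# Attenuation and concurrency are overridden on the cue, adding distance-based filtering and a low voice limit.
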SoundCue /AmbientAudioContent/SoundCues/SpottedLoops/Spot_Water_WallMounted_Fountain_Stream_Loop_NoOcc_Sphere_02_Cue
root=Enveloper_1
nodes [
 {id=Delay_0, children [QualityLevel_0], properties {delayMax=0.5, delayMin=0.35}}
 {
  id=Enveloper_1
  children [Random_0]
  properties {
   pitchCurve {editorCurveData {keys [{time=0}]}}
   volumeCurve {editorCurveData {keys [<at -0.0001 0 RCIM_Cubic>, {arriveTangent=0.4839, interpMode=RCIM_Cubic, leaveTangent=0.4839, tangentMode=RCTM_User, time=1, value=1}]}}
  }
 }
 {id=Mixer_1, children [WavePlayer_0,WavePlayer_4], properties {inputVolume [1,1]}}
 {id=Modulator_1, children [Mixer_1], properties {pitchMax=1.2, pitchMin=0.9, volumeMax=1.2, volumeMin=1.2}}
 {id=QualityLevel_0, children [Modulator_1,Modulator_1,Modulator_1,Modulator_1]}
 {id=Random_0, children [Delay_0, QualityLevel_0], properties {weights [1,1]}}
 {id=WavePlayer_0, properties {bLooping=true, soundWaveAssetPtr=/AmbientAudioContent/Waves/Water/Spot_stream_medium_B_01}}
 {id=WavePlayer_4, properties {bLooping=true, soundWaveAssetPtr=/AmbientAudioContent/Waves/Water/Spot_Water_Stream_Med_Close_Loop_03}}
]
properties {
 bEnableSubmixSends=false
 bOverrideAttenuation=true
 bOverrideConcurrency=true
 maxDistance=900
 priority=10
 soundClassObject=/Game/Sounds/FortSoundClassesAndModes/Fort_Water_Ambient
 volumeMultiplier=1.2
 attenuationOverrides {
  bAttenuateWithLPF=true
  bEnableLogFrequencyScaling=true
  falloffDistance=650
  hPFFrequencyAtMax=300
  lPFFrequencyAtMax=3000
  lPFRadiusMax=1500
  lPFRadiusMin=950
  nonFocusVolumeAttenuation=0.75
  occlusionInterpolationTime=0.2
  occlusionLowPassFilterFrequency=5000
  occlusionTraceChannel=ECC_GameTraceChannel11
  occlusionVolumeAttenuation=0.75
  attenuationShapeExtents=<250 174.7142>
 }
 concurrencyOverrides {maxCount=2, resolutionRule=StopLowestPriority}
 concurrencySet [/AmbientAudioBase/Concurrency/AmbientConc_3Max_Spot_Default_Concurrency]
}
