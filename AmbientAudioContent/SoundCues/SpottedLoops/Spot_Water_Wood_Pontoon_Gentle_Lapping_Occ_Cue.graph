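# Ambient spot Sound Cue of gentle water lapping on a wooden pontoon, in the water ambient sound class.
# A Mixer layers a slapping loop, swelled and dipped by a repeating volume envelope, with randomly chosen wood impacts that repeat after random delays with modulated pitch.
# It uses a shared occlusion-named attenuation asset, sends to the default reverb submix and overrides concurrency to allow one instance, blocking new ones.
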SoundCue /AmbientAudioContent/SoundCues/SpottedLoops/Spot_Water_Wood_Pontoon_Gentle_Lapping_Occ_Cue
root=Mixer_0
nodes [
 {id=Delay_0, children [Modulator_0], properties {delayMax=8, delayMin=3}}
 {
  id=Enveloper_1
  children [WavePlayer_0]
  properties {
   bLoop=true
   bLoopIndefinitely=true
   loopEnd=10
   pitchCurve {editorCurveData {keys [{time=0}]}}
   volumeCurve {editorCurveData {keys [
      <at 0 1 RCIM_Cubic>
      <at 5 0.25 RCIM_Cubic>
      {arriveTangent=0.011, interpMode=RCIM_Cubic, leaveTangent=0.011, tangentMode=RCTM_User, time=10, value=1}
     ]}}
  }
 }
 {id=Looping_0, children [Delay_0]}
 {id=Mixer_0, children [Looping_0,Enveloper_1], properties {inputVolume [1,1.2]}}
 {id=Modulator_0, children [Random_0], properties {pitchMax=1.1, pitchMin=0.9}}
 {id=Random_0, children [WavePlayer_3,WavePlayer_4,WavePlayer_5,WavePlayer_6,WavePlayer_7,WavePlayer_8], properties {weights [1,1,1,1,1,1]}}
 {
  id=WavePlayer_0
  properties {bLooping=true, soundWaveAssetPtr=/Game/Sounds/Asteria/POI/IndustrialTown/Spots/Spot_Water_Slapping_Against_Pontoon_Loop}
 }
 {id=WavePlayer_3, properties {soundWaveAssetPtr=/Game/Athena/Sounds/Water/Ambient/Spot_Water_Pontoon_Wood_Impact_01}}
 {id=WavePlayer_4, properties {soundWaveAssetPtr=/Game/Athena/Sounds/Water/Ambient/Spot_Water_Pontoon_Wood_Impact_02}}
 {id=WavePlayer_5, properties {soundWaveAssetPtr=/Game/Athena/Sounds/Water/Ambient/Spot_Water_Pontoon_Wood_Impact_03}}
 {id=WavePlayer_6, properties {soundWaveAssetPtr=/Game/Athena/Sounds/Water/Ambient/Spot_Water_Pontoon_Wood_Impact_04}}
 {id=WavePlayer_7, properties {soundWaveAssetPtr=/Game/Athena/Sounds/Water/Ambient/Spot_Water_Pontoon_Wood_Impact_05}}
 {id=WavePlayer_8, properties {soundWaveAssetPtr=/Game/Athena/Sounds/Water/Ambient/Spot_Water_Pontoon_Wood_Impact_06}}
]
properties {
 attenuationSettings=/AmbientAudioBase/Attenuations/Apollo_Water_Slapping_Impact_Foley_Focus_Occ_Att
 bOverrideConcurrency=true
 maxDistance=1850
 soundClassObject=/Game/Sounds/FortSoundClassesAndModes/Fort_Water_Ambient
 volumeMultiplier=0.8
 attenuationOverrides {
  bAttenuateWithLPF=true
  falloffDistance=6000
  lPFFrequencyAtMax=500
  nonSpatializedRadiusStart=700
  stereoSpread=1000
  attenuationShapeExtents {x=925}
 }
 concurrencyOverrides {maxCount=1, resolutionRule=PreventNew}
 soundSubmixSends [{minSendLevel=0.4, soundSubmix=&Fortnite_Reverb_Default}]
}
