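# Ambient sound cue for a spot-placed nylon windsock flapping outdoors, in the Fort_Ambient_Outdoors_Athena sound class.
# A louder flapping loop mixes with a looping chain of randomly chosen flap one-shots, spaced by short random delays and modulated.
# Two nested looping envelopers add slow, out-of-step variation: pitch over a minute, volume over a shorter cycle.
# Overridden logarithmic attenuation adds occlusion and listener focus to the quiet, short-range source.
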SoundCue /AmbientAudioContent/SoundCues/SpottedLoops/Spot_WindSockLoop_Loop_Cue
root=Enveloper_2
nodes [
 {id=Delay_0, children [Modulator_0], properties {delayMax=3, delayMin=2}}
 {
  id=Enveloper_1
  children [Mixer_0]
  properties {
   bLoop=true
   bLoopIndefinitely=true
   loopEnd=60
   pitchCurve {editorCurveData {keys [
      {interpMode=RCIM_Cubic, time=0}
      <at 15 0.95 RCIM_Cubic>
      {arriveTangent=0.0072, interpMode=RCIM_Cubic, leaveTangent=0.0072, tangentMode=RCTM_User, time=30}
      <at 45 1.1 RCIM_Cubic>
      {time=60}
     ]}}
   volumeCurve {editorCurveData {keys [<at 0 1>]}}
  }
 }
 {
  id=Enveloper_2
  children [Enveloper_1]
  properties {
   bLoop=true
   bLoopIndefinitely=true
   loopEnd=20
   pitchCurve {editorCurveData {keys [{time=0}]}}
   volumeCurve {editorCurveData {keys [
      {arriveTangent=-0.01, interpMode=RCIM_Cubic, time=0, value=1}
      <at 10 0.8 RCIM_Cubic>
      {interpMode=RCIM_Cubic, leaveTangent=0.01, time=20, value=1}
     ]}}
  }
 }
 {id=Looping_0, children [Delay_0]}
 {id=Mixer_0, children [WavePlayer_0,Looping_0], properties {inputVolume [3,1]}}
 {id=Modulator_0, children [Random_1], properties {pitchMin=0.9}}
 {id=Random_1, children [WavePlayer_10, WavePlayer_11, WavePlayer_12, WavePlayer_13], properties {weights [1,1,1,1]}}
 {id=WavePlayer_0, properties {bLooping=true, soundWaveAssetPtr=/AmbientAudioContent/Waves/SpottedLoops/Spot_WindSock_Nylon_Flapping_Loop}}
 {id=WavePlayer_10, properties {soundWaveAssetPtr=/AmbientAudioContent/Waves/OneShots/Spot_WindSock_Nylon_Flap_Sweet_01}}
 {id=WavePlayer_11, properties {soundWaveAssetPtr=/AmbientAudioContent/Waves/OneShots/Spot_WindSock_Nylon_Flap_Sweet_02}}
 {id=WavePlayer_12, properties {soundWaveAssetPtr=/AmbientAudioContent/Waves/OneShots/Spot_WindSock_Nylon_Flap_Sweet_03}}
 {id=WavePlayer_13, properties {soundWaveAssetPtr=/AmbientAudioContent/Waves/OneShots/Spot_WindSock_Nylon_Flap_Sweet_04}}
]
properties {
 bOverrideAttenuation=true
 maxDistance=1600
 soundClassObject=&Fort_Ambient_Outdoors_Athena
 volumeMultiplier=0.15
 attenuationOverrides {
  bEnableFocusInterpolation=true
  bEnableListenerFocus=true
  bEnableOcclusion=true
  dBAttenuationAtMax=-40
  distanceAlgorithm=Logarithmic
  falloffDistance=1500
  focusAttackInterpSpeed=0.1
  focusAzimuth=15
  focusReleaseInterpSpeed=0.1
  focusVolumeAttenuation=1.1
  nonFocusAzimuth=45
  nonFocusVolumeAttenuation=0.4
  nonSpatializedRadiusStart=100
  occlusionTraceChannel=ECC_GameTraceChannel11
  occlusionVolumeAttenuation=0.7
  stereoSpread=2000
  attenuationShapeExtents {x=100}
 }
}
