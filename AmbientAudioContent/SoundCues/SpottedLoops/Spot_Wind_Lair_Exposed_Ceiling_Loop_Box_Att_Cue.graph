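# Spotted looping wind ambience for the Lair's exposed ceiling, played as an indoor ambient sound with box attenuation and limited to one instance.
# After a short random delay it fades in, then a quality switch selects quad waves for the first two levels and stereo for the other two.
# Each branch randomly picks an evening or morning loop, with out-of-step looping volume and pitch envelopes adding slow variation.
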SoundCue /AmbientAudioContent/SoundCues/SpottedLoops/Spot_Wind_Lair_Exposed_Ceiling_Loop_Box_Att_Cue
root=Delay_0
nodes [
 {id=Delay_0, children [Enveloper_4], properties {delayMax=2, delayMin=0.6}}
 {
  id=Enveloper_0
  children [Random_0]
  properties {
   bLoop=true
   bLoopIndefinitely=true
   loopEnd=30
   pitchCurve {editorCurveData {keys [{interpMode=RCIM_Cubic, time=0}, <at 15.9468 0.9181 RCIM_Cubic>, {interpMode=RCIM_Cubic, time=30}]}}
   volumeCurve {editorCurveData {keys [<at 0 1>]}}
  }
 }
 {
  id=Enveloper_1
  children [Random_1]
  properties {
   bLoop=true
   bLoopIndefinitely=true
   loopEnd=30
   pitchCurve {editorCurveData {keys [{interpMode=RCIM_Cubic, time=0}, <at 15.9468 0.9181 RCIM_Cubic>, {interpMode=RCIM_Cubic, time=30}]}}
   volumeCurve {editorCurveData {keys [<at 0 1>]}}
  }
 }
 {
  id=Enveloper_2
  children [Enveloper_1]
  properties {
   bLoop=true
   bLoopIndefinitely=true
   loopEnd=20
   pitchCurve {editorCurveData {keys [{time=0}]}}
   volumeCurve {editorCurveData {keys [
      {arriveTangent=-0.01, interpMode=RCIM_Cubic, time=0, value=1}
      <at 10 1.2 RCIM_Cubic>
      {interpMode=RCIM_Cubic, leaveTangent=0.01, time=20, value=1}
     ]}}
  }
 }
 {
  id=Enveloper_3
  children [Enveloper_0]
  properties {
   bLoop=true
   bLoopIndefinitely=true
   loopEnd=20
   pitchCurve {editorCurveData {keys [{time=0}]}}
   volumeCurve {editorCurveData {keys [
      {arriveTangent=-0.01, interpMode=RCIM_Cubic, time=0, value=1}
      <at 10 1.2 RCIM_Cubic>
      {interpMode=RCIM_Cubic, leaveTangent=0.01, time=20, value=1}
     ]}}
  }
 }
 {
  id=Enveloper_4
  children [QualityLevel_0]
  properties {pitchCurve {editorCurveData {keys [{time=0}]}}, volumeCurve {editorCurveData {keys [<at 0 0>,<at 2 1>]}}}
 }
 {id=QualityLevel_0, children [Enveloper_3,Enveloper_3,Enveloper_2,Enveloper_2]}
 {id=Random_0, children [WavePlayer_0,WavePlayer_2], properties {weights [1,1]}}
 {id=Random_1, children [WavePlayer_4,WavePlayer_3], properties {weights [1,1]}}
 {id=WavePlayer_0, properties {bLooping=true, soundWaveAssetPtr=/AmbientAudioContent/Waves/POI/Ambience_Lair_Evening_IntExt_Loop_Quad}}
 {id=WavePlayer_2, properties {bLooping=true, soundWaveAssetPtr=/AmbientAudioContent/Waves/POI/Ambience_Lair_Morning_Ext_Loop_Quad}}
 {id=WavePlayer_3, properties {bLooping=true, soundWaveAssetPtr=/AmbientAudioContent/Waves/POI/Ambience_Lair_Morning_IntExt_Loop_Stereo}}
 {id=WavePlayer_4, properties {bLooping=true, soundWaveAssetPtr=/AmbientAudioContent/Waves/POI/Ambience_Lair_Evening_IntExt_Loop_Stereo}}
]
properties {
 attenuationSettings=/Game/Sounds/Asteria/POI/The_Lair/Wind/Spot_Wind_Lair_Wind_Box_Focus_Large_Att
 bOverrideConcurrency=true
 maxDistance=1289.2665
 soundClassObject=&Fort_Ambient_Indoors_Athena
 volumeMultiplier=0.4
 attenuationOverrides {falloffDistance=1600}
 concurrencyOverrides {maxCount=1, resolutionRule=PreventNew}
}
