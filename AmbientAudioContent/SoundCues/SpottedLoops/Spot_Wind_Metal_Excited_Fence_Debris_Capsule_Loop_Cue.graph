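# Looping spot ambience cue for wind-rattled metal fence debris, attenuated over distance and limited to one concurrent instance.
# A Random node picks either a delayed dusty metal vibration loop or a mix of exterior and wire wind loops.
# Each branch runs through nested, indefinitely looping Envelopers of differing lengths that modulate volume and pitch over time.
# The first two quality levels share that Random node, while the other two are empty.
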SoundCue /AmbientAudioContent/SoundCues/SpottedLoops/Spot_Wind_Metal_Excited_Fence_Debris_Capsule_Loop_Cue
root=QualityLevel_0
nodes [
 {id=Delay_0, children [Enveloper_3], properties {delayMin=0.35}}
 {
  id=Enveloper_0
  children [WavePlayer_0]
  properties {
   bLoop=true
   bLoopIndefinitely=true
   loopEnd=30
   pitchCurve {editorCurveData {keys [{interpMode=RCIM_Cubic, time=0}, <at 15.9468 0.9181 RCIM_Cubic>, {interpMode=RCIM_Cubic, time=30}]}}
   volumeCurve {editorCurveData {keys [<at 0 1>]}}
  }
 }
 {
  id=Enveloper_1
  children [Mixer_1]
  properties {
   bLoop=true
   bLoopIndefinitely=true
   loopEnd=30
   pitchCurve {editorCurveData {keys [{interpMode=RCIM_Cubic, time=0}, <at 15.9468 1.03 RCIM_Cubic>, {interpMode=RCIM_Cubic, time=30}]}}
   volumeCurve {editorCurveData {keys [<at 0 1>]}}
  }
 }
 {
  id=Enveloper_2
  children [Enveloper_1]
  properties {
   bLoop=true
   bLoopIndefinitely=true
   loopEnd=20
   pitchCurve {editorCurveData {keys [{time=0}]}}
   volumeCurve {editorCurveData {keys [
      {arriveTangent=-0.01, interpMode=RCIM_Cubic, time=0, value=1}
      <at 10 1.2 RCIM_Cubic>
      {interpMode=RCIM_Cubic, leaveTangent=0.01, time=20, value=1}
     ]}}
  }
 }
 {
  id=Enveloper_3
  children [Enveloper_0]
  properties {
   bLoop=true
   bLoopIndefinitely=true
   loopEnd=20
   pitchCurve {editorCurveData {keys [{time=0}]}}
   volumeCurve {editorCurveData {keys [
      {arriveTangent=-0.01, interpMode=RCIM_Cubic, time=0, value=1}
      <at 10 1.2 RCIM_Cubic>
      {interpMode=RCIM_Cubic, leaveTangent=0.01, time=20, value=1}
     ]}}
  }
 }
 {id=Mixer_0, children [WavePlayer_1,WavePlayer_2], properties {inputVolume [0.6,1]}}
 {id=Mixer_1, children [Mixer_0], properties {inputVolume [0.6]}}
 {id=QualityLevel_0, children [Random_0,Random_0,null,null]}
 {id=Random_0, children [Delay_0,Enveloper_2], properties {weights [1,1]}}
 {
  id=WavePlayer_0
  properties {bLooping=true, soundWaveAssetPtr=/AmbientAudioContent/Waves/SpottedLoops/Spot_Wind_Excited_Dusty_Metal_Vibrating_Loop}
 }
 {id=WavePlayer_1, properties {bLooping=true, soundWaveAssetPtr=/AmbientAudioContent/Waves/SpottedLoops/Spot_Winterfest_Exterior_Wind_Loop}}
 {id=WavePlayer_2, properties {bLooping=true, soundWaveAssetPtr=/AmbientAudioContent/Waves/SpottedLoops/Spot_Ascender_Ambient_Wire_Wind_Loop}}
]
properties {
 attenuationSettings=/Game/Sounds/Artemis/POI/Hollow_Base/Spots/Spot_Exterior_Metal_Rattles_Occlusion_Att
 bOverrideConcurrency=true
 maxDistance=1625.1821
 soundClassObject=&Fort_Gadgets
 volumeMultiplier=0.5
 attenuationOverrides {falloffDistance=1600}
 concurrencyOverrides {maxCount=1, voiceStealReleaseTime=3}
}
