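# Looping spot ambience cue that plays low-pitched wooden board creaks at random intervals, with the whole cue pitched down.
# A Looping node repeats a random Delay into a Quality Level switch, whose first branch draws on fifteen creaks through either a normal or raised-pitch Modulator.
# The other quality branches use smaller randomized pools of four or two waves, also modulated in pitch and volume.
# Overrides add listener-focus attenuation and a tight, distance-scaled voice limit.
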SoundCue /AmbientAudioContent/SoundCues/SpottedLoops/Spot_Wood_Creak_Board_Low_Loop_Cue
root=Looping_1
nodes [
 {id=Delay_1, children [QualityLevel_0], properties {delayMax=8, delayMin=3}}
 {id=Looping_1, children [Delay_1]}
 {id=Modulator_0, children [Random_5], properties {pitchMax=1.1, pitchMin=0.8, volumeMax=1, volumeMin=0.3}}
 {id=Modulator_1, children [Random_0], properties {pitchMax=1.1, pitchMin=0.8, volumeMax=1, volumeMin=0.3}}
 {id=Modulator_2, children [Random_3], properties {pitchMax=1.1, pitchMin=0.8, volumeMax=1, volumeMin=0.3}}
 {id=Modulator_4, children [Random_0], properties {pitchMax=1.6, pitchMin=1.4, volumeMax=1, volumeMin=0.3}}
 {id=QualityLevel_0, children [Random_2,Modulator_2,Modulator_0,Modulator_0]}
 {
  id=Random_0
  children [
   WavePlayer_2
   WavePlayer_3
   WavePlayer_4
   WavePlayer_5
   WavePlayer_6
   WavePlayer_7
   WavePlayer_8
   WavePlayer_9
   WavePlayer_10
   WavePlayer_11
   WavePlayer_12
   WavePlayer_26
   WavePlayer_27
   WavePlayer_28
   WavePlayer_29
  ]
  properties {weights [1,1,1,1,1,1,1,1,1,1,1,1,1,1,1]}
 }
 {id=Random_2, children [Modulator_1,Modulator_4], properties {weights [1,1]}}
 {id=Random_3, children [WavePlayer_30, WavePlayer_31, WavePlayer_32, WavePlayer_33], properties {weights [1,1,1,1]}}
 {id=Random_5, children [WavePlayer_34, WavePlayer_35], properties {weights [1,1]}}
 {id=WavePlayer_10, properties {soundWaveAssetPtr=/AmbientAudioContent/Waves/OneShots/SC_Wood_Creak_Board_09}}
 {id=WavePlayer_11, properties {soundWaveAssetPtr=/AmbientAudioContent/Waves/OneShots/SC_Wood_Creak_Board_10}}
 {id=WavePlayer_12, properties {soundWaveAssetPtr=/AmbientAudioContent/Waves/OneShots/SC_Wood_Creak_Board_11}}
 {id=WavePlayer_2, properties {soundWaveAssetPtr=/AmbientAudioContent/Waves/OneShots/SC_Wood_Creak_Board_01}}
 {id=WavePlayer_26, properties {soundWaveAssetPtr=/AmbientAudioContent/Waves/OneShots/SC_Wood_Creak_Board_12}}
 {id=WavePlayer_27, properties {soundWaveAssetPtr=/AmbientAudioContent/Waves/OneShots/SC_Wood_Creak_Board_13}}
 {id=WavePlayer_28, properties {soundWaveAssetPtr=/AmbientAudioContent/Waves/OneShots/SC_Wood_Creak_Board_14}}
 {id=WavePlayer_29, properties {soundWaveAssetPtr=/AmbientAudioContent/Waves/OneShots/SC_Wood_Creak_Board_15}}
 {id=WavePlayer_3, properties {soundWaveAssetPtr=/AmbientAudioContent/Waves/OneShots/SC_Wood_Creak_Board_02}}
 {id=WavePlayer_30, properties {soundWaveAssetPtr=/AmbientAudioContent/Waves/OneShots/SC_Wood_Creak_Board_05}}
 {id=WavePlayer_31, properties {soundWaveAssetPtr=/AmbientAudioContent/Waves/OneShots/SC_Wood_Creak_Board_04}}
 {id=WavePlayer_32, properties {soundWaveAssetPtr=&SC_Wood_Creak_Board_02}}
 {id=WavePlayer_33, properties {soundWaveAssetPtr=&SC_Wood_Creak_Board_01}}
 {id=WavePlayer_34, properties {soundWaveAssetPtr=/AmbientAudioContent/Waves/OneShots/SC_Wood_Creak_Board_07}}
 {id=WavePlayer_35, properties {soundWaveAssetPtr=/AmbientAudioContent/Waves/OneShots/SC_Wood_Creak_Board_06}}
 {id=WavePlayer_4, properties {soundWaveAssetPtr=/AmbientAudioContent/Waves/OneShots/SC_Wood_Creak_Board_03}}
 {id=WavePlayer_5, properties {soundWaveAssetPtr=&SC_Wood_Creak_Board_04}}
 {id=WavePlayer_6, properties {soundWaveAssetPtr=&SC_Wood_Creak_Board_05}}
 {id=WavePlayer_7, properties {soundWaveAssetPtr=&SC_Wood_Creak_Board_06}}
 {id=WavePlayer_8, properties {soundWaveAssetPtr=&SC_Wood_Creak_Board_07}}
 {id=WavePlayer_9, properties {soundWaveAssetPtr=/AmbientAudioContent/Waves/OneShots/SC_Wood_Creak_Board_08}}
]
properties {
 bOverrideAttenuation=true
 bOverrideConcurrency=true
 maxDistance=1350
 pitchMultiplier=0.5
 soundClassObject=&Fort_Ambient_OneShots
 volumeMultiplier=0.8
 attenuationOverrides {bEnableListenerFocus=true, falloffDistance=1000, nonFocusVolumeAttenuation=0.75, attenuationShapeExtents {x=350}}
 concurrencyOverrides {maxCount=2, voiceStealReleaseTime=2, volumeScaleMode=Distance}
}
