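# Looping ambient SoundCue that plays intermittent spot wood creaks, routed to the Fort_Ambient_OneShots sound class.
# A Looping node repeats a random delay into a Quality Level switch, with modulators varying pitch and volume.
# The first quality slot randomizes eleven waves across normal or raised pitch ranges, the second uses three waves, and the last two share one.
# Cue-level overrides set distance attenuation and cap concurrency at two voices.
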SoundCue /AmbientAudioContent/SoundCues/SpottedLoops/Spot_Wood_Creak_Loop_Cue
root=Looping_1
nodes [
 {id=Delay_1, children [QualityLevel_0], properties {delayMax=9, delayMin=3}}
 {id=Looping_1, children [Delay_1]}
 {id=Modulator_0, children [WavePlayer_0], properties {pitchMax=1.1, pitchMin=0.8, volumeMax=1, volumeMin=0.3}}
 {id=Modulator_1, children [Random_4], properties {pitchMax=1.1, pitchMin=0.8, volumeMax=1, volumeMin=0.3}}
 {id=Modulator_2, children [Random_1], properties {pitchMax=1.1, pitchMin=0.8, volumeMax=1, volumeMin=0.3}}
 {id=Modulator_4, children [Random_4], properties {pitchMax=1.6, pitchMin=1.4, volumeMax=1, volumeMin=0.3}}
 {id=QualityLevel_0, children [Random_2,Modulator_2,Modulator_0,Modulator_0]}
 {id=Random_1, children [WavePlayer_13, WavePlayer_14, WavePlayer_15], properties {weights [1,1,1]}}
 {id=Random_2, children [Modulator_1,Modulator_4], properties {weights [1,1]}}
 {
  id=Random_4
  children [
   WavePlayer_1
   WavePlayer_16
   WavePlayer_17
   WavePlayer_18
   WavePlayer_19
   WavePlayer_20
   WavePlayer_21
   WavePlayer_22
   WavePlayer_23
   WavePlayer_24
   WavePlayer_25
  ]
  properties {weights [1,1,1,1,1,1,1,1,1,1,1]}
 }
 {id=WavePlayer_0, properties {soundWaveAssetPtr=/AmbientAudioContent/Waves/OneShots/Spot_wood_creak_medium_01}}
 {id=WavePlayer_1, properties {soundWaveAssetPtr=/AmbientAudioContent/Waves/OneShots/Spot_wood_creak_medium_11}}
 {id=WavePlayer_13, properties {soundWaveAssetPtr=&Spot_wood_creak_medium_01}}
 {id=WavePlayer_14, properties {soundWaveAssetPtr=/AmbientAudioContent/Waves/OneShots/Spot_wood_creak_medium_03}}
 {id=WavePlayer_15, properties {soundWaveAssetPtr=/AmbientAudioContent/Waves/OneShots/Spot_wood_creak_medium_06}}
 {id=WavePlayer_16, properties {soundWaveAssetPtr=/AmbientAudioContent/Waves/OneShots/Spot_wood_creak_medium_10}}
 {id=WavePlayer_17, properties {soundWaveAssetPtr=/AmbientAudioContent/Waves/OneShots/Spot_wood_creak_medium_09}}
 {id=WavePlayer_18, properties {soundWaveAssetPtr=/AmbientAudioContent/Waves/OneShots/Spot_wood_creak_medium_08}}
 {id=WavePlayer_19, properties {soundWaveAssetPtr=/AmbientAudioContent/Waves/OneShots/Spot_wood_creak_medium_07}}
 {id=WavePlayer_20, properties {soundWaveAssetPtr=&Spot_wood_creak_medium_06}}
 {id=WavePlayer_21, properties {soundWaveAssetPtr=/AmbientAudioContent/Waves/OneShots/Spot_wood_creak_medium_05}}
 {id=WavePlayer_22, properties {soundWaveAssetPtr=/AmbientAudioContent/Waves/OneShots/Spot_wood_creak_medium_04}}
 {id=WavePlayer_23, properties {soundWaveAssetPtr=&Spot_wood_creak_medium_03}}
 {id=WavePlayer_24, properties {soundWaveAssetPtr=/AmbientAudioContent/Waves/OneShots/Spot_wood_creak_medium_02}}
 {id=WavePlayer_25, properties {soundWaveAssetPtr=&Spot_wood_creak_medium_01}}
]
properties {
 bOverrideAttenuation=true
 bOverrideConcurrency=true
 maxDistance=1200
 soundClassObject=&Fort_Ambient_OneShots
 volumeMultiplier=0.6
 attenuationOverrides {falloffDistance=1000, stereoSpread=100, attenuationShapeExtents {x=200}}
 concurrencyOverrides {maxCount=2, voiceStealReleaseTime=2, volumeScaleMode=Distance}
 soundSubmixSends [{minSendLevel=0, sendLevel=0, soundSubmix=null}]
}
