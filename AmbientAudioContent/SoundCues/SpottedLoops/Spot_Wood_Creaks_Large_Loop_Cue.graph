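# Looping ambient Sound Cue that plays spotted, large wooden house creak one-shots at irregular intervals.
# A Looping node repeats a random delay of a few seconds, then a Quality Level switch feeding modulated random wave picks.
# The first quality slot draws from six creaks with pitch and volume variation; the other three share a three-creak branch.
# Attenuation and a tight concurrency limit are overridden on the cue, which uses the Fort_Doors sound class.
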SoundCue /AmbientAudioContent/SoundCues/SpottedLoops/Spot_Wood_Creaks_Large_Loop_Cue
root=Looping_0
nodes [
 {id=Delay_0, children [QualityLevel_0], properties {delayMax=8, delayMin=3}}
 {id=Looping_0, children [Delay_0]}
 {id=Modulator_0, children [Random_3], properties {pitchMin=0.75, volumeMin=0.5}}
 {id=Modulator_36, children [Random_2], properties {pitchMax=1, pitchMin=0.8, volumeMax=1, volumeMin=0.5}}
 {id=QualityLevel_0, children [Modulator_36,Modulator_0,Modulator_0,Modulator_0]}
 {
  id=Random_2
  children [WavePlayer_13, WavePlayer_14, WavePlayer_15, WavePlayer_16, WavePlayer_17, WavePlayer_18]
  properties {weights [1,1,1,1,1,1]}
 }
 {id=Random_3, children [WavePlayer_19, WavePlayer_20, WavePlayer_21], properties {weights [1,1,1]}}
 {id=WavePlayer_13, properties {soundWaveAssetPtr=/AmbientAudioContent/Waves/OneShots/Spot_CreakHouseA001}}
 {id=WavePlayer_14, properties {soundWaveAssetPtr=/AmbientAudioContent/Waves/OneShots/Spot_CreakHouseA002}}
 {id=WavePlayer_15, properties {soundWaveAssetPtr=/AmbientAudioContent/Waves/OneShots/Spot_CreakHouseA003}}
 {id=WavePlayer_16, properties {soundWaveAssetPtr=/AmbientAudioContent/Waves/OneShots/Spot_CreakHouseB001}}
 {id=WavePlayer_17, properties {soundWaveAssetPtr=/AmbientAudioContent/Waves/OneShots/Spot_CreakHouseB002}}
 {id=WavePlayer_18, properties {soundWaveAssetPtr=/AmbientAudioContent/Waves/OneShots/Spot_CreakHouseB003}}
 {id=WavePlayer_19, properties {soundWaveAssetPtr=&Spot_CreakHouseA001}}
 {id=WavePlayer_20, properties {soundWaveAssetPtr=&Spot_CreakHouseA002}}
 {id=WavePlayer_21, properties {soundWaveAssetPtr=&Spot_CreakHouseB001}}
]
properties {
 bOverrideAttenuation=true
 bOverrideConcurrency=true
 maxDistance=1350
 soundClassObject=&Fort_Doors
 volumeMultiplier=1
 attenuationOverrides {falloffDistance=1000, reverbWetLevelMin=1, attenuationShapeExtents {x=350}}
 concurrencyOverrides {maxCount=2}
 soundSubmixSends [{minSendLevel=0, sendLevel=0, soundSubmix=null}]
}
