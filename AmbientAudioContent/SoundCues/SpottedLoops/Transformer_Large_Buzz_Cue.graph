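# Ambient Sound Cue for a looping large electrical transformer buzz, assigned to the Athena ambient props sound class.
# A top Random picks either a shared two-wave Random directly, or that same node through a fixed short Delay and raised-pitch Modulator, varying each instance.
# Overridden box attenuation adds logarithmic falloff, listener focus and low-pass occlusion.
# Overridden concurrency caps simultaneous instances, with distance-based volume scaling.
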SoundCue /AmbientAudioContent/SoundCues/SpottedLoops/Transformer_Large_Buzz_Cue
root=Random_1
nodes [
 {id=Delay_0, children [Modulator_0], properties {delayMax=0.3, delayMin=0.3}}
 {id=Modulator_0, children [Random_0], properties {pitchMax=1.1, pitchMin=1.1}}
 {id=Random_0, children [WavePlayer_1,WavePlayer_2], properties {weights [1,1]}}
 {id=Random_1, children [Delay_0,Random_0], properties {weights [1,1]}}
 {id=WavePlayer_1, properties {bLooping=true, soundWaveAssetPtr=/AmbientAudioContent/Waves/SpottedLoops/Spot_Transformer_Large_Buzz_01}}
 {id=WavePlayer_2, properties {bLooping=true, soundWaveAssetPtr=/AmbientAudioContent/Waves/SpottedLoops/Spot_Transformer_Large_Buzz_02}}
]
properties {
 bEnableSubmixSends=false
 bOverrideAttenuation=true
 bOverrideConcurrency=true
 maxDistance=1600
 soundClassObject=&Fort_Ambient_Props_Athena
 volumeMultiplier=0.45
 attenuationOverrides {
  attenuationShape=Box
  bEnableListenerFocus=true
  bEnableOcclusion=true
  distanceAlgorithm=Logarithmic
  falloffDistance=1200
  nonFocusVolumeAttenuation=0.4
  occlusionInterpolationTime=2
  occlusionLowPassFilterFrequency=600
  occlusionTraceChannel=ECC_GameTraceChannel11
  reverbWetLevelMin=0.6905
  stereoSpread=500
  attenuationShapeExtents=<350 350 400>
 }
 concurrencyOverrides {maxCount=3, volumeScaleMode=Distance}
}
