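# SoundCue for a looping ambient turbine prop sound, played quietly and pitched down through the Athena ambient props sound class.
# A root random node picks either a shared random choice of two looping turbine waves directly, or that same choice after a short delay and slight upward pitch modulation, adding variation between instances.
# Overridden capsule attenuation uses logarithmic falloff, with occlusion low-pass filtering on a custom trace channel.
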SoundCue /AmbientAudioContent/SoundCues/SpottedLoops/Turbine_Loop_02_Capsule_Cue
root=Random_1
nodes [
 {id=Delay_0, children [Modulator_0], properties {delayMax=0.5, delayMin=0.5}}
 {id=Modulator_0, children [Random_0], properties {pitchMax=1.2, pitchMin=1.1}}
 {id=Random_0, children [WavePlayer_1,WavePlayer_2], properties {weights [1,1]}}
 {id=Random_1, children [Delay_0,Random_0], properties {weights [1,1]}}
 {id=WavePlayer_1, properties {bLooping=true, soundWaveAssetPtr=/AmbientAudioContent/Waves/SpottedLoops/Spot_Turbine_Loop_01}}
 {id=WavePlayer_2, properties {bLooping=true, soundWaveAssetPtr=/AmbientAudioContent/Waves/SpottedLoops/Spot_Turbine_Loop_02}}
]
properties {
 bOverrideAttenuation=true
 maxDistance=1100
 pitchMultiplier=0.75
 soundClassObject=&Fort_Ambient_Props_Athena
 volumeMultiplier=0.33
 attenuationOverrides {
  attenuationShape=Capsule
  bEnableOcclusion=true
  distanceAlgorithm=Logarithmic
  falloffDistance=650
  occlusionInterpolationTime=0.5
  occlusionLowPassFilterFrequency=1000
  occlusionTraceChannel=ECC_GameTraceChannel11
  reverbWetLevelMin=0.6905
  attenuationShapeExtents=<450 200 800>
 }
}
