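# Looping ambient cue for Sky Island vortex debris, using spline-named attenuation, limited to one instance.
# A distant vortex loop runs through nested 60-second pitch and volume envelopes, with that branch reused several times across two mixers.
# Delayed, looping random picks layer branch debris, mountain gusts and desert wind whistles over the bed.
# The quality node plays this mix on its first two levels and leaves the other two empty.
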SoundCue /AmbientAudioContent/SoundCues/SpottedLoops/Vortex_Leaf_Debris_Only_Loop_Cue
root=QualityLevel_0
nodes [
 {id=Delay_0, children [Looping_0], properties {delayMax=12, delayMin=4}}
 {id=Delay_5, children [Looping_5], properties {delayMax=12, delayMin=8}}
 {id=Delay_6, children [Looping_6], properties {delayMax=14, delayMin=9}}
 {
  id=Enveloper_0
  children [Mixer_0]
  properties {
   bLoop=true
   bLoopIndefinitely=true
   loopEnd=60
   pitchCurve {editorCurveData {keys [
      {interpMode=RCIM_Cubic, time=0}
      <at 15 0.85 RCIM_Cubic>
      {arriveTangent=0.0091, interpMode=RCIM_Cubic, leaveTangent=0.0091, tangentMode=RCTM_User, time=30}
      {interpMode=RCIM_Cubic, time=45}
      {time=60}
     ]}}
   volumeCurve {editorCurveData {keys [<at 0 1>]}}
  }
 }
 {
  id=Enveloper_3
  children [Enveloper_0]
  properties {
   bLoop=true
   bLoopIndefinitely=true
   loopEnd=60
   pitchCurve {editorCurveData {keys [{time=0}]}}
   volumeCurve {editorCurveData {keys [
      {arriveTangent=-0.01, interpMode=RCIM_Cubic, time=0, value=1}
      {arriveTangent=-0.0035, interpMode=RCIM_Cubic, leaveTangent=-0.0035, tangentMode=RCTM_User, time=4.9208, value=0.7335}
      {arriveTangent=0.0016, interpMode=RCIM_Cubic, leaveTangent=0.0016, tangentMode=RCTM_User, time=7.5626, value=0.9613}
      {arriveTangent=0.0008, interpMode=RCIM_Cubic, leaveTangent=0.0008, tangentMode=RCTM_User, time=11.4857, value=0.747}
      {arriveTangent=0.0016, interpMode=RCIM_Cubic, leaveTangent=0.0016, tangentMode=RCTM_User, time=13.3554, value=0.7326}
      {arriveTangent=0.0009, interpMode=RCIM_Cubic, leaveTangent=0.0009, time=15.1401, value=0.8437}
      {arriveTangent=-0.0011, interpMode=RCIM_Cubic, leaveTangent=-0.0011, tangentMode=RCTM_User, time=18.9647, value=0.7376}
      {arriveTangent=0.0002, interpMode=RCIM_Cubic, leaveTangent=0.0002, time=24.105, value=0.9798}
      {arriveTangent=0.0016, interpMode=RCIM_Cubic, leaveTangent=0.0016, tangentMode=RCTM_User, time=28.412, value=0.7398}
      {arriveTangent=-0.0003, interpMode=RCIM_Cubic, leaveTangent=-0.0003, time=31.3775, value=0.8584}
      {arriveTangent=0.0009, interpMode=RCIM_Cubic, leaveTangent=0.0009, tangentMode=RCTM_User, time=35.3006, value=0.7378}
      {arriveTangent=-0.0008, interpMode=RCIM_Cubic, leaveTangent=-0.0008, time=38.0791, value=0.9798}
      <at 42.3861 0.7324>
      {arriveTangent=0.0024, interpMode=RCIM_Cubic, leaveTangent=0.0024, tangentMode=RCTM_User, time=44.9447, value=0.7376}
      {arriveTangent=-0.0006, interpMode=RCIM_Cubic, leaveTangent=-0.0006, time=48.1071, value=0.9613}
      {arriveTangent=-0.0056, interpMode=RCIM_Cubic, leaveTangent=-0.0056, tangentMode=RCTM_User, time=53.5122, value=0.7326}
      <at 55.4914 0.8685>
      {arriveTangent=-0.0003, interpMode=RCIM_Cubic, leaveTangent=-0.0003, tangentMode=RCTM_User, time=58.1315, value=0.7323}
      <at 60 1 RCIM_Cubic>
     ]}}
  }
 }
 {id=Looping_0, children [Mixer_3]}
 {id=Looping_5, children [Random_6]}
 {id=Looping_6, children [Random_5]}
 {id=Mixer_0, children [WavePlayer_0], properties {inputVolume [0.3]}}
 {id=Mixer_1, children [Enveloper_3,Enveloper_3,Delay_0,Delay_6,Delay_5], properties {inputVolume [1.5,2,1,8,8]}}
 {id=Mixer_3, children [Random_0], properties {inputVolume [1]}}
 {id=Mixer_4, children [Enveloper_3,Mixer_1], properties {inputVolume [1,0.25]}}
 {id=QualityLevel_0, children [Mixer_4,Mixer_4,null,null]}
 {
  id=Random_0
  children [WavePlayer_1,WavePlayer_2,WavePlayer_3,WavePlayer_4,WavePlayer_5,WavePlayer_6,WavePlayer_7,WavePlayer_8,WavePlayer_9]
  properties {bShouldExcludeFromBranchCulling=true, weights [1,1,1,1,1,1,1,1,1]}
 }
 {
  id=Random_5
  children [WavePlayer_53, WavePlayer_54, WavePlayer_55, WavePlayer_56, WavePlayer_57, WavePlayer_58]
  properties {weights [1,1,1,1,1,1]}
 }
 {id=Random_6, children [WavePlayer_59, WavePlayer_60, WavePlayer_61, WavePlayer_62, WavePlayer_63], properties {weights [1,1,1,1,1]}}
 {id=WavePlayer_0, properties {bLooping=true, soundWaveAssetPtr=/Game/Sounds/Asteria/POI/SkyIsland/Vortex/Vortex_Outside_Distant_Loop}}
 {id=WavePlayer_1, properties {soundWaveAssetPtr=/Game/Sounds/Asteria/POI/SkyIsland/Vortex/Spot_SkyIsland_Branch_Debris_01}}
 {id=WavePlayer_2, properties {soundWaveAssetPtr=/Game/Sounds/Asteria/POI/SkyIsland/Vortex/Spot_SkyIsland_Branch_Debris_02}}
 {id=WavePlayer_3, properties {soundWaveAssetPtr=/Game/Sounds/Asteria/POI/SkyIsland/Vortex/Spot_SkyIsland_Branch_Debris_03}}
 {id=WavePlayer_4, properties {soundWaveAssetPtr=/Game/Sounds/Asteria/POI/SkyIsland/Vortex/Spot_SkyIsland_Branch_Debris_04}}
 {id=WavePlayer_5, properties {soundWaveAssetPtr=/Game/Sounds/Asteria/POI/SkyIsland/Vortex/Spot_SkyIsland_Branch_Debris_05}}
 {id=WavePlayer_53, properties {soundWaveAssetPtr=/Game/Sounds/Artemis/Biomes/Mountains/Mountain_Spots/Spot_Mountain_Day_Gusty_Wind_03}}
 {id=WavePlayer_54, properties {soundWaveAssetPtr=/Game/Sounds/Artemis/Biomes/Mountains/Mountain_Spots/Spot_Mountain_Day_Gusty_Wind_05}}
 {id=WavePlayer_55, properties {soundWaveAssetPtr=/Game/Sounds/Artemis/Biomes/Mountains/Mountain_Spots/Spot_Mountain_Day_Gusty_Wind_01}}
 {id=WavePlayer_56, properties {soundWaveAssetPtr=/Game/Sounds/Artemis/Biomes/Mountains/Mountain_Spots/Spot_Mountain_Day_Gusty_Wind_02}}
 {id=WavePlayer_57, properties {soundWaveAssetPtr=/Game/Sounds/Artemis/Biomes/Mountains/Mountain_Spots/Spot_Mountain_Day_Gusty_Wind_04}}
 {id=WavePlayer_58, properties {soundWaveAssetPtr=/Game/Sounds/Artemis/Biomes/Mountains/Mountain_Spots/Spot_Mountain_Day_Gusty_Wind_06}}
 {id=WavePlayer_59, properties {soundWaveAssetPtr=/Game/Sounds/Fort_Ambient_NonLoop/Desert/Wind_DesertGust_Whistle14}}
 {id=WavePlayer_6, properties {soundWaveAssetPtr=/Game/Sounds/Asteria/POI/SkyIsland/Vortex/Spot_SkyIsland_Branch_Debris_06}}
 {id=WavePlayer_60, properties {soundWaveAssetPtr=/Game/Sounds/Fort_Ambient_NonLoop/Desert/Wind_DesertGust_Whistle15}}
 {id=WavePlayer_61, properties {soundWaveAssetPtr=/Game/Sounds/Fort_Ambient_NonLoop/Desert/Wind_DesertGust_Whistle16}}
 {id=WavePlayer_62, properties {soundWaveAssetPtr=/Game/Sounds/Fort_Ambient_NonLoop/Desert/Wind_DesertGust_Whistle18}}
 {id=WavePlayer_63, properties {soundWaveAssetPtr=/Game/Sounds/Fort_Ambient_NonLoop/Desert/Wind_DesertGust_Whistle26}}
 {id=WavePlayer_7, properties {soundWaveAssetPtr=/Game/Sounds/Asteria/POI/SkyIsland/Vortex/Spot_SkyIsland_Branch_Debris_07}}
 {id=WavePlayer_8, properties {soundWaveAssetPtr=/Game/Sounds/Asteria/POI/SkyIsland/Vortex/Spot_SkyIsland_Branch_Debris_08}}
 {id=WavePlayer_9, properties {soundWaveAssetPtr=/Game/Sounds/Asteria/POI/SkyIsland/Vortex/Spot_SkyIsland_Branch_Debris_09}}
]
properties {
 attenuationSettings=/Game/Sounds/Asteria/POI/SkyIsland/Vortex/Fort_Amb_Debris_Feild_Spline_Att
 bOverrideConcurrency=true
 maxDistance=2925
 soundClassObject=/Game/Sounds/FortSoundClassesAndModes/Fort_Storms_PSM
 volumeMultiplier=1
 attenuationOverrides {
  bAttenuateWithLPF=true
  bEnableLogFrequencyScaling=true
  distanceAlgorithm=Logarithmic
  falloffDistance=10000
  lPFFrequencyAtMax=197.5057
  lPFFrequencyAtMin=20511.4102
  lPFRadiusMax=3789.3618
  lPFRadiusMin=1822.3112
  stereoSpread=2000
  attenuationShapeExtents {x=0}
 }
 concurrencyOverrides {maxCount=1}
 soundSubmixSends [{minSendLevel=0.6, soundSubmix=&Fortnite_Reverb_Default}]
}
